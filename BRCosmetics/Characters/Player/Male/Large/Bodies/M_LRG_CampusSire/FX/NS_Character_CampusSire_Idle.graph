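# Niagara idle effect for the CampusSire large male player character, rendering glowing headlights on the body.
# One Headlights emitter burst-spawns particles in two spawn groups with mirror axes and pins them through a socket reader.
# Particle attributes drive layered sprite flares, glows and a light renderer, all faded by a system FadeIn value.
# Glow brightness follows camera angle, and front-end mode zeroes the smaller glow's size and camera offset.
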NiagaraSystem /BRCosmetics/Characters/Player/Male/Large/Bodies/M_LRG_CampusSire/FX/NS_Character_CampusSire_Idle
dependencies {
 usedDataInterfaces [
  &NiagaraDataInterfaceCamera
  &NiagaraDataInterfaceSocketReader
  &NiagaraDataInterfaceColorCurve
  &NiagaraDataInterfaceCurve
 ]
 usedDynamicInputs [
  &ColorFromCurve
  &FloatFromCurve
  &Subtract_Float
 ]
 usedModules [
  &self:SystemSpawnScript.SystemScriptSource.SystemScriptGraph.NiagaraNodeAssignment_0.SetVariables_#1
  &self:SystemSpawnScript.SystemScriptSource.SystemScriptGraph.NiagaraNodeAssignment_4.SetVariables_#2
  &SystemState
  &self:SystemSpawnScript.SystemScriptSource.SystemScriptGraph.NiagaraNodeAssignment_6.SetVariables_#3
 ]
}
emitters {Headlights {
  emitterUpdateScript [
   &EmitterState(
    Inactive Response=::NewEnumerator1
    Life Cycle Mode=::NewEnumerator1
    Loop Behavior=::NewEnumerator1
    Loop Duration Mode=::NewEnumerator1
    UseLoopDelay=false
   )
   &SpawnBurst_Instantaneous(Use Spawn Probability=false, UseLoopCountLimit=false)
   &SpawnBurst_Instantaneous(Spawn Time=0.2, Use Spawn Probability=false, UseLoopCountLimit=false) {module=SpawnBurst_Instantaneous001}
  ]
  particleSpawnScript [
   &FN_SetSpawnGroup(Number of Spawn Groups=2)
   /SetVariables(Particles.MirrorAxis:Vector3f=&MaskVectorBySpawnGroup(Masked Vector=<-1 1 1>, Passthrough Vector=<1 1 1>)) {isSetParameters=true}
   &InitializeParticle(
    Color=<rgba 0.7083 0.5463 0.3468 0.15>
    Mesh Scale Mode=::NewEnumerator3
    Sprite Size Mode=::NewEnumerator3
    Uniform Sprite Size=50
    UsePositionOffset=true
    Write Material Random=true
   )
   &CameraOffset(Camera Offset Amount=5)
   &DynamicMaterialParameters(
    Index 0 Param 1=&Multiply_Float(A=-19, B=&MakeFloatFromVector(Vector=<0 0 0>, Channel=::NewEnumerator0))
    Index 0 Param 2=-1
    Module.Param0WriteEnabled=true
    Module.Param1WriteEnabled=true
    Module.Param2WriteEnabled=false
    Module.Param3WriteEnabled=false
   )
  ]
  particleUpdateScript [
   &ParticleState(Kill Particles When Lifetime Has Elapsed=false)
   &SocketLocation(
    Orientation Sampling=::NewEnumerator1
    SamplePosOffset=true
    Sampled Position Offset=<-1 0 1>
    Socket Reader=$System.Socket Reader
    Socket Sampling Mode=::NewEnumerator3
    VelocitySpeedLimit=false
   )
   &ScaleColor(Scale Alpha=$System.FadeIn, ScaleA=true, ScaleRGB=true)
   /SetVariables(
    Particles.FlareAlignment:Vector3f=<0 0 1>
    Particles.FlareCameraOffset=10
    Particles.FlareColor=&MultiplyLinearColor(Linear Color=$Particles.Color, Scale Factor=&MakeVector4_FromFloat(Float=2))
    Particles.FlareFacing:Vector3f=<0 1 0>
    Particles.FlarePosition:NiagaraPosition=&AddVectorToPosition(Position=$Particles.Position, Vector=<1 3 -1>)
    Particles.FlareSize:Vector2f=&Multiply_Vector2D_ByFloat(Float=0.5, Vector2D=<100 50>)
   ) {isSetParameters=true}
   /SetVariables(
    Particles.GlowCameraOffset:NiagaraFloat=&FN_FrontEndFloatModifier(Gameplay=3, Front End=0)
    Particles.GlowColor=&MultiplyLinearColor(
     Linear Color=$Particles.Color
     Scale Factor=&MakeVector4_FromFloat(Float=&Multiply_Float(
       A=&Multiply_Float(A=8, B=$System.FadeIn)
       B=&ClampFloat(
        Float=&DotProduct(Vector A=&ReturnCameraVector(Camera Vectors=::NewEnumerator1, Coordinate Space=::Local), Vector B=<0 -1 0>)
        Max=0.5
       )
      ))
    )
    Particles.GlowPosition:NiagaraPosition=&AddVectorToPosition(Position=$Particles.Position, Vector=<0.2 3 -1.2>)
    Particles.GlowSize=&Vector2DFromFloat(Value=&FN_FrontEndFloatModifier(Gameplay=10, Front End=0))
    Particles.glowalignement:Vector3f=<0 0 1>
    Particles.glowfacing:Vector3f=<1 0 0>
   ) {isSetParameters=true}
   /SetVariables(
    Particles.BigGlowColor:LinearColor=&MultiplyLinearColor(
     Linear Color=$Particles.Color
     Scale Factor=&MakeVector4_FromFloat(Float=&Multiply_Float(
       A=&Multiply_Float(A=1.2, B=$System.FadeIn)
       B=&ClampFloat(Float=&OneMinusFloat(Float=&Multiply_Float(
          A=0.4
          B=&VectorLength(
           Vector=/Niagara/DynamicInputs/Vectors/CrossProduct(
            Vector A:Vector3f=&ReturnCameraVector(Camera Vectors=::NewEnumerator1, Coordinate Space=::Simulation)
            Vector B:Vector3f=<0 -1 0>
            Normalize Incoming Vectors:NiagaraBool=true
           )
           Fallback Vector=<0 -1 0>
          )
         )))
      ))
    )
    Particles.BigGlowSize:Vector2f=&Vector2DFromFloat(Value=25)
   ) {isSetParameters=true}
   &SpriteFacingAndAlignment(Sprite Alignment=<0 -1 0>, WriteAlignment=true, WriteFacing=false) {module=SpriteFacingAndAlignment001}
   &ScaleColor(Scale Alpha=$System.FadeIn, ScaleA=true, ScaleRGB=true) {module=ScaleColor002}
   &FN_Light_Attributes(Color=$Particles.Color, Create Color Attribute=true, Create Position Attribute=true, Position=<-1 25 -1>, Scale Color=$System.FadeIn)
  ]
  rendererClasses [&NiagaraSpriteRendererProperties, &NiagaraLightRendererProperties]
  renderers [
   {rendererClass=&NiagaraSpriteRendererProperties}
   {rendererClass=&NiagaraSpriteRendererProperties, rendererIndex=1}
   {rendererClass=&NiagaraSpriteRendererProperties, rendererIndex=2}
   {rendererClass=&NiagaraSpriteRendererProperties, rendererIndex=3}
   {rendererClass=&NiagaraLightRendererProperties, rendererIndex=4}
  ]
 }}
summary {
 systemName=NS_Character_CampusSire_Idle
 emitters [{emitterName=Headlights, rendererClasses [&NiagaraSpriteRendererProperties, &NiagaraLightRendererProperties]}]
 userVariables [
  User.bIsFrontEnd {type=NiagaraBool}
  User.bIsFrontEndPreview {type=NiagaraBool}
 ]
}
