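# Control Rig for the large male CampusSire player body; the visible part measures upper-arm orientation as directional pose values.
# Spherical pose readers with tuned offsets and falloffs output Up, Down, Front, Back and in-between results.
# One function hard-codes both upper arms, using clavicle_l as parent even for the right arm.
# A generic version takes any driver bone, looks up its parent, and adds per-pose debug drawing and printing.
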ControlRigBlueprint /BRCosmetics/Characters/Player/Male/Large/Bodies/M_LRG_CampusSire/Meshes/M_LRG_CampusSire_CtrlRig
graphs [
 Add_14_ContainedGraph {nodes [
   n1 = RigVMFunctionEntryNode()
   n2 = /Script/RigVM.RigVMFunction_MathVectorAdd(A=@n1.A, B=@n1.B) {is=struct RigVMUnitNode, template="Add::Execute(in A,in B,out Result)"}
   RigVMFunctionReturnNode(Result=&RigVMFunction_MathVectorAdd(A=&RigVMFunction_MathVectorAdd(A=@n2.Result, B=@n1.C) {template="Add::Execute(in A,in B,out Result)"}.Result, B=@n1.D) {template="Add::Execute(in A,in B,out Result)"}.Result)
  ]}
 Add_14_ContainedGraph {nodes [
   n1 = RigVMFunctionEntryNode()
   RigVMFunctionReturnNode(Result=&RigVMFunction_MathVectorAdd(A=&RigVMFunction_MathVectorAdd(A=@n1.A, B=@n1.B) {template="Add::Execute(in A,in B,out Result)"}.Result, B=@n1.C) {template="Add::Execute(in A,in B,out Result)"}.Result)
  ]}
 New Function_ContainedGraph {nodes [
   RigVMFunctionEntryNode()
   > RigVMFunction_Sequence = RigVMAggregateNode()
   RigVMFunctionReturnNode(
    ExecuteContext=@RigVMFunction_Sequence.E
    Up Post L=RigVMVariableNode() {variable=Result Up Post L}.Value
    Up Post R=RigVMVariableNode() {variable=Result Up Post R}.Value
    Dn L=RigVMVariableNode() {variable=Result Dn L}.Value
    Dn R=RigVMVariableNode() {variable=Result Dn R}.Value
    Dn Pre L=RigVMVariableNode() {variable=Result Dn Pre L}.Value
    Dn Pre R=RigVMVariableNode() {variable=Result Dn Pre R}.Value
    Dn Post L=RigVMVariableNode() {variable=Result Dn Post L}.Value
    Dn Post R=RigVMVariableNode() {variable=Result Dn Post R}.Value
    Fr L=RigVMVariableNode() {variable=Result Fr L}.Value
    Fr R=RigVMVariableNode() {variable=Result Fr R}.Value
    Fr Pre L=RigVMVariableNode() {variable=Result Fr Pre L}.Value
    Fr Pre R=RigVMVariableNode() {variable=Result Fr Pre R}.Value
    Fr Post L=RigVMVariableNode() {variable=Result Fr Post L}.Value
    Fr Post R=RigVMVariableNode() {variable=Result Fr Post R}.Value
    Fr Up L=RigVMVariableNode() {variable=Result Fr Up L}.Value
    Fr Up R=RigVMVariableNode() {variable=Result Fr Up R}.Value
    Fr Dn L=RigVMVariableNode() {variable=Result Fr Dn L}.Value
    Fr Dn R=RigVMVariableNode() {variable=Result Fr Dn R}.Value
    Ba L=RigVMVariableNode() {variable=Result Ba L}.Value
    Ba R=RigVMVariableNode() {variable=Result Ba R}.Value
    Ba Pre L=RigVMVariableNode() {variable=Result Ba Pre L}.Value
    Ba Pre R=RigVMVariableNode() {variable=Result Ba Pre R}.Value
    Ba Post L=RigVMVariableNode() {variable=Result Ba Post L}.Value
    Ba Post R=RigVMVariableNode() {variable=Result Ba Post R}.Value
    Up L=RigVMVariableNode() {variable=Result Up L}.Value
    Ba Up L=RigVMVariableNode() {variable=Result Ba Up L}.Value
    Ba Up R=RigVMVariableNode() {variable=Result Ba Up R}.Value
    Ba Dn L=RigVMVariableNode() {variable=Result Ba Dn L}.Value
    Ba Dn R=RigVMVariableNode() {variable=Result Ba Dn R}.Value
    Up R=RigVMVariableNode() {variable=Result Up R}.Value
    Up Pre L=RigVMVariableNode() {variable=Result Up Pre L}.Value
    Up Pre R=RigVMVariableNode() {variable=Result Up Pre R}.Value
   )
   n1 = &RigUnit_SphericalPoseReader(
    ActiveRegionScaleFactors=(PositiveWidth=1,NegativeWidth=1,PositiveHeight=1,NegativeHeight=1)
    ActiveRegionSize=0.001
    Debug=(bDrawDebug=false,bDraw2D=False,bDrawLocalAxes=False,DebugScale=25,DebugSegments=20,DebugThickness=0.05)
    DriverAxis=(X=1,Y=0,Z=0)
    DriverItem=(Type=Bone,Name="upperarm_l")
    FalloffRegionScaleFactors=(PositiveWidth=1.3471,NegativeWidth=1,PositiveHeight=1,NegativeHeight=1)
    FalloffSize=0.5
    FlipHeightScaling=False
    FlipWidthScaling=False
    OptionalParentItem=(Type=Bone,Name="clavicle_l")
    RotationOffset=(X=0,Y=31.239,Z=0)
    ExecutePin=@RigVMFunction_Sequence.A
   )
   > RigVMVariableNode(Value=@n1.OutputParam) {variable=Result Up L}
   > n2 = &RigUnit_SphericalPoseReader(
    ActiveRegionScaleFactors=(PositiveWidth=1,NegativeWidth=1,PositiveHeight=1,NegativeHeight=1)
    ActiveRegionSize=0.001
    Debug=(bDrawDebug=false,bDraw2D=False,bDrawLocalAxes=False,DebugScale=25,DebugSegments=20,DebugThickness=0.05)
    DriverAxis=(X=1,Y=0,Z=0)
    DriverItem=(Type=Bone,Name="upperarm_r")
    FalloffRegionScaleFactors=(PositiveWidth=1.3471,NegativeWidth=1,PositiveHeight=1,NegativeHeight=1)
    FalloffSize=0.5
    FlipHeightScaling=False
    FlipWidthScaling=False
    OptionalParentItem=(Type=Bone,Name="clavicle_l")
    RotationOffset=(X=0,Y=31.239,Z=0)
   )
   > RigVMVariableNode(Value=@n2.OutputParam) {variable=Result Up R}
   > n3 = &RigUnit_SphericalPoseReader(
    ActiveRegionScaleFactors=(PositiveWidth=1,NegativeWidth=1,PositiveHeight=1,NegativeHeight=1)
    ActiveRegionSize=0.001
    Debug=(bDrawDebug=false,bDraw2D=False,bDrawLocalAxes=False,DebugScale=25,DebugSegments=20,DebugThickness=0.05)
    DriverAxis=(X=1,Y=0,Z=0)
    DriverItem=(Type=Bone,Name="upperarm_l")
    FalloffRegionScaleFactors=(PositiveWidth=1.6942,NegativeWidth=1,PositiveHeight=2,NegativeHeight=2)
    FalloffSize=0.25
    FlipHeightScaling=False
    FlipWidthScaling=False
    OptionalParentItem=(Type=Bone,Name="clavicle_l")
    RotationOffset=(X=0,Y=-13.761,Z=0)
   )
   > RigVMVariableNode(Value=@n3.OutputParam) {variable=Result Up Pre L}
   > n4 = &RigUnit_SphericalPoseReader(
    ActiveRegionScaleFactors=(PositiveWidth=1,NegativeWidth=1,PositiveHeight=1,NegativeHeight=1)
    ActiveRegionSize=0.001
    Debug=(bDrawDebug=false,bDraw2D=False,bDrawLocalAxes=False,DebugScale=25,DebugSegments=20,DebugThickness=0.05)
    DriverAxis=(X=1,Y=0,Z=0)
    DriverItem=(Type=Bone,Name="upperarm_r")
    FalloffRegionScaleFactors=(PositiveWidth=1.6942,NegativeWidth=1,PositiveHeight=2,NegativeHeight=2)
    FalloffSize=0.25
    FlipHeightScaling=False
    FlipWidthScaling=False
    OptionalParentItem=(Type=Bone,Name="clavicle_l")
    RotationOffset=(X=0,Y=-13.761,Z=0)
   )
   > RigVMVariableNode(Value=@n4.OutputParam) {variable=Result Up Pre R}
   > n5 = &RigUnit_SphericalPoseReader(
    ActiveRegionScaleFactors=(PositiveWidth=1,NegativeWidth=1,PositiveHeight=1,NegativeHeight=1)
    ActiveRegionSize=0.001
    Debug=(bDrawDebug=false,bDraw2D=False,bDrawLocalAxes=False,DebugScale=25,DebugSegments=20,DebugThickness=0.05)
    DriverAxis=(X=1,Y=0,Z=0)
    DriverItem=(Type=Bone,Name="upperarm_l")
    FalloffRegionScaleFactors=(PositiveWidth=0.5,NegativeWidth=0.5,PositiveHeight=1,NegativeHeight=1)
    FalloffSize=0.5
    FlipHeightScaling=False
    FlipWidthScaling=False
    OptionalParentItem=(Type=Bone,Name="clavicle_l")
    RotationOffset=(X=0,Y=76.239,Z=0)
   )
   > RigVMVariableNode(Value=@n5.OutputParam) {variable=Result Up Post L}
   > n6 = &RigUnit_SphericalPoseReader(
    ActiveRegionScaleFactors=(PositiveWidth=1,NegativeWidth=1,PositiveHeight=1,NegativeHeight=1)
    ActiveRegionSize=0.001
    Debug=(bDrawDebug=false,bDraw2D=False,bDrawLocalAxes=False,DebugScale=25,DebugSegments=20,DebugThickness=0.05)
    DriverAxis=(X=1,Y=0,Z=0)
    DriverItem=(Type=Bone,Name="upperarm_r")
    FalloffRegionScaleFactors=(PositiveWidth=0.5,NegativeWidth=0.5,PositiveHeight=1,NegativeHeight=1)
    FalloffSize=0.5
    FlipHeightScaling=False
    FlipWidthScaling=False
    OptionalParentItem=(Type=Bone,Name="clavicle_l")
    RotationOffset=(X=0,Y=76.239,Z=0)
   )
   > RigVMVariableNode(Value=@n6.OutputParam) {variable=Result Up Post R}
   n7 = &RigUnit_SphericalPoseReader(
    ActiveRegionScaleFactors=(PositiveWidth=1,NegativeWidth=1,PositiveHeight=1,NegativeHeight=1)
    ActiveRegionSize=0.001
    Debug=(bDrawDebug=false,bDraw2D=False,bDrawLocalAxes=False,DebugScale=25,DebugSegments=20,DebugThickness=0.05)
    DriverAxis=(X=1,Y=0,Z=0)
    DriverItem=(Type=Bone,Name="upperarm_l")
    FalloffRegionScaleFactors=(PositiveWidth=1,NegativeWidth=1,PositiveHeight=1,NegativeHeight=1)
    FalloffSize=0.5
    FlipHeightScaling=False
    FlipWidthScaling=False
    OptionalParentItem=(Type=Bone,Name="clavicle_l")
    RotationOffset=(X=90,Y=31.239,Z=0)
    ExecutePin=@RigVMFunction_Sequence.B
   )
   > RigVMVariableNode(Value=@n7.OutputParam) {variable=Result Fr L}
   > n8 = &RigUnit_SphericalPoseReader(
    ActiveRegionScaleFactors=(PositiveWidth=1,NegativeWidth=1,PositiveHeight=1,NegativeHeight=1)
    ActiveRegionSize=0.001
    Debug=(bDrawDebug=false,bDraw2D=False,bDrawLocalAxes=False,DebugScale=25,DebugSegments=20,DebugThickness=0.05)
    DriverAxis=(X=1,Y=0,Z=0)
    DriverItem=(Type=Bone,Name="upperarm_r")
    FalloffRegionScaleFactors=(PositiveWidth=1,NegativeWidth=1,PositiveHeight=1,NegativeHeight=1)
    FalloffSize=0.5
    FlipHeightScaling=False
    FlipWidthScaling=False
    OptionalParentItem=(Type=Bone,Name="clavicle_l")
    RotationOffset=(X=90,Y=31.239,Z=0)
   )
   > RigVMVariableNode(Value=@n8.OutputParam) {variable=Result Fr R}
   > n9 = &RigUnit_SphericalPoseReader(
    ActiveRegionScaleFactors=(PositiveWidth=1,NegativeWidth=1,PositiveHeight=1,NegativeHeight=1)
    ActiveRegionSize=0.001
    Debug=(bDrawDebug=false,bDraw2D=False,bDrawLocalAxes=False,DebugScale=25,DebugSegments=20,DebugThickness=0.05)
    DriverAxis=(X=1,Y=0,Z=0)
    DriverItem=(Type=Bone,Name="upperarm_l")
    FalloffRegionScaleFactors=(PositiveWidth=2,NegativeWidth=2,PositiveHeight=1,NegativeHeight=1)
    FalloffSize=0.25
    FlipHeightScaling=False
    FlipWidthScaling=False
    OptionalParentItem=(Type=Bone,Name="clavicle_l")
    RotationOffset=(X=45,Y=-58.761,Z=0)
   )
   > RigVMVariableNode(Value=@n9.OutputParam) {variable=Result Fr Pre L}
   > n10 = &RigUnit_SphericalPoseReader(
    ActiveRegionScaleFactors=(PositiveWidth=1,NegativeWidth=1,PositiveHeight=1,NegativeHeight=1)
    ActiveRegionSize=0.001
    Debug=(bDrawDebug=false,bDraw2D=False,bDrawLocalAxes=False,DebugScale=25,DebugSegments=20,DebugThickness=0.05)
    DriverAxis=(X=1,Y=0,Z=0)
    DriverItem=(Type=Bone,Name="upperarm_r")
    FalloffRegionScaleFactors=(PositiveWidth=2,NegativeWidth=2,PositiveHeight=1,NegativeHeight=1)
    FalloffSize=0.25
    FlipHeightScaling=False
    FlipWidthScaling=False
    OptionalParentItem=(Type=Bone,Name="clavicle_l")
    RotationOffset=(X=45,Y=-58.761,Z=0)
   )
   > RigVMVariableNode(Value=@n10.OutputParam) {variable=Result Fr Pre R}
   > n11 = &RigUnit_SphericalPoseReader(
    ActiveRegionScaleFactors=(PositiveWidth=1,NegativeWidth=1,PositiveHeight=1,NegativeHeight=1)
    ActiveRegionSize=0.001
    Debug=(bDrawDebug=false,bDraw2D=False,bDrawLocalAxes=False,DebugScale=25,DebugSegments=20,DebugThickness=0.05)
    DriverAxis=(X=1,Y=0,Z=0)
    DriverItem=(Type=Bone,Name="upperarm_l")
    FalloffRegionScaleFactors=(PositiveWidth=2,NegativeWidth=2,PositiveHeight=1,NegativeHeight=1)
    FalloffSize=0.25
    FlipHeightScaling=False
    FlipWidthScaling=False
    OptionalParentItem=(Type=Bone,Name="clavicle_l")
    RotationOffset=(X=135,Y=-58.761,Z=0)
   )
   > RigVMVariableNode(Value=@n11.OutputParam) {variable=Result Fr Post L}
   > n12 = &RigUnit_SphericalPoseReader(
    ActiveRegionScaleFactors=(PositiveWidth=1,NegativeWidth=1,PositiveHeight=1,NegativeHeight=1)
    ActiveRegionSize=0.001
    Debug=(bDrawDebug=false,bDraw2D=False,bDrawLocalAxes=False,DebugScale=25,DebugSegments=20,DebugThickness=0.05)
    DriverAxis=(X=1,Y=0,Z=0)
    DriverItem=(Type=Bone,Name="upperarm_r")
    FalloffRegionScaleFactors=(PositiveWidth=2,NegativeWidth=2,PositiveHeight=1,NegativeHeight=1)
    FalloffSize=0.25
    FlipHeightScaling=False
    FlipWidthScaling=False
    OptionalParentItem=(Type=Bone,Name="clavicle_l")
    RotationOffset=(X=135,Y=-58.761,Z=0)
   )
   > RigVMVariableNode(Value=@n12.OutputParam) {variable=Result Fr Post R}
   > n13 = &RigUnit_SphericalPoseReader(
    ActiveRegionScaleFactors=(PositiveWidth=1,NegativeWidth=1,PositiveHeight=1,NegativeHeight=1)
    ActiveRegionSize=0.001
    Debug=(bDrawDebug=false,bDraw2D=False,bDrawLocalAxes=False,DebugScale=25,DebugSegments=20,DebugThickness=0.05)
    DriverAxis=(X=1,Y=0,Z=0)
    DriverItem=(Type=Bone,Name="upperarm_l")
    FalloffRegionScaleFactors=(PositiveWidth=1,NegativeWidth=1,PositiveHeight=0.5,NegativeHeight=0.5)
    FalloffSize=0.5
    FlipHeightScaling=False
    FlipWidthScaling=False
    OptionalParentItem=(Type=Bone,Name="clavicle_l")
    RotationOffset=(X=45,Y=31.239,Z=0)
   )
   > RigVMVariableNode(Value=@n13.OutputParam) {variable=Result Fr Up L}
   > n14 = &RigUnit_SphericalPoseReader(
    ActiveRegionScaleFactors=(PositiveWidth=1,NegativeWidth=1,PositiveHeight=1,NegativeHeight=1)
    ActiveRegionSize=0.001
    Debug=(bDrawDebug=false,bDraw2D=False,bDrawLocalAxes=False,DebugScale=25,DebugSegments=20,DebugThickness=0.05)
    DriverAxis=(X=1,Y=0,Z=0)
    DriverItem=(Type=Bone,Name="upperarm_r")
    FalloffRegionScaleFactors=(PositiveWidth=1,NegativeWidth=1,PositiveHeight=0.5,NegativeHeight=0.5)
    FalloffSize=0.5
    FlipHeightScaling=False
    FlipWidthScaling=False
    OptionalParentItem=(Type=Bone,Name="clavicle_l")
    RotationOffset=(X=45,Y=31.239,Z=0)
   )
   > RigVMVariableNode(Value=@n14.OutputParam) {variable=Result Fr Up R}
   > n15 = &RigUnit_SphericalPoseReader(
    ActiveRegionScaleFactors=(PositiveWidth=1,NegativeWidth=1,PositiveHeight=1,NegativeHeight=1)
    ActiveRegionSize=0.001
    Debug=(bDrawDebug=false,bDraw2D=False,bDrawLocalAxes=False,DebugScale=25,DebugSegments=20,DebugThickness=0.05)
    DriverAxis=(X=1,Y=0,Z=0)
    DriverItem=(Type=Bone,Name="upperarm_l")
    FalloffRegionScaleFactors=(PositiveWidth=1,NegativeWidth=1,PositiveHeight=0.5,NegativeHeight=0.5)
    FalloffSize=0.5
    FlipHeightScaling=False
    FlipWidthScaling=False
    OptionalParentItem=(Type=Bone,Name="clavicle_l")
    RotationOffset=(X=135,Y=31.239,Z=0)
   )
   > RigVMVariableNode(Value=@n15.OutputParam) {variable=Result Fr Dn L}
   > n16 = &RigUnit_SphericalPoseReader(
    ActiveRegionScaleFactors=(PositiveWidth=1,NegativeWidth=1,PositiveHeight=1,NegativeHeight=1)
    ActiveRegionSize=0.001
    Debug=(bDrawDebug=false,bDraw2D=False,bDrawLocalAxes=False,DebugScale=25,DebugSegments=20,DebugThickness=0.05)
    DriverAxis=(X=1,Y=0,Z=0)
    DriverItem=(Type=Bone,Name="upperarm_r")
    FalloffRegionScaleFactors=(PositiveWidth=1,NegativeWidth=1,PositiveHeight=0.5,NegativeHeight=0.5)
    FalloffSize=0.5
    FlipHeightScaling=False
    FlipWidthScaling=False
    OptionalParentItem=(Type=Bone,Name="clavicle_l")
    RotationOffset=(X=135,Y=31.239,Z=0)
   )
   > RigVMVariableNode(Value=@n16.OutputParam) {variable=Result Fr Dn R}
   n17 = &RigUnit_SphericalPoseReader(
    ActiveRegionScaleFactors=(PositiveWidth=1,NegativeWidth=1,PositiveHeight=1,NegativeHeight=1)
    ActiveRegionSize=0.001
    Debug=(bDrawDebug=false,bDraw2D=False,bDrawLocalAxes=False,DebugScale=25,DebugSegments=20,DebugThickness=0.05)
    DriverAxis=(X=1,Y=0,Z=0)
    DriverItem=(Type=Bone,Name="upperarm_l")
    FalloffRegionScaleFactors=(PositiveWidth=1,NegativeWidth=1,PositiveHeight=1,NegativeHeight=1)
    FalloffSize=0.5
    FlipHeightScaling=False
    FlipWidthScaling=False
    OptionalParentItem=(Type=Bone,Name="clavicle_l")
    RotationOffset=(X=-90,Y=31.239,Z=0)
    ExecutePin=@RigVMFunction_Sequence.C
   )
   > RigVMVariableNode(Value=@n17.OutputParam) {variable=Result Ba L}
   > n18 = &RigUnit_SphericalPoseReader(
    ActiveRegionScaleFactors=(PositiveWidth=1,NegativeWidth=1,PositiveHeight=1,NegativeHeight=1)
    ActiveRegionSize=0.001
    Debug=(bDrawDebug=false,bDraw2D=False,bDrawLocalAxes=False,DebugScale=25,DebugSegments=20,DebugThickness=0.05)
    DriverAxis=(X=1,Y=0,Z=0)
    DriverItem=(Type=Bone,Name="upperarm_r")
    FalloffRegionScaleFactors=(PositiveWidth=1,NegativeWidth=1,PositiveHeight=1,NegativeHeight=1)
    FalloffSize=0.5
    FlipHeightScaling=False
    FlipWidthScaling=False
    OptionalParentItem=(Type=Bone,Name="clavicle_l")
    RotationOffset=(X=-90,Y=31.239,Z=0)
   )
   > RigVMVariableNode(Value=@n18.OutputParam) {variable=Result Ba R}
   > n19 = &RigUnit_SphericalPoseReader(
    ActiveRegionScaleFactors=(PositiveWidth=1,NegativeWidth=1,PositiveHeight=1,NegativeHeight=1)
    ActiveRegionSize=0.001
    Debug=(bDrawDebug=false,bDraw2D=False,bDrawLocalAxes=False,DebugScale=25,DebugSegments=20,DebugThickness=0.05)
    DriverAxis=(X=1,Y=0,Z=0)
    DriverItem=(Type=Bone,Name="upperarm_l")
    FalloffRegionScaleFactors=(PositiveWidth=2,NegativeWidth=2,PositiveHeight=1,NegativeHeight=1)
    FalloffSize=0.25
    FlipHeightScaling=False
    FlipWidthScaling=False
    OptionalParentItem=(Type=Bone,Name="clavicle_l")
    RotationOffset=(X=-45,Y=-58.761,Z=0)
   )
   > RigVMVariableNode(Value=@n19.OutputParam) {variable=Result Ba Pre L}
   > n20 = &RigUnit_SphericalPoseReader(
    ActiveRegionScaleFactors=(PositiveWidth=1,NegativeWidth=1,PositiveHeight=1,NegativeHeight=1)
    ActiveRegionSize=0.001
    Debug=(bDrawDebug=false,bDraw2D=False,bDrawLocalAxes=False,DebugScale=25,DebugSegments=20,DebugThickness=0.05)
    DriverAxis=(X=1,Y=0,Z=0)
    DriverItem=(Type=Bone,Name="upperarm_r")
    FalloffRegionScaleFactors=(PositiveWidth=2,NegativeWidth=2,PositiveHeight=1,NegativeHeight=1)
    FalloffSize=0.25
    FlipHeightScaling=False
    FlipWidthScaling=False
    OptionalParentItem=(Type=Bone,Name="clavicle_l")
    RotationOffset=(X=-45,Y=-58.761,Z=0)
   )
   > RigVMVariableNode(Value=@n20.OutputParam) {variable=Result Ba Pre R}
   > n21 = &RigUnit_SphericalPoseReader(
    ActiveRegionScaleFactors=(PositiveWidth=1,NegativeWidth=1,PositiveHeight=1,NegativeHeight=1)
    ActiveRegionSize=0.001
    Debug=(bDrawDebug=false,bDraw2D=False,bDrawLocalAxes=False,DebugScale=25,DebugSegments=20,DebugThickness=0.05)
    DriverAxis=(X=1,Y=0,Z=0)
    DriverItem=(Type=Bone,Name="upperarm_l")
    FalloffRegionScaleFactors=(PositiveWidth=2,NegativeWidth=2,PositiveHeight=1,NegativeHeight=1)
    FalloffSize=0.25
    FlipHeightScaling=False
    FlipWidthScaling=False
    OptionalParentItem=(Type=Bone,Name="clavicle_l")
    RotationOffset=(X=-135,Y=-58.761,Z=0)
   )
   > RigVMVariableNode(Value=@n21.OutputParam) {variable=Result Ba Post L}
   > n22 = &RigUnit_SphericalPoseReader(
    ActiveRegionScaleFactors=(PositiveWidth=1,NegativeWidth=1,PositiveHeight=1,NegativeHeight=1)
    ActiveRegionSize=0.001
    Debug=(bDrawDebug=false,bDraw2D=False,bDrawLocalAxes=False,DebugScale=25,DebugSegments=20,DebugThickness=0.05)
    DriverAxis=(X=1,Y=0,Z=0)
    DriverItem=(Type=Bone,Name="upperarm_r")
    FalloffRegionScaleFactors=(PositiveWidth=2,NegativeWidth=2,PositiveHeight=1,NegativeHeight=1)
    FalloffSize=0.25
    FlipHeightScaling=False
    FlipWidthScaling=False
    OptionalParentItem=(Type=Bone,Name="clavicle_l")
    RotationOffset=(X=-135,Y=-58.761,Z=0)
   )
   > RigVMVariableNode(Value=@n22.OutputParam) {variable=Result Ba Post R}
   > n23 = &RigUnit_SphericalPoseReader(
    ActiveRegionScaleFactors=(PositiveWidth=1,NegativeWidth=1,PositiveHeight=1,NegativeHeight=1)
    ActiveRegionSize=0.001
    Debug=(bDrawDebug=false,bDraw2D=False,bDrawLocalAxes=False,DebugScale=25,DebugSegments=20,DebugThickness=0.05)
    DriverAxis=(X=1,Y=0,Z=0)
    DriverItem=(Type=Bone,Name="upperarm_l")
    FalloffRegionScaleFactors=(PositiveWidth=1,NegativeWidth=1,PositiveHeight=0.5,NegativeHeight=0.5)
    FalloffSize=0.5
    FlipHeightScaling=False
    FlipWidthScaling=False
    OptionalParentItem=(Type=Bone,Name="clavicle_l")
    RotationOffset=(X=-45,Y=31.239,Z=0)
   )
   > RigVMVariableNode(Value=@n23.OutputParam) {variable=Result Ba Up L}
   > n24 = &RigUnit_SphericalPoseReader(
    ActiveRegionScaleFactors=(PositiveWidth=1,NegativeWidth=1,PositiveHeight=1,NegativeHeight=1)
    ActiveRegionSize=0.001
    Debug=(bDrawDebug=false,bDraw2D=False,bDrawLocalAxes=False,DebugScale=25,DebugSegments=20,DebugThickness=0.05)
    DriverAxis=(X=1,Y=0,Z=0)
    DriverItem=(Type=Bone,Name="upperarm_r")
    FalloffRegionScaleFactors=(PositiveWidth=1,NegativeWidth=1,PositiveHeight=0.5,NegativeHeight=0.5)
    FalloffSize=0.5
    FlipHeightScaling=False
    FlipWidthScaling=False
    OptionalParentItem=(Type=Bone,Name="clavicle_l")
    RotationOffset=(X=-45,Y=31.239,Z=0)
   )
   > RigVMVariableNode(Value=@n24.OutputParam) {variable=Result Ba Up R}
   > n25 = &RigUnit_SphericalPoseReader(
    ActiveRegionScaleFactors=(PositiveWidth=1,NegativeWidth=1,PositiveHeight=1,NegativeHeight=1)
    ActiveRegionSize=0.001
    Debug=(bDrawDebug=false,bDraw2D=False,bDrawLocalAxes=False,DebugScale=25,DebugSegments=20,DebugThickness=0.05)
    DriverAxis=(X=1,Y=0,Z=0)
    DriverItem=(Type=Bone,Name="upperarm_l")
    FalloffRegionScaleFactors=(PositiveWidth=1,NegativeWidth=1,PositiveHeight=0.5,NegativeHeight=0.5)
    FalloffSize=0.5
    FlipHeightScaling=False
    FlipWidthScaling=False
    OptionalParentItem=(Type=Bone,Name="clavicle_l")
    RotationOffset=(X=-135,Y=31.239,Z=0)
   )
   > RigVMVariableNode(Value=@n25.OutputParam) {variable=Result Ba Dn L}
   > n26 = &RigUnit_SphericalPoseReader(
    ActiveRegionScaleFactors=(PositiveWidth=1,NegativeWidth=1,PositiveHeight=1,NegativeHeight=1)
    ActiveRegionSize=0.001
    Debug=(bDrawDebug=false,bDraw2D=False,bDrawLocalAxes=False,DebugScale=25,DebugSegments=20,DebugThickness=0.05)
    DriverAxis=(X=1,Y=0,Z=0)
    DriverItem=(Type=Bone,Name="upperarm_r")
    FalloffRegionScaleFactors=(PositiveWidth=1,NegativeWidth=1,PositiveHeight=0.5,NegativeHeight=0.5)
    FalloffSize=0.5
    FlipHeightScaling=False
    FlipWidthScaling=False
    OptionalParentItem=(Type=Bone,Name="clavicle_l")
    RotationOffset=(X=-135,Y=31.239,Z=0)
   )
   > RigVMVariableNode(Value=@n26.OutputParam) {variable=Result Ba Dn R}
   n27 = &RigUnit_SphericalPoseReader(
    ActiveRegionScaleFactors=(PositiveWidth=1,NegativeWidth=1,PositiveHeight=1,NegativeHeight=1)
    ActiveRegionSize=0.001
    Debug=(bDrawDebug=false,bDraw2D=False,bDrawLocalAxes=False,DebugScale=25,DebugSegments=20,DebugThickness=0.05)
    DriverAxis=(X=1,Y=0,Z=0)
    DriverItem=(Type=Bone,Name="upperarm_l")
    FalloffRegionScaleFactors=(PositiveWidth=1,NegativeWidth=0.6529,PositiveHeight=1,NegativeHeight=1)
    FalloffSize=0.5
    FlipHeightScaling=False
    FlipWidthScaling=False
    OptionalParentItem=(Type=Bone,Name="clavicle_l")
    RotationOffset=(X=0,Y=-148.761,Z=0)
    ExecutePin=@RigVMFunction_Sequence.D
   )
   > RigVMVariableNode(Value=@n27.OutputParam) {variable=Result Dn L}
   > n28 = &RigUnit_SphericalPoseReader(
    ActiveRegionScaleFactors=(PositiveWidth=1,NegativeWidth=1,PositiveHeight=1,NegativeHeight=1)
    ActiveRegionSize=0.001
    Debug=(bDrawDebug=false,bDraw2D=False,bDrawLocalAxes=False,DebugScale=25,DebugSegments=20,DebugThickness=0.05)
    DriverAxis=(X=1,Y=0,Z=0)
    DriverItem=(Type=Bone,Name="upperarm_r")
    FalloffRegionScaleFactors=(PositiveWidth=1,NegativeWidth=0.6529,PositiveHeight=1,NegativeHeight=1)
    FalloffSize=0.5
    FlipHeightScaling=False
    FlipWidthScaling=False
    OptionalParentItem=(Type=Bone,Name="clavicle_l")
    RotationOffset=(X=0,Y=-148.761,Z=0)
   )
   > RigVMVariableNode(Value=@n28.OutputParam) {variable=Result Dn R}
   > n29 = &RigUnit_SphericalPoseReader(
    ActiveRegionScaleFactors=(PositiveWidth=1,NegativeWidth=1,PositiveHeight=1,NegativeHeight=1)
    ActiveRegionSize=0.001
    Debug=(bDrawDebug=false,bDraw2D=False,bDrawLocalAxes=False,DebugScale=25,DebugSegments=20,DebugThickness=0.05)
    DriverAxis=(X=1,Y=0,Z=0)
    DriverItem=(Type=Bone,Name="upperarm_l")
    FalloffRegionScaleFactors=(PositiveWidth=1,NegativeWidth=0.3471,PositiveHeight=2,NegativeHeight=2)
    FalloffSize=0.25
    FlipHeightScaling=False
    FlipWidthScaling=False
    OptionalParentItem=(Type=Bone,Name="clavicle_l")
    RotationOffset=(X=0,Y=-103.761,Z=0)
   )
   > RigVMVariableNode(Value=@n29.OutputParam) {variable=Result Dn Pre L}
   > n30 = &RigUnit_SphericalPoseReader(
    ActiveRegionScaleFactors=(PositiveWidth=1,NegativeWidth=1,PositiveHeight=1,NegativeHeight=1)
    ActiveRegionSize=0.001
    Debug=(bDrawDebug=false,bDraw2D=False,bDrawLocalAxes=False,DebugScale=25,DebugSegments=20,DebugThickness=0.05)
    DriverAxis=(X=1,Y=0,Z=0)
    DriverItem=(Type=Bone,Name="upperarm_r")
    FalloffRegionScaleFactors=(PositiveWidth=1,NegativeWidth=0.3471,PositiveHeight=2,NegativeHeight=2)
    FalloffSize=0.25
    FlipHeightScaling=False
    FlipWidthScaling=False
    OptionalParentItem=(Type=Bone,Name="clavicle_l")
    RotationOffset=(X=0,Y=-103.761,Z=0)
   )
   > RigVMVariableNode(Value=@n30.OutputParam) {variable=Result Dn Pre R}
   > n31 = &RigUnit_SphericalPoseReader(
    ActiveRegionScaleFactors=(PositiveWidth=1,NegativeWidth=1,PositiveHeight=1,NegativeHeight=1)
    ActiveRegionSize=0.001
    Debug=(bDrawDebug=false,bDraw2D=False,bDrawLocalAxes=False,DebugScale=25,DebugSegments=20,DebugThickness=0.05)
    DriverAxis=(X=1,Y=0,Z=0)
    DriverItem=(Type=Bone,Name="upperarm_l")
    FalloffRegionScaleFactors=(PositiveWidth=1,NegativeWidth=1,PositiveHeight=2,NegativeHeight=2)
    FalloffSize=0.25
    FlipHeightScaling=False
    FlipWidthScaling=False
    OptionalParentItem=(Type=Bone,Name="clavicle_l")
    RotationOffset=(X=0,Y=-193.761,Z=0)
   )
   > RigVMVariableNode(Value=@n31.OutputParam) {variable=Result Dn Post L}
   > n32 = &RigUnit_SphericalPoseReader(
    ActiveRegionScaleFactors=(PositiveWidth=1,NegativeWidth=1,PositiveHeight=1,NegativeHeight=1)
    ActiveRegionSize=0.001
    Debug=(bDrawDebug=false,bDraw2D=False,bDrawLocalAxes=False,DebugScale=25,DebugSegments=20,DebugThickness=0.05)
    DriverAxis=(X=1,Y=0,Z=0)
    DriverItem=(Type=Bone,Name="upperarm_r")
    FalloffRegionScaleFactors=(PositiveWidth=1,NegativeWidth=1,PositiveHeight=2,NegativeHeight=2)
    FalloffSize=0.25
    FlipHeightScaling=False
    FlipWidthScaling=False
    OptionalParentItem=(Type=Bone,Name="clavicle_l")
    RotationOffset=(X=0,Y=-193.761,Z=0)
   )
   > RigVMVariableNode(Value=@n32.OutputParam) {variable=Result Dn Post R}
  ]}
 New Function_ContainedGraph {nodes [
   n1 = RigVMFunctionEntryNode()
   > RigVMVariableNode(Value=/Script/ControlRig.RigUnit_HierarchyGetParent(bDefaultParent=True, Child=@n1.Driver Item) {is=struct RigVMUnitNode}.Parent) {variable=Driver Item Parent}
   To_String = RigVMDispatchNode(Value=@n1.Driver Item.Name) {template="DISPATCH_RigDispatch_ToString(in Value,out Result)"}
   > RigVMVariableNode(Value=@To_String.Result) {variable=Driver Item Name}
   > RigVMFunction_Sequence = RigVMAggregateNode()
   RigVMFunctionReturnNode(
    ExecuteContext=@RigVMFunction_Sequence.F
    Up Post=RigVMVariableNode() {variable=Result Up Post}.Value
    Aim Up In=RigVMVariableNode() {variable=Result Aim Up In}.Value
    Aim Up Ot=RigVMVariableNode() {variable=Result Aim Up Ot}.Value
    Aim Dn In=RigVMVariableNode() {variable=Result Aim Dn In}.Value
    Aim Dn Ot=RigVMVariableNode() {variable=Result Aim Dn Ot}.Value
    Dn=RigVMVariableNode() {variable=Result Dn}.Value
    Dn Pre=RigVMVariableNode() {variable=Result Dn Pre}.Value
    Dn Post=RigVMVariableNode() {variable=Result Dn Post}.Value
    Fr=RigVMVariableNode() {variable=Result Fr}.Value
    Fr Pre=RigVMVariableNode() {variable=Result Fr Pre}.Value
    Fr Post=RigVMVariableNode() {variable=Result Fr Post}.Value
    Fr Up=RigVMVariableNode() {variable=Result Fr Up}.Value
    Fr Dn=RigVMVariableNode() {variable=Result Fr Dn}.Value
    Ba=RigVMVariableNode() {variable=Result Ba}.Value
    Ba Pre=RigVMVariableNode() {variable=Result Ba Pre}.Value
    Ba Post=RigVMVariableNode() {variable=Result Ba Post}.Value
    Up=RigVMVariableNode() {variable=Result Up}.Value
    Ba Up=RigVMVariableNode() {variable=Result Ba Up}.Value
    Ba Dn=RigVMVariableNode() {variable=Result Ba Dn}.Value
    Up Pre=RigVMVariableNode() {variable=Result Up Pre}.Value
   )
   n2 = &RigUnit_SphericalPoseReader(
    ActiveRegionScaleFactors=(PositiveWidth=1,NegativeWidth=1,PositiveHeight=1,NegativeHeight=1)
    ActiveRegionSize=0.001
    Debug=(bDrawDebug=false,bDraw2D=False,bDrawLocalAxes=False,DebugScale=25,DebugSegments=20,DebugThickness=0.05)
    DriverAxis=(X=1,Y=0,Z=0)
    FalloffRegionScaleFactors=(PositiveWidth=1.3471,NegativeWidth=1,PositiveHeight=1,NegativeHeight=1)
    FalloffSize=0.5
    FlipHeightScaling=False
    FlipWidthScaling=False
    RotationOffset=(X=0,Y=31.239,Z=0)
    Debug.bDrawDebug=RigVMVariableNode() {variable=Up Viz}.Value
    DriverItem=RigVMVariableNode() {variable=Driver Item}.Value
    OptionalParentItem=RigVMVariableNode() {variable=Driver Item Parent}.Value
    ExecutePin=@RigVMFunction_Sequence.A
   )
   > n3 = RigVMVariableNode(Value=@n2.OutputParam) {variable=Result Up}
   RigVMFunction_ControlFlowBranch = &RigVMFunction_ControlFlowBranch(ExecuteContext=@n3.ExecuteContext, Condition=RigVMVariableNode() {variable=Up Viz}.Value)
   n4 = &RigUnit_SphericalPoseReader(
    ActiveRegionScaleFactors=(PositiveWidth=1,NegativeWidth=1,PositiveHeight=1,NegativeHeight=1)
    ActiveRegionSize=0.001
    Debug=(bDrawDebug=false,bDraw2D=False,bDrawLocalAxes=False,DebugScale=25,DebugSegments=20,DebugThickness=0.05)
    DriverAxis=(X=1,Y=0,Z=0)
    FalloffRegionScaleFactors=(PositiveWidth=1.6942,NegativeWidth=1,PositiveHeight=2,NegativeHeight=2)
    FalloffSize=0.25
    FlipHeightScaling=False
    FlipWidthScaling=False
    RotationOffset=(X=0,Y=-13.761,Z=0)
    ExecutePin=@RigVMFunction_ControlFlowBranch.Completed
    DriverItem=RigVMVariableNode() {variable=Driver Item}.Value
    OptionalParentItem=RigVMVariableNode() {variable=Driver Item Parent}.Value
    Debug.bDrawDebug=RigVMVariableNode() {variable=Up Pre Viz}.Value
   )
   > n5 = RigVMVariableNode(Value=@n4.OutputParam) {variable=Result Up Pre}
   n6 = &RigVMFunction_ControlFlowBranch(Condition=RigVMVariableNode() {variable=Up Pre Viz}.Value, ExecuteContext=@n5.ExecuteContext)
   n7 = &RigUnit_SphericalPoseReader(
    ActiveRegionScaleFactors=(PositiveWidth=1,NegativeWidth=1,PositiveHeight=1,NegativeHeight=1)
    ActiveRegionSize=0.001
    Debug=(bDrawDebug=false,bDraw2D=False,bDrawLocalAxes=False,DebugScale=25,DebugSegments=20,DebugThickness=0.05)
    DriverAxis=(X=1,Y=0,Z=0)
    FalloffRegionScaleFactors=(PositiveWidth=0.5,NegativeWidth=0.5,PositiveHeight=1,NegativeHeight=1)
    FalloffSize=0.5
    FlipHeightScaling=False
    FlipWidthScaling=False
    RotationOffset=(X=0,Y=76.239,Z=0)
    ExecutePin=@n6.Completed
    Debug.bDrawDebug=RigVMVariableNode() {variable=Up Post Viz}.Value
    DriverItem=RigVMVariableNode() {variable=Driver Item}.Value
    OptionalParentItem=RigVMVariableNode() {variable=Driver Item Parent}.Value
   )
   > n8 = RigVMVariableNode(Value=@n7.OutputParam) {variable=Result Up Post}
   n9 = &RigVMFunction_ControlFlowBranch(Condition=RigVMVariableNode() {variable=Up Post Viz}.Value, ExecuteContext=@n8.ExecuteContext)
   RigVMDispatchNode(
    Enabled=True
    ScreenColor=(R=1,G=1,B=1,A=1)
    ScreenDuration=0.05
    Prefix=/Script/RigVM.RigVMFunction_StringConcat(B=" - UP POST:   ", A=RigVMVariableNode() {variable=Driver Item Name}.Value) {is=struct RigVMUnitNode, template="Concat::Execute(in A,in B,out Result)"}.Result
    Value=RigVMVariableNode() {variable=Result Up Post}.Value
    ExecuteContext=@n9.True
   ) {template="DISPATCH_RigVMDispatch_Print(in Prefix,in Value,in Enabled,in ScreenDuration,in ScreenColor)"}
   RigVMDispatchNode(
    Enabled=True
    ScreenColor=(R=1,G=1,B=1,A=1)
    ScreenDuration=0.05
    Prefix=&RigVMFunction_StringConcat(B=" - UP PRE:   ", A=RigVMVariableNode() {variable=Driver Item Name}.Value) {template="Concat::Execute(in A,in B,out Result)"}.Result
    Value=RigVMVariableNode() {variable=Result Up Pre}.Value
    ExecuteContext=@n6.True
   ) {template="DISPATCH_RigVMDispatch_Print(in Prefix,in Value,in Enabled,in ScreenDuration,in ScreenColor)"}
   Concat = &RigVMFunction_StringConcat(B=" - UP:   ", A=RigVMVariableNode() {variable=Driver Item Name}.Value) {template="Concat::Execute(in A,in B,out Result)"}
   RigVMDispatchNode(
    Enabled=True
    ScreenColor=(R=1,G=1,B=1,A=1)
    ScreenDuration=0.05
    Prefix=@Concat.Result
    Value=RigVMVariableNode() {variable=Result Up}.Value
    ExecuteContext=@RigVMFunction_ControlFlowBranch.True
   ) {template="DISPATCH_RigVMDispatch_Print(in Prefix,in Value,in Enabled,in ScreenDuration,in ScreenColor)"}
   n10 = &RigUnit_SphericalPoseReader(
    ActiveRegionScaleFactors=(PositiveWidth=1,NegativeWidth=1,PositiveHeight=1,NegativeHeight=1)
    ActiveRegionSize=0.001
    Debug=(bDrawDebug=false,bDraw2D=False,bDrawLocalAxes=False,DebugScale=25,DebugSegments=20,DebugThickness=0.05)
    DriverAxis=(X=1,Y=0,Z=0)
    FalloffRegionScaleFactors=(PositiveWidth=1,NegativeWidth=1,PositiveHeight=1,NegativeHeight=1)
    FalloffSize=0.5
    FlipHeightScaling=False
    FlipWidthScaling=False
    RotationOffset=(X=90,Y=31.239,Z=0)
    DriverItem=RigVMVariableNode() {variable=Driver Item}.Value
    OptionalParentItem=RigVMVariableNode() {variable=Driver Item Parent}.Value
    Debug.bDrawDebug=RigVMVariableNode() {variable=Fr Viz}.Value
    ExecutePin=@RigVMFunction_Sequence.C
   )
   > n11 = RigVMVariableNode(Value=@n10.OutputParam) {variable=Result Fr}
   n12 = &RigVMFunction_ControlFlowBranch(ExecuteContext=@n11.ExecuteContext, Condition=RigVMVariableNode() {variable=Fr Viz}.Value)
   n13 = &RigUnit_SphericalPoseReader(
    ActiveRegionScaleFactors=(PositiveWidth=1,NegativeWidth=1,PositiveHeight=1,NegativeHeight=1)
    ActiveRegionSize=0.001
    Debug=(bDrawDebug=false,bDraw2D=False,bDrawLocalAxes=False,DebugScale=25,DebugSegments=20,DebugThickness=0.05)
    DriverAxis=(X=1,Y=0,Z=0)
    FalloffRegionScaleFactors=(PositiveWidth=2,NegativeWidth=2,PositiveHeight=1,NegativeHeight=1)
    FalloffSize=0.25
    FlipHeightScaling=False
    FlipWidthScaling=False
    RotationOffset=(X=45,Y=-58.761,Z=0)
    ExecutePin=@n12.Completed
    DriverItem=RigVMVariableNode() {variable=Driver Item}.Value
    OptionalParentItem=RigVMVariableNode() {variable=Driver Item Parent}.Value
    Debug.bDrawDebug=RigVMVariableNode() {variable=Fr Pre Viz}.Value
   )
   > n14 = RigVMVariableNode(Value=@n13.OutputParam) {variable=Result Fr Pre}
   n15 = &RigVMFunction_ControlFlowBranch(ExecuteContext=@n14.ExecuteContext, Condition=RigVMVariableNode() {variable=Fr Pre Viz}.Value)
   n16 = &RigUnit_SphericalPoseReader(
    ActiveRegionScaleFactors=(PositiveWidth=1,NegativeWidth=1,PositiveHeight=1,NegativeHeight=1)
    ActiveRegionSize=0.001
    Debug=(bDrawDebug=false,bDraw2D=False,bDrawLocalAxes=False,DebugScale=25,DebugSegments=20,DebugThickness=0.05)
    DriverAxis=(X=1,Y=0,Z=0)
    FalloffRegionScaleFactors=(PositiveWidth=2,NegativeWidth=2,PositiveHeight=1,NegativeHeight=1)
    FalloffSize=0.25
    FlipHeightScaling=False
    FlipWidthScaling=False
    RotationOffset=(X=135,Y=-58.761,Z=0)
    ExecutePin=@n15.Completed
    DriverItem=RigVMVariableNode() {variable=Driver Item}.Value
    OptionalParentItem=RigVMVariableNode() {variable=Driver Item Parent}.Value
    Debug.bDrawDebug=RigVMVariableNode() {variable=Fr Post Viz}.Value
   )
   > n17 = RigVMVariableNode(Value=@n16.OutputParam) {variable=Result Fr Post}
   n18 = &RigVMFunction_ControlFlowBranch(ExecuteContext=@n17.ExecuteContext, Condition=RigVMVariableNode() {variable=Fr Post Viz}.Value)
   n19 = &RigUnit_SphericalPoseReader(
    ActiveRegionScaleFactors=(PositiveWidth=1,NegativeWidth=1,PositiveHeight=1,NegativeHeight=1)
    ActiveRegionSize=0.001
    Debug=(bDrawDebug=false,bDraw2D=False,bDrawLocalAxes=False,DebugScale=25,DebugSegments=20,DebugThickness=0.05)
    DriverAxis=(X=1,Y=0,Z=0)
    FalloffRegionScaleFactors=(PositiveWidth=1,NegativeWidth=1,PositiveHeight=0.5,NegativeHeight=0.5)
    FalloffSize=0.5
    FlipHeightScaling=False
    FlipWidthScaling=False
    RotationOffset=(X=45,Y=31.239,Z=0)
    ExecutePin=@n18.Completed
    DriverItem=RigVMVariableNode() {variable=Driver Item}.Value
    OptionalParentItem=RigVMVariableNode() {variable=Driver Item Parent}.Value
    Debug.bDrawDebug=RigVMVariableNode() {variable=Fr Up Viz}.Value
   )
   > n20 = RigVMVariableNode(Value=@n19.OutputParam) {variable=Result Fr Up}
   n21 = &RigVMFunction_ControlFlowBranch(ExecuteContext=@n20.ExecuteContext, Condition=RigVMVariableNode() {variable=Fr Up Viz}.Value)
   n22 = &RigUnit_SphericalPoseReader(
    ActiveRegionScaleFactors=(PositiveWidth=1,NegativeWidth=1,PositiveHeight=1,NegativeHeight=1)
    ActiveRegionSize=0.001
    Debug=(bDrawDebug=false,bDraw2D=False,bDrawLocalAxes=False,DebugScale=25,DebugSegments=20,DebugThickness=0.05)
    DriverAxis=(X=1,Y=0,Z=0)
    FalloffRegionScaleFactors=(PositiveWidth=1,NegativeWidth=1,PositiveHeight=0.5,NegativeHeight=0.5)
    FalloffSize=0.5
    FlipHeightScaling=False
    FlipWidthScaling=False
    RotationOffset=(X=135,Y=31.239,Z=0)
    ExecutePin=@n21.Completed
    DriverItem=RigVMVariableNode() {variable=Driver Item}.Value
    OptionalParentItem=RigVMVariableNode() {variable=Driver Item Parent}.Value
    Debug.bDrawDebug=RigVMVariableNode() {variable=Fr Dn Viz}.Value
   )
   > n23 = RigVMVariableNode(Value=@n22.OutputParam) {variable=Result Fr Dn}
   n24 = &RigVMFunction_ControlFlowBranch(ExecuteContext=@n23.ExecuteContext, Condition=RigVMVariableNode() {variable=Fr Dn Viz}.Value)
   RigVMDispatchNode(
    Enabled=True
    ScreenColor=(R=1,G=1,B=1,A=1)
    ScreenDuration=0.05
    Prefix=&RigVMFunction_StringConcat(B=" - FR DN:   ", A=RigVMVariableNode() {variable=Driver Item Name}.Value) {template="Concat::Execute(in A,in B,out Result)"}.Result
    ExecuteContext=@n24.True
    Value=RigVMVariableNode() {variable=Result Fr Dn}.Value
   ) {template="DISPATCH_RigVMDispatch_Print(in Prefix,in Value,in Enabled,in ScreenDuration,in ScreenColor)"}
   RigVMDispatchNode(
    Enabled=True
    ScreenColor=(R=1,G=1,B=1,A=1)
    ScreenDuration=0.05
    Prefix=&RigVMFunction_StringConcat(B=" - FR UP:   ", A=RigVMVariableNode() {variable=Driver Item Name}.Value) {template="Concat::Execute(in A,in B,out Result)"}.Result
    ExecuteContext=@n21.True
    Value=RigVMVariableNode() {variable=Result Fr Up}.Value
   ) {template="DISPATCH_RigVMDispatch_Print(in Prefix,in Value,in Enabled,in ScreenDuration,in ScreenColor)"}
   RigVMDispatchNode(
    Enabled=True
    ScreenColor=(R=1,G=1,B=1,A=1)
    ScreenDuration=0.05
    Prefix=&RigVMFunction_StringConcat(B=" - FR POST:   ", A=RigVMVariableNode() {variable=Driver Item Name}.Value) {template="Concat::Execute(in A,in B,out Result)"}.Result
    ExecuteContext=@n18.True
    Value=RigVMVariableNode() {variable=Result Fr Post}.Value
   ) {template="DISPATCH_RigVMDispatch_Print(in Prefix,in Value,in Enabled,in ScreenDuration,in ScreenColor)"}
   RigVMDispatchNode(
    Enabled=True
    ScreenColor=(R=1,G=1,B=1,A=1)
    ScreenDuration=0.05
    Prefix=&RigVMFunction_StringConcat(B=" - FR PRE:   ", A=RigVMVariableNode() {variable=Driver Item Name}.Value) {template="Concat::Execute(in A,in B,out Result)"}.Result
    ExecuteContext=@n15.True
    Value=RigVMVariableNode() {variable=Result Fr Pre}.Value
   ) {template="DISPATCH_RigVMDispatch_Print(in Prefix,in Value,in Enabled,in ScreenDuration,in ScreenColor)"}
   RigVMDispatchNode(
    Enabled=True
    ScreenColor=(R=1,G=1,B=1,A=1)
    ScreenDuration=0.05
    Prefix=&RigVMFunction_StringConcat(B=" - FR:   ", A=RigVMVariableNode() {variable=Driver Item Name}.Value) {template="Concat::Execute(in A,in B,out Result)"}.Result
    ExecuteContext=@n12.True
    Value=RigVMVariableNode() {variable=Result Fr}.Value
   ) {template="DISPATCH_RigVMDispatch_Print(in Prefix,in Value,in Enabled,in ScreenDuration,in ScreenColor)"}
   n25 = &RigUnit_SphericalPoseReader(
    ActiveRegionScaleFactors=(PositiveWidth=1,NegativeWidth=1,PositiveHeight=1,NegativeHeight=1)
    ActiveRegionSize=0.001
    Debug=(bDrawDebug=false,bDraw2D=False,bDrawLocalAxes=False,DebugScale=25,DebugSegments=20,DebugThickness=0.05)
    DriverAxis=(X=1,Y=0,Z=0)
    FalloffRegionScaleFactors=(PositiveWidth=1,NegativeWidth=1,PositiveHeight=1,NegativeHeight=1)
    FalloffSize=0.5
    FlipHeightScaling=False
    FlipWidthScaling=False
    RotationOffset=(X=-90,Y=31.239,Z=0)
    DriverItem=RigVMVariableNode() {variable=Driver Item}.Value
    OptionalParentItem=RigVMVariableNode() {variable=Driver Item Parent}.Value
    Debug.bDrawDebug=RigVMVariableNode() {variable=Ba Viz}.Value
    ExecutePin=@RigVMFunction_Sequence.D
   )
   > n26 = RigVMVariableNode(Value=@n25.OutputParam) {variable=Result Ba}
   n27 = &RigVMFunction_ControlFlowBranch(ExecuteContext=@n26.ExecuteContext, Condition=RigVMVariableNode() {variable=Ba Viz}.Value)
   n28 = &RigUnit_SphericalPoseReader(
    ActiveRegionScaleFactors=(PositiveWidth=1,NegativeWidth=1,PositiveHeight=1,NegativeHeight=1)
    ActiveRegionSize=0.001
    Debug=(bDrawDebug=false,bDraw2D=False,bDrawLocalAxes=False,DebugScale=25,DebugSegments=20,DebugThickness=0.05)
    DriverAxis=(X=1,Y=0,Z=0)
    FalloffRegionScaleFactors=(PositiveWidth=2,NegativeWidth=2,PositiveHeight=1,NegativeHeight=1)
    FalloffSize=0.25
    FlipHeightScaling=False
    FlipWidthScaling=False
    RotationOffset=(X=-45,Y=-58.761,Z=0)
    ExecutePin=@n27.Completed
    Debug.bDrawDebug=RigVMVariableNode() {variable=Ba Pre Viz}.Value
    DriverItem=RigVMVariableNode() {variable=Driver Item}.Value
    OptionalParentItem=RigVMVariableNode() {variable=Driver Item Parent}.Value
   )
   > n29 = RigVMVariableNode(Value=@n28.OutputParam) {variable=Result Ba Pre}
   n30 = &RigVMFunction_ControlFlowBranch(Condition=RigVMVariableNode() {variable=Ba Pre Viz}.Value, ExecuteContext=@n29.ExecuteContext)
   n31 = &RigUnit_SphericalPoseReader(
    ActiveRegionScaleFactors=(PositiveWidth=1,NegativeWidth=1,PositiveHeight=1,NegativeHeight=1)
    ActiveRegionSize=0.001
    Debug=(bDrawDebug=false,bDraw2D=False,bDrawLocalAxes=False,DebugScale=25,DebugSegments=20,DebugThickness=0.05)
    DriverAxis=(X=1,Y=0,Z=0)
    FalloffRegionScaleFactors=(PositiveWidth=2,NegativeWidth=2,PositiveHeight=1,NegativeHeight=1)
    FalloffSize=0.25
    FlipHeightScaling=False
    FlipWidthScaling=False
    RotationOffset=(X=-135,Y=-58.761,Z=0)
    ExecutePin=@n30.Completed
    Debug.bDrawDebug=RigVMVariableNode() {variable=Ba Post Viz}.Value
    DriverItem=RigVMVariableNode() {variable=Driver Item}.Value
    OptionalParentItem=RigVMVariableNode() {variable=Driver Item Parent}.Value
   )
   > n32 = RigVMVariableNode(Value=@n31.OutputParam) {variable=Result Ba Post}
   n33 = &RigVMFunction_ControlFlowBranch(Condition=RigVMVariableNode() {variable=Ba Post Viz}.Value, ExecuteContext=@n32.ExecuteContext)
   n34 = &RigUnit_SphericalPoseReader(
    ActiveRegionScaleFactors=(PositiveWidth=1,NegativeWidth=1,PositiveHeight=1,NegativeHeight=1)
    ActiveRegionSize=0.001
    Debug=(bDrawDebug=false,bDraw2D=False,bDrawLocalAxes=False,DebugScale=25,DebugSegments=20,DebugThickness=0.05)
    DriverAxis=(X=1,Y=0,Z=0)
    FalloffRegionScaleFactors=(PositiveWidth=1,NegativeWidth=1,PositiveHeight=0.5,NegativeHeight=0.5)
    FalloffSize=0.5
    FlipHeightScaling=False
    FlipWidthScaling=False
    RotationOffset=(X=-45,Y=31.239,Z=0)
    ExecutePin=@n33.Completed
    Debug.bDrawDebug=RigVMVariableNode() {variable=Ba Up Viz}.Value
    DriverItem=RigVMVariableNode() {variable=Driver Item}.Value
    OptionalParentItem=RigVMVariableNode() {variable=Driver Item Parent}.Value
   )
   > n35 = RigVMVariableNode(Value=@n34.OutputParam) {variable=Result Ba Up}
   n36 = &RigVMFunction_ControlFlowBranch(Condition=RigVMVariableNode() {variable=Ba Up Viz}.Value, ExecuteContext=@n35.ExecuteContext)
   n37 = &RigUnit_SphericalPoseReader(
    ActiveRegionScaleFactors=(PositiveWidth=1,NegativeWidth=1,PositiveHeight=1,NegativeHeight=1)
    ActiveRegionSize=0.001
    Debug=(bDrawDebug=false,bDraw2D=False,bDrawLocalAxes=False,DebugScale=25,DebugSegments=20,DebugThickness=0.05)
    DriverAxis=(X=1,Y=0,Z=0)
    FalloffRegionScaleFactors=(PositiveWidth=1,NegativeWidth=1,PositiveHeight=0.5,NegativeHeight=0.5)
    FalloffSize=0.5
    FlipHeightScaling=False
    FlipWidthScaling=False
    RotationOffset=(X=-135,Y=31.239,Z=0)
    ExecutePin=@n36.Completed
    Debug.bDrawDebug=RigVMVariableNode() {variable=Ba Dn Viz}.Value
    DriverItem=RigVMVariableNode() {variable=Driver Item}.Value
    OptionalParentItem=RigVMVariableNode() {variable=Driver Item Parent}.Value
   )
   > n38 = RigVMVariableNode(Value=@n37.OutputParam) {variable=Result Ba Dn}
   n39 = &RigVMFunction_ControlFlowBranch(Condition=RigVMVariableNode() {variable=Ba Dn Viz}.Value, ExecuteContext=@n38.ExecuteContext)
   RigVMDispatchNode(
    Enabled=True
    ScreenColor=(R=1,G=1,B=1,A=1)
    ScreenDuration=0.05
    Prefix=&RigVMFunction_StringConcat(B=" - BA DN:   ", A=RigVMVariableNode() {variable=Driver Item Name}.Value) {template="Concat::Execute(in A,in B,out Result)"}.Result
    ExecuteContext=@n39.True
    Value=RigVMVariableNode() {variable=Result Ba Dn}.Value
   ) {template="DISPATCH_RigVMDispatch_Print(in Prefix,in Value,in Enabled,in ScreenDuration,in ScreenColor)"}
   RigVMDispatchNode(
    Enabled=True
    ScreenColor=(R=1,G=1,B=1,A=1)
    ScreenDuration=0.05
    Prefix=&RigVMFunction_StringConcat(B=" - BA UP:   ", A=RigVMVariableNode() {variable=Driver Item Name}.Value) {template="Concat::Execute(in A,in B,out Result)"}.Result
    ExecuteContext=@n36.True
    Value=RigVMVariableNode() {variable=Result Ba Up}.Value
   ) {template="DISPATCH_RigVMDispatch_Print(in Prefix,in Value,in Enabled,in ScreenDuration,in ScreenColor)"}
   RigVMDispatchNode(
    Enabled=True
    ScreenColor=(R=1,G=1,B=1,A=1)
    ScreenDuration=0.05
    Prefix=&RigVMFunction_StringConcat(B=" - BA POST:   ", A=RigVMVariableNode() {variable=Driver Item Name}.Value) {template="Concat::Execute(in A,in B,out Result)"}.Result
    ExecuteContext=@n33.True
    Value=RigVMVariableNode() {variable=Result Ba Post}.Value
   ) {template="DISPATCH_RigVMDispatch_Print(in Prefix,in Value,in Enabled,in ScreenDuration,in ScreenColor)"}
   RigVMDispatchNode(
    Enabled=True
    ScreenColor=(R=1,G=1,B=1,A=1)
    ScreenDuration=0.05
    Prefix=&RigVMFunction_StringConcat(B=" - BA PRE:   ", A=RigVMVariableNode() {variable=Driver Item Name}.Value) {template="Concat::Execute(in A,in B,out Result)"}.Result
    ExecuteContext=@n30.True
    Value=RigVMVariableNode() {variable=Result Ba Pre}.Value
   ) {template="DISPATCH_RigVMDispatch_Print(in Prefix,in Value,in Enabled,in ScreenDuration,in ScreenColor)"}
   RigVMDispatchNode(
    Enabled=True
    ScreenColor=(R=1,G=1,B=1,A=1)
    ScreenDuration=0.05
    Prefix=&RigVMFunction_StringConcat(B=" - BA:   ", A=RigVMVariableNode() {variable=Driver Item Name}.Value) {template="Concat::Execute(in A,in B,out Result)"}.Result
    ExecuteContext=@n27.True
    Value=RigVMVariableNode() {variable=Result Ba}.Value
   ) {template="DISPATCH_RigVMDispatch_Print(in Prefix,in Value,in Enabled,in ScreenDuration,in ScreenColor)"}
   n40 = &RigUnit_SphericalPoseReader(
    ActiveRegionScaleFactors=(PositiveWidth=1,NegativeWidth=1,PositiveHeight=1,NegativeHeight=1)
    ActiveRegionSize=0.001
    Debug=(bDrawDebug=false,bDraw2D=False,bDrawLocalAxes=False,DebugScale=25,DebugSegments=20,DebugThickness=0.05)
    DriverAxis=(X=1,Y=0,Z=0)
    FalloffRegionScaleFactors=(PositiveWidth=1,NegativeWidth=0.65,PositiveHeight=1,NegativeHeight=1)
    FalloffSize=0.5
    FlipHeightScaling=False
    FlipWidthScaling=False
    RotationOffset=(X=0,Y=-148.761,Z=0)
    DriverItem=RigVMVariableNode() {variable=Driver Item}.Value
    OptionalParentItem=RigVMVariableNode() {variable=Driver Item Parent}.Value
    Debug.bDrawDebug=RigVMVariableNode() {variable=Dn Viz}.Value
    ExecutePin=@RigVMFunction_Sequence.B
   )
   > n41 = RigVMVariableNode(Value=@n40.OutputParam) {variable=Result Dn}
   n42 = &RigVMFunction_ControlFlowBranch(ExecuteContext=@n41.ExecuteContext, Condition=RigVMVariableNode() {variable=Dn Viz}.Value)
   n43 = &RigUnit_SphericalPoseReader(
    ActiveRegionScaleFactors=(PositiveWidth=1,NegativeWidth=1,PositiveHeight=1,NegativeHeight=1)
    ActiveRegionSize=0.001
    Debug=(bDrawDebug=false,bDraw2D=False,bDrawLocalAxes=False,DebugScale=25,DebugSegments=20,DebugThickness=0.05)
    DriverAxis=(X=1,Y=0,Z=0)
    FalloffRegionScaleFactors=(PositiveWidth=1,NegativeWidth=0.298,PositiveHeight=2,NegativeHeight=2)
    FalloffSize=0.25
    FlipHeightScaling=False
    FlipWidthScaling=False
    RotationOffset=(X=0,Y=-103.761,Z=0)
    ExecutePin=@n42.Completed
    DriverItem=RigVMVariableNode() {variable=Driver Item}.Value
    OptionalParentItem=RigVMVariableNode() {variable=Driver Item Parent}.Value
    Debug.bDrawDebug=RigVMVariableNode() {variable=Dn Pre Viz}.Value
   )
   > n44 = RigVMVariableNode(Value=@n43.OutputParam) {variable=Result Dn Pre}
   n45 = &RigVMFunction_ControlFlowBranch(ExecuteContext=@n44.ExecuteContext, Condition=RigVMVariableNode() {variable=Dn Pre Viz}.Value)
   n46 = &RigUnit_SphericalPoseReader(
    ActiveRegionScaleFactors=(PositiveWidth=1,NegativeWidth=1,PositiveHeight=1,NegativeHeight=1)
    ActiveRegionSize=0.001
    Debug=(bDrawDebug=false,bDraw2D=False,bDrawLocalAxes=False,DebugScale=25,DebugSegments=20,DebugThickness=0.05)
    DriverAxis=(X=1,Y=0,Z=0)
    FalloffRegionScaleFactors=(PositiveWidth=1,NegativeWidth=1,PositiveHeight=2,NegativeHeight=2)
    FalloffSize=0.25
    FlipHeightScaling=False
    FlipWidthScaling=False
    RotationOffset=(X=0,Y=-193.761,Z=0)
    ExecutePin=@n45.Completed
    DriverItem=RigVMVariableNode() {variable=Driver Item}.Value
    OptionalParentItem=RigVMVariableNode() {variable=Driver Item Parent}.Value
    Debug.bDrawDebug=RigVMVariableNode() {variable=Dn Post Viz}.Value
   )
   > n47 = RigVMVariableNode(Value=@n46.OutputParam) {variable=Result Dn Post}
   n48 = &RigVMFunction_ControlFlowBranch(ExecuteContext=@n47.ExecuteContext, Condition=RigVMVariableNode() {variable=Dn Post Viz}.Value)
   RigVMDispatchNode(
    Enabled=True
    ScreenColor=(R=1,G=1,B=1,A=1)
    ScreenDuration=0.05
    Prefix=&RigVMFunction_StringConcat(B=" - DN POST:   ", A=RigVMVariableNode() {variable=Driver Item Name}.Value) {template="Concat::Execute(in A,in B,out Result)"}.Result
    ExecuteContext=@n48.True
    Value=RigVMVariableNode() {variable=Result Dn Post}.Value
   ) {template="DISPATCH_RigVMDispatch_Print(in Prefix,in Value,in Enabled,in ScreenDuration,in ScreenColor)"}
   RigVMDispatchNode(
    Enabled=True
    ScreenColor=(R=1,G=1,B=1,A=1)
    ScreenDuration=0.05
    Prefix=&RigVMFunction_StringConcat(B=" - DN PRE:   ", A=RigVMVariableNode() {variable=Driver Item Name}.Value) {template="Concat::Execute(in A,in B,out Result)"}.Result
    ExecuteContext=@n45.True
    Value=RigVMVariableNode() {variable=Result Dn Pre}.Value
   ) {template="DISPATCH_RigVMDispatch_Print(in Prefix,in Value,in Enabled,in ScreenDuration,in ScreenColor)"}
   RigVMDispatchNode(
    Enabled=True
    ScreenColor=(R=1,G=1,B=1,A=1)
    ScreenDuration=0.05
    Prefix=&RigVMFunction_StringConcat(B=" - DN:   ", A=RigVMVariableNode() {variable=Driver Item Name}.Value) {template="Concat::Execute(in A,in B,out Result)"}.Result
    ExecuteContext=@n42.True
    Value=RigVMVariableNode() {variable=Result Dn}.Value
   ) {template="DISPATCH_RigVMDispatch_Print(in Prefix,in Value,in Enabled,in ScreenDuration,in ScreenColor)"}
   n49 = &RigUnit_SphericalPoseReader(
    ActiveRegionScaleFactors=(PositiveWidth=1,NegativeWidth=1,PositiveHeight=1,NegativeHeight=1)
    ActiveRegionSize=0.001
    Debug=(bDrawDebug=true,bDraw2D=False,bDrawLocalAxes=False,DebugScale=25,DebugSegments=20,DebugThickness=0.05)
    DriverAxis=(X=1,Y=0,Z=0)
    FalloffRegionScaleFactors=(PositiveWidth=1,NegativeWidth=1,PositiveHeight=1,NegativeHeight=1)
    FalloffSize=0.18
    FlipHeightScaling=False
    FlipWidthScaling=False
    RotationOffset=(X=28.4478,Y=74.4998,Z=0)
    DriverItem=RigVMVariableNode() {variable=Driver Item}.Value
    OptionalParentItem=RigVMVariableNode() {variable=Driver Item Parent}.Value
    Debug.bDrawDebug=RigVMVariableNode() {variable=Aim Up In Viz}.Value
    ExecutePin=@RigVMFunction_Sequence.E
   )
   > n50 = RigVMVariableNode(Value=@n49.OutputParam) {variable=Result Aim Up In}
   n51 = &RigVMFunction_ControlFlowBranch(Condition=RigVMVariableNode() {variable=Aim Up In Viz}.Value, ExecuteContext=@n50.ExecuteContext)
   n52 = &RigUnit_SphericalPoseReader(
    ActiveRegionScaleFactors=(PositiveWidth=1,NegativeWidth=1,PositiveHeight=1,NegativeHeight=1)
    ActiveRegionSize=0.001
    Debug=(bDrawDebug=true,bDraw2D=False,bDrawLocalAxes=False,DebugScale=25,DebugSegments=20,DebugThickness=0.05)
    DriverAxis=(X=1,Y=0,Z=0)
    FalloffRegionScaleFactors=(PositiveWidth=1,NegativeWidth=1,PositiveHeight=1,NegativeHeight=1)
    FalloffSize=0.18
    FlipHeightScaling=False
    FlipWidthScaling=False
    RotationOffset=(X=31.239,Y=-1.4062,Z=0)
    ExecutePin=@n51.Completed
    DriverItem=RigVMVariableNode() {variable=Driver Item}.Value
    OptionalParentItem=RigVMVariableNode() {variable=Driver Item Parent}.Value
    Debug.bDrawDebug=RigVMVariableNode() {variable=Aim Up Ot Viz}.Value
   )
   > n53 = RigVMVariableNode(Value=@n52.OutputParam) {variable=Result Aim Up Ot}
   n54 = &RigVMFunction_ControlFlowBranch(Condition=RigVMVariableNode() {variable=Aim Up Ot Viz}.Value, ExecuteContext=@n53.ExecuteContext)
   n55 = &RigUnit_SphericalPoseReader(
    ActiveRegionScaleFactors=(PositiveWidth=1,NegativeWidth=1,PositiveHeight=1,NegativeHeight=1)
    ActiveRegionSize=0.001
    Debug=(bDrawDebug=true,bDraw2D=False,bDrawLocalAxes=False,DebugScale=25,DebugSegments=20,DebugThickness=0.05)
    DriverAxis=(X=1,Y=0,Z=0)
    FalloffRegionScaleFactors=(PositiveWidth=1,NegativeWidth=1,PositiveHeight=1,NegativeHeight=1)
    FalloffSize=0.18
    FlipHeightScaling=False
    FlipWidthScaling=False
    RotationOffset=(X=147.9195,Y=-7.4723,Z=1.8383)
    ExecutePin=@n54.Completed
    DriverItem=RigVMVariableNode() {variable=Driver Item}.Value
    OptionalParentItem=RigVMVariableNode() {variable=Driver Item Parent}.Value
    Debug.bDrawDebug=RigVMVariableNode() {variable=Aim Dn In Viz}.Value
   )
   > n56 = RigVMVariableNode(Value=@n55.OutputParam) {variable=Result Aim Dn In}
   n57 = &RigVMFunction_ControlFlowBranch(Condition=RigVMVariableNode() {variable=Aim Dn In Viz}.Value, ExecuteContext=@n56.ExecuteContext)
   RigVMDispatchNode(
    Enabled=True
    ScreenColor=(R=1,G=1,B=1,A=1)
    ScreenDuration=0.05
    Prefix=&RigVMFunction_StringConcat(B=" - AIM DN IN:   ", A=RigVMVariableNode() {variable=Driver Item Name}.Value) {template="Concat::Execute(in A,in B,out Result)"}.Result
    ExecuteContext=@n57.True
    Value=RigVMVariableNode() {variable=Result Aim Dn In}.Value
   ) {template="DISPATCH_RigVMDispatch_Print(in Prefix,in Value,in Enabled,in ScreenDuration,in ScreenColor)"}
   RigVMDispatchNode(
    Enabled=True
    ScreenColor=(R=1,G=1,B=1,A=1)
    ScreenDuration=0.05
    Prefix=&RigVMFunction_StringConcat(B=" - AIM UP OT:   ", A=RigVMVariableNode() {variable=Driver Item Name}.Value) {template="Concat::Execute(in A,in B,out Result)"}.Result
    ExecuteContext=@n54.True
    Value=RigVMVariableNode() {variable=Result Aim Up Ot}.Value
   ) {template="DISPATCH_RigVMDispatch_Print(in Prefix,in Value,in Enabled,in ScreenDuration,in ScreenColor)"}
   RigVMDispatchNode(
    Enabled=True
    ScreenColor=(R=1,G=1,B=1,A=1)
    ScreenDuration=0.05
    Prefix=&RigVMFunction_StringConcat(B=" - AIM UP IN:   ", A=RigVMVariableNode() {variable=Driver Item Name}.Value) {template="Concat::Execute(in A,in B,out Result)"}.Result
    ExecuteContext=@n51.True
    Value=RigVMVariableNode() {variable=Result Aim Up In}.Value
   ) {template="DISPATCH_RigVMDispatch_Print(in Prefix,in Value,in Enabled,in ScreenDuration,in ScreenColor)"}
   n58 = &RigUnit_SphericalPoseReader(
    ActiveRegionScaleFactors=(PositiveWidth=1,NegativeWidth=1,PositiveHeight=1,NegativeHeight=1)
    ActiveRegionSize=0.001
    Debug=(bDrawDebug=true,bDraw2D=False,bDrawLocalAxes=False,DebugScale=25,DebugSegments=20,DebugThickness=0.05)
    DriverAxis=(X=1,Y=0,Z=0)
    ExecutePin=()
    FalloffRegionScaleFactors=(PositiveWidth=1,NegativeWidth=1,PositiveHeight=1,NegativeHeight=1)
    FalloffSize=0.18
    FlipHeightScaling=False
    FlipWidthScaling=False
    RotationOffset=(X=147.9195,Y=-7.4723,Z=1.8383)
    DriverItem=RigVMVariableNode() {variable=Driver Item}.Value
    OptionalParentItem=RigVMVariableNode() {variable=Driver Item Parent}.Value
    Debug.bDrawDebug=RigVMVariableNode() {variable=Aim Dn In Viz}.Value
   )
   > n59 = RigVMVariableNode(Value=@n58.OutputParam) {variable=Result Aim Dn In}
   n60 = &RigVMFunction_ControlFlowBranch(Condition=RigVMVariableNode() {variable=Aim Dn In Viz}.Value, ExecuteContext=@n59.ExecuteContext)
   RigVMDispatchNode(
    Enabled=True
    ScreenColor=(R=1,G=1,B=1,A=1)
    ScreenDuration=0.05
    Prefix=&RigVMFunction_StringConcat(B=" - AIM DN IN:   ", A=RigVMVariableNode() {variable=Driver Item Name}.Value) {template="Concat::Execute(in A,in B,out Result)"}.Result
    ExecuteContext=@n60.True
    Value=RigVMVariableNode() {variable=Result Aim Dn In}.Value
   ) {template="DISPATCH_RigVMDispatch_Print(in Prefix,in Value,in Enabled,in ScreenDuration,in ScreenColor)"}
  ]}
 RigVMFunctionLibrary {nodes [
   Sided UpperArm Readers L R = RigVMCollapseNode(ExecuteContext=())
   UpperArm Readers = RigVMCollapseNode(Driver Item=(Type=None,Name="None"), ExecuteContext=())
  ]}
 RigVMFunction_Sequence_1_ContainedGraph {nodes [
   n1 = RigVMFunctionEntryNode()
   RigVMFunction_Sequence = &RigVMFunction_Sequence(ExecuteContext=@n1.ExecuteContext)
   n2 = &RigVMFunction_Sequence(ExecuteContext=@RigVMFunction_Sequence.B)
   n3 = &RigVMFunction_Sequence(ExecuteContext=@n2.B)
   n4 = &RigVMFunction_Sequence(ExecuteContext=@n3.B)
   RigVMFunctionReturnNode(A=@RigVMFunction_Sequence.A, B=@n2.A, C=@n3.A, D=@n4.A, E=@n4.B)
  ]}
 RigVMFunction_Sequence_1_ContainedGraph {nodes [
   n1 = RigVMFunctionEntryNode()
   RigVMFunction_Sequence = &RigVMFunction_Sequence(ExecuteContext=@n1.ExecuteContext)
   n2 = &RigVMFunction_Sequence(ExecuteContext=@RigVMFunction_Sequence.B)
   n3 = &RigVMFunction_Sequence(ExecuteContext=@n2.B)
   n4 = &RigVMFunction_Sequence(ExecuteContext=@n3.B)
   n5 = &RigVMFunction_Sequence(ExecuteContext=@n4.B)
   RigVMFunctionReturnNode(A=@RigVMFunction_Sequence.A, B=@n2.A, C=@n3.A, D=@n4.A, E=@n5.A, F=@n5.B)
  ]}
 RigVMFunction_Sequence_2_ContainedGraph {nodes [
   n1 = RigVMFunctionEntryNode()
   n2 = &RigVMFunction_Sequence(ExecuteContext=@n1.ExecuteContext)
   n3 = &RigVMFunction_Sequence(ExecuteContext=@n2.B)
   n4 = &RigVMFunction_Sequence(ExecuteContext=@n3.B)
   n5 = &RigVMFunction_Sequence(ExecuteContext=@n4.B)
   n6 = &RigVMFunction_Sequence(ExecuteContext=@n5.B)
   n7 = &RigVMFunction_Sequence(ExecuteContext=@n6.B)
   n8 = &RigVMFunction_Sequence(ExecuteContext=@n7.B)
   n9 = &RigVMFunction_Sequence(ExecuteContext=@n8.B)
   n10 = &RigVMFunction_Sequence(ExecuteContext=@n9.B)
   n11 = &RigVMFunction_Sequence(ExecuteContext=@n10.B)
   n12 = &RigVMFunction_Sequence(ExecuteContext=@n11.B)
   RigVMFunctionReturnNode(A=@n2.A, J=@n11.A, K=@n12.A, L=@n12.B, B=@n3.A, C=@n4.A, D=@n5.A, E=@n6.A, F=@n7.A, G=@n8.A, H=@n9.A, I=@n10.A)
  ]}
 RigVMModel {nodes [
   UpperArm Readers_2 = RigVMFunctionReferenceNode(
    Aim Dn In Viz=false
    Aim Up In Viz=false
    Aim Up Ot Viz=false
    Ba Viz=false
    Dn Post Viz=false
    Dn Pre Viz=false
    Dn Viz=false
    Driver Item=(Type=Bone,Name="upperarm_l")
    Fr Dn Viz=false
    Fr Post Viz=false
    Fr Pre Viz=false
    Fr Up Viz=false
    Fr Viz=false
    Up Post Viz=false
    Up Pre Viz=false
    Up Viz=false
    ExecuteContext=@RigVMFunction_Sequence_1.L
   )
   &RigUnit_ModifyTransforms(
    ExecutePin=()
    ItemToModify=((Item=(Type=Bone,Name="shoulder_pad_r_MSTR"),Transform=(Rotation=(X=0.6422,Y=0.028,Z=-0.0334,W=0.7653),Translation=(X=0.0209,Y=-0.5107,Z=1.6885),Scale3D=(X=1,Y=1,Z=1))))
    WeightMinimum=0
    Weight=@"UpperArm Readers_2".Up
   )
   > &RigUnit_ModifyTransforms(
    ItemToModify=((Item=(Type=Bone,Name="shoulder_pad_r_TWIST"),Transform=(Rotation=(X=0,Y=0,Z=0,W=1),Translation=(X=0,Y=0,Z=0),Scale3D=(X=1,Y=1,Z=1))))
    WeightMinimum=0
    Weight=@"UpperArm Readers_2".Up
   )
   > &RigUnit_ModifyTransforms(
    ItemToModify=((Item=(Type=Bone,Name="shoulder_pad_r"),Transform=(Rotation=(X=0.0634,Y=0.3505,Z=-0.0224,W=0.9342),Translation=(X=0,Y=0,Z=0),Scale3D=(X=1,Y=1,Z=1))))
    WeightMinimum=0
    Weight=@"UpperArm Readers_2".Up
   )
   UpperArm Readers_3 = RigVMFunctionReferenceNode(
    Aim Dn In Viz=false
    Aim Up In Viz=false
    Aim Up Ot Viz=false
    Ba Viz=false
    Dn Post Viz=false
    Dn Pre Viz=false
    Dn Viz=false
    Driver Item=(Type=Bone,Name="upperarm_r")
    Fr Dn Viz=false
    Fr Post Viz=false
    Fr Pre Viz=false
    Fr Up Viz=false
    Fr Viz=false
    Up Post Viz=false
    Up Pre Viz=false
    Up Viz=false
    ExecuteContext=@RigVMFunction_Sequence_1.K
   )
   &RigUnit_ModifyTransforms(
    ExecutePin=()
    ItemToModify=((Item=(Type=Bone,Name="shoulder_pad_r_MSTR"),Transform=(Rotation=(X=0.6422,Y=0.028,Z=-0.0334,W=0.7653),Translation=(X=0.0209,Y=-0.5107,Z=1.6885),Scale3D=(X=1,Y=1,Z=1))))
    WeightMinimum=0
    Weight=@"UpperArm Readers_3".Up
   )
   > &RigUnit_ModifyTransforms(
    ItemToModify=((Item=(Type=Bone,Name="shoulder_pad_r_TWIST"),Transform=(Rotation=(X=0,Y=0,Z=0,W=1),Translation=(X=0,Y=0,Z=0),Scale3D=(X=1,Y=1,Z=1))))
    WeightMinimum=0
    Weight=@"UpperArm Readers_3".Up
   )
   > &RigUnit_ModifyTransforms(
    ItemToModify=((Item=(Type=Bone,Name="shoulder_pad_r"),Transform=(Rotation=(X=0.0634,Y=0.3505,Z=-0.0224,W=0.9342),Translation=(X=0,Y=0,Z=0),Scale3D=(X=1,Y=1,Z=1))))
    WeightMinimum=0
    Weight=@"UpperArm Readers_3".Up
   )
   &RigUnit_ModifyTransforms(
    ExecutePin=()
    ItemToModify=((Item=(Type=Bone,Name="shoulder_pad_r"),Transform=(Rotation=(X=0,Y=-0.352,Z=0,W=0.936),Translation=(X=0,Y=0,Z=0),Scale3D=(X=1,Y=1,Z=1))))
    Weight=1
    WeightMinimum=0
   )
   > &RigUnit_ModifyTransforms(
    ItemToModify=((Item=(Type=Bone,Name="shoulder_pad_r"),Transform=(Rotation=(X=0.0755,Y=-0.3507,Z=0.0284,W=0.933),Translation=(X=0,Y=0,Z=0),Scale3D=(X=1,Y=1,Z=1))))
    Weight=1
    WeightMinimum=0
   )
   &RigUnit_ModifyTransforms(
    ExecutePin=()
    ItemToModify=((Item=(Type=Bone,Name="link_a_BASE"),Transform=(Rotation=(X=0,Y=-0.1045,Z=0,W=0.9945),Translation=(X=0,Y=0,Z=0),Scale3D=(X=1,Y=1,Z=1))))
    Weight=1
    WeightMinimum=0
   )
   > &RigUnit_ModifyTransforms(
    ItemToModify=((Item=(Type=Bone,Name="link_b_BASE"),Transform=(Rotation=(X=0,Y=-0.1305,Z=0,W=0.9914),Translation=(X=0,Y=0,Z=0),Scale3D=(X=1,Y=1,Z=1))))
    Weight=1
    WeightMinimum=0
   )
   > &RigUnit_ModifyTransforms(
    ItemToModify=((Item=(Type=Bone,Name="rope_b_BASE"),Transform=(Rotation=(X=-0.2588,Y=0,Z=0,W=0.9659),Translation=(X=0,Y=0,Z=0),Scale3D=(X=1,Y=1,Z=1))))
    Weight=1
    WeightMinimum=0
   )
   &RigUnit_ModifyTransforms(
    ExecutePin=()
    ItemToModify=((Item=(Type=Bone,Name="shoulder_pad_r"),Transform=(Rotation=(X=0,Y=-0.352,Z=0,W=0.936),Translation=(X=0,Y=0,Z=0),Scale3D=(X=1,Y=1,Z=1))))
    Weight=1
    WeightMinimum=0
   )
   > &RigUnit_ModifyTransforms(
    ItemToModify=((Item=(Type=Bone,Name="shoulder_pad_r"),Transform=(Rotation=(X=0.0755,Y=-0.3507,Z=0.0284,W=0.933),Translation=(X=0,Y=0,Z=0),Scale3D=(X=1,Y=1,Z=1))))
    Weight=1
    WeightMinimum=0
   )
   RigUnit_BeginExecution = &RigUnit_BeginExecution()
   RigVMFunction_Sequence_1 = RigVMAggregateNode(ExecuteContext=@RigUnit_BeginExecution.ExecutePin)
   &RigUnit_RotationConstraintLocalSpaceOffset(
    AdvancedSettings=(InterpolationType=Average,RotationOrderForFilter=XZY)
    Child=(Type=Bone,Name="elbow_l")
    Filter=(bX=True,bY=True,bZ=True)
    Parents=((Item=(Type=Bone,Name="lowerarm_l"),Weight=0.5),(Item=(Type=Bone,Name="upperarm_l"),Weight=0.5))
    Weight=1
    ExecutePin=@RigVMFunction_Sequence_1.C
   )
   > &RigUnit_RotationConstraintLocalSpaceOffset(
    AdvancedSettings=(InterpolationType=Average,RotationOrderForFilter=XZY)
    Child=(Type=Bone,Name="elbow_r")
    Filter=(bX=True,bY=True,bZ=True)
    Parents=((Item=(Type=Bone,Name="lowerarm_r"),Weight=0.5),(Item=(Type=Bone,Name="upperarm_r"),Weight=0.5))
    Weight=1
   )
   > n1 = &RigUnit_SphericalPoseReader(
    ActiveRegionScaleFactors=(PositiveWidth=1,NegativeWidth=1,PositiveHeight=1,NegativeHeight=1)
    ActiveRegionSize=0.001
    Debug=(bDrawDebug=false,bDraw2D=False,bDrawLocalAxes=False,DebugScale=25,DebugSegments=20,DebugThickness=0.25)
    DriverAxis=(X=1,Y=0,Z=0)
    DriverItem=(Type=Bone,Name="lowerarm_l")
    FalloffRegionScaleFactors=(PositiveWidth=1,NegativeWidth=1,PositiveHeight=1,NegativeHeight=1)
    FalloffSize=0.5
    FlipHeightScaling=False
    FlipWidthScaling=False
    OptionalParentItem=(Type=Bone,Name="None")
    RotationOffset=(X=90,Y=0,Z=0)
   )
   > &RigUnit_ModifyTransforms(
    ItemToModify=((Item=(Type=Bone,Name="elbow_fr_l"),Transform=(Rotation=(X=0,Y=0,Z=0,W=-1),Translation=(X=1.3766,Y=4.2219,Z=0),Scale3D=(X=1,Y=1,Z=1))))
    WeightMinimum=0
    Weight=@n1.OutputParam
   )
   > &RigUnit_ModifyTransforms(
    ItemToModify=((Item=(Type=Bone,Name="elbow_ba_l"),Transform=(Rotation=(X=0,Y=0,Z=-0.0725,W=-0.9974),Translation=(X=-4.1649,Y=-2.719,Z=0),Scale3D=(X=1,Y=1,Z=1))))
    WeightMinimum=0
    Weight=@n1.OutputParam
   )
   > &RigUnit_ModifyTransforms(
    ItemToModify=((Item=(Type=Bone,Name="bicep_l"),Transform=(Rotation=(X=0,Y=0,Z=0,W=1),Translation=(X=-3.2843,Y=0.743,Z=0),Scale3D=(X=1,Y=1,Z=1))))
    WeightMinimum=0
    Weight=@n1.OutputParam
   )
   > &RigUnit_ModifyTransforms(
    ItemToModify=((Item=(Type=Bone,Name="tricep_l"),Transform=(Rotation=(X=0,Y=0,Z=0,W=1),Translation=(X=0,Y=-0.4249,Z=0),Scale3D=(X=1,Y=1,Z=1))))
    WeightMinimum=0
    Weight=@n1.OutputParam
   )
   > n2 = &RigUnit_SphericalPoseReader(
    ActiveRegionScaleFactors=(PositiveWidth=1,NegativeWidth=1,PositiveHeight=1,NegativeHeight=1)
    ActiveRegionSize=0.001
    Debug=(bDrawDebug=false,bDraw2D=False,bDrawLocalAxes=False,DebugScale=25,DebugSegments=20,DebugThickness=0.25)
    DriverAxis=(X=1,Y=0,Z=0)
    DriverItem=(Type=Bone,Name="lowerarm_r")
    FalloffRegionScaleFactors=(PositiveWidth=1,NegativeWidth=1,PositiveHeight=1,NegativeHeight=1)
    FalloffSize=0.5
    FlipHeightScaling=False
    FlipWidthScaling=False
    OptionalParentItem=(Type=Bone,Name="None")
    RotationOffset=(X=90,Y=0,Z=0)
   )
   > &RigUnit_ModifyTransforms(
    ItemToModify=((Item=(Type=Bone,Name="elbow_fr_r"),Transform=(Rotation=(X=0,Y=0,Z=0.031,W=-0.9995),Translation=(X=-3.3932,Y=-5.411,Z=0),Scale3D=(X=0.671,Y=1,Z=1))))
    WeightMinimum=0
    Weight=@n2.OutputParam
   )
   > &RigUnit_ModifyTransforms(
    ItemToModify=((Item=(Type=Bone,Name="elbow_ba_r"),Transform=(Rotation=(X=0,Y=0,Z=-0.0782,W=0.9969),Translation=(X=4.7647,Y=-1.4237,Z=0),Scale3D=(X=1,Y=1,Z=1))))
    WeightMinimum=0
    Weight=@n2.OutputParam
   )
   > &RigUnit_ModifyTransforms(
    ItemToModify=((Item=(Type=Bone,Name="bicep_r"),Transform=(Rotation=(X=0,Y=0,Z=0,W=1),Translation=(X=3.2843,Y=-0.743,Z=0),Scale3D=(X=1,Y=1,Z=1))))
    WeightMinimum=0
    Weight=@n2.OutputParam
   )
   > &RigUnit_ModifyTransforms(
    ItemToModify=((Item=(Type=Bone,Name="tricep_r"),Transform=(Rotation=(X=0,Y=0,Z=0,W=1),Translation=(X=0,Y=0.4249,Z=0),Scale3D=(X=1,Y=1,Z=1))))
    WeightMinimum=0
    Weight=@n2.OutputParam
   )
   > &RigUnit_RotationConstraintLocalSpaceOffset(
    AdvancedSettings=(InterpolationType=Average,RotationOrderForFilter=XZY)
    Child=(Type=Bone,Name="knee_fr_base_l")
    Filter=(bX=True,bY=True,bZ=True)
    Parents=((Item=(Type=Bone,Name="calf_l"),Weight=0.667),(Item=(Type=Bone,Name="thigh_l"),Weight=0.333))
    Weight=1
   )
   > &RigUnit_RotationConstraintLocalSpaceOffset(
    AdvancedSettings=(InterpolationType=Average,RotationOrderForFilter=XZY)
    Child=(Type=Bone,Name="knee_back_base_l")
    Filter=(bX=True,bY=True,bZ=True)
    Parents=((Item=(Type=Bone,Name="calf_l"),Weight=0.75),(Item=(Type=Bone,Name="thigh_l"),Weight=0.25))
    Weight=1
   )
   > &RigUnit_RotationConstraintLocalSpaceOffset(
    AdvancedSettings=(InterpolationType=Average,RotationOrderForFilter=XZY)
    Child=(Type=Bone,Name="knee_pad_base_l")
    Filter=(bX=True,bY=True,bZ=True)
    Parents=((Item=(Type=Bone,Name="calf_l"),Weight=0.75),(Item=(Type=Bone,Name="thigh_l"),Weight=0.25))
    Weight=1
   )
   > &RigUnit_RotationConstraintLocalSpaceOffset(
    AdvancedSettings=(InterpolationType=Average,RotationOrderForFilter=XZY)
    Child=(Type=Bone,Name="knee_wrinkle_base_l")
    Filter=(bX=True,bY=True,bZ=True)
    Parents=((Item=(Type=Bone,Name="calf_l"),Weight=0.5),(Item=(Type=Bone,Name="thigh_l"),Weight=0.5))
    Weight=1
   )
   > &RigUnit_RotationConstraintLocalSpaceOffset(
    AdvancedSettings=(InterpolationType=Average,RotationOrderForFilter=XZY)
    Child=(Type=Bone,Name="knee_fr_base_r")
    Filter=(bX=True,bY=True,bZ=True)
    Parents=((Item=(Type=Bone,Name="calf_r"),Weight=0.667),(Item=(Type=Bone,Name="thigh_r"),Weight=0.333))
    Weight=1
   )
   > &RigUnit_RotationConstraintLocalSpaceOffset(
    AdvancedSettings=(InterpolationType=Average,RotationOrderForFilter=XZY)
    Child=(Type=Bone,Name="knee_back_base_r")
    Filter=(bX=True,bY=True,bZ=True)
    Parents=((Item=(Type=Bone,Name="calf_r"),Weight=0.75),(Item=(Type=Bone,Name="thigh_r"),Weight=0.25))
    Weight=1
   )
   > &RigUnit_RotationConstraintLocalSpaceOffset(
    AdvancedSettings=(InterpolationType=Average,RotationOrderForFilter=XZY)
    Child=(Type=Bone,Name="knee_pad_base_r")
    Filter=(bX=True,bY=True,bZ=True)
    Parents=((Item=(Type=Bone,Name="calf_r"),Weight=0.75),(Item=(Type=Bone,Name="thigh_r"),Weight=0.25))
    Weight=1
   )
   > &RigUnit_RotationConstraintLocalSpaceOffset(
    AdvancedSettings=(InterpolationType=Average,RotationOrderForFilter=XZY)
    Child=(Type=Bone,Name="knee_wrinkle_base_r")
    Filter=(bX=True,bY=True,bZ=True)
    Parents=((Item=(Type=Bone,Name="calf_r"),Weight=0.5),(Item=(Type=Bone,Name="thigh_r"),Weight=0.5))
    Weight=1
   )
   > n3 = &RigUnit_SphericalPoseReader(
    ActiveRegionScaleFactors=(PositiveWidth=0.1,NegativeWidth=0.1,PositiveHeight=0.1,NegativeHeight=0.1)
    ActiveRegionSize=0.1
    Debug=(bDrawDebug=false,bDraw2D=false,DebugScale=25,DebugSegments=20,DebugThickness=0.25)
    DriverAxis=(X=-1,Y=0,Z=0)
    DriverItem=(Type=Bone,Name="calf_l")
    FalloffRegionScaleFactors=(PositiveWidth=0,NegativeWidth=1.5,PositiveHeight=1.5,NegativeHeight=1.5)
    FalloffSize=0.6634
    OptionalParentItem=(Type=Bone,Name="thigh_l")
    RotationOffset=(X=-90,Y=0,Z=90)
   )
   > &RigUnit_ModifyTransforms(
    ItemToModify=((Item=(Type=Bone,Name="knee_ba_l"),Transform=(Rotation=(X=0,Y=0,Z=0,W=1),Translation=(X=-9,Y=-15,Z=0),Scale3D=(X=1,Y=1,Z=1))))
    WeightMinimum=0
    Weight=@n3.OutputParam
   )
   > &RigUnit_ModifyTransforms(
    ItemToModify=((Item=(Type=Bone,Name="knee_pad_l"),Transform=(Rotation=(X=0,Y=0,Z=-0.061,W=0.9981),Translation=(X=0,Y=-8,Z=0),Scale3D=(X=1,Y=1,Z=1))))
    WeightMinimum=0
    Weight=@n3.OutputParam
   )
   > &RigUnit_ModifyTransforms(
    ItemToModify=((Item=(Type=Bone,Name="knee_fr_l"),Transform=(Rotation=(X=0,Y=0,Z=-0.061,W=0.9981),Translation=(X=0,Y=-9.5,Z=0),Scale3D=(X=1,Y=1,Z=1))))
    WeightMinimum=0
    Weight=@n3.OutputParam
   )
   > &RigUnit_ModifyTransforms(
    ItemToModify=((Item=(Type=Bone,Name="knee_wrinkle_l"),Transform=(Rotation=(X=0,Y=0,Z=-0.1736,W=0.9848),Translation=(X=-8,Y=-12,Z=0),Scale3D=(X=1,Y=1,Z=1))))
    WeightMinimum=0
    Weight=@n3.OutputParam
   )
   > n4 = &RigUnit_SphericalPoseReader(
    ActiveRegionScaleFactors=(PositiveWidth=0.1,NegativeWidth=0.1,PositiveHeight=0.1,NegativeHeight=0.1)
    ActiveRegionSize=0.1
    Debug=(bDrawDebug=false,bDraw2D=false,DebugScale=25,DebugSegments=20,DebugThickness=0.25)
    DriverAxis=(X=1,Y=0,Z=0)
    DriverItem=(Type=Bone,Name="calf_r")
    FalloffRegionScaleFactors=(PositiveWidth=0,NegativeWidth=1.5,PositiveHeight=1.5,NegativeHeight=1.5)
    FalloffSize=0.6634
    OptionalParentItem=(Type=Bone,Name="thigh_r")
    RotationOffset=(X=-90,Y=0,Z=-90)
   )
   > &RigUnit_ModifyTransforms(
    ItemToModify=((Item=(Type=Bone,Name="knee_ba_r"),Transform=(Rotation=(X=0,Y=0,Z=0,W=1),Translation=(X=11,Y=13,Z=0),Scale3D=(X=0.2,Y=1,Z=1))))
    WeightMinimum=0
    Weight=@n4.OutputParam
   )
   > &RigUnit_ModifyTransforms(
    ItemToModify=((Item=(Type=Bone,Name="knee_pad_r"),Transform=(Rotation=(X=0,Y=0,Z=0.0436,W=0.999),Translation=(X=0,Y=8,Z=0),Scale3D=(X=1,Y=1,Z=1))))
    WeightMinimum=0
    Weight=@n4.OutputParam
   )
   > &RigUnit_ModifyTransforms(
    ItemToModify=((Item=(Type=Bone,Name="knee_fr_r"),Transform=(Rotation=(X=0,Y=0,Z=0,W=1),Translation=(X=0,Y=7,Z=0),Scale3D=(X=1,Y=1,Z=1))))
    WeightMinimum=0
    Weight=@n4.OutputParam
   )
   > &RigUnit_ModifyTransforms(
    ItemToModify=((Item=(Type=Bone,Name="knee_wrinkle_r"),Transform=(Rotation=(X=0,Y=0,Z=0,W=1),Translation=(X=10,Y=5,Z=0),Scale3D=(X=1,Y=1,Z=1))))
    WeightMinimum=0
    Weight=@n4.OutputParam
   )
   n5 = &RigUnit_SphericalPoseReader(
    ActiveRegionScaleFactors=(PositiveWidth=1,NegativeWidth=1,PositiveHeight=1,NegativeHeight=1)
    ActiveRegionSize=0.001
    Debug=(bDrawDebug=false,bDraw2D=False,bDrawLocalAxes=False,DebugScale=25,DebugSegments=20,DebugThickness=0.25)
    DriverAxis=(X=1,Y=0,Z=0)
    DriverItem=(Type=Bone,Name="upperarm_l")
    FalloffRegionScaleFactors=(PositiveWidth=1,NegativeWidth=1,PositiveHeight=1,NegativeHeight=1)
    FalloffSize=0.5
    FlipHeightScaling=False
    FlipWidthScaling=False
    OptionalParentItem=(Type=Bone,Name="None")
    RotationOffset=(X=0,Y=0,Z=0)
    ExecutePin=@RigVMFunction_Sequence_1.E
   )
   > &RigUnit_ModifyTransforms(
    ItemToModify=((Item=(Type=Bone,Name="def_arm_l_up_tp"),Transform=(Rotation=(X=0,Y=-0.364,Z=0,W=0.9314),Translation=(X=-6.3873,Y=0,Z=6.234),Scale3D=(X=1,Y=1,Z=1))))
    WeightMinimum=0
    Weight=@n5.OutputParam
   )
   > &RigUnit_ModifyTransforms(
    ItemToModify=((Item=(Type=Bone,Name="def_arm_l_up_fr"),Transform=(Rotation=(X=0,Y=-0.2943,Z=0,W=0.9557),Translation=(X=-0.298,Y=0,Z=5.9308),Scale3D=(X=1,Y=1,Z=1))))
    WeightMinimum=0
    Weight=@n5.OutputParam
   )
   > &RigUnit_ModifyTransforms(
    ItemToModify=((Item=(Type=Bone,Name="def_arm_l_up_ba"),Transform=(Rotation=(X=0,Y=-0.2919,Z=0,W=0.9564),Translation=(X=-1.8249,Y=0,Z=6.7006),Scale3D=(X=1,Y=1,Z=1))))
    WeightMinimum=0
    Weight=@n5.OutputParam
   )
   > n6 = &RigUnit_SphericalPoseReader(
    ActiveRegionScaleFactors=(PositiveWidth=1,NegativeWidth=1,PositiveHeight=1,NegativeHeight=1)
    ActiveRegionSize=0.001
    Debug=(bDrawDebug=false,bDraw2D=False,bDrawLocalAxes=False,DebugScale=25,DebugSegments=20,DebugThickness=0.25)
    DriverAxis=(X=1,Y=0,Z=0)
    DriverItem=(Type=Bone,Name="upperarm_l")
    FalloffRegionScaleFactors=(PositiveWidth=1,NegativeWidth=1,PositiveHeight=1,NegativeHeight=1)
    FalloffSize=0.5
    FlipHeightScaling=False
    FlipWidthScaling=False
    OptionalParentItem=(Type=Bone,Name="None")
    RotationOffset=(X=90,Y=0,Z=0)
   )
   > &RigUnit_ModifyTransforms(
    ItemToModify=((Item=(Type=Bone,Name="def_arm_l_up_fr"),Transform=(Rotation=(X=0,Y=0,Z=0.3619,W=0.9322),Translation=(X=-9.7104,Y=7.4518,Z=0),Scale3D=(X=0.8987,Y=1,Z=1))))
    WeightMinimum=0
    Weight=@n6.OutputParam
   )
   > &RigUnit_ModifyTransforms(
    ItemToModify=((Item=(Type=Bone,Name="bicep_l"),Transform=(Rotation=(X=0,Y=0,Z=0,W=1),Translation=(X=2.5884,Y=1.0949,Z=0),Scale3D=(X=1,Y=1,Z=1))))
    WeightMinimum=0
    Weight=@n6.OutputParam
   )
   > &RigUnit_ModifyTransforms(
    ItemToModify=((Item=(Type=Bone,Name="def_arm_l_up_tp"),Transform=(Rotation=(X=0.0729,Y=-0.102,Z=0.5833,W=0.8025),Translation=(X=4.4244,Y=2.5399,Z=0),Scale3D=(X=1,Y=1,Z=1))))
    WeightMinimum=0
    Weight=@n6.OutputParam
   )
   > &RigUnit_ModifyTransforms(
    ItemToModify=((Item=(Type=Bone,Name="def_arm_l_up_ba"),Transform=(Rotation=(X=0,Y=0,Z=0.3603,W=0.9328),Translation=(X=8.3993,Y=5.5607,Z=0),Scale3D=(X=1,Y=1,Z=1))))
    WeightMinimum=0
    Weight=@n6.OutputParam
   )
   > n7 = &RigUnit_SphericalPoseReader(
    ActiveRegionScaleFactors=(PositiveWidth=1,NegativeWidth=1,PositiveHeight=1,NegativeHeight=1)
    ActiveRegionSize=0.001
    Debug=(bDrawDebug=false,bDraw2D=False,bDrawLocalAxes=False,DebugScale=25,DebugSegments=20,DebugThickness=0.25)
    DriverAxis=(X=1,Y=0,Z=0)
    DriverItem=(Type=Bone,Name="upperarm_l")
    FalloffRegionScaleFactors=(PositiveWidth=1,NegativeWidth=1,PositiveHeight=1,NegativeHeight=1)
    FalloffSize=0.5
    FlipHeightScaling=False
    FlipWidthScaling=False
    OptionalParentItem=(Type=Bone,Name="None")
    RotationOffset=(X=-90,Y=0,Z=0)
   )
   > &RigUnit_ModifyTransforms(
    ItemToModify=((Item=(Type=Bone,Name="def_arm_l_up_ba"),Transform=(Rotation=(X=0,Y=0,Z=-0.301,W=0.9536),Translation=(X=-9.7887,Y=-5.2744,Z=0),Scale3D=(X=1,Y=1,Z=1))))
    WeightMinimum=0
    Weight=@n7.OutputParam
   )
   > &RigUnit_ModifyTransforms(
    ItemToModify=((Item=(Type=Bone,Name="def_arm_l_up_tp"),Transform=(Rotation=(X=0,Y=0,Z=-0.202,W=0.9794),Translation=(X=3.1761,Y=-6.8939,Z=0),Scale3D=(X=1,Y=1,Z=1))))
    WeightMinimum=0
    Weight=@n7.OutputParam
   )
   > &RigUnit_ModifyTransforms(
    ItemToModify=((Item=(Type=Bone,Name="def_arm_l_up_fr"),Transform=(Rotation=(X=0,Y=0,Z=-0.2688,W=0.9632),Translation=(X=6.4704,Y=-3.9161,Z=0),Scale3D=(X=1,Y=1,Z=1))))
    WeightMinimum=0
    Weight=@n7.OutputParam
   )
   > &RigUnit_SphericalPoseReader(
    ActiveRegionScaleFactors=(PositiveWidth=1,NegativeWidth=1,PositiveHeight=1,NegativeHeight=1)
    ActiveRegionSize=0.001
    Debug=(bDrawDebug=false,bDraw2D=False,bDrawLocalAxes=False,DebugScale=25,DebugSegments=20,DebugThickness=0.25)
    DriverAxis=(X=1,Y=0,Z=0)
    DriverItem=(Type=Bone,Name="upperarm_r")
    FalloffRegionScaleFactors=(PositiveWidth=1,NegativeWidth=1,PositiveHeight=1,NegativeHeight=1)
    FalloffSize=0.27
    FlipHeightScaling=False
    FlipWidthScaling=False
    OptionalParentItem=(Type=Bone,Name="None")
    RotationOffset=(X=0,Y=60,Z=-60)
   )
   > n8 = &RigUnit_SphericalPoseReader(
    ActiveRegionScaleFactors=(PositiveWidth=1,NegativeWidth=1,PositiveHeight=1,NegativeHeight=1)
    ActiveRegionSize=0.001
    Debug=(bDrawDebug=false,bDraw2D=False,bDrawLocalAxes=False,DebugScale=25,DebugSegments=20,DebugThickness=0.25)
    DriverAxis=(X=1,Y=0,Z=0)
    DriverItem=(Type=Bone,Name="upperarm_r")
    FalloffRegionScaleFactors=(PositiveWidth=1,NegativeWidth=1,PositiveHeight=1,NegativeHeight=1)
    FalloffSize=0.5
    FlipHeightScaling=False
    FlipWidthScaling=False
    OptionalParentItem=(Type=Bone,Name="None")
    RotationOffset=(X=0,Y=0,Z=0)
   )
   > &RigUnit_ModifyTransforms(
    ItemToModify=((Item=(Type=Bone,Name="def_arm_r_up_tp"),Transform=(Rotation=(X=0,Y=0.3337,Z=0,W=-0.9427),Translation=(X=6.0709,Y=0,Z=-6.3272),Scale3D=(X=1,Y=1,Z=1))))
    WeightMinimum=0
    Weight=@n8.OutputParam
   )
   > &RigUnit_ModifyTransforms(
    ItemToModify=((Item=(Type=Bone,Name="def_arm_r_up_fr"),Transform=(Rotation=(X=0,Y=-0.2943,Z=0,W=0.9557),Translation=(X=-0.298,Y=0,Z=-5.9308),Scale3D=(X=1,Y=1,Z=1))))
    WeightMinimum=0
    Weight=@n8.OutputParam
   )
   > &RigUnit_ModifyTransforms(
    ItemToModify=((Item=(Type=Bone,Name="def_arm_r_up_ba"),Transform=(Rotation=(X=0,Y=-0.2919,Z=0,W=0.9564),Translation=(X=1.8249,Y=0,Z=-6.7006),Scale3D=(X=1,Y=1,Z=1))))
    WeightMinimum=0
    Weight=@n8.OutputParam
   )
   > n9 = &RigUnit_SphericalPoseReader(
    ActiveRegionScaleFactors=(PositiveWidth=1,NegativeWidth=1,PositiveHeight=1,NegativeHeight=1)
    ActiveRegionSize=0.001
    Debug=(bDrawDebug=false,bDraw2D=False,bDrawLocalAxes=False,DebugScale=25,DebugSegments=20,DebugThickness=0.25)
    DriverAxis=(X=1,Y=0,Z=0)
    DriverItem=(Type=Bone,Name="upperarm_r")
    FalloffRegionScaleFactors=(PositiveWidth=1,NegativeWidth=1,PositiveHeight=1,NegativeHeight=1)
    FalloffSize=0.5
    FlipHeightScaling=False
    FlipWidthScaling=False
    OptionalParentItem=(Type=Bone,Name="None")
    RotationOffset=(X=90,Y=0,Z=0)
   )
   > &RigUnit_ModifyTransforms(
    ItemToModify=((Item=(Type=Bone,Name="def_arm_r_up_fr"),Transform=(Rotation=(X=-0.0054,Y=-0.0042,Z=0.3813,W=0.9244),Translation=(X=7.9778,Y=-0.3701,Z=-0.0002),Scale3D=(X=-0.8747,Y=1,Z=1))))
    WeightMinimum=0
    Weight=@n9.OutputParam
   )
   > &RigUnit_ModifyTransforms(
    ItemToModify=((Item=(Type=Bone,Name="bicep_r"),Transform=(Rotation=(X=0,Y=0,Z=0,W=1),Translation=(X=-3.8879,Y=0,Z=0),Scale3D=(X=0.5878,Y=1,Z=1))))
    WeightMinimum=0
    Weight=@n9.OutputParam
   )
   > &RigUnit_ModifyTransforms(
    ItemToModify=((Item=(Type=Bone,Name="def_arm_r_up_tp"),Transform=(Rotation=(X=-0.1746,Y=0.1127,Z=-0.3483,W=-0.9141),Translation=(X=0,Y=0,Z=0),Scale3D=(X=1,Y=1,Z=1))))
    WeightMinimum=0
    Weight=@n9.OutputParam
   )
   > &RigUnit_ModifyTransforms(
    ItemToModify=((Item=(Type=Bone,Name="def_arm_r_up_ba"),Transform=(Rotation=(X=0.0122,Y=0.0293,Z=-0.3849,W=-0.9224),Translation=(X=-7.4389,Y=-3.7057,Z=-0.0538),Scale3D=(X=1,Y=1,Z=1))))
    WeightMinimum=0
    Weight=@n9.OutputParam
   )
   > n10 = &RigUnit_SphericalPoseReader(
    ActiveRegionScaleFactors=(PositiveWidth=1,NegativeWidth=1,PositiveHeight=1,NegativeHeight=1)
    ActiveRegionSize=0.001
    Debug=(bDrawDebug=false,bDraw2D=False,bDrawLocalAxes=False,DebugScale=25,DebugSegments=20,DebugThickness=0.25)
    DriverAxis=(X=1,Y=0,Z=0)
    DriverItem=(Type=Bone,Name="upperarm_r")
    FalloffRegionScaleFactors=(PositiveWidth=1,NegativeWidth=1,PositiveHeight=1,NegativeHeight=1)
    FalloffSize=0.5
    FlipHeightScaling=False
    FlipWidthScaling=False
    OptionalParentItem=(Type=Bone,Name="None")
    RotationOffset=(X=-90,Y=0,Z=0)
   )
   > &RigUnit_ModifyTransforms(
    ItemToModify=((Item=(Type=Bone,Name="def_arm_r_up_ba"),Transform=(Rotation=(X=0,Y=0,Z=0.3091,W=-0.951),Translation=(X=10.2472,Y=7.7939,Z=0),Scale3D=(X=0.2697,Y=1,Z=1))))
    WeightMinimum=0
    Weight=@n10.OutputParam
   )
   > &RigUnit_ModifyTransforms(
    ItemToModify=((Item=(Type=Bone,Name="def_arm_r_up_tp"),Transform=(Rotation=(X=-0.0096,Y=-0.0249,Z=-0.3598,W=0.9326),Translation=(X=0.2797,Y=1.2864,Z=0),Scale3D=(X=1,Y=1,Z=1))))
    WeightMinimum=0
    Weight=@n10.OutputParam
   )
   > &RigUnit_ModifyTransforms(
    ItemToModify=((Item=(Type=Bone,Name="def_arm_r_up_fr"),Transform=(Rotation=(X=0,Y=0,Z=-0.2522,W=0.9677),Translation=(X=-5.9172,Y=2.4633,Z=0),Scale3D=(X=1,Y=1,Z=1))))
    WeightMinimum=0
    Weight=@n10.OutputParam
   )
   UpperArm Readers = RigVMFunctionReferenceNode(
    Dn Post Viz=false
    Dn Pre Viz=false
    Dn Viz=false
    Driver Item=(Type=Bone,Name="upperarm_l")
    Fr Dn Viz=false
    Fr Post Viz=false
    Fr Pre Viz=false
    Fr Up Viz=false
    Fr Viz=false
    Up Post Viz=false
    Up Pre Viz=false
    Up Viz=false
    ExecuteContext=@RigVMFunction_Sequence_1.J
   )
   n11 = &RigUnit_GetTransform(Item=(Type=Bone,Name="shoulder_pad_l_TWIST"), bInitial=false)
   Multiply = &RigVMFunction_MathDoubleMul(B=-25, A=@"UpperArm Readers".Up)
   Add_6 = RigVMAggregateNode(
    D=(X=0,Y=0,Z=0)
    C=/Script/RigVM.RigVMFunction_MathVectorScale(
     Factor=-1.5
     Value=/Script/RigVM.RigVMFunction_MathTransformRotateVector(Vector=(X=0,Y=1,Z=0), Transform=@n11.Transform) {is=struct RigVMUnitNode, template="Rotate Vector::Execute(in Transform,in Vector,out Result)"}.Result
    ) {is=struct RigVMUnitNode, template="Scale::Execute(in Value,in Factor,out Result)"}.Result
    B=&RigVMFunction_MathVectorScale(
     Factor=@Multiply.Result
     Value=&RigVMFunction_MathTransformRotateVector(Vector=(X=1,Y=0,Z=0), Transform=@n11.Transform) {template="Rotate Vector::Execute(in Transform,in Vector,out Result)"}.Result
    ) {template="Scale::Execute(in Value,in Factor,out Result)"}.Result
    A=RigVMVariableNode() {variable=Pad L Up Pos}.Value
   )
   > RigVMVariableNode(Value=@Add_6.Result) {variable=Pad L Up Pos}
   n12 = &RigVMFunction_MathDoubleMul(B=-25, A=@"UpperArm Readers".Up Post)
   n13 = &RigVMFunction_MathTransformRotateVector(
    Vector=(X=1,Y=0,Z=0)
    Transform=&RigUnit_GetTransform(Item=(Type=Bone,Name="shoulder_pad_l_TWIST"), bInitial=false).Transform
   ) {template="Rotate Vector::Execute(in Transform,in Vector,out Result)"}
   > RigVMVariableNode(Value=&RigVMFunction_MathVectorAdd(
     B=&RigVMFunction_MathVectorScale(Factor=@n12.Result, Value=@n13.Result) {template="Scale::Execute(in Value,in Factor,out Result)"}.Result
     A=RigVMVariableNode() {variable=Pad L Up Pos}.Value
    ) {template="Add::Execute(in A,in B,out Result)"}.Result) {variable=Pad L Up Pos}
   n14 = &RigVMFunction_MathDoubleMul(B=25, A=@"UpperArm Readers".Dn)
   n15 = &RigVMFunction_MathTransformRotateVector(
    Vector=(X=1,Y=0,Z=0)
    Transform=&RigUnit_GetTransform(Item=(Type=Bone,Name="shoulder_pad_l_TWIST"), bInitial=false).Transform
   ) {template="Rotate Vector::Execute(in Transform,in Vector,out Result)"}
   > RigVMVariableNode(Value=&RigVMFunction_MathVectorAdd(
     B=&RigVMFunction_MathVectorScale(Factor=@n14.Result, Value=@n15.Result) {template="Scale::Execute(in Value,in Factor,out Result)"}.Result
     A=RigVMVariableNode() {variable=Pad L Up Pos}.Value
    ) {template="Add::Execute(in A,in B,out Result)"}.Result) {variable=Pad L Up Pos}
   n16 = &RigVMFunction_MathDoubleMul(B=25, A=@"UpperArm Readers".Dn Post)
   n17 = &RigVMFunction_MathTransformRotateVector(
    Vector=(X=1,Y=0,Z=0)
    Transform=&RigUnit_GetTransform(Item=(Type=Bone,Name="shoulder_pad_l_TWIST"), bInitial=false).Transform
   ) {template="Rotate Vector::Execute(in Transform,in Vector,out Result)"}
   > RigVMVariableNode(Value=&RigVMFunction_MathVectorAdd(
     B=&RigVMFunction_MathVectorScale(Factor=@n16.Result, Value=@n17.Result) {template="Scale::Execute(in Value,in Factor,out Result)"}.Result
     A=RigVMVariableNode() {variable=Pad L Up Pos}.Value
    ) {template="Add::Execute(in A,in B,out Result)"}.Result) {variable=Pad L Up Pos}
   > /Script/RigVM.RigVMFunction_DebugTransformMutableNoSpace(
    Color=(R=1,G=0,B=0,A=1)
    DebugDrawSettings=(DepthPriority=SDPG_Foreground,Lifetime=-1)
    Mode=Point
    Scale=10
    Thickness=0
    Transform=(Rotation=(X=0,Y=0,Z=0,W=1),Translation=(X=0,Y=0,Z=0),Scale3D=(X=1,Y=1,Z=1))
    WorldOffset=(Rotation=(X=0,Y=0,Z=0,W=1),Translation=(X=0,Y=0,Z=0),Scale3D=(X=1,Y=1,Z=1))
    bEnabled=false
    Transform.Translation=RigVMVariableNode() {variable=Pad L Up Pos}.Value
   ) {is=struct RigVMUnitNode}
   > UpperArm Readers_1 = RigVMFunctionReferenceNode(
    Dn Post Viz=false
    Dn Pre Viz=false
    Dn Viz=false
    Driver Item=(Type=Bone,Name="upperarm_r")
    Fr Dn Viz=false
    Fr Post Viz=false
    Fr Pre Viz=false
    Fr Up Viz=false
    Fr Viz=false
    Up Post Viz=false
    Up Pre Viz=false
    Up Viz=false
   )
   n18 = &RigUnit_GetTransform(Item=(Type=Bone,Name="shoulder_pad_r_TWIST"), bInitial=false)
   Add_10 = RigVMAggregateNode(
    C=&RigVMFunction_MathVectorScale(
     Factor=1.5
     Value=&RigVMFunction_MathTransformRotateVector(Vector=(X=0,Y=1,Z=0), Transform=@n18.Transform) {template="Rotate Vector::Execute(in Transform,in Vector,out Result)"}.Result
    ) {template="Scale::Execute(in Value,in Factor,out Result)"}.Result
    B=&RigVMFunction_MathVectorScale(
     Factor=&RigVMFunction_MathDoubleMul(B=25, A=@"UpperArm Readers_1".Up).Result
     Value=&RigVMFunction_MathTransformRotateVector(Vector=(X=1,Y=0,Z=0), Transform=@n18.Transform) {template="Rotate Vector::Execute(in Transform,in Vector,out Result)"}.Result
    ) {template="Scale::Execute(in Value,in Factor,out Result)"}.Result
    A=RigVMVariableNode() {variable=Pad R Up Pos}.Value
   )
   > RigVMVariableNode(Value=@Add_10.Result) {variable=Pad R Up Pos}
   n19 = &RigVMFunction_MathDoubleMul(B=25, A=@"UpperArm Readers_1".Up Post)
   n20 = &RigVMFunction_MathTransformRotateVector(
    Vector=(X=1,Y=0,Z=0)
    Transform=&RigUnit_GetTransform(Item=(Type=Bone,Name="shoulder_pad_r_TWIST"), bInitial=false).Transform
   ) {template="Rotate Vector::Execute(in Transform,in Vector,out Result)"}
   > RigVMVariableNode(Value=&RigVMFunction_MathVectorAdd(
     B=&RigVMFunction_MathVectorScale(Factor=@n19.Result, Value=@n20.Result) {template="Scale::Execute(in Value,in Factor,out Result)"}.Result
     A=RigVMVariableNode() {variable=Pad R Up Pos}.Value
    ) {template="Add::Execute(in A,in B,out Result)"}.Result) {variable=Pad R Up Pos}
   n21 = &RigVMFunction_MathDoubleMul(B=-25, A=@"UpperArm Readers_1".Dn)
   n22 = &RigVMFunction_MathTransformRotateVector(
    Vector=(X=1,Y=0,Z=0)
    Transform=&RigUnit_GetTransform(Item=(Type=Bone,Name="shoulder_pad_r_TWIST"), bInitial=false).Transform
   ) {template="Rotate Vector::Execute(in Transform,in Vector,out Result)"}
   > RigVMVariableNode(Value=&RigVMFunction_MathVectorAdd(
     B=&RigVMFunction_MathVectorScale(Factor=@n21.Result, Value=@n22.Result) {template="Scale::Execute(in Value,in Factor,out Result)"}.Result
     A=RigVMVariableNode() {variable=Pad R Up Pos}.Value
    ) {template="Add::Execute(in A,in B,out Result)"}.Result) {variable=Pad R Up Pos}
   n23 = &RigVMFunction_MathDoubleMul(B=-25, A=@"UpperArm Readers_1".Dn Post)
   n24 = &RigVMFunction_MathTransformRotateVector(
    Vector=(X=1,Y=0,Z=0)
    Transform=&RigUnit_GetTransform(Item=(Type=Bone,Name="shoulder_pad_r_TWIST"), bInitial=false).Transform
   ) {template="Rotate Vector::Execute(in Transform,in Vector,out Result)"}
   > RigVMVariableNode(Value=&RigVMFunction_MathVectorAdd(
     B=&RigVMFunction_MathVectorScale(Factor=@n23.Result, Value=@n24.Result) {template="Scale::Execute(in Value,in Factor,out Result)"}.Result
     A=RigVMVariableNode() {variable=Pad R Up Pos}.Value
    ) {template="Add::Execute(in A,in B,out Result)"}.Result) {variable=Pad R Up Pos}
   > &RigVMFunction_DebugTransformMutableNoSpace(
    Color=(R=1,G=0,B=0,A=1)
    DebugDrawSettings=(DepthPriority=SDPG_Foreground,Lifetime=-1)
    Mode=Point
    Scale=10
    Thickness=0
    Transform=(Rotation=(X=0,Y=0,Z=0,W=1),Translation=(X=0,Y=0,Z=0),Scale3D=(X=1,Y=1,Z=1))
    WorldOffset=(Rotation=(X=0,Y=0,Z=0,W=1),Translation=(X=0,Y=0,Z=0),Scale3D=(X=1,Y=1,Z=1))
    bEnabled=false
    Transform.Translation=RigVMVariableNode() {variable=Pad R Up Pos}.Value
   )
   n25 = &RigUnit_SphericalPoseReader(
    ActiveRegionScaleFactors=(PositiveWidth=1,NegativeWidth=1,PositiveHeight=1,NegativeHeight=1)
    ActiveRegionSize=0.001
    Debug=(bDrawDebug=false,bDraw2D=False,bDrawLocalAxes=False,DebugScale=25,DebugSegments=20,DebugThickness=0.25)
    DriverAxis=(X=1,Y=0,Z=0)
    DriverItem=(Type=Bone,Name="thigh_r")
    FalloffRegionScaleFactors=(PositiveWidth=1,NegativeWidth=1,PositiveHeight=1,NegativeHeight=1)
    FalloffSize=0.475
    FlipHeightScaling=False
    FlipWidthScaling=False
    OptionalParentItem=(Type=Bone,Name="None")
    RotationOffset=(X=90,Y=0,Z=0)
    ExecutePin=@RigVMFunction_Sequence_1.G
   )
   > RigVMVariableNode(Value=@n25.OutputParam) {variable=thigh spr r ba}
   n26 = RigVMVariableNode() {variable=thigh spr r ba}
   > &RigUnit_ModifyTransforms(
    ItemToModify=((Item=(Type=Bone,Name="butt_r"),Transform=(Rotation=(X=0,Y=0,Z=0.5,W=0.866),Translation=(X=-20,Y=10,Z=0),Scale3D=(X=1,Y=1,Z=1))))
    WeightMinimum=0
    Weight=@n26.Value
   )
   > &RigUnit_ModifyTransforms(
    ItemToModify=((Item=(Type=Bone,Name="belt_glute_r"),Transform=(Rotation=(X=0,Y=0,Z=0,W=1),Translation=(X=7.5,Y=0,Z=0),Scale3D=(X=1,Y=1,Z=1))))
    WeightMinimum=0
    Weight=@n26.Value
   )
   > &RigUnit_ModifyTransforms(
    ItemToModify=((Item=(Type=Bone,Name="belt_r"),Transform=(Rotation=(X=0,Y=0,Z=0,W=1),Translation=(X=2.5,Y=0,Z=0),Scale3D=(X=1,Y=1,Z=1))))
    WeightMinimum=0
    Weight=@n26.Value
   )
   > &RigUnit_ModifyTransforms(
    ItemToModify=((Item=(Type=Bone,Name="belt_r_rots"),Transform=(Rotation=(X=0,Y=0,Z=0.1736,W=0.9848),Translation=(X=0,Y=0,Z=0),Scale3D=(X=1,Y=1,Z=1))))
    WeightMinimum=0
    Weight=@n26.Value
   )
   > n27 = &RigUnit_SphericalPoseReader(
    ActiveRegionScaleFactors=(PositiveWidth=1,NegativeWidth=1,PositiveHeight=1,NegativeHeight=1)
    ActiveRegionSize=0.001
    Debug=(bDrawDebug=false,bDraw2D=False,bDrawLocalAxes=False,DebugScale=25,DebugSegments=20,DebugThickness=0.25)
    DriverAxis=(X=1,Y=0,Z=0)
    DriverItem=(Type=Bone,Name="thigh_r")
    FalloffRegionScaleFactors=(PositiveWidth=1,NegativeWidth=1,PositiveHeight=1,NegativeHeight=1)
    FalloffSize=0.475
    FlipHeightScaling=False
    FlipWidthScaling=False
    OptionalParentItem=(Type=Bone,Name="None")
    RotationOffset=(X=0,Y=0,Z=0)
   )
   > &RigUnit_ModifyTransforms(
    ItemToModify=((Item=(Type=Bone,Name="pouch_r_MSTR"),Transform=(Rotation=(X=0,Y=0.7071,Z=0,W=0.7071),Translation=(X=0,Y=0,Z=0),Scale3D=(X=1,Y=1,Z=1))))
    WeightMinimum=0
    Weight=@n27.OutputParam
   )
   n28 = &RigUnit_SphericalPoseReader(
    ActiveRegionScaleFactors=(PositiveWidth=1,NegativeWidth=1,PositiveHeight=1,NegativeHeight=1)
    ActiveRegionSize=0.001
    Debug=(bDrawDebug=false,bDraw2D=False,bDrawLocalAxes=False,DebugScale=25,DebugSegments=20,DebugThickness=0.25)
    DriverAxis=(X=1,Y=0,Z=0)
    DriverItem=(Type=Bone,Name="thigh_r")
    FalloffRegionScaleFactors=(PositiveWidth=1,NegativeWidth=1,PositiveHeight=1,NegativeHeight=1)
    FalloffSize=0.25
    FlipHeightScaling=False
    FlipWidthScaling=False
    OptionalParentItem=(Type=Bone,Name="None")
    RotationOffset=(X=-90,Y=0,Z=0)
    ExecutePin=@RigVMFunction_Sequence_1.F
   )
   > RigVMVariableNode(Value=@n28.OutputParam) {variable=thigh spr r fr}
   n29 = RigVMVariableNode() {variable=thigh spr r fr}
   > &RigUnit_ModifyTransforms(
    ItemToModify=((Item=(Type=Bone,Name="belt_leg_r"),Transform=(Rotation=(X=0,Y=0,Z=0,W=1),Translation=(X=8.125,Y=0,Z=0),Scale3D=(X=1,Y=1,Z=1))))
    WeightMinimum=0
    Weight=@n29.Value
   )
   > &RigUnit_ModifyTransforms(
    ItemToModify=((Item=(Type=Bone,Name="belt_fr"),Transform=(Rotation=(X=0,Y=0.1564,Z=0,W=0.9877),Translation=(X=3.75,Y=0,Z=0),Scale3D=(X=1,Y=1,Z=1))))
    WeightMinimum=0
    Weight=@n29.Value
   )
   > &RigUnit_ModifyTransforms(
    ItemToModify=((Item=(Type=Bone,Name="belt_r"),Transform=(Rotation=(X=0,Y=0,Z=0,W=1),Translation=(X=2.5,Y=0,Z=0),Scale3D=(X=1,Y=1,Z=1))))
    WeightMinimum=0
    Weight=@n29.Value
   )
   > &RigUnit_ModifyTransforms(
    ItemToModify=((Item=(Type=Bone,Name="belt_r_rots"),Transform=(Rotation=(X=0,Y=0,Z=-0.0872,W=0.9962),Translation=(X=0,Y=0,Z=0),Scale3D=(X=1,Y=1,Z=1))))
    WeightMinimum=0
    Weight=@n29.Value
   )
   > n30 = &RigUnit_SphericalPoseReader(
    ActiveRegionScaleFactors=(PositiveWidth=1,NegativeWidth=1,PositiveHeight=1,NegativeHeight=1)
    ActiveRegionSize=0.001
    Debug=(bDrawDebug=false,bDraw2D=False,bDrawLocalAxes=False,DebugScale=25,DebugSegments=20,DebugThickness=0.25)
    DriverAxis=(X=1,Y=0,Z=0)
    DriverItem=(Type=Bone,Name="thigh_r")
    FalloffRegionScaleFactors=(PositiveWidth=1,NegativeWidth=1,PositiveHeight=1,NegativeHeight=1)
    FalloffSize=0.475
    FlipHeightScaling=False
    FlipWidthScaling=False
    OptionalParentItem=(Type=Bone,Name="None")
    RotationOffset=(X=-90,Y=0,Z=0)
   )
   > RigVMVariableNode(Value=@n30.OutputParam) {variable=thigh spr r fr}
   n31 = RigVMVariableNode() {variable=thigh spr r fr}
   > &RigUnit_ModifyTransforms(
    ItemToModify=((Item=(Type=Bone,Name="groin_r"),Transform=(Rotation=(X=0,Y=0,Z=0,W=1),Translation=(X=-30,Y=-15,Z=0),Scale3D=(X=1,Y=1,Z=1))))
    WeightMinimum=0
    Weight=@n31.Value
   )
   > &RigUnit_ModifyTransforms(
    ItemToModify=((Item=(Type=Bone,Name="link_a_BASE"),Transform=(Rotation=(X=0,Y=-0.9659,Z=0,W=0.2588),Translation=(X=0,Y=0,Z=0),Scale3D=(X=1,Y=1,Z=1))))
    WeightMinimum=0
    Weight=@n31.Value
   )
   > &RigUnit_ModifyTransforms(
    ItemToModify=((Item=(Type=Bone,Name="link_b_BASE"),Transform=(Rotation=(X=0,Y=-0.9063,Z=0,W=0.4226),Translation=(X=0,Y=0,Z=0),Scale3D=(X=1,Y=1,Z=1))))
    WeightMinimum=0
    Weight=@n31.Value
   )
   > &RigUnit_ModifyTransforms(
    ItemToModify=((Item=(Type=Bone,Name="rope_a_MSTR"),Transform=(Rotation=(X=0,Y=-0.8572,Z=0,W=0.515),Translation=(X=0,Y=0,Z=0),Scale3D=(X=1,Y=1,Z=1))))
    WeightMinimum=0
    Weight=@n31.Value
   )
   > &RigUnit_ModifyTransforms(
    ItemToModify=((Item=(Type=Bone,Name="rope_b_MSTR"),Transform=(Rotation=(X=0,Y=-0.9063,Z=0,W=0.4226),Translation=(X=0,Y=0,Z=0),Scale3D=(X=1,Y=1,Z=1))))
    WeightMinimum=0
    Weight=@n31.Value
   )
   > &RigUnit_ModifyTransforms(
    ItemToModify=((Item=(Type=Bone,Name="rope_c_MSTR"),Transform=(Rotation=(X=0,Y=-0.887,Z=0,W=0.4617),Translation=(X=0,Y=0,Z=0),Scale3D=(X=1,Y=1,Z=1))))
    WeightMinimum=0
    Weight=@n31.Value
   )
   > n32 = &RigUnit_SphericalPoseReader(
    ActiveRegionScaleFactors=(PositiveWidth=1,NegativeWidth=1,PositiveHeight=1,NegativeHeight=1)
    ActiveRegionSize=0.01
    Debug=(bDrawDebug=false,bDraw2D=False,bDrawLocalAxes=False,DebugScale=25,DebugSegments=20,DebugThickness=0.25)
    DriverAxis=(X=1,Y=0,Z=0)
    DriverItem=(Type=Bone,Name="thigh_r")
    FalloffRegionScaleFactors=(PositiveWidth=1,NegativeWidth=1,PositiveHeight=1,NegativeHeight=1)
    FalloffSize=0.5
    FlipHeightScaling=False
    FlipWidthScaling=False
    OptionalParentItem=(Type=Bone,Name="None")
    RotationOffset=(X=-90,Y=0,Z=-90)
   )
   > RigVMVariableNode(Value=@n32.OutputParam) {variable=thigh spr r fr post}
   n33 = RigVMVariableNode() {variable=thigh spr r fr post}
   > &RigUnit_ModifyTransforms(
    ItemToModify=((Item=(Type=Bone,Name="belt_leg_r"),Transform=(Rotation=(X=0,Y=0,Z=0,W=1),Translation=(X=30,Y=0,Z=0),Scale3D=(X=1,Y=1,Z=1))))
    WeightMinimum=0
    Weight=@n33.Value
   )
   > &RigUnit_ModifyTransforms(
    ItemToModify=((Item=(Type=Bone,Name="belt_fr"),Transform=(Rotation=(X=0,Y=0.3827,Z=0,W=0.9239),Translation=(X=12.5,Y=0,Z=0),Scale3D=(X=1,Y=1,Z=1))))
    WeightMinimum=0
    Weight=@n33.Value
   )
   > &RigUnit_ModifyTransforms(
    ItemToModify=((Item=(Type=Bone,Name="belt_r"),Transform=(Rotation=(X=0,Y=0,Z=0,W=1),Translation=(X=15,Y=0,Z=0),Scale3D=(X=1,Y=1,Z=1))))
    WeightMinimum=0
    Weight=@n33.Value
   )
   > &RigUnit_ModifyTransforms(
    ItemToModify=((Item=(Type=Bone,Name="belt_r_rots"),Transform=(Rotation=(X=0,Y=0,Z=-0.3827,W=0.9239),Translation=(X=0,Y=0,Z=0),Scale3D=(X=1,Y=1,Z=1))))
    WeightMinimum=0
    Weight=@n33.Value
   )
   > &RigUnit_ModifyTransforms(
    ItemToModify=((Item=(Type=Bone,Name="groin_r"),Transform=(Rotation=(X=0,Y=0,Z=0,W=1),Translation=(X=-60,Y=-30,Z=0),Scale3D=(X=1,Y=1,Z=1))))
    WeightMinimum=0
    Weight=@n33.Value
   )
   > &RigUnit_ModifyTransforms(
    ItemToModify=((Item=(Type=Bone,Name="link_a_BASE"),Transform=(Rotation=(X=0,Y=-1,Z=0,W=0),Translation=(X=0,Y=0,Z=0),Scale3D=(X=1,Y=1,Z=1))))
    WeightMinimum=0
    Weight=@n33.Value
   )
   > &RigUnit_ModifyTransforms(
    ItemToModify=((Item=(Type=Bone,Name="link_b_BASE"),Transform=(Rotation=(X=0,Y=-1,Z=0,W=0),Translation=(X=0,Y=0,Z=0),Scale3D=(X=1,Y=1,Z=1))))
    WeightMinimum=0
    Weight=@n33.Value
   )
   > &RigUnit_ModifyTransforms(
    ItemToModify=((Item=(Type=Bone,Name="rope_a_MSTR"),Transform=(Rotation=(X=0,Y=-1,Z=0,W=0),Translation=(X=0,Y=0,Z=0),Scale3D=(X=1,Y=1,Z=1))))
    WeightMinimum=0
    Weight=@n33.Value
   )
   > &RigUnit_ModifyTransforms(
    ItemToModify=((Item=(Type=Bone,Name="rope_b_MSTR"),Transform=(Rotation=(X=0,Y=-1,Z=0,W=0),Translation=(X=0,Y=0,Z=0),Scale3D=(X=1,Y=1,Z=1))))
    WeightMinimum=0
    Weight=@n33.Value
   )
   > &RigUnit_ModifyTransforms(
    ItemToModify=((Item=(Type=Bone,Name="rope_c_MSTR"),Transform=(Rotation=(X=0,Y=-1,Z=0,W=0),Translation=(X=0,Y=0,Z=0),Scale3D=(X=1,Y=1,Z=1))))
    WeightMinimum=0
    Weight=@n33.Value
   )
   > &RigUnit_ModifyTransforms(
    ItemToModify=((Item=(Type=Bone,Name="rope_d_MSTR"),Transform=(Rotation=(X=0,Y=-1,Z=0,W=0),Translation=(X=0,Y=0,Z=0),Scale3D=(X=1,Y=1,Z=1))))
    WeightMinimum=0
    Weight=@n33.Value
   )
   > &RigUnit_ModifyTransforms(
    ItemToModify=((Item=(Type=Bone,Name="rope_e_MSTR"),Transform=(Rotation=(X=0,Y=-1,Z=0,W=0),Translation=(X=0,Y=0,Z=0),Scale3D=(X=1,Y=1,Z=1))))
    WeightMinimum=0
    Weight=@n33.Value
   )
   RigVMFunction_ControlFlowBranch = &RigVMFunction_ControlFlowBranch(Condition=RigVMVariableNode() {variable=Debug Controls On}.Value, ExecuteContext=@RigVMFunction_Sequence_1.A)
   &RigUnit_SetTransform(
    Item=(Type=Bone,Name="thigh_r")
    Weight=1
    bInitial=False
    Value=&RigUnit_GetTransform(bInitial=False, Item=RigVMVariableNode() {variable=Thigh R Ctrl}.Value).Transform
    ExecutePin=@RigVMFunction_ControlFlowBranch.True
   )
   > &RigUnit_SetTransform(
    Item=(Type=Bone,Name="thigh_l")
    Weight=1
    bInitial=False
    Value=&RigUnit_GetTransform(bInitial=False, Item=RigVMVariableNode() {variable=Thigh L Ctrl}.Value).Transform
   )
   > For_Each = RigVMDispatchNode(Array=RigVMVariableNode() {variable=Debug Ctrls}.Value) {template="DISPATCH_RigVMDispatch_ArrayIterator(in Array,out Element,out Index,out Count,out Ratio)"}
   > &RigUnit_SetTransform(
    Weight=1
    bInitial=False
    Item=RigVMDispatchNode(Array=RigVMVariableNode() {variable=Debug Bones}.Value, Index=@For_Each.Index) {template="DISPATCH_RigVMDispatch_ArrayGetAtIndex(in Array,in Index,out Element)"}.Element
    Value=&RigUnit_GetTransform(bInitial=False, Item=@For_Each.Element).Transform
   )
   n34 = &RigUnit_SphericalPoseReader(
    ActiveRegionScaleFactors=(PositiveWidth=1,NegativeWidth=1,PositiveHeight=1,NegativeHeight=1)
    ActiveRegionSize=0.001
    Debug=(bDrawDebug=false,bDraw2D=False,bDrawLocalAxes=False,DebugScale=25,DebugSegments=20,DebugThickness=0.25)
    DriverAxis=(X=1,Y=0,Z=0)
    DriverItem=(Type=Bone,Name="thigh_l")
    FalloffRegionScaleFactors=(PositiveWidth=1,NegativeWidth=1,PositiveHeight=1,NegativeHeight=1)
    FalloffSize=0.475
    FlipHeightScaling=False
    FlipWidthScaling=False
    OptionalParentItem=(Type=Bone,Name="None")
    RotationOffset=(X=90,Y=0,Z=0)
    ExecutePin=@RigVMFunction_Sequence_1.I
   )
   > RigVMVariableNode(Value=@n34.OutputParam) {variable=thigh spr l ba}
   n35 = RigVMVariableNode() {variable=thigh spr l ba}
   > &RigUnit_ModifyTransforms(
    ItemToModify=((Item=(Type=Bone,Name="butt_l"),Transform=(Rotation=(X=0,Y=0,Z=0.5,W=0.866),Translation=(X=20,Y=-10,Z=0),Scale3D=(X=1,Y=1,Z=1))))
    WeightMinimum=0
    Weight=@n35.Value
   )
   > &RigUnit_ModifyTransforms(
    ItemToModify=((Item=(Type=Bone,Name="belt_glute_l"),Transform=(Rotation=(X=0,Y=0,Z=0,W=1),Translation=(X=7.5,Y=0,Z=0),Scale3D=(X=1,Y=1,Z=1))))
    WeightMinimum=0
    Weight=@n35.Value
   )
   > &RigUnit_ModifyTransforms(
    ItemToModify=((Item=(Type=Bone,Name="belt_l"),Transform=(Rotation=(X=0,Y=0,Z=0,W=1),Translation=(X=2.5,Y=0,Z=0),Scale3D=(X=1,Y=1,Z=1))))
    WeightMinimum=0
    Weight=@n35.Value
   )
   > &RigUnit_ModifyTransforms(
    ItemToModify=((Item=(Type=Bone,Name="belt_l_rots"),Transform=(Rotation=(X=0,Y=0,Z=0.1736,W=0.9848),Translation=(X=0,Y=0,Z=0),Scale3D=(X=1,Y=1,Z=1))))
    WeightMinimum=0
    Weight=@n35.Value
   )
   > n36 = &RigUnit_SphericalPoseReader(
    ActiveRegionScaleFactors=(PositiveWidth=1,NegativeWidth=1,PositiveHeight=1,NegativeHeight=1)
    ActiveRegionSize=0.001
    Debug=(bDrawDebug=false,bDraw2D=False,bDrawLocalAxes=False,DebugScale=25,DebugSegments=20,DebugThickness=0.25)
    DriverAxis=(X=1,Y=0,Z=0)
    DriverItem=(Type=Bone,Name="thigh_l")
    FalloffRegionScaleFactors=(PositiveWidth=1,NegativeWidth=1,PositiveHeight=1,NegativeHeight=1)
    FalloffSize=0.475
    FlipHeightScaling=False
    FlipWidthScaling=False
    OptionalParentItem=(Type=Bone,Name="None")
    RotationOffset=(X=0,Y=0,Z=0)
   )
   > &RigUnit_ModifyTransforms(
    ItemToModify=((Item=(Type=Bone,Name="pouch_l_MSTR"),Transform=(Rotation=(X=0,Y=-0.7071,Z=0,W=0.7071),Translation=(X=0,Y=0,Z=0),Scale3D=(X=1,Y=1,Z=1))))
    WeightMinimum=0
    Weight=@n36.OutputParam
   )
   n37 = &RigUnit_SphericalPoseReader(
    ActiveRegionScaleFactors=(PositiveWidth=1,NegativeWidth=1,PositiveHeight=1,NegativeHeight=1)
    ActiveRegionSize=0.001
    Debug=(bDrawDebug=false,bDraw2D=False,bDrawLocalAxes=False,DebugScale=25,DebugSegments=20,DebugThickness=0.25)
    DriverAxis=(X=1,Y=0,Z=0)
    DriverItem=(Type=Bone,Name="thigh_l")
    FalloffRegionScaleFactors=(PositiveWidth=1,NegativeWidth=1,PositiveHeight=1,NegativeHeight=1)
    FalloffSize=0.25
    FlipHeightScaling=False
    FlipWidthScaling=False
    OptionalParentItem=(Type=Bone,Name="None")
    RotationOffset=(X=-90,Y=0,Z=0)
    ExecutePin=@RigVMFunction_Sequence_1.H
   )
   > RigVMVariableNode(Value=@n37.OutputParam) {variable=thigh spr l fr}
   n38 = RigVMVariableNode() {variable=thigh spr l fr}
   > &RigUnit_ModifyTransforms(
    ItemToModify=((Item=(Type=Bone,Name="belt_leg_l"),Transform=(Rotation=(X=0,Y=0,Z=0,W=1),Translation=(X=8.125,Y=0,Z=0),Scale3D=(X=1,Y=1,Z=1))))
    WeightMinimum=0
    Weight=@n38.Value
   )
   > &RigUnit_ModifyTransforms(
    ItemToModify=((Item=(Type=Bone,Name="belt_fr"),Transform=(Rotation=(X=0,Y=-0.1564,Z=0,W=0.9877),Translation=(X=3.75,Y=0,Z=0),Scale3D=(X=1,Y=1,Z=1))))
    WeightMinimum=0
    Weight=@n38.Value
   )
   > &RigUnit_ModifyTransforms(
    ItemToModify=((Item=(Type=Bone,Name="belt_l"),Transform=(Rotation=(X=0,Y=0,Z=0,W=1),Translation=(X=2.5,Y=0,Z=0),Scale3D=(X=1,Y=1,Z=1))))
    WeightMinimum=0
    Weight=@n38.Value
   )
   > &RigUnit_ModifyTransforms(
    ItemToModify=((Item=(Type=Bone,Name="belt_l_rots"),Transform=(Rotation=(X=0,Y=0,Z=-0.0872,W=0.9962),Translation=(X=0,Y=0,Z=0),Scale3D=(X=1,Y=1,Z=1))))
    WeightMinimum=0
    Weight=@n38.Value
   )
   > n39 = &RigUnit_SphericalPoseReader(
    ActiveRegionScaleFactors=(PositiveWidth=1,NegativeWidth=1,PositiveHeight=1,NegativeHeight=1)
    ActiveRegionSize=0.001
    Debug=(bDrawDebug=false,bDraw2D=False,bDrawLocalAxes=False,DebugScale=25,DebugSegments=20,DebugThickness=0.25)
    DriverAxis=(X=1,Y=0,Z=0)
    DriverItem=(Type=Bone,Name="thigh_l")
    FalloffRegionScaleFactors=(PositiveWidth=1,NegativeWidth=1,PositiveHeight=1,NegativeHeight=1)
    FalloffSize=0.475
    FlipHeightScaling=False
    FlipWidthScaling=False
    OptionalParentItem=(Type=Bone,Name="None")
    RotationOffset=(X=-90,Y=0,Z=0)
   )
   > &RigUnit_ModifyTransforms(
    ItemToModify=((Item=(Type=Bone,Name="groin_l"),Transform=(Rotation=(X=0,Y=0,Z=0,W=1),Translation=(X=27.5,Y=15,Z=0),Scale3D=(X=1,Y=1,Z=1))))
    WeightMinimum=0
    Weight=@n39.OutputParam
   )
   > n40 = &RigUnit_SphericalPoseReader(
    ActiveRegionScaleFactors=(PositiveWidth=1,NegativeWidth=1,PositiveHeight=1,NegativeHeight=1)
    ActiveRegionSize=0.01
    Debug=(bDrawDebug=false,bDraw2D=False,bDrawLocalAxes=False,DebugScale=25,DebugSegments=20,DebugThickness=0.25)
    DriverAxis=(X=1,Y=0,Z=0)
    DriverItem=(Type=Bone,Name="thigh_l")
    FalloffRegionScaleFactors=(PositiveWidth=1,NegativeWidth=1,PositiveHeight=1,NegativeHeight=1)
    FalloffSize=0.5
    FlipHeightScaling=False
    FlipWidthScaling=False
    OptionalParentItem=(Type=Bone,Name="None")
    RotationOffset=(X=-90,Y=0,Z=-90)
   )
   > RigVMVariableNode(Value=@n40.OutputParam) {variable=thigh spr l fr post}
   n41 = RigVMVariableNode() {variable=thigh spr l fr post}
   > &RigUnit_ModifyTransforms(
    ItemToModify=((Item=(Type=Bone,Name="belt_leg_l"),Transform=(Rotation=(X=0,Y=0,Z=0,W=1),Translation=(X=30,Y=0,Z=0),Scale3D=(X=1,Y=1,Z=1))))
    WeightMinimum=0
    Weight=@n41.Value
   )
   > &RigUnit_ModifyTransforms(
    ItemToModify=((Item=(Type=Bone,Name="belt_fr"),Transform=(Rotation=(X=0,Y=-0.3827,Z=0,W=0.9239),Translation=(X=12.5,Y=0,Z=0),Scale3D=(X=1,Y=1,Z=1))))
    WeightMinimum=0
    Weight=@n41.Value
   )
   > &RigUnit_ModifyTransforms(
    ItemToModify=((Item=(Type=Bone,Name="belt_l"),Transform=(Rotation=(X=0,Y=0,Z=0,W=1),Translation=(X=15,Y=0,Z=0),Scale3D=(X=1,Y=1,Z=1))))
    WeightMinimum=0
    Weight=@n41.Value
   )
   > &RigUnit_ModifyTransforms(
    ItemToModify=((Item=(Type=Bone,Name="belt_l_rots"),Transform=(Rotation=(X=0,Y=0,Z=-0.3827,W=0.9239),Translation=(X=0,Y=0,Z=0),Scale3D=(X=1,Y=1,Z=1))))
    WeightMinimum=0
    Weight=@n41.Value
   )
   > &RigUnit_ModifyTransforms(
    ItemToModify=((Item=(Type=Bone,Name="groin_l"),Transform=(Rotation=(X=0,Y=0,Z=0,W=1),Translation=(X=60,Y=30,Z=0),Scale3D=(X=1,Y=1,Z=1))))
    WeightMinimum=0
    Weight=@n41.Value
   )
   n42 = &RigUnit_SphericalPoseReader(
    ActiveRegionScaleFactors=(PositiveWidth=1,NegativeWidth=1,PositiveHeight=1,NegativeHeight=1)
    ActiveRegionSize=0.001
    Debug=(bDrawDebug=false,bDraw2D=False,bDrawLocalAxes=False,DebugScale=25,DebugSegments=20,DebugThickness=0.25)
    DriverAxis=(X=1,Y=0,Z=0)
    DriverItem=(Type=Bone,Name="clavicle_l")
    FalloffRegionScaleFactors=(PositiveWidth=1,NegativeWidth=1,PositiveHeight=1,NegativeHeight=1)
    FalloffSize=0.5
    FlipHeightScaling=False
    FlipWidthScaling=False
    OptionalParentItem=(Type=Bone,Name="None")
    RotationOffset=(X=0,Y=0,Z=0)
    ExecutePin=@RigVMFunction_Sequence_1.D
   )
   > &RigUnit_ModifyTransforms(
    ItemToModify=((Item=(Type=Bone,Name="None"),Transform=(Rotation=(X=0,Y=0,Z=0,W=1),Translation=(X=0,Y=0,Z=0),Scale3D=(X=1,Y=1,Z=1))),(Item=(Type=Bone,Name="shoulder_pad_l_TWIST"),Transform=(Rotation=(X=0,Y=0,Z=0,W=1),Translation=(X=20,Y=0,Z=-5),Scale3D=(X=1,Y=1,Z=1))))
    WeightMinimum=0
    Weight=@n42.OutputParam
   )
   > n43 = &RigUnit_SphericalPoseReader(
    ActiveRegionScaleFactors=(PositiveWidth=1,NegativeWidth=1,PositiveHeight=1,NegativeHeight=1)
    ActiveRegionSize=0.001
    Debug=(bDrawDebug=false,bDraw2D=False,bDrawLocalAxes=False,DebugScale=25,DebugSegments=20,DebugThickness=0.25)
    DriverAxis=(X=1,Y=0,Z=0)
    DriverItem=(Type=Bone,Name="clavicle_r")
    FalloffRegionScaleFactors=(PositiveWidth=1,NegativeWidth=1,PositiveHeight=1,NegativeHeight=1)
    FalloffSize=0.5
    FlipHeightScaling=False
    FlipWidthScaling=False
    OptionalParentItem=(Type=Bone,Name="None")
    RotationOffset=(X=0,Y=0,Z=0)
   )
   > &RigUnit_ModifyTransforms(
    ItemToModify=((Item=(Type=Bone,Name="None"),Transform=(Rotation=(X=0,Y=0,Z=0,W=1),Translation=(X=0,Y=0,Z=0),Scale3D=(X=1,Y=1,Z=1))),(Item=(Type=Bone,Name="shoulder_pad_r_TWIST"),Transform=(Rotation=(X=0,Y=0,Z=0,W=1),Translation=(X=-20,Y=0,Z=5),Scale3D=(X=1,Y=1,Z=1))))
    WeightMinimum=0
    Weight=@n43.OutputParam
   )
   n44 = &RigUnit_GetTransform(Item=(Type=Bone,Name="shoulder_pad_l_TWIST"), bInitial=false)
   Scale = &RigVMFunction_MathVectorScale(
    Factor=30
    Value=&RigVMFunction_MathTransformRotateVector(Vector=(X=0,Y=0,Z=1), Transform=@n44.Transform) {template="Rotate Vector::Execute(in Transform,in Vector,out Result)"}.Result
   ) {template="Scale::Execute(in Value,in Factor,out Result)"}
   > RigVMVariableNode(Value=&RigVMFunction_MathVectorAdd(A=@n44.Transform.Translation, B=@Scale.Result) {template="Add::Execute(in A,in B,out Result)"}.Result) {variable=Pad L Up Pos}
   n45 = &RigUnit_GetTransform(Item=(Type=Bone,Name="shoulder_pad_r_TWIST"), bInitial=false)
   n46 = &RigVMFunction_MathTransformRotateVector(Vector=(X=0,Y=0,Z=1), Transform=@n45.Transform) {template="Rotate Vector::Execute(in Transform,in Vector,out Result)"}
   > RigVMVariableNode(Value=&RigVMFunction_MathVectorAdd(
     A=@n45.Transform.Translation
     B=&RigVMFunction_MathVectorScale(Factor=-30, Value=@n46.Result) {template="Scale::Execute(in Value,in Factor,out Result)"}.Result
    ) {template="Add::Execute(in A,in B,out Result)"}.Result) {variable=Pad R Up Pos}
   RigVMDispatchNode(
    Enabled=false
    Prefix=clav Fii
    ScreenColor=(R=1,G=1,B=1,A=1)
    ScreenDuration=0.05
    Value=/Script/RigVM.RigVMFunction_MathDoubleMin(A=@"UpperArm Readers_3".Up, B=@"UpperArm Readers_3".Up) {is=struct RigVMUnitNode, template="Minimum::Execute(in A,in B,out Result)"}.Result
    ExecuteContext=@n53.ExecutePin
   ) {template="DISPATCH_RigVMDispatch_Print(in Prefix,in Value,in Enabled,in ScreenDuration,in ScreenColor)"}
   RigVMDispatchNode(
    Enabled=false
    Prefix=clav Fii
    ScreenColor=(R=1,G=1,B=1,A=1)
    ScreenDuration=0.05
    Value=&RigVMFunction_MathDoubleMin(A=@"UpperArm Readers_2".Up, B=@"UpperArm Readers_2".Up) {template="Minimum::Execute(in A,in B,out Result)"}.Result
    ExecuteContext=@n52.ExecutePin
   ) {template="DISPATCH_RigVMDispatch_Print(in Prefix,in Value,in Enabled,in ScreenDuration,in ScreenColor)"}
   RigVMRerouteNode(Value=@"UpperArm Readers_2".Fr Dn)
   RigVMRerouteNode(Value=@"UpperArm Readers_2".Aim Up Ot)
   RigVMRerouteNode(Value=@"UpperArm Readers_2".Aim Dn In)
   RigVMRerouteNode(Value=@"UpperArm Readers_3".Fr Dn)
   RigVMRerouteNode(Value=@"UpperArm Readers_3".Aim Up Ot)
   RigVMRerouteNode(Value=@"UpperArm Readers_3".Aim Dn In)
   n47 = &RigUnit_SphericalPoseReader(
    ActiveRegionScaleFactors=(PositiveWidth=1,NegativeWidth=1,PositiveHeight=1,NegativeHeight=1)
    ActiveRegionSize=0.001
    Debug=(bDrawDebug=false,bDraw2D=False,bDrawLocalAxes=False,DebugScale=25,DebugSegments=20,DebugThickness=0.25)
    DriverAxis=(X=1,Y=0,Z=0)
    DriverItem=(Type=Bone,Name="clavicle_r")
    FalloffRegionScaleFactors=(PositiveWidth=1,NegativeWidth=1,PositiveHeight=1,NegativeHeight=1)
    FalloffSize=0.5
    FlipHeightScaling=False
    FlipWidthScaling=False
    OptionalParentItem=(Type=Bone,Name="None")
    RotationOffset=(X=0,Y=0,Z=0)
    ExecutePin=@n51.ExecutePin
   )
   Minimum = &RigVMFunction_MathDoubleMin(A=@n47.OutputParam, B=@"UpperArm Readers_3".Fr) {template="Minimum::Execute(in A,in B,out Result)"}
   n48 = &RigUnit_SphericalPoseReader(
    ActiveRegionScaleFactors=(PositiveWidth=1,NegativeWidth=1,PositiveHeight=1,NegativeHeight=1)
    ActiveRegionSize=0.001
    Debug=(bDrawDebug=false,bDraw2D=False,bDrawLocalAxes=False,DebugScale=25,DebugSegments=20,DebugThickness=0.25)
    DriverAxis=(X=1,Y=0,Z=0)
    DriverItem=(Type=Bone,Name="clavicle_l")
    FalloffRegionScaleFactors=(PositiveWidth=1,NegativeWidth=1,PositiveHeight=1,NegativeHeight=1)
    FalloffSize=0.5
    FlipHeightScaling=False
    FlipWidthScaling=False
    OptionalParentItem=(Type=Bone,Name="None")
    RotationOffset=(X=0,Y=0,Z=0)
    ExecutePin=@n50.ExecutePin
   )
   n49 = &RigVMFunction_MathDoubleMin(A=@n48.OutputParam, B=@"UpperArm Readers_2".Fr) {template="Minimum::Execute(in A,in B,out Result)"}
   &RigUnit_ModifyTransforms(
    ItemToModify=((Item=(Type=Bone,Name="shoulder_pad_l_TWIST"),Transform=(Rotation=(X=0.2947,Y=0,Z=0,W=0.9556),Translation=(X=0,Y=0,Z=0),Scale3D=(X=1,Y=1,Z=1))))
    WeightMinimum=0
    ExecutePin=@n54.ExecutePin
    Weight=@"UpperArm Readers_2".Up
   )
   > &RigUnit_ModifyTransforms(
    ItemToModify=((Item=(Type=Bone,Name="shoulder_pad_l_MSTR"),Transform=(Rotation=(X=-0.1875,Y=0.1576,Z=-0.536,W=-0.8079),Translation=(X=2.3059,Y=-7.1602,Z=1.0491),Scale3D=(X=1,Y=1,Z=1))))
    WeightMinimum=0
    Weight=@"UpperArm Readers_2".Up
   )
   > &RigUnit_ModifyTransforms(
    ItemToModify=((Item=(Type=Bone,Name="shoulder_pad_l"),Transform=(Rotation=(X=0.0416,Y=0.1461,Z=0.0009,W=-0.9884),Translation=(X=0,Y=0,Z=0),Scale3D=(X=1,Y=1,Z=1))))
    WeightMinimum=0
    Weight=@"UpperArm Readers_2".Up Pre
   )
   > &RigUnit_ModifyTransforms(
    ItemToModify=((Item=(Type=Bone,Name="shoulder_pad_l_TWIST"),Transform=(Rotation=(X=-0.0884,Y=0,Z=0,W=0.9961),Translation=(X=0,Y=0,Z=0),Scale3D=(X=1,Y=1,Z=1))))
    WeightMinimum=0
    Weight=@"UpperArm Readers_2".Up Pre
   )
   > &RigUnit_ModifyTransforms(
    ItemToModify=((Item=(Type=Bone,Name="shoulder_pad_l_MSTR"),Transform=(Rotation=(X=0,Y=0,Z=0.1184,W=-0.993),Translation=(X=0,Y=0,Z=0),Scale3D=(X=1,Y=1,Z=1))))
    WeightMinimum=0
    Weight=@"UpperArm Readers_2".Up Pre
   )
   > &RigUnit_ModifyTransforms(
    ItemToModify=((Item=(Type=Bone,Name="shoulder_pad_l"),Transform=(Rotation=(X=0.177,Y=-0.1009,Z=-0.1451,W=0.9682),Translation=(X=0,Y=0,Z=0),Scale3D=(X=1,Y=1,Z=1))))
    WeightMinimum=0
    Weight=@"UpperArm Readers_2".Up Post
   )
   > &RigUnit_ModifyTransforms(
    ItemToModify=((Item=(Type=Bone,Name="shoulder_pad_l_TWIST"),Transform=(Rotation=(X=0.2947,Y=0,Z=0,W=0.9556),Translation=(X=0,Y=0,Z=0),Scale3D=(X=1,Y=1,Z=1))))
    WeightMinimum=0
    Weight=@"UpperArm Readers_2".Up Post
   )
   > &RigUnit_ModifyTransforms(
    ItemToModify=((Item=(Type=Bone,Name="shoulder_pad_l_MSTR"),Transform=(Rotation=(X=0,Y=0,Z=-0.5724,W=-0.82),Translation=(X=0.0466,Y=-6.2549,Z=1.0967),Scale3D=(X=1,Y=1,Z=1))))
    WeightMinimum=0
    Weight=@"UpperArm Readers_2".Up Post
   )
   > &RigUnit_ModifyTransforms(
    ItemToModify=((Item=(Type=Bone,Name="shoulder_pad_l_MSTR"),Transform=(Rotation=(X=0,Y=0,Z=-0.239,W=-0.971),Translation=(X=0.4814,Y=-1.9168,Z=2.7905),Scale3D=(X=1,Y=1,Z=1))))
    WeightMinimum=0
    Weight=@"UpperArm Readers_2".Fr
   )
   > &RigUnit_ModifyTransforms(
    ItemToModify=((Item=(Type=Bone,Name="shoulder_pad_l_TWIST"),Transform=(Rotation=(X=0.1888,Y=0,Z=0,W=0.982),Translation=(X=0,Y=0,Z=0),Scale3D=(X=1,Y=1,Z=1))))
    WeightMinimum=0
    Weight=@"UpperArm Readers_2".Fr
   )
   > &RigUnit_ModifyTransforms(
    ItemToModify=((Item=(Type=Bone,Name="shoulder_pad_l"),Transform=(Rotation=(X=-0.0318,Y=0.0147,Z=0.0009,W=-0.9994),Translation=(X=0,Y=0,Z=0),Scale3D=(X=1,Y=1,Z=1))))
    WeightMinimum=0
    Weight=@"UpperArm Readers_2".Fr
   )
   > &RigUnit_ModifyTransforms(
    ItemToModify=((Item=(Type=Bone,Name="shoulder_pad_l_MSTR"),Transform=(Rotation=(X=0,Y=0,Z=-0.3853,W=-0.9228),Translation=(X=1.5682,Y=-1.752,Z=0.7045),Scale3D=(X=1,Y=1,Z=1))))
    WeightMinimum=0
    Weight=@"UpperArm Readers_2".Fr Post
   )
   > &RigUnit_ModifyTransforms(
    ItemToModify=((Item=(Type=Bone,Name="shoulder_pad_l_TWIST"),Transform=(Rotation=(X=0.1972,Y=0,Z=0,W=0.9804),Translation=(X=0,Y=0,Z=0),Scale3D=(X=1,Y=1,Z=1))))
    WeightMinimum=0
    Weight=@"UpperArm Readers_2".Fr Post
   )
   > &RigUnit_ModifyTransforms(
    ItemToModify=((Item=(Type=Bone,Name="shoulder_pad_l"),Transform=(Rotation=(X=-0.0518,Y=-0.0355,Z=0.2585,W=-0.964),Translation=(X=0,Y=0,Z=0),Scale3D=(X=1,Y=1,Z=1))))
    WeightMinimum=0
    Weight=@"UpperArm Readers_2".Fr Post
   )
   > &RigUnit_ModifyTransforms(
    ItemToModify=((Item=(Type=Bone,Name="shoulder_pad_l_MSTR"),Transform=(Rotation=(X=0,Y=0,Z=-0.342,W=0.9397),Translation=(X=0,Y=0,Z=0),Scale3D=(X=1,Y=1,Z=1))))
    WeightMinimum=0
    Weight=@"UpperArm Readers_2".Ba
   )
   > &RigUnit_ModifyTransforms(
    ItemToModify=((Item=(Type=Bone,Name="shoulder_pad_l_TWIST"),Transform=(Rotation=(X=0,Y=0,Z=0,W=1),Translation=(X=0,Y=0,Z=0),Scale3D=(X=1,Y=1,Z=1))))
    WeightMinimum=0
    Weight=@"UpperArm Readers_2".Ba
   )
   > &RigUnit_ModifyTransforms(
    ItemToModify=((Item=(Type=Bone,Name="shoulder_pad_l"),Transform=(Rotation=(X=0,Y=0.1736,Z=0,W=-0.9848),Translation=(X=0,Y=0,Z=0),Scale3D=(X=1,Y=1,Z=1))))
    WeightMinimum=0
    Weight=@"UpperArm Readers_2".Ba
   )
   > &RigUnit_ModifyTransforms(
    ItemToModify=((Item=(Type=Bone,Name="shoulder_pad_l_MSTR"),Transform=(Rotation=(X=-0.0979,Y=0.0263,Z=0.0391,W=0.9941),Translation=(X=0,Y=0,Z=0),Scale3D=(X=1,Y=1,Z=1))))
    WeightMinimum=0
    Weight=@"UpperArm Readers_2".Fr Up
   )
   > &RigUnit_ModifyTransforms(
    ItemToModify=((Item=(Type=Bone,Name="shoulder_pad_l_TWIST"),Transform=(Rotation=(X=-0.0952,Y=0,Z=0,W=0.9955),Translation=(X=0,Y=0,Z=0),Scale3D=(X=1,Y=1,Z=1))))
    WeightMinimum=0
    Weight=@"UpperArm Readers_2".Fr Up
   )
   > &RigUnit_ModifyTransforms(
    ItemToModify=((Item=(Type=Bone,Name="shoulder_pad_l"),Transform=(Rotation=(X=0.1293,Y=0.0642,Z=-0.0698,W=0.9871),Translation=(X=0,Y=0,Z=0),Scale3D=(X=1,Y=1,Z=1))))
    WeightMinimum=0
    Weight=@"UpperArm Readers_2".Fr Up
   )
   > &RigUnit_ModifyTransforms(
    ItemToModify=((Item=(Type=Bone,Name="shoulder_pad_l_MSTR"),Transform=(Rotation=(X=-0.2715,Y=0.0234,Z=-0.2067,W=0.9397),Translation=(X=5.2552,Y=-1.0501,Z=2.6723),Scale3D=(X=1,Y=1,Z=1))))
    WeightMinimum=0
    Weight=@"UpperArm Readers_2".Aim Up In
   )
   > &RigUnit_ModifyTransforms(
    ItemToModify=((Item=(Type=Bone,Name="shoulder_pad_l_TWIST"),Transform=(Rotation=(X=0,Y=0,Z=0,W=1),Translation=(X=0,Y=0,Z=0),Scale3D=(X=1,Y=1,Z=1))))
    WeightMinimum=0
    Weight=@"UpperArm Readers_2".Aim Up In
   )
   > n50 = &RigUnit_ModifyTransforms(
    ItemToModify=((Item=(Type=Bone,Name="shoulder_pad_l"),Transform=(Rotation=(X=0.0327,Y=0.3081,Z=-0.1461,W=0.9395),Translation=(X=0,Y=0,Z=0),Scale3D=(X=1,Y=1,Z=1))))
    WeightMinimum=0
    Weight=@"UpperArm Readers_2".Aim Up In
   )
   &RigUnit_ModifyTransforms(
    ItemToModify=((Item=(Type=Bone,Name="shoulder_pad_r"),Transform=(Rotation=(X=-0.0161,Y=0.212,Z=0.1163,W=-0.9702),Translation=(X=0,Y=0,Z=0),Scale3D=(X=1,Y=1,Z=1))))
    WeightMinimum=0
    ExecutePin=@"UpperArm Readers_3".ExecuteContext
    Weight=@"UpperArm Readers_3".Up
   )
   > &RigUnit_ModifyTransforms(
    ItemToModify=((Item=(Type=Bone,Name="shoulder_pad_r_TWIST"),Transform=(Rotation=(X=0.2947,Y=0,Z=0,W=0.9556),Translation=(X=0,Y=0,Z=0),Scale3D=(X=1,Y=1,Z=1))))
    WeightMinimum=0
    Weight=@"UpperArm Readers_3".Up
   )
   > &RigUnit_ModifyTransforms(
    ItemToModify=((Item=(Type=Bone,Name="shoulder_pad_r_MSTR"),Transform=(Rotation=(X=-0.1875,Y=0.1576,Z=-0.536,W=-0.8079),Translation=(X=-0.9039,Y=7.341,Z=-1.8189),Scale3D=(X=1,Y=1,Z=1))))
    WeightMinimum=0
    Weight=@"UpperArm Readers_3".Up
   )
   > &RigUnit_ModifyTransforms(
    ItemToModify=((Item=(Type=Bone,Name="shoulder_pad_r"),Transform=(Rotation=(X=0.0416,Y=0.1461,Z=0.0009,W=-0.9884),Translation=(X=0,Y=0,Z=0),Scale3D=(X=1,Y=1,Z=1))))
    WeightMinimum=0
    Weight=@"UpperArm Readers_3".Up Pre
   )
   > &RigUnit_ModifyTransforms(
    ItemToModify=((Item=(Type=Bone,Name="shoulder_pad_r_TWIST"),Transform=(Rotation=(X=-0.0884,Y=0,Z=0,W=0.9961),Translation=(X=0,Y=0,Z=0),Scale3D=(X=1,Y=1,Z=1))))
    WeightMinimum=0
    Weight=@"UpperArm Readers_3".Up Pre
   )
   > &RigUnit_ModifyTransforms(
    ItemToModify=((Item=(Type=Bone,Name="shoulder_pad_r_MSTR"),Transform=(Rotation=(X=0,Y=0,Z=0.1184,W=-0.993),Translation=(X=0,Y=0,Z=0),Scale3D=(X=1,Y=1,Z=1))))
    WeightMinimum=0
    Weight=@"UpperArm Readers_3".Up Pre
   )
   > &RigUnit_ModifyTransforms(
    ItemToModify=((Item=(Type=Bone,Name="shoulder_pad_r"),Transform=(Rotation=(X=0.177,Y=-0.1009,Z=-0.1451,W=0.9682),Translation=(X=0,Y=0,Z=0),Scale3D=(X=1,Y=1,Z=1))))
    WeightMinimum=0
    Weight=@"UpperArm Readers_3".Up Post
   )
   > &RigUnit_ModifyTransforms(
    ItemToModify=((Item=(Type=Bone,Name="shoulder_pad_r_TWIST"),Transform=(Rotation=(X=0.2947,Y=0,Z=0,W=0.9556),Translation=(X=0,Y=0,Z=0),Scale3D=(X=1,Y=1,Z=1))))
    WeightMinimum=0
    Weight=@"UpperArm Readers_3".Up Post
   )
   > &RigUnit_ModifyTransforms(
    ItemToModify=((Item=(Type=Bone,Name="shoulder_pad_r_MSTR"),Transform=(Rotation=(X=0,Y=0,Z=0.5724,W=0.82),Translation=(X=-0.9039,Y=7.341,Z=-1.8189),Scale3D=(X=1,Y=1,Z=1))))
    WeightMinimum=0
    Weight=@"UpperArm Readers_3".Up Post
   )
   > &RigUnit_ModifyTransforms(
    ItemToModify=((Item=(Type=Bone,Name="shoulder_pad_r_MSTR"),Transform=(Rotation=(X=0,Y=0,Z=-0.239,W=-0.971),Translation=(X=-0.4814,Y=1.9168,Z=-2.7905),Scale3D=(X=1,Y=1,Z=1))))
    WeightMinimum=0
    Weight=@"UpperArm Readers_3".Fr
   )
   > &RigUnit_ModifyTransforms(
    ItemToModify=((Item=(Type=Bone,Name="shoulder_pad_r_TWIST"),Transform=(Rotation=(X=0.1888,Y=0,Z=0,W=0.982),Translation=(X=0,Y=0,Z=0),Scale3D=(X=1,Y=1,Z=1))))
    WeightMinimum=0
    Weight=@"UpperArm Readers_3".Fr
   )
   > &RigUnit_ModifyTransforms(
    ItemToModify=((Item=(Type=Bone,Name="shoulder_pad_r"),Transform=(Rotation=(X=-0.0318,Y=0.0147,Z=0.0009,W=-0.9994),Translation=(X=0,Y=0,Z=0),Scale3D=(X=1,Y=1,Z=1))))
    WeightMinimum=0
    Weight=@"UpperArm Readers_3".Fr
   )
   > &RigUnit_ModifyTransforms(
    ItemToModify=((Item=(Type=Bone,Name="shoulder_pad_r_MSTR"),Transform=(Rotation=(X=0,Y=0,Z=-0.3853,W=-0.9228),Translation=(X=-1.5682,Y=1.752,Z=-0.7045),Scale3D=(X=1,Y=1,Z=1))))
    WeightMinimum=0
    Weight=@"UpperArm Readers_3".Fr Post
   )
   > &RigUnit_ModifyTransforms(
    ItemToModify=((Item=(Type=Bone,Name="shoulder_pad_r_TWIST"),Transform=(Rotation=(X=0.1972,Y=0,Z=0,W=0.9804),Translation=(X=0,Y=0,Z=0),Scale3D=(X=1,Y=1,Z=1))))
    WeightMinimum=0
    Weight=@"UpperArm Readers_3".Fr Post
   )
   > &RigUnit_ModifyTransforms(
    ItemToModify=((Item=(Type=Bone,Name="shoulder_pad_r"),Transform=(Rotation=(X=-0.0518,Y=-0.0355,Z=0.2585,W=-0.964),Translation=(X=0,Y=0,Z=0),Scale3D=(X=1,Y=1,Z=1))))
    WeightMinimum=0
    Weight=@"UpperArm Readers_3".Fr Post
   )
   > &RigUnit_ModifyTransforms(
    ItemToModify=((Item=(Type=Bone,Name="shoulder_pad_r_MSTR"),Transform=(Rotation=(X=0,Y=0,Z=-0.342,W=0.9397),Translation=(X=0,Y=0,Z=0),Scale3D=(X=1,Y=1,Z=1))))
    WeightMinimum=0
    Weight=@"UpperArm Readers_3".Ba
   )
   > &RigUnit_ModifyTransforms(
    ItemToModify=((Item=(Type=Bone,Name="shoulder_pad_r_TWIST"),Transform=(Rotation=(X=0,Y=0,Z=0,W=1),Translation=(X=0,Y=0,Z=0),Scale3D=(X=1,Y=1,Z=1))))
    WeightMinimum=0
    Weight=@"UpperArm Readers_3".Ba
   )
   > &RigUnit_ModifyTransforms(
    ItemToModify=((Item=(Type=Bone,Name="shoulder_pad_r"),Transform=(Rotation=(X=0,Y=0.1736,Z=0,W=-0.9848),Translation=(X=0,Y=0,Z=0),Scale3D=(X=1,Y=1,Z=1))))
    WeightMinimum=0
    Weight=@"UpperArm Readers_3".Ba
   )
   > &RigUnit_ModifyTransforms(
    ItemToModify=((Item=(Type=Bone,Name="shoulder_pad_r_MSTR"),Transform=(Rotation=(X=-0.0979,Y=0.0263,Z=0.0391,W=0.9941),Translation=(X=0,Y=0,Z=0),Scale3D=(X=1,Y=1,Z=1))))
    WeightMinimum=0
    Weight=@"UpperArm Readers_3".Fr Up
   )
   > &RigUnit_ModifyTransforms(
    ItemToModify=((Item=(Type=Bone,Name="shoulder_pad_r_TWIST"),Transform=(Rotation=(X=-0.0952,Y=0,Z=0,W=0.9955),Translation=(X=0,Y=0,Z=0),Scale3D=(X=1,Y=1,Z=1))))
    WeightMinimum=0
    Weight=@"UpperArm Readers_3".Fr Up
   )
   > &RigUnit_ModifyTransforms(
    ItemToModify=((Item=(Type=Bone,Name="shoulder_pad_r"),Transform=(Rotation=(X=0.1293,Y=0.0642,Z=-0.0698,W=0.9871),Translation=(X=0,Y=0,Z=0),Scale3D=(X=1,Y=1,Z=1))))
    WeightMinimum=0
    Weight=@"UpperArm Readers_3".Fr Up
   )
   > &RigUnit_ModifyTransforms(
    ItemToModify=((Item=(Type=Bone,Name="shoulder_pad_r_MSTR"),Transform=(Rotation=(X=-0.2715,Y=0.0234,Z=-0.2067,W=0.9397),Translation=(X=-5.2552,Y=-1.0501,Z=-2.6723),Scale3D=(X=1,Y=1,Z=1))))
    WeightMinimum=0
    Weight=@"UpperArm Readers_3".Aim Up In
   )
   > &RigUnit_ModifyTransforms(
    ItemToModify=((Item=(Type=Bone,Name="shoulder_pad_r_TWIST"),Transform=(Rotation=(X=0,Y=0,Z=0,W=1),Translation=(X=0,Y=0,Z=0),Scale3D=(X=1,Y=1,Z=1))))
    WeightMinimum=0
    Weight=@"UpperArm Readers_3".Aim Up In
   )
   > n51 = &RigUnit_ModifyTransforms(
    ItemToModify=((Item=(Type=Bone,Name="shoulder_pad_r"),Transform=(Rotation=(X=0.0327,Y=0.3081,Z=-0.1461,W=0.9395),Translation=(X=0,Y=0,Z=0),Scale3D=(X=1,Y=1,Z=1))))
    WeightMinimum=0
    Weight=@"UpperArm Readers_3".Aim Up In
   )
   &RigUnit_ModifyTransforms(
    ItemToModify=((Item=(Type=Bone,Name="shoulder_pad_l_MSTR"),Transform=(Rotation=(X=0.1305,Y=0,Z=0,W=0.9914),Translation=(X=0,Y=0,Z=0),Scale3D=(X=1,Y=1,Z=1))))
    WeightMinimum=0
    Weight=@n49.Result
    ExecutePin=@n56.ExecuteContext
   )
   > &RigUnit_ModifyTransforms(
    ItemToModify=((Item=(Type=Bone,Name="shoulder_pad_l_TWIST"),Transform=(Rotation=(X=0.5373,Y=0,Z=0,W=0.8434),Translation=(X=-4.1975,Y=0,Z=0),Scale3D=(X=1,Y=1,Z=1))))
    WeightMinimum=0
    Weight=@n49.Result
   )
   > n52 = &RigUnit_ModifyTransforms(
    ItemToModify=((Item=(Type=Bone,Name="shoulder_pad_l"),Transform=(Rotation=(X=0.0772,Y=0.3406,Z=-0.0402,W=0.9362),Translation=(X=0,Y=0,Z=0),Scale3D=(X=1,Y=1,Z=1))))
    WeightMinimum=0
    Weight=@n49.Result
   )
   &RigUnit_ModifyTransforms(
    ItemToModify=((Item=(Type=Bone,Name="shoulder_pad_r_MSTR"),Transform=(Rotation=(X=0.1305,Y=0,Z=0,W=0.9914),Translation=(X=0,Y=0,Z=0),Scale3D=(X=1,Y=1,Z=1))))
    WeightMinimum=0
    Weight=@Minimum.Result
    ExecutePin=@n55.ExecuteContext
   )
   > &RigUnit_ModifyTransforms(
    ItemToModify=((Item=(Type=Bone,Name="shoulder_pad_r_TWIST"),Transform=(Rotation=(X=0.5373,Y=0,Z=0,W=0.8434),Translation=(X=4.1975,Y=0,Z=0),Scale3D=(X=1,Y=1,Z=1))))
    WeightMinimum=0
    Weight=@Minimum.Result
   )
   > n53 = &RigUnit_ModifyTransforms(
    ItemToModify=((Item=(Type=Bone,Name="shoulder_pad_r"),Transform=(Rotation=(X=0.0772,Y=0.3406,Z=-0.0402,W=0.9362),Translation=(X=0,Y=0,Z=0),Scale3D=(X=1,Y=1,Z=1))))
    WeightMinimum=0
    Weight=@Minimum.Result
   )
   n54 = &RigUnit_ModifyTransforms(
    ItemToModify=((Item=(Type=Bone,Name="shoulder_pad_l"),Transform=(Rotation=(X=-0.0161,Y=0.212,Z=0.1163,W=-0.9702),Translation=(X=0,Y=0,Z=0),Scale3D=(X=1,Y=1,Z=1))))
    WeightMinimum=0
    ExecutePin=@"UpperArm Readers_2".ExecuteContext
    Weight=@"UpperArm Readers_2".Up
   )
   n55 = RigVMDispatchNode(Enabled=false, Prefix=clav, ScreenColor=(R=1,G=1,B=1,A=1), ScreenDuration=0.05, Value=@Minimum.Result, ExecuteContext=@n47.ExecutePin) {template="DISPATCH_RigVMDispatch_Print(in Prefix,in Value,in Enabled,in ScreenDuration,in ScreenColor)"}
   n56 = RigVMDispatchNode(Enabled=false, Prefix=clav, ScreenColor=(R=1,G=1,B=1,A=1), ScreenDuration=0.05, Value=@n49.Result, ExecuteContext=@n48.ExecutePin) {template="DISPATCH_RigVMDispatch_Print(in Prefix,in Value,in Enabled,in ScreenDuration,in ScreenColor)"}
  ]}
 RigVMModel Rig {nodes [
   n1 = &RigUnit_PrepareForExecution()
   RigVMFunction_ControlFlowBranch = &RigVMFunction_ControlFlowBranch(ExecuteContext=@n1.ExecutePin, Condition=RigVMVariableNode() {variable=Debug Controls On}.Value)
   RigUnit_GetTransform = &RigUnit_GetTransform(Item=(Type=Bone,Name="thigh_r"), bInitial=False)
   SpawnControl = /Script/ControlRig.RigUnit_HierarchyAddControlTransform(
    InitialValue=(Rotation=(X=0,Y=0,Z=0,W=1),Translation=(X=0,Y=0,Z=0),Scale3D=(X=1,Y=1,Z=1))
    Name=Thigh R
    OffsetSpace=GlobalSpace
    Parent=(Type=Bone,Name="None")
    Settings="(InitialSpace=LocalSpace,bUsePreferredRotationOrder=False,PreferredRotationOrder=YZX,Limits=(LimitTranslationX=(bMinimum=False,bMaximum=False),LimitTranslationY=(bMinimum=False,bMaximum=False),LimitTranslationZ=(bMinimum=False,bMaximum=False),LimitPitch=(b ...(truncated)"
    OffsetTransform=@RigUnit_GetTransform.Transform
    ExecutePin=@RigVMFunction_ControlFlowBranch.True
   ) {
    is=struct RigVMUnitNode
    template="SpawnControl::Execute(in InitialValue,in Settings,in OffsetTransform,in OffsetSpace,in Parent,in Name,out Item)"
   }
   > RigVMVariableNode(Value=@SpawnControl.Item) {variable=Thigh R Ctrl}
   > n2 = &RigUnit_HierarchyAddControlTransform(
    InitialValue=(Rotation=(X=0,Y=0,Z=0,W=1),Translation=(X=0,Y=0,Z=0),Scale3D=(X=1,Y=1,Z=1))
    Name=Thigh L
    OffsetSpace=GlobalSpace
    Parent=(Type=Bone,Name="None")
    Settings="(InitialSpace=LocalSpace,bUsePreferredRotationOrder=False,PreferredRotationOrder=YZX,Limits=(LimitTranslationX=(bMinimum=False,bMaximum=False),LimitTranslationY=(bMinimum=False,bMaximum=False),LimitTranslationZ=(bMinimum=False,bMaximum=False),LimitPitch=(b ...(truncated)"
    OffsetTransform=&RigUnit_GetTransform(Item=(Type=Bone,Name="thigh_l"), bInitial=False).Transform
   ) {template="SpawnControl::Execute(in InitialValue,in Settings,in OffsetTransform,in OffsetSpace,in Parent,in Name,out Item)"}
   > RigVMVariableNode(Value=@n2.Item) {variable=Thigh L Ctrl}
   > RigVMDispatchNode(Array=RigVMVariableNode() {variable=Debug Ctrls}.Value) {template="DISPATCH_RigVMDispatch_ArrayReset(io Array)"}
   ItemArray_1 = RigVMDispatchNode(Value=((Type=Bone,Name="calf_l"),(Type=Bone,Name="calf_r"),(Type=Bone,Name="clavicle_l"),(Type=Bone,Name="clavicle_r"),(Type=Bone,Name="upperarm_l"),(Type=Bone,Name="upperarm_r"),(Type=Bone,Name="lowerarm_l"),(Type=Bone,Name="lowerarm_r"))) {template="DISPATCH_RigVMDispatch_Constant(io Value)"}
   > RigVMVariableNode(Value=@ItemArray_1.Value) {variable=Debug Bones}
   > For_Each = RigVMDispatchNode(Array=@ItemArray_1.Value) {template="DISPATCH_RigVMDispatch_ArrayIterator(in Array,out Element,out Index,out Count,out Ratio)"}
   Concat = /Script/RigVM.RigVMFunction_NameConcat(B=" Ctrl", A=@For_Each.Element.Name) {is=struct RigVMUnitNode, template="Concat::Execute(in A,in B,out Result)"}
   > n3 = &RigUnit_HierarchyAddControlTransform(
    InitialValue=(Rotation=(X=0,Y=0,Z=0,W=1),Translation=(X=0,Y=0,Z=0),Scale3D=(X=1,Y=1,Z=1))
    OffsetSpace=GlobalSpace
    Settings="(InitialSpace=LocalSpace,bUsePreferredRotationOrder=False,PreferredRotationOrder=YZX,Limits=(LimitTranslationX=(bMinimum=False,bMaximum=False),LimitTranslationY=(bMinimum=False,bMaximum=False),LimitTranslationZ=(bMinimum=False,bMaximum=False),LimitPitch=(b ...(truncated)"
    Name=@Concat.Result
    Parent=&RigUnit_HierarchyGetParent(bDefaultParent=True, Child=@For_Each.Element).Parent
    OffsetTransform=&RigUnit_GetTransform(bInitial=False, Item=@For_Each.Element).Transform
   ) {template="SpawnControl::Execute(in InitialValue,in Settings,in OffsetTransform,in OffsetSpace,in Parent,in Name,out Item)"}
   > Add = RigVMDispatchNode(Element=@n3.Item, Array=RigVMVariableNode() {variable=Debug Ctrls}.Value) {template="DISPATCH_RigVMDispatch_ArrayAdd(io Array,in Element,out Index)"}
  ]}
]
hierarchy [
 root {loc=<0 0 0>, rot=<0 0 0 1>}
 pelvis {parent=root, loc=<0 -0.7321 92.4218>, rot=<0.0127 0.707 -0.0127 -0.707>}
 spine_01 {parent=pelvis, loc=<5.6002 -0.5128 0>, rot=<0 0 0.034 -0.9994>}
 spine_02 {parent=spine_01, loc=<8.0756 0.0139 0>, rot=<0 0 0.0238 -0.9997>}
 spine_03 {parent=spine_02, loc=<7.0395 0 0>, rot=<0 0 0.0149 -0.9999>}
 spine_04 {parent=spine_03, loc=<8.1023 0 0>, rot=<0 0 -0.0132 -0.9999>}
 spine_05 {parent=spine_04, loc=<10.0141 0 0>, rot=<0 0 -0.0015 -1>}
 clavicle_l {parent=spine_05, loc=<2.5856 -0.2866 -4.3777>, rot=<0.0465 -0.7289 -0.009 -0.6829>}
 upperarm_l {parent=clavicle_l, loc=<18.2264 0.1342 -0.001>, rot=<0.0253 -0.2693 0.0013 -0.9627>}
 lowerarm_l {parent=upperarm_l, loc=<29.0071 0.1686 -1.239>, rot=<0.0391 -0.0099 -0.2453 -0.9686>}
 hand_l {parent=lowerarm_l, loc=<25.8607 0.222 1.0274>, rot=<-0.7193 -0.0988 0.1193 -0.6773>}
 thumb_01_l {parent=hand_l, loc=<1.3118 -0.3191 -2.459>, rot=<0.6546 -0.4042 0.0664 -0.6354>}
 thumb_02_l {parent=thumb_01_l, loc=<5.9336 -0.0448 0.0401>, rot=<-0.0037 0.0213 0.298 -0.9543>}
 thumb_03_l {parent=thumb_02_l, loc=<2.2223 0.3682 -0.0907>, rot=<0.0118 -0.013 0.0066 -0.9998>}
 index_metacarpal_l {parent=hand_l, loc=<1.6535 0.0183 -2.96>, rot=<0.0158 -0.0124 -0.0173 -0.9996>}
 index_01_l {parent=index_metacarpal_l, loc=<8.3008 0.0357 -0.0524>, rot=<-0.0044 0.0485 0.0924 -0.9945>}
 index_02_l {parent=index_01_l, loc=<3.3587 -0.3004 0.0452>, rot=<-0.002 0.0027 0.3225 -0.9466>}
 index_03_l {parent=index_02_l, loc=<2.5578 -0.123 -0.0252>, rot=<-0.002 0.0067 0.2558 -0.9667>}
 middle_metacarpal_l {parent=hand_l, loc=<2.1856 -0.1725 -0.5984>, rot=<0.006 0.0168 -0.0329 -0.9993>}
 middle_01_l {parent=middle_metacarpal_l, loc=<7.3275 -0.2447 0.0129>, rot=<-0.0208 0.0042 0.0858 -0.9961>}
 middle_02_l {parent=middle_01_l, loc=<4.0123 0.1401 0.1175>, rot=<-0.0218 0.0258 0.3512 -0.9357>}
 middle_03_l {parent=middle_02_l, loc=<2.6077 0.0534 0.1373>, rot=<0.0082 0.0188 0.1708 -0.9851>}
 ring_metacarpal_l {parent=hand_l, loc=<2.2074 -0.3625 1.3638>, rot=<-0.0284 0.0722 -0.0458 -0.9959>}
 ring_01_l {parent=ring_metacarpal_l, loc=<7.0295 -0.6197 -0.1446>, rot=<-0.0489 -0.0252 0.1892 -0.9804>}
 ring_02_l {parent=ring_01_l, loc=<3.6174 0.4647 -0.1027>, rot=<0 0 0.2806 -0.9598>}
 ring_03_l {parent=ring_02_l, loc=<2.4152 0.114 0.0199>, rot=<0 0 0.1189 -0.9929>}
 pinky_metacarpal_l {parent=hand_l, loc=<1.7616 -0.9753 3.3814>, rot=<-0.0397 0.0814 -0.069 -0.9935>}
 pinky_01_l {parent=pinky_metacarpal_l, loc=<7.1709 -0.4831 -0.1094>, rot=<-0.0654 -0.0035 0.2242 -0.9723>}
 pinky_02_l {parent=pinky_01_l, loc=<2.8274 0.0893 -0.1043>, rot=<0.0037 0.0011 0.2854 -0.9584>}
 pinky_03_l {parent=pinky_02_l, loc=<2.2046 0.0307 -0.0078>, rot=<-0.0002 0.0098 0.1445 -0.9895>}
 lowerarm_twist_02_l {parent=lowerarm_l, loc=<9.6199 0 0>, rot=<0 0 0 -1>}
 lowerarm_twist_01_l {parent=lowerarm_l, loc=<17.8549 0 0>, rot=<0 0 0 -1>}
 upperarm_twist_01_l {parent=upperarm_l, loc=<7.9796 0.0843 -0.6195>, rot=<0 0 0 -1>}
 upperarm_twist_02_l {parent=upperarm_l, loc=<19.5059 0.0843 -0.6195>, rot=<0 0 0 -1>}
 bicep_l {parent=upperarm_twist_02_l, loc=<-4.4819 10.1464 0.498>, rot=<0 0 0 1>}
 tricep_l {parent=upperarm_twist_02_l, loc=<-2.386 -12.6709 -1.0486>, rot=<0 0 0 1>}
 elbow_l {parent=upperarm_l, loc=<29.0071 0.1686 -1.239>, rot=<0.0391 -0.0099 -0.2453 -0.9686>}
 elbow_fr_l {parent=elbow_l, loc=<-0.6205 10.8664 1.2615>, rot=<0 0 0 1>}
 elbow_ba_l {parent=elbow_l, loc=<1.2931 -9.3843 -1.0027>, rot=<0 0 0 1>}
 shoulder_pad_l_TWIST {parent=clavicle_l, loc=<8.5118 0 0>, rot=<0 0 0 -1>}
 shoulder_pad_l_MSTR {parent=shoulder_pad_l_TWIST, loc=<0 -1.6216 12.8061>, rot=<0 0 0 -1>}
 shoulder_pad_l {parent=shoulder_pad_l_MSTR, loc=<0 0 0>, rot=<0 0 0 1>}
 FX_Shoulder_HeadLight {parent=shoulder_pad_l, type=Null, loc=<12.6019 4.7088 5.8658>, rot=<0 0 0.0493 0.9988>}
 def_arm_l_up_tp {parent=clavicle_l, loc=<15.0864 -1.8096 10.6756>, rot=<0 0 0 -1>}
 def_arm_l_up_fr {parent=clavicle_l, loc=<18.5408 11.4069 -0.0275>, rot=<0 0 0 -1>}
 def_arm_l_up_ba {parent=clavicle_l, loc=<19.9091 -13.7314 -1.0539>, rot=<0 0 0 -1>}
 neck_01 {parent=spine_05, loc=<10.815 1.9196 0>, rot=<0 0 -0.1864 -0.9825>}
 neck_02 {parent=neck_01, loc=<4.7347 0.3045 0>, rot=<0 0 0.0363 -0.9993>}
 head {parent=neck_02, loc=<5.7632 0.5552 0>, rot=<0 0 0.115 -0.9934>}
 clavicle_r {parent=spine_05, loc=<2.5855 -0.2866 4.3777>, rot=<0.7289 0.0465 -0.6829 0.009>}
 upperarm_r {parent=clavicle_r, loc=<-18.2264 -0.1342 0.001>, rot=<0.0253 -0.2693 0.0013 -0.9627>}
 lowerarm_r {parent=upperarm_r, loc=<-29.0071 -0.1686 1.239>, rot=<0.0391 -0.0099 -0.2453 -0.9686>}
 hand_r {parent=lowerarm_r, loc=<-25.8607 -0.222 -1.0274>, rot=<-0.7193 -0.0988 0.1193 -0.6773>}
 thumb_01_r {parent=hand_r, loc=<-1.3118 0.3191 2.459>, rot=<0.6546 -0.4042 0.0664 -0.6354>}
 thumb_02_r {parent=thumb_01_r, loc=<-5.9336 0.0448 -0.0401>, rot=<-0.0037 0.0213 0.298 -0.9543>}
 thumb_03_r {parent=thumb_02_r, loc=<-2.2223 -0.3682 0.0907>, rot=<0.0118 -0.013 0.0066 -0.9998>}
 index_metacarpal_r {parent=hand_r, loc=<-1.6535 -0.0183 2.96>, rot=<0.0158 -0.0124 -0.0173 -0.9996>}
 index_01_r {parent=index_metacarpal_r, loc=<-8.3008 -0.0357 0.0524>, rot=<-0.0044 0.0485 0.0924 -0.9945>}
 index_02_r {parent=index_01_r, loc=<-3.3587 0.3004 -0.0452>, rot=<-0.002 0.0027 0.3225 -0.9466>}
 index_03_r {parent=index_02_r, loc=<-2.5578 0.123 0.0252>, rot=<-0.002 0.0067 0.2558 -0.9667>}
 middle_metacarpal_r {parent=hand_r, loc=<-2.1856 0.1725 0.5984>, rot=<0.006 0.0168 -0.0329 -0.9993>}
 middle_01_r {parent=middle_metacarpal_r, loc=<-7.3275 0.2447 -0.0129>, rot=<-0.0208 0.0042 0.0858 -0.9961>}
 middle_02_r {parent=middle_01_r, loc=<-4.0123 -0.1401 -0.1175>, rot=<-0.0218 0.0258 0.3512 -0.9357>}
 middle_03_r {parent=middle_02_r, loc=<-2.6077 -0.0534 -0.1373>, rot=<0.0082 0.0188 0.1708 -0.9851>}
 ring_metacarpal_r {parent=hand_r, loc=<-2.2074 0.3625 -1.3638>, rot=<-0.0284 0.0722 -0.0458 -0.9959>}
 ring_01_r {parent=ring_metacarpal_r, loc=<-7.0295 0.6197 0.1446>, rot=<-0.0489 -0.0252 0.1892 -0.9804>}
 ring_02_r {parent=ring_01_r, loc=<-3.6174 -0.4647 0.1027>, rot=<0 0 0.2806 -0.9598>}
 ring_03_r {parent=ring_02_r, loc=<-2.4152 -0.114 -0.0199>, rot=<0 0 0.1189 -0.9929>}
 pinky_metacarpal_r {parent=hand_r, loc=<-1.7616 0.9753 -3.3814>, rot=<-0.0397 0.0814 -0.069 -0.9935>}
 pinky_01_r {parent=pinky_metacarpal_r, loc=<-7.1709 0.4831 0.1094>, rot=<-0.0654 -0.0035 0.2242 -0.9723>}
 pinky_02_r {parent=pinky_01_r, loc=<-2.8274 -0.0893 0.1043>, rot=<0.0037 0.0011 0.2854 -0.9584>}
 pinky_03_r {parent=pinky_02_r, loc=<-2.2046 -0.0307 0.0078>, rot=<-0.0002 0.0098 0.1445 -0.9895>}
 lowerarm_twist_02_r {parent=lowerarm_r, loc=<-9.6199 0 0>, rot=<0 0 0 -1>}
 lowerarm_twist_01_r {parent=lowerarm_r, loc=<-17.8549 0 0>, rot=<0 0 0 -1>}
 upperarm_twist_01_r {parent=upperarm_r, loc=<-7.9796 -0.0843 0.6195>, rot=<0 0 0 -1>}
 upperarm_twist_02_r {parent=upperarm_r, loc=<-19.5059 -0.0843 0.6195>, rot=<0 0 0 -1>}
 bicep_r {parent=upperarm_twist_02_r, loc=<4.482 -10.146 -0.498>, rot=<0 0 0 1>}
 tricep_r {parent=upperarm_twist_02_r, loc=<2.386 12.671 1.049>, rot=<0 0 0 1>}
 elbow_r {parent=upperarm_r, loc=<-29.0071 -0.1686 1.239>, rot=<0.0391 -0.0099 -0.2453 -0.9686>}
 elbow_fr_r {parent=elbow_r, loc=<0.621 -10.866 -1.261>, rot=<0 0 0 1>}
 elbow_ba_r {parent=elbow_r, loc=<-1.293 9.384 1.003>, rot=<0 0 0 1>}
 shoulder_pad_r_TWIST {parent=clavicle_r, loc=<-8.5118 0 0>, rot=<0 0 0 -1>}
 shoulder_pad_r_MSTR {parent=shoulder_pad_r_TWIST, loc=<0 1.6216 -12.8061>, rot=<0 0 0 -1>}
 shoulder_pad_r {parent=shoulder_pad_r_MSTR, loc=<0 0 0>, rot=<0 0 0 1>}
 def_arm_r_up_tp {parent=clavicle_r, loc=<-15.086 1.81 -10.676>, rot=<0 0 0 -1>}
 def_arm_r_up_fr {parent=clavicle_r, loc=<-18.541 -11.407 -0.027>, rot=<0 0 0 -1>}
 def_arm_r_up_ba {parent=clavicle_r, loc=<-19.909 13.731 1.054>, rot=<0 0 0 -1>}
 attach_cape {parent=spine_05, loc=<9.017 -11.714 0>, rot=<0.0616 -0.7044 -0.0616 -0.7044>}
 cape {parent=attach_cape, type=Null, loc=<0 -6.0897 0>, rot=<0 0 0 1>}
 attach_backpack {parent=spine_05, loc=<-0.123 -13.473 0>, rot=<-0.0093 -0.707 0.0093 -0.707>}
 Backpack_BR {parent=attach_backpack, type=Null, loc=<0 -12.1803 0>, rot=<0 0 0 1>}
 thigh_l {parent=pelvis, loc=<-1.8716 -0.0109 -11.9936>, rot=<0.0012 0.0572 0.0155 -0.9982>}
 calf_l {parent=thigh_l, loc=<-41.9178 -0.2246 0.0926>, rot=<0.0006 -0.0161 -0.0391 -0.9991>}
 foot_l {parent=calf_l, loc=<-38.7057 1.2438 0.2452>, rot=<0.0975 -0.0474 0.0317 -0.9936>}
 ball_l {parent=foot_l, loc=<-6.1614 13.1469 0.6807>, rot=<0.0001 -0.0001 -0.7063 -0.7079>}
 calf_twist_02_l {parent=calf_l, loc=<-15.9268 -0.0967 0.0397>, rot=<0.0237 0.0095 -0.0035 -0.9997>}
 calf_twist_01_l {parent=calf_l, loc=<-32.9335 -0.0967 0.0397>, rot=<0.0237 0.0095 -0.0035 -0.9997>}
 thigh_twist_01_l {parent=thigh_l, loc=<-21.6303 0 0>, rot=<0.0013 0 0 -1>}
 knee_wrinkle_base_l {parent=thigh_l, loc=<-41.9178 -0.2246 0.0926>, rot=<0.0006 -0.0161 -0.0391 -0.9991>}
 knee_wrinkle_l {parent=knee_wrinkle_base_l, loc=<12.5 16.5 0>, rot=<0 0 0 -1>}
 knee_pad_base_l {parent=thigh_l, loc=<-41.9178 -0.2246 0.0926>, rot=<0.0006 -0.0161 -0.0391 -0.9991>}
 knee_pad_l {parent=knee_pad_base_l, loc=<-12.9405 10.1876 0>, rot=<0 0 0 1>}
 knee_back_base_l {parent=thigh_l, loc=<-41.9178 -0.2246 0.0926>, rot=<0.0006 -0.0161 -0.0391 -0.9991>}
 knee_ba_l {parent=knee_back_base_l, loc=<0 -11.25 0>, rot=<0 0 0 1>}
 knee_fr_base_l {parent=thigh_l, loc=<-41.9178 -0.2246 0.0926>, rot=<0.0006 -0.0161 -0.0391 -0.9991>}
 knee_fr_l {parent=knee_fr_base_l, loc=<-12.9405 10.1876 0>, rot=<0 0 0 -1>}
 thigh_r {parent=pelvis, loc=<-1.8716 -0.0109 11.9936>, rot=<0.0572 -0.0012 0.9982 0.0155>}
 calf_r {parent=thigh_r, loc=<41.9178 0.2246 -0.0926>, rot=<0.0006 -0.0161 -0.0391 -0.9991>}
 foot_r {parent=calf_r, loc=<38.7057 -1.2438 -0.2452>, rot=<0.0975 -0.0474 0.0317 -0.9936>}
 ball_r {parent=foot_r, loc=<6.1614 -13.1469 -0.6807>, rot=<0.0001 -0.0001 -0.7063 -0.7079>}
 calf_twist_02_r {parent=calf_r, loc=<15.9268 0.0967 -0.0397>, rot=<0.0237 0.0095 -0.0035 -0.9997>}
 calf_twist_01_r {parent=calf_r, loc=<32.9335 0.0967 -0.0397>, rot=<0.0237 0.0095 -0.0035 -0.9997>}
 thigh_twist_01_r {parent=thigh_r, loc=<21.6303 0 0>, rot=<0.0013 0 0 -1>}
 knee_wrinkle_base_r {parent=thigh_r, loc=<41.9178 0.2246 -0.0926>, rot=<0.0006 -0.0161 -0.0391 -0.9991>}
 knee_wrinkle_r {parent=knee_wrinkle_base_r, loc=<-12.5 -16.5 0>, rot=<0 0 0 1>}
 knee_pad_base_r {parent=thigh_r, loc=<41.9178 0.2246 -0.0926>, rot=<0.0006 -0.0161 -0.0391 -0.9991>}
 knee_pad_r {parent=knee_pad_base_r, loc=<12.9405 -10.1876 0>, rot=<0 0 0 1>}
 knee_back_base_r {parent=thigh_r, loc=<41.9178 0.2246 -0.0926>, rot=<0.0006 -0.0161 -0.0391 -0.9991>}
 knee_ba_r {parent=knee_back_base_r, loc=<0 11.25 0>, rot=<0 0 0 1>}
 knee_fr_base_r {parent=thigh_r, loc=<41.9178 0.2246 -0.0926>, rot=<0.0006 -0.0161 -0.0391 -0.9991>}
 knee_fr_r {parent=knee_fr_base_r, loc=<12.9405 -10.1876 0>, rot=<0 0 0 -1>}
 belt {parent=pelvis, loc=<0 0 0>, rot=<0 0 0 1>}
 belt_fr {parent=belt, loc=<4.5916 23.0737 0>, rot=<0 0 0 1>}
 belt_l {parent=belt, loc=<5.3128 2.72 -21.9967>, rot=<0 0 0 1>}
 belt_l_rots {parent=belt_l, loc=<0 0 0>, rot=<0 0 0 1>}
 belt_l_fr {parent=belt_l, loc=<0 0 0>, rot=<0 0 0 1>}
 belt_l_ba {parent=belt_l, loc=<0 0 0>, rot=<0 0 0 1>}
 belt_glute_l {parent=belt, loc=<5.587 -15.8125 -9.869>, rot=<0 0 0 1>}
 belt_glute_r {parent=belt, loc=<5.587 -15.8125 9.869>, rot=<0 0 0 1>}
 belt_r {parent=belt, loc=<5.3128 2.72 21.997>, rot=<0 0 0 1>}
 belt_r_fr {parent=belt_r, loc=<0 0 0>, rot=<0 0 0 1>}
 belt_r_rots {parent=belt_r, loc=<0 0 0>, rot=<0 0 0 1>}
 belt_r_ba {parent=belt_r, loc=<0 0 0>, rot=<0 0 0 1>}
 belt_leg_l {parent=belt, loc=<4.3907 17.3936 -13.6396>, rot=<0 0 0 1>}
 belt_leg_r {parent=belt, loc=<4.3907 17.3936 13.6396>, rot=<0 0 0 1>}
 rope_a_BASE {parent=belt_leg_r, loc=<0.721 -1.1444 1.6984>, rot=<0.0047 0.5656 0.8174 -0.1089>}
 rope_a_MSTR {parent=rope_a_BASE, loc=<0 0 0>, rot=<0 0 0 1>}
 rope_a_01 {parent=rope_a_MSTR, loc=<0 0 0>, rot=<0 0 0 1>}
 rope_a_02 {parent=rope_a_01, loc=<5.6772 0 0>, rot=<-0.0596 0.1091 0.0968 -0.9875>}
 rope_a_03 {parent=rope_a_02, loc=<6.5591 0 0>, rot=<-0.0056 -0.012 0.0323 -0.9994>}
 rope_a_04 {parent=rope_a_03, loc=<6.585 0 0>, rot=<0.0046 0.0121 -0.0157 -0.9998>}
 rope_a_05 {parent=rope_a_04, loc=<4.0929 0 0>, rot=<0 0 0 1>}
 rope_c_BASE {parent=belt_leg_r, loc=<-4.0625 1.2485 -3.6766>, rot=<-0.0351 0.5554 0.8298 0.0426>}
 rope_c_MSTR {parent=rope_c_BASE, loc=<0 0 0>, rot=<0 0 0 1>}
 rope_c_01 {parent=rope_c_MSTR, loc=<0 0 0>, rot=<0 0 0 1>}
 rope_c_02 {parent=rope_c_01, loc=<6.718 0 0>, rot=<0 0 0 1>}
 rope_c_03 {parent=rope_c_02, loc=<6.5935 0 0>, rot=<-0.0001 0.004 -0.0303 -0.9995>}
 rope_c_04 {parent=rope_c_03, loc=<5.9606 0 0>, rot=<0.0001 -0.004 0.0303 -0.9995>}
 link_a_BASE {parent=belt_leg_r, loc=<1.523 1.5681 -0.9919>, rot=<0.0347 0.5139 0.8503 -0.1082>}
 link_a_MSTR {parent=link_a_BASE, loc=<0 0 0>, rot=<0 0 0 1>}
 link_a {parent=link_a_MSTR, loc=<0 0 0>, rot=<0 0 0 1>}
 link_a_end {parent=link_a, loc=<9.439 0 0>, rot=<0 0 0 1>}
 link_a_FIX {parent=link_a_BASE, loc=<0 0 0>, rot=<0 0 0 1>}
 link_b_BASE {parent=belt_leg_r, loc=<-0.7563 4.315 -4.4391>, rot=<-0.0079 -0.7104 -0.7036 -0.0157>}
 link_b_MSTR {parent=link_b_BASE, loc=<0 0 0>, rot=<0 0 0 1>}
 link_b {parent=link_b_MSTR, loc=<0 0 0>, rot=<0 0 0 1>}
 link_b_end {parent=link_b, loc=<9.1224 0 0>, rot=<0 0 0 1>}
 link_b_FIX {parent=link_b_BASE, loc=<0 0 0>, rot=<0 0 0 1>}
 loop_tp_a {parent=belt_leg_r, loc=<4.6783 0.7129 -1.8016>, rot=<0.0257 0.5129 0.858 0.0057>}
 loop_tp_b {parent=belt_leg_r, loc=<4.4812 2.177 -6.0543>, rot=<0.0262 0.6196 0.7845 0.0023>}
 loop_bt_a {parent=belt_leg_r, loc=<-4.1594 -0.4335 -1.6561>, rot=<0.0257 0.5129 0.858 0.0057>}
 loop_bt_b {parent=belt_leg_r, loc=<-3.9283 1.4571 -5.9108>, rot=<0.0262 0.6196 0.7845 0.0023>}
 rope_b_BASE {parent=belt_leg_r, loc=<-2.4442 1.0115 -0.9218>, rot=<-0.0294 -0.7218 -0.6882 0.0677>}
 rope_b_MSTR {parent=rope_b_BASE, loc=<0 0 0>, rot=<0 0 0 1>}
 rope_b_01 {parent=rope_b_MSTR, loc=<0 0 0>, rot=<0 0 0 1>}
 rope_b_02 {parent=rope_b_01, loc=<3.1773 0 0>, rot=<0.0034 0.105 0.0156 -0.9943>}
 rope_b_03 {parent=rope_b_02, loc=<6.9967 0 0>, rot=<-0.0151 0.0011 0.0374 -0.9992>}
 rope_b_04 {parent=rope_b_03, loc=<4.1185 0 0>, rot=<-0.0239 -0.0083 0.0052 -0.9997>}
 rope_b_05 {parent=rope_b_04, loc=<3.3594 0 0>, rot=<0 0 0 1>}
 pouch_l_MSTR {parent=belt, loc=<1.9435 2.6203 -22.1798>, rot=<-0.0178 -0.9958 0.0016 -0.0893>}
 pouch_l {parent=pouch_l_MSTR, loc=<0 0 0>, rot=<0 0 0 1>}
 pouch_r_MSTR {parent=belt, loc=<2.1321 2.4725 22.3192>, rot=<-0.0178 -0.994 -0.0019 0.1078>}
 pouch_r {parent=pouch_r_MSTR, loc=<0 0 0>, rot=<0 0 0 1>}
 groin_l {parent=pelvis, loc=<-6.923 15.3775 -12.9987>, rot=<0.0012 0.0572 0.0155 -0.9982>}
 groin_r {parent=pelvis, loc=<-6.923 15.377 12.999>, rot=<0.0572 -0.0012 0.9982 0.0155>}
 groin {parent=pelvis, loc=<-17.2026 10.6411 0>, rot=<0 0 0 1>}
 butt_l_base {parent=pelvis, loc=<-1.8716 -0.0109 -11.9936>, rot=<0.0012 0.0572 0.0155 -0.9982>}
 butt_l {parent=butt_l_base, loc=<-8.0841 -16.2944 1.7922>, rot=<0 0 0 1>}
 butt_r_base {parent=pelvis, loc=<-1.8716 -0.0109 11.9936>, rot=<0.0572 -0.0012 0.9982 0.0155>}
 butt_r {parent=butt_r_base, loc=<8.0844 16.294 -1.7918>, rot=<0 0 0 1>}
]
variables [
 Debug Bones {type="TArray<FRigElementKey>", default=((Type=Bone,Name="calf_l"),(Type=Bone,Name="calf_r"))}
 Debug Controls On {type=bool, default=False}
 Debug Ctrls {type="TArray<FRigElementKey>", default=(())}
 ItemToModify_0_Transform_Translation {type=FVector, default=(X=-0.4814,Y=1.9167,Z=-2.7904)}
 ItemToModify_0_Transform_Translation_0 {type=FVector, default=(X=-0.4814,Y=1.9167,Z=-2.7904)}
 ItemToModify_0_Transform_Translation_1 {type=FVector, default=(X=0,Y=0,Z=0)}
 ItemToModify_0_Transform_Translation_2 {type=FVector, default=(X=0.9039,Y=-7.341,Z=-1.8189)}
 Pad L Up Pos {type=FVector, default=(X=0,Y=0,Z=0)}
 Pad R Up Pos {type=FVector, default=(X=0,Y=0,Z=0)}
 Thigh L Ctrl {type=FRigElementKey, default=(Type=Bone,Name="")}
 Thigh R Ctrl {type=FRigElementKey, default=(Type=Bone,Name="")}
 thigh spr l ba {type=double, default=0}
 thigh spr l fr {type=double, default=0}
 thigh spr l fr post {type=double, default=0}
 thigh spr r ba {type=double, default=0}
 thigh spr r fr {type=double, default=0}
 thigh spr r fr post {type=double, default=0}
]
