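# Control Rig for the M_LRG_CampusSire large male body cosmetic that corrects the link_a_FIX and link_b_FIX bones under the right belt.
# The forward solve aims each FIX bone's X axis at its matching _end bone, taking the Z up direction from that end bone.
# When Debug Controls On is set, construction spawns controls for a hardcoded bone list, and the solve copies their transforms onto those bones before aiming.
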ControlRigBlueprint /BRCosmetics/Characters/Player/Male/Large/Bodies/M_LRG_CampusSire/Meshes/M_LRG_CampusSire_DeferredFix_CtrlRig
graphs [
 RigVMFunctionLibrary {links [], nodes []}
 RigVMModel {nodes [
   RigUnit_BeginExecution = &RigUnit_BeginExecution()
   RigVMFunction_ControlFlowBranch = &RigVMFunction_ControlFlowBranch(ExecuteContext=@RigUnit_BeginExecution.ExecutePin, Condition=RigVMVariableNode() {variable=Debug Controls On}.Value)
   RigUnit_GetTransform = &RigUnit_GetTransform(Item=(Type=Bone,Name="link_a_FIX"), bInitial=False)
   n1 = &RigUnit_GetTransform(Item=(Type=Bone,Name="link_a_end"), bInitial=False)
   n2 = /Script/ControlRig.RigUnit_AimBoneMath(
    DebugSettings=(bEnabled=False,Scale=10,WorldOffset=(Rotation=(X=0,Y=0,Z=0,W=1),Translation=(X=0,Y=0,Z=0),Scale3D=(X=1,Y=1,Z=1)))
    Primary=(Weight=1,Axis=(X=1,Y=0,Z=0),Target=(X=1,Y=0,Z=0),Kind=Location,Space=(Type=Bone,Name="None"))
    Secondary=(Weight=1,Axis=(X=0,Y=0,Z=1),Target=(X=1,Y=0,Z=0),Kind=Direction,Space=(Type=Bone,Name="None"))
    Weight=1
    InputTransform=@RigUnit_GetTransform.Transform
    Primary.Target=@n1.Transform.Translation
    Secondary.Target=/Script/RigVM.RigVMFunction_MathTransformRotateVector(Vector=(X=0,Y=0,Z=1), Transform=@n1.Transform) {is=struct RigVMUnitNode, template="Rotate Vector::Execute(in Transform,in Vector,out Result)"}.Result
   ) {is=struct RigVMUnitNode}
   &RigUnit_SetTransform(Item=(Type=Bone,Name="link_a_FIX"), Weight=1, bInitial=False, Value=@n2.Result, ExecutePin=@RigVMFunction_ControlFlowBranch.Completed)
   n3 = &RigUnit_GetTransform(Item=(Type=Bone,Name="link_b_end"), bInitial=False)
   n4 = &RigUnit_AimBoneMath(
    DebugSettings=(bEnabled=False,Scale=10,WorldOffset=(Rotation=(X=0,Y=0,Z=0,W=1),Translation=(X=0,Y=0,Z=0),Scale3D=(X=1,Y=1,Z=1)))
    Primary=(Weight=1,Axis=(X=1,Y=0,Z=0),Target=(X=1,Y=0,Z=0),Kind=Location,Space=(Type=Bone,Name="None"))
    Secondary=(Weight=1,Axis=(X=0,Y=0,Z=1),Target=(X=1,Y=0,Z=0),Kind=Direction,Space=(Type=Bone,Name="None"))
    Weight=1
    InputTransform=&RigUnit_GetTransform(Item=(Type=Bone,Name="link_b_FIX"), bInitial=False).Transform
    Primary.Target=@n3.Transform.Translation
    Secondary.Target=&RigVMFunction_MathTransformRotateVector(Vector=(X=0,Y=0,Z=1), Transform=@n3.Transform) {template="Rotate Vector::Execute(in Transform,in Vector,out Result)"}.Result
   )
   > &RigUnit_SetTransform(Item=(Type=Bone,Name="link_b_FIX"), Weight=1, bInitial=False, Value=@n4.Result)
   For_Each_1 = RigVMDispatchNode(ExecuteContext=@RigVMFunction_ControlFlowBranch.True, Array=RigVMVariableNode() {variable=Debug Ctrls}.Value) {template="DISPATCH_RigVMDispatch_ArrayIterator(in Array,out Element,out Index,out Count,out Ratio)"}
   > &RigUnit_SetTransform(
    Weight=1
    bInitial=False
    Item=RigVMDispatchNode(Array=RigVMVariableNode() {variable=Debug Bones}.Value, Index=@For_Each_1.Index) {template="DISPATCH_RigVMDispatch_ArrayGetAtIndex(in Array,in Index,out Element)"}.Element
    Value=&RigUnit_GetTransform(bInitial=False, Item=@For_Each_1.Element).Transform
   )
  ]}
 RigVMModel Rig {nodes [
   n1 = &RigUnit_PrepareForExecution()
   RigVMFunction_ControlFlowBranch = &RigVMFunction_ControlFlowBranch(ExecuteContext=@n1.ExecutePin, Condition=RigVMVariableNode() {variable=Debug Controls On}.Value)
   RigVMDispatchNode(ExecuteContext=@RigVMFunction_ControlFlowBranch.True, Array=RigVMVariableNode() {variable=Debug Ctrls}.Value) {template="DISPATCH_RigVMDispatch_ArrayReset(io Array)"}
   ItemArray_1 = RigVMDispatchNode(Value="((Type=Bone,Name=\"calf_l\"),(Type=Bone,Name=\"calf_r\"),(Type=Bone,Name=\"clavicle_l\"),(Type=Bone,Name=\"clavicle_r\"),(Type=Bone,Name=\"upperarm_l\"),(Type=Bone,Name=\"upperarm_r\"),(Type=Bone,Name=\"lowerarm_l\"),(Type=Bone,Name=\"lowerarm_r\"),(Type=Bone,Name=\"link_a ...(truncated)") {template="DISPATCH_RigVMDispatch_Constant(io Value)"}
   > RigVMVariableNode(Value=@ItemArray_1.Value) {variable=Debug Bones}
   > For_Each = RigVMDispatchNode(Array=@ItemArray_1.Value) {template="DISPATCH_RigVMDispatch_ArrayIterator(in Array,out Element,out Index,out Count,out Ratio)"}
   Concat = /Script/RigVM.RigVMFunction_NameConcat(B=" Ctrl", A=@For_Each.Element.Name) {is=struct RigVMUnitNode, template="Concat::Execute(in A,in B,out Result)"}
   > n2 = /Script/ControlRig.RigUnit_HierarchyAddControlTransform(
    InitialValue=(Rotation=(X=0,Y=0,Z=0,W=1),Translation=(X=0,Y=0,Z=0),Scale3D=(X=1,Y=1,Z=1))
    OffsetSpace=GlobalSpace
    Settings="(InitialSpace=LocalSpace,bUsePreferredRotationOrder=False,PreferredRotationOrder=YZX,Limits=(LimitTranslationX=(bMinimum=False,bMaximum=False),LimitTranslationY=(bMinimum=False,bMaximum=False),LimitTranslationZ=(bMinimum=False,bMaximum=False),LimitPitch=(b ...(truncated)"
    Name=@Concat.Result
    Parent=/Script/ControlRig.RigUnit_HierarchyGetParent(bDefaultParent=True, Child=@For_Each.Element) {is=struct RigVMUnitNode}.Parent
    OffsetTransform=&RigUnit_GetTransform(bInitial=False, Item=@For_Each.Element).Transform
   ) {
    is=struct RigVMUnitNode
    template="SpawnControl::Execute(in InitialValue,in Settings,in OffsetTransform,in OffsetSpace,in Parent,in Name,out Item)"
   }
   > RigVMDispatchNode(Element=@n2.Item, Array=RigVMVariableNode() {variable=Debug Ctrls}.Value) {template="DISPATCH_RigVMDispatch_ArrayAdd(io Array,in Element,out Index)"}
  ]}
]
hierarchy [
 root {loc=<0 0 0>, rot=<0 0 0 1>}
 pelvis {parent=root, loc=<0 -0.7321 92.4218>, rot=<0.0127 0.707 -0.0127 -0.707>}
 spine_01 {parent=pelvis, loc=<5.6002 -0.5128 0>, rot=<0 0 0.034 -0.9994>}
 spine_02 {parent=spine_01, loc=<8.0756 0.0139 0>, rot=<0 0 0.0238 -0.9997>}
 spine_03 {parent=spine_02, loc=<7.0395 0 0>, rot=<0 0 0.0149 -0.9999>}
 spine_04 {parent=spine_03, loc=<8.1023 0 0>, rot=<0 0 -0.0132 -0.9999>}
 spine_05 {parent=spine_04, loc=<10.0141 0 0>, rot=<0 0 -0.0015 -1>}
 clavicle_l {parent=spine_05, loc=<2.5856 -0.2866 -4.3777>, rot=<0.0465 -0.7289 -0.009 -0.6829>}
 upperarm_l {parent=clavicle_l, loc=<18.2264 0.1342 -0.001>, rot=<0.0253 -0.2693 0.0013 -0.9627>}
 lowerarm_l {parent=upperarm_l, loc=<29.0071 0.1686 -1.239>, rot=<0.0391 -0.0099 -0.2453 -0.9686>}
 hand_l {parent=lowerarm_l, loc=<25.8607 0.222 1.0274>, rot=<-0.7193 -0.0988 0.1193 -0.6773>}
 thumb_01_l {parent=hand_l, loc=<1.3118 -0.3191 -2.459>, rot=<0.6546 -0.4042 0.0664 -0.6354>}
 thumb_02_l {parent=thumb_01_l, loc=<5.9336 -0.0448 0.0401>, rot=<-0.0037 0.0213 0.298 -0.9543>}
 thumb_03_l {parent=thumb_02_l, loc=<2.2223 0.3682 -0.0907>, rot=<0.0118 -0.013 0.0066 -0.9998>}
 index_metacarpal_l {parent=hand_l, loc=<1.6535 0.0183 -2.96>, rot=<0.0158 -0.0124 -0.0173 -0.9996>}
 index_01_l {parent=index_metacarpal_l, loc=<8.3008 0.0357 -0.0524>, rot=<-0.0044 0.0485 0.0924 -0.9945>}
 index_02_l {parent=index_01_l, loc=<3.3587 -0.3004 0.0452>, rot=<-0.002 0.0027 0.3225 -0.9466>}
 index_03_l {parent=index_02_l, loc=<2.5578 -0.123 -0.0252>, rot=<-0.002 0.0067 0.2558 -0.9667>}
 middle_metacarpal_l {parent=hand_l, loc=<2.1856 -0.1725 -0.5984>, rot=<0.006 0.0168 -0.0329 -0.9993>}
 middle_01_l {parent=middle_metacarpal_l, loc=<7.3275 -0.2447 0.0129>, rot=<-0.0208 0.0042 0.0858 -0.9961>}
 middle_02_l {parent=middle_01_l, loc=<4.0123 0.1401 0.1175>, rot=<-0.0218 0.0258 0.3512 -0.9357>}
 middle_03_l {parent=middle_02_l, loc=<2.6077 0.0534 0.1373>, rot=<0.0082 0.0188 0.1708 -0.9851>}
 ring_metacarpal_l {parent=hand_l, loc=<2.2074 -0.3625 1.3638>, rot=<-0.0284 0.0722 -0.0458 -0.9959>}
 ring_01_l {parent=ring_metacarpal_l, loc=<7.0295 -0.6197 -0.1446>, rot=<-0.0489 -0.0252 0.1892 -0.9804>}
 ring_02_l {parent=ring_01_l, loc=<3.6174 0.4647 -0.1027>, rot=<0 0 0.2806 -0.9598>}
 ring_03_l {parent=ring_02_l, loc=<2.4152 0.114 0.0199>, rot=<0 0 0.1189 -0.9929>}
 pinky_metacarpal_l {parent=hand_l, loc=<1.7616 -0.9753 3.3814>, rot=<-0.0397 0.0814 -0.069 -0.9935>}
 pinky_01_l {parent=pinky_metacarpal_l, loc=<7.1709 -0.4831 -0.1094>, rot=<-0.0654 -0.0035 0.2242 -0.9723>}
 pinky_02_l {parent=pinky_01_l, loc=<2.8274 0.0893 -0.1043>, rot=<0.0037 0.0011 0.2854 -0.9584>}
 pinky_03_l {parent=pinky_02_l, loc=<2.2046 0.0307 -0.0078>, rot=<-0.0002 0.0098 0.1445 -0.9895>}
 lowerarm_twist_02_l {parent=lowerarm_l, loc=<9.6199 0 0>, rot=<0 0 0 -1>}
 lowerarm_twist_01_l {parent=lowerarm_l, loc=<17.8549 0 0>, rot=<0 0 0 -1>}
 upperarm_twist_01_l {parent=upperarm_l, loc=<7.9796 0.0843 -0.6195>, rot=<0 0 0 -1>}
 upperarm_twist_02_l {parent=upperarm_l, loc=<19.5059 0.0843 -0.6195>, rot=<0 0 0 -1>}
 bicep_l {parent=upperarm_twist_02_l, loc=<-4.4819 10.1464 0.498>, rot=<0 0 0 1>}
 tricep_l {parent=upperarm_twist_02_l, loc=<-2.386 -12.6709 -1.0486>, rot=<0 0 0 1>}
 elbow_l {parent=upperarm_l, loc=<29.0071 0.1686 -1.239>, rot=<0.0391 -0.0099 -0.2453 -0.9686>}
 elbow_fr_l {parent=elbow_l, loc=<-0.6205 10.8664 1.2615>, rot=<0 0 0 1>}
 elbow_ba_l {parent=elbow_l, loc=<1.2931 -9.3843 -1.0027>, rot=<0 0 0 1>}
 shoulder_pad_l_TWIST {parent=clavicle_l, loc=<8.5118 0 0>, rot=<0 0 0 -1>}
 shoulder_pad_l_MSTR {parent=shoulder_pad_l_TWIST, loc=<0 -1.6216 12.8061>, rot=<0 0 0 -1>}
 shoulder_pad_l {parent=shoulder_pad_l_MSTR, loc=<0 0 0>, rot=<0 0 0 1>}
 FX_Shoulder_HeadLight {parent=shoulder_pad_l, type=Null, loc=<12.6019 4.7088 5.8658>, rot=<0 0 0.0493 0.9988>}
 def_arm_l_up_tp {parent=clavicle_l, loc=<15.0864 -1.8096 10.6756>, rot=<0 0 0 -1>}
 def_arm_l_up_fr {parent=clavicle_l, loc=<18.5408 11.4069 -0.0275>, rot=<0 0 0 -1>}
 def_arm_l_up_ba {parent=clavicle_l, loc=<19.9091 -13.7314 -1.0539>, rot=<0 0 0 -1>}
 neck_01 {parent=spine_05, loc=<10.815 1.9196 0>, rot=<0 0 -0.1864 -0.9825>}
 neck_02 {parent=neck_01, loc=<4.7347 0.3045 0>, rot=<0 0 0.0363 -0.9993>}
 head {parent=neck_02, loc=<5.7632 0.5552 0>, rot=<0 0 0.115 -0.9934>}
 clavicle_r {parent=spine_05, loc=<2.5855 -0.2866 4.3777>, rot=<0.7289 0.0465 -0.6829 0.009>}
 upperarm_r {parent=clavicle_r, loc=<-18.2264 -0.1342 0.001>, rot=<0.0253 -0.2693 0.0013 -0.9627>}
 lowerarm_r {parent=upperarm_r, loc=<-29.0071 -0.1686 1.239>, rot=<0.0391 -0.0099 -0.2453 -0.9686>}
 hand_r {parent=lowerarm_r, loc=<-25.8607 -0.222 -1.0274>, rot=<-0.7193 -0.0988 0.1193 -0.6773>}
 thumb_01_r {parent=hand_r, loc=<-1.3118 0.3191 2.459>, rot=<0.6546 -0.4042 0.0664 -0.6354>}
 thumb_02_r {parent=thumb_01_r, loc=<-5.9336 0.0448 -0.0401>, rot=<-0.0037 0.0213 0.298 -0.9543>}
 thumb_03_r {parent=thumb_02_r, loc=<-2.2223 -0.3682 0.0907>, rot=<0.0118 -0.013 0.0066 -0.9998>}
 index_metacarpal_r {parent=hand_r, loc=<-1.6535 -0.0183 2.96>, rot=<0.0158 -0.0124 -0.0173 -0.9996>}
 index_01_r {parent=index_metacarpal_r, loc=<-8.3008 -0.0357 0.0524>, rot=<-0.0044 0.0485 0.0924 -0.9945>}
 index_02_r {parent=index_01_r, loc=<-3.3587 0.3004 -0.0452>, rot=<-0.002 0.0027 0.3225 -0.9466>}
 index_03_r {parent=index_02_r, loc=<-2.5578 0.123 0.0252>, rot=<-0.002 0.0067 0.2558 -0.9667>}
 middle_metacarpal_r {parent=hand_r, loc=<-2.1856 0.1725 0.5984>, rot=<0.006 0.0168 -0.0329 -0.9993>}
 middle_01_r {parent=middle_metacarpal_r, loc=<-7.3275 0.2447 -0.0129>, rot=<-0.0208 0.0042 0.0858 -0.9961>}
 middle_02_r {parent=middle_01_r, loc=<-4.0123 -0.1401 -0.1175>, rot=<-0.0218 0.0258 0.3512 -0.9357>}
 middle_03_r {parent=middle_02_r, loc=<-2.6077 -0.0534 -0.1373>, rot=<0.0082 0.0188 0.1708 -0.9851>}
 ring_metacarpal_r {parent=hand_r, loc=<-2.2074 0.3625 -1.3638>, rot=<-0.0284 0.0722 -0.0458 -0.9959>}
 ring_01_r {parent=ring_metacarpal_r, loc=<-7.0295 0.6197 0.1446>, rot=<-0.0489 -0.0252 0.1892 -0.9804>}
 ring_02_r {parent=ring_01_r, loc=<-3.6174 -0.4647 0.1027>, rot=<0 0 0.2806 -0.9598>}
 ring_03_r {parent=ring_02_r, loc=<-2.4152 -0.114 -0.0199>, rot=<0 0 0.1189 -0.9929>}
 pinky_metacarpal_r {parent=hand_r, loc=<-1.7616 0.9753 -3.3814>, rot=<-0.0397 0.0814 -0.069 -0.9935>}
 pinky_01_r {parent=pinky_metacarpal_r, loc=<-7.1709 0.4831 0.1094>, rot=<-0.0654 -0.0035 0.2242 -0.9723>}
 pinky_02_r {parent=pinky_01_r, loc=<-2.8274 -0.0893 0.1043>, rot=<0.0037 0.0011 0.2854 -0.9584>}
 pinky_03_r {parent=pinky_02_r, loc=<-2.2046 -0.0307 0.0078>, rot=<-0.0002 0.0098 0.1445 -0.9895>}
 lowerarm_twist_02_r {parent=lowerarm_r, loc=<-9.6199 0 0>, rot=<0 0 0 -1>}
 lowerarm_twist_01_r {parent=lowerarm_r, loc=<-17.8549 0 0>, rot=<0 0 0 -1>}
 upperarm_twist_01_r {parent=upperarm_r, loc=<-7.9796 -0.0843 0.6195>, rot=<0 0 0 -1>}
 upperarm_twist_02_r {parent=upperarm_r, loc=<-19.5059 -0.0843 0.6195>, rot=<0 0 0 -1>}
 bicep_r {parent=upperarm_twist_02_r, loc=<4.482 -10.146 -0.498>, rot=<0 0 0 1>}
 tricep_r {parent=upperarm_twist_02_r, loc=<2.386 12.671 1.049>, rot=<0 0 0 1>}
 elbow_r {parent=upperarm_r, loc=<-29.0071 -0.1686 1.239>, rot=<0.0391 -0.0099 -0.2453 -0.9686>}
 elbow_fr_r {parent=elbow_r, loc=<0.621 -10.866 -1.261>, rot=<0 0 0 1>}
 elbow_ba_r {parent=elbow_r, loc=<-1.293 9.384 1.003>, rot=<0 0 0 1>}
 shoulder_pad_r_TWIST {parent=clavicle_r, loc=<-8.5118 0 0>, rot=<0 0 0 -1>}
 shoulder_pad_r_MSTR {parent=shoulder_pad_r_TWIST, loc=<0 1.6216 -12.8061>, rot=<0 0 0 -1>}
 shoulder_pad_r {parent=shoulder_pad_r_MSTR, loc=<0 0 0>, rot=<0 0 0 1>}
 def_arm_r_up_tp {parent=clavicle_r, loc=<-15.086 1.81 -10.676>, rot=<0 0 0 -1>}
 def_arm_r_up_fr {parent=clavicle_r, loc=<-18.541 -11.407 -0.027>, rot=<0 0 0 -1>}
 def_arm_r_up_ba {parent=clavicle_r, loc=<-19.909 13.731 1.054>, rot=<0 0 0 -1>}
 attach_cape {parent=spine_05, loc=<9.017 -11.714 0>, rot=<0.0616 -0.7044 -0.0616 -0.7044>}
 cape {parent=attach_cape, type=Null, loc=<0 -6.0897 0>, rot=<0 0 0 1>}
 attach_backpack {parent=spine_05, loc=<-0.123 -13.473 0>, rot=<-0.0093 -0.707 0.0093 -0.707>}
 Backpack_BR {parent=attach_backpack, type=Null, loc=<0 -12.1803 0>, rot=<0 0 0 1>}
 thigh_l {parent=pelvis, loc=<-1.8716 -0.0109 -11.9936>, rot=<0.0012 0.0572 0.0155 -0.9982>}
 calf_l {parent=thigh_l, loc=<-41.9178 -0.2246 0.0926>, rot=<0.0006 -0.0161 -0.0391 -0.9991>}
 foot_l {parent=calf_l, loc=<-38.7057 1.2438 0.2452>, rot=<0.0975 -0.0474 0.0317 -0.9936>}
 ball_l {parent=foot_l, loc=<-6.1614 13.1469 0.6807>, rot=<0.0001 -0.0001 -0.7063 -0.7079>}
 calf_twist_02_l {parent=calf_l, loc=<-15.9268 -0.0967 0.0397>, rot=<0.0237 0.0095 -0.0035 -0.9997>}
 calf_twist_01_l {parent=calf_l, loc=<-32.9335 -0.0967 0.0397>, rot=<0.0237 0.0095 -0.0035 -0.9997>}
 thigh_twist_01_l {parent=thigh_l, loc=<-21.6303 0 0>, rot=<0.0013 0 0 -1>}
 knee_wrinkle_base_l {parent=thigh_l, loc=<-41.9178 -0.2246 0.0926>, rot=<0.0006 -0.0161 -0.0391 -0.9991>}
 knee_wrinkle_l {parent=knee_wrinkle_base_l, loc=<12.5 16.5 0>, rot=<0 0 0 -1>}
 knee_pad_base_l {parent=thigh_l, loc=<-41.9178 -0.2246 0.0926>, rot=<0.0006 -0.0161 -0.0391 -0.9991>}
 knee_pad_l {parent=knee_pad_base_l, loc=<-12.9405 10.1876 0>, rot=<0 0 0 1>}
 knee_back_base_l {parent=thigh_l, loc=<-41.9178 -0.2246 0.0926>, rot=<0.0006 -0.0161 -0.0391 -0.9991>}
 knee_ba_l {parent=knee_back_base_l, loc=<0 -11.25 0>, rot=<0 0 0 1>}
 knee_fr_base_l {parent=thigh_l, loc=<-41.9178 -0.2246 0.0926>, rot=<0.0006 -0.0161 -0.0391 -0.9991>}
 knee_fr_l {parent=knee_fr_base_l, loc=<-12.9405 10.1876 0>, rot=<0 0 0 -1>}
 thigh_r {parent=pelvis, loc=<-1.8716 -0.0109 11.9936>, rot=<0.0572 -0.0012 0.9982 0.0155>}
 calf_r {parent=thigh_r, loc=<41.9178 0.2246 -0.0926>, rot=<0.0006 -0.0161 -0.0391 -0.9991>}
 foot_r {parent=calf_r, loc=<38.7057 -1.2438 -0.2452>, rot=<0.0975 -0.0474 0.0317 -0.9936>}
 ball_r {parent=foot_r, loc=<6.1614 -13.1469 -0.6807>, rot=<0.0001 -0.0001 -0.7063 -0.7079>}
 calf_twist_02_r {parent=calf_r, loc=<15.9268 0.0967 -0.0397>, rot=<0.0237 0.0095 -0.0035 -0.9997>}
 calf_twist_01_r {parent=calf_r, loc=<32.9335 0.0967 -0.0397>, rot=<0.0237 0.0095 -0.0035 -0.9997>}
 thigh_twist_01_r {parent=thigh_r, loc=<21.6303 0 0>, rot=<0.0013 0 0 -1>}
 knee_wrinkle_base_r {parent=thigh_r, loc=<41.9178 0.2246 -0.0926>, rot=<0.0006 -0.0161 -0.0391 -0.9991>}
 knee_wrinkle_r {parent=knee_wrinkle_base_r, loc=<-12.5 -16.5 0>, rot=<0 0 0 1>}
 knee_pad_base_r {parent=thigh_r, loc=<41.9178 0.2246 -0.0926>, rot=<0.0006 -0.0161 -0.0391 -0.9991>}
 knee_pad_r {parent=knee_pad_base_r, loc=<12.9405 -10.1876 0>, rot=<0 0 0 1>}
 knee_back_base_r {parent=thigh_r, loc=<41.9178 0.2246 -0.0926>, rot=<0.0006 -0.0161 -0.0391 -0.9991>}
 knee_ba_r {parent=knee_back_base_r, loc=<0 11.25 0>, rot=<0 0 0 1>}
 knee_fr_base_r {parent=thigh_r, loc=<41.9178 0.2246 -0.0926>, rot=<0.0006 -0.0161 -0.0391 -0.9991>}
 knee_fr_r {parent=knee_fr_base_r, loc=<12.9405 -10.1876 0>, rot=<0 0 0 -1>}
 belt {parent=pelvis, loc=<0 0 0>, rot=<0 0 0 1>}
 belt_fr {parent=belt, loc=<4.5916 23.0737 0>, rot=<0 0 0 1>}
 belt_l {parent=belt, loc=<5.3128 2.72 -21.9967>, rot=<0 0 0 1>}
 belt_l_rots {parent=belt_l, loc=<0 0 0>, rot=<0 0 0 1>}
 belt_l_fr {parent=belt_l, loc=<0 0 0>, rot=<0 0 0 1>}
 belt_l_ba {parent=belt_l, loc=<0 0 0>, rot=<0 0 0 1>}
 belt_glute_l {parent=belt, loc=<5.587 -15.8125 -9.869>, rot=<0 0 0 1>}
 belt_glute_r {parent=belt, loc=<5.587 -15.8125 9.869>, rot=<0 0 0 1>}
 belt_r {parent=belt, loc=<5.3128 2.72 21.997>, rot=<0 0 0 1>}
 belt_r_fr {parent=belt_r, loc=<0 0 0>, rot=<0 0 0 1>}
 belt_r_rots {parent=belt_r, loc=<0 0 0>, rot=<0 0 0 1>}
 belt_r_ba {parent=belt_r, loc=<0 0 0>, rot=<0 0 0 1>}
 belt_leg_l {parent=belt, loc=<4.3907 17.3936 -13.6396>, rot=<0 0 0 1>}
 belt_leg_r {parent=belt, loc=<4.3907 17.3936 13.6396>, rot=<0 0 0 1>}
 rope_a_BASE {parent=belt_leg_r, loc=<0.721 -1.1444 1.6984>, rot=<0.0047 0.5656 0.8174 -0.1089>}
 rope_a_MSTR {parent=rope_a_BASE, loc=<0 0 0>, rot=<0 0 0 1>}
 rope_a_01 {parent=rope_a_MSTR, loc=<0 0 0>, rot=<0 0 0 1>}
 rope_a_02 {parent=rope_a_01, loc=<5.6772 0 0>, rot=<-0.0596 0.1091 0.0968 -0.9875>}
 rope_a_03 {parent=rope_a_02, loc=<6.5591 0 0>, rot=<-0.0056 -0.012 0.0323 -0.9994>}
 rope_a_04 {parent=rope_a_03, loc=<6.585 0 0>, rot=<0.0046 0.0121 -0.0157 -0.9998>}
 rope_a_05 {parent=rope_a_04, loc=<4.0929 0 0>, rot=<0 0 0 1>}
 rope_c_BASE {parent=belt_leg_r, loc=<-4.0625 1.2485 -3.6766>, rot=<-0.0351 0.5554 0.8298 0.0426>}
 rope_c_MSTR {parent=rope_c_BASE, loc=<0 0 0>, rot=<0 0 0 1>}
 rope_c_01 {parent=rope_c_MSTR, loc=<0 0 0>, rot=<0 0 0 1>}
 rope_c_02 {parent=rope_c_01, loc=<6.718 0 0>, rot=<0 0 0 1>}
 rope_c_03 {parent=rope_c_02, loc=<6.5935 0 0>, rot=<-0.0001 0.004 -0.0303 -0.9995>}
 rope_c_04 {parent=rope_c_03, loc=<5.9606 0 0>, rot=<0.0001 -0.004 0.0303 -0.9995>}
 link_a_BASE {parent=belt_leg_r, loc=<1.523 1.5681 -0.9919>, rot=<0.0347 0.5139 0.8503 -0.1082>}
 link_a_MSTR {parent=link_a_BASE, loc=<0 0 0>, rot=<0 0 0 1>}
 link_a {parent=link_a_MSTR, loc=<0 0 0>, rot=<0 0 0 1>}
 link_a_end {parent=link_a, loc=<9.439 0 0>, rot=<0 0 0 1>}
 link_a_FIX {parent=link_a_BASE, loc=<0 0 0>, rot=<0 0 0 1>}
 link_b_BASE {parent=belt_leg_r, loc=<-0.7563 4.315 -4.4391>, rot=<-0.0079 -0.7104 -0.7036 -0.0157>}
 link_b_MSTR {parent=link_b_BASE, loc=<0 0 0>, rot=<0 0 0 1>}
 link_b {parent=link_b_MSTR, loc=<0 0 0>, rot=<0 0 0 1>}
 link_b_end {parent=link_b, loc=<9.1224 0 0>, rot=<0 0 0 1>}
 link_b_FIX {parent=link_b_BASE, loc=<0 0 0>, rot=<0 0 0 1>}
 loop_tp_a {parent=belt_leg_r, loc=<4.6783 0.7129 -1.8016>, rot=<0.0257 0.5129 0.858 0.0057>}
 loop_tp_b {parent=belt_leg_r, loc=<4.4812 2.177 -6.0543>, rot=<0.0262 0.6196 0.7845 0.0023>}
 loop_bt_a {parent=belt_leg_r, loc=<-4.1594 -0.4335 -1.6561>, rot=<0.0257 0.5129 0.858 0.0057>}
 loop_bt_b {parent=belt_leg_r, loc=<-3.9283 1.4571 -5.9108>, rot=<0.0262 0.6196 0.7845 0.0023>}
 rope_b_BASE {parent=belt_leg_r, loc=<-2.4442 1.0115 -0.9218>, rot=<-0.0294 -0.7218 -0.6882 0.0677>}
 rope_b_MSTR {parent=rope_b_BASE, loc=<0 0 0>, rot=<0 0 0 1>}
 rope_b_01 {parent=rope_b_MSTR, loc=<0 0 0>, rot=<0 0 0 1>}
 rope_b_02 {parent=rope_b_01, loc=<3.1773 0 0>, rot=<0.0034 0.105 0.0156 -0.9943>}
 rope_b_03 {parent=rope_b_02, loc=<6.9967 0 0>, rot=<-0.0151 0.0011 0.0374 -0.9992>}
 rope_b_04 {parent=rope_b_03, loc=<4.1185 0 0>, rot=<-0.0239 -0.0083 0.0052 -0.9997>}
 rope_b_05 {parent=rope_b_04, loc=<3.3594 0 0>, rot=<0 0 0 1>}
 pouch_l_MSTR {parent=belt, loc=<1.9435 2.6203 -22.1798>, rot=<-0.0178 -0.9958 0.0016 -0.0893>}
 pouch_l {parent=pouch_l_MSTR, loc=<0 0 0>, rot=<0 0 0 1>}
 pouch_r_MSTR {parent=belt, loc=<2.1321 2.4725 22.3192>, rot=<-0.0178 -0.994 -0.0019 0.1078>}
 pouch_r {parent=pouch_r_MSTR, loc=<0 0 0>, rot=<0 0 0 1>}
 groin_l {parent=pelvis, loc=<-6.923 15.3775 -12.9987>, rot=<0.0012 0.0572 0.0155 -0.9982>}
 groin_r {parent=pelvis, loc=<-6.923 15.377 12.999>, rot=<0.0572 -0.0012 0.9982 0.0155>}
 groin {parent=pelvis, loc=<-17.2026 10.6411 0>, rot=<0 0 0 1>}
 butt_l_base {parent=pelvis, loc=<-1.8716 -0.0109 -11.9936>, rot=<0.0012 0.0572 0.0155 -0.9982>}
 butt_l {parent=butt_l_base, loc=<-8.0841 -16.2944 1.7922>, rot=<0 0 0 1>}
 butt_r_base {parent=pelvis, loc=<-1.8716 -0.0109 11.9936>, rot=<0.0572 -0.0012 0.9982 0.0155>}
 butt_r {parent=butt_r_base, loc=<8.0844 16.294 -1.7918>, rot=<0 0 0 1>}
]
variables [
 Debug Bones {type="TArray<FRigElementKey>", default=((Type=Bone,Name="calf_l"),(Type=Bone,Name="calf_r"))}
 Debug Controls On {type=bool, default=False}
 Debug Ctrls {type="TArray<FRigElementKey>", default=(())}
]
