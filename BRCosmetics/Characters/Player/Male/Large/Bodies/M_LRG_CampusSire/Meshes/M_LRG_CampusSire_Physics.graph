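# Physics asset for the CampusSire large male player body, setting up collision and joints for dangling rope and link accessories around the hips.
# Pelvis, spine and doubled thigh capsules form body collision, joined by free angular constraints.
# Three rope chains (two flat-box, one capsule) hang from pelvis-attached master bones with tight swing and twist limits.
# Two four-capsule links swing without twist from masters angularly locked to the pelvis.
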PhysicsAsset /BRCosmetics/Characters/Player/Male/Large/Bodies/M_LRG_CampusSire/Meshes/M_LRG_CampusSire_Physics
bodies {
 link_a [
  {shapeType=Capsule, length=6, radius=0.7, shapeName=link_a_capsule, center=<3.6024 3.2226 0>, rotation=<rot 90 12.6415 0>}
  {shapeType=Capsule, length=6, radius=0.67, shapeName=link_a_capsule, center=<3.5832 -3.3081 0>, rotation=<rot 90 -12.6415 0>}
  {shapeType=Capsule, length=4.8932, radius=0.6098, shapeName=link_a_capsule, center=<8.2619 1.8267 0>, rotation=<rot 90 124.0551 0>}
  {shapeType=Capsule, length=4.753, radius=0.75, shapeName=link_a_capsule, center=<8.1732 -1.958 0>, rotation=<rot 90 -124.0551 0>}
 ]
 link_a_MSTR [{shapeType=Sphere, radius=0.505, shapeName=link_a_MSTR_sphere}]
 link_b [
  {shapeType=Capsule, length=6, radius=0.7, shapeName=link_a_capsule, center=<3.6024 3.2226 0>, rotation=<rot 90 12.6415 0>}
  {shapeType=Capsule, length=6, radius=0.67, shapeName=link_a_capsule, center=<3.5832 -3.3081 0>, rotation=<rot 90 -12.6415 0>}
  {shapeType=Capsule, length=4.8932, radius=0.6098, shapeName=link_a_capsule, center=<8.2619 1.8267 0>, rotation=<rot 90 124.0551 0>}
  {shapeType=Capsule, length=4.753, radius=0.75, shapeName=link_a_capsule, center=<8.1732 -1.958 0>, rotation=<rot 90 -124.0551 0>}
 ]
 link_b_MSTR [{shapeType=Sphere, radius=0.505, shapeName=link_b_MSTR_sphere}]
 pelvis [{shapeType=Capsule, length=7.5, radius=18, center=<-1.9952 0 0>}]
 rope_a_01 [{shapeType=Box, extentX=5.677, extentY=6.2861, extentZ=0.75, shapeName=rope_a_01_box, center=<2.8385 0 0>}]
 rope_a_02 [{shapeType=Box, extentX=6.559, extentY=7.8814, extentZ=0.75, shapeName=rope_a_02_box, center=<3.2795 0 0>}]
 rope_a_03 [{shapeType=Box, extentX=6.585, extentY=9.0295, extentZ=0.75, shapeName=rope_a_03_box, center=<3.2925 0 0>}]
 rope_a_04 [{shapeType=Box, extentX=4.093, extentY=8.0796, extentZ=0.75, shapeName=rope_a_04_box, center=<2.0465 0 0>}]
 rope_a_05 [{shapeType=Box, extentX=1.8432, extentY=4.9981, extentZ=0.75, shapeName=rope_a_05_box, center=<-0.12 0 0>}]
 rope_a_MSTR [{shapeType=Capsule, length=0.505, radius=0.505, shapeName=rope_a_MSTR_capsule, rotation=<rot 0 -90 0>}]
 rope_b_01 [{shapeType=Box, extentX=3.177, extentY=6.3171, extentZ=0.75, shapeName=rope_b_01_box, center=<1.5885 -0.0039 0.3963>}]
 rope_b_02 [{shapeType=Box, extentX=6.997, extentY=7.8015, extentZ=0.75, shapeName=rope_b_02_box, center=<3.4985 0.1472 0>}]
 rope_b_03 [{shapeType=Box, extentX=4.118, extentY=7.8814, extentZ=0.75, shapeName=rope_b_03_box, center=<2.059 0 0>}]
 rope_b_04 [{shapeType=Box, extentX=3.359, extentY=7.3792, extentZ=0.75, shapeName=rope_b_04_box, center=<1.6795 0 0>}]
 rope_b_05 [{shapeType=Box, extentX=1.438, extentY=4.7494, extentZ=0.75, shapeName=rope_b_05_box}]
 rope_b_MSTR [{shapeType=Capsule, length=0.505, radius=0.505, shapeName=rope_b_MSTR_capsule, rotation=<rot 0 -90 0>}]
 rope_c_01 [{shapeType=Capsule, length=6.718, radius=0.75, shapeName=rope_c_01_capsule, center=<3.359 0 0>, rotation=<rot 90 0 0>}]
 rope_c_02 [{shapeType=Capsule, length=6.594, radius=0.75, shapeName=rope_c_02_capsule, center=<3.297 0 0>, rotation=<rot 90 0 0>}]
 rope_c_03 [{shapeType=Capsule, length=5.961, radius=0.75, shapeName=rope_c_03_capsule, center=<2.9805 0 0>, rotation=<rot 90 0 0>}]
 rope_c_04 [{shapeType=Capsule, length=1, radius=0.75, shapeName=rope_c_04_capsule, rotation=<rot 90 0 0>}]
 rope_c_MSTR [{shapeType=Capsule, length=0.505, radius=0.505, shapeName=rope_b_MSTR_capsule, rotation=<rot 0 -90 0>}]
 spine_01 [{shapeType=Capsule, length=8, radius=17, center=<3.9839 3.0237 0>}]
 spine_02 [{shapeType=Capsule, length=0.505, radius=0.505, shapeName=spine_02_capsule, rotation=<rot 0 -90 0>}]
 thigh_l [
  {shapeType=Capsule, length=40, radius=9.5, center=<-21 -1.7 0>, rotation=<rot 90 -5.6 0>}
  {shapeType=Capsule, length=40, radius=9.5, center=<-21 5.5 0>, rotation=<rot 90 -5.6 0>}
 ]
 thigh_r [
  {shapeType=Capsule, length=40, radius=9.5, center=<21 1.7 0>, rotation=<rot 90 -5.6 0>}
  {shapeType=Capsule, length=40, radius=10, center=<20.8103 -7.4346 0>, rotation=<rot 90 -1.9879 0>}
 ]
}
constraints [
 {
  bone1Name=spine_01
  bone2Name=pelvis
  swing1LimitDegrees=45
  swing1Motion=Free
  swing2LimitDegrees=45
  swing2Motion=Free
  twistLimitDegrees=45
  twistMotion=Free
 }
 {
  bone1Name=thigh_l
  bone2Name=pelvis
  swing1LimitDegrees=45
  swing1Motion=Free
  swing2LimitDegrees=45
  swing2Motion=Free
  twistLimitDegrees=45
  twistMotion=Free
 }
 {
  bone1Name=thigh_r
  bone2Name=pelvis
  swing1LimitDegrees=45
  swing1Motion=Free
  swing2LimitDegrees=45
  swing2Motion=Free
  twistLimitDegrees=45
  twistMotion=Free
 }
 {
  bone1Name=spine_02
  bone2Name=spine_01
  swing1LimitDegrees=45
  swing1Motion=Free
  swing2LimitDegrees=45
  swing2Motion=Free
  twistLimitDegrees=45
  twistMotion=Free
 }
 {bone1Name=link_a, bone2Name=link_a_MSTR, swing1LimitDegrees=2.8125, swing2LimitDegrees=15, twistLimitDegrees=0}
 {bone1Name=link_b, bone2Name=link_b_MSTR, swing1LimitDegrees=2.8125, swing2LimitDegrees=15, twistLimitDegrees=0}
 {
  bone1Name=link_b_MSTR
  bone2Name=pelvis
  swing1LimitDegrees=45
  swing1Motion=Locked
  swing2LimitDegrees=45
  swing2Motion=Locked
  twistLimitDegrees=45
  twistMotion=Locked
 }
 {
  bone1Name=link_a_MSTR
  bone2Name=pelvis
  swing1LimitDegrees=45
  swing1Motion=Locked
  swing2LimitDegrees=45
  swing2Motion=Locked
  twistLimitDegrees=45
  twistMotion=Locked
 }
 {
  bone1Name=rope_a_MSTR
  bone2Name=pelvis
  swing1LimitDegrees=45
  swing1Motion=Free
  swing2LimitDegrees=45
  swing2Motion=Free
  twistLimitDegrees=45
  twistMotion=Free
 }
 {bone1Name=rope_a_01, bone2Name=rope_a_MSTR, swing1LimitDegrees=2.5, swing2LimitDegrees=10, twistLimitDegrees=2.5}
 {bone1Name=rope_a_02, bone2Name=rope_a_01, swing1LimitDegrees=2.5, swing2LimitDegrees=10, twistLimitDegrees=2.5}
 {bone1Name=rope_a_03, bone2Name=rope_a_02, swing1LimitDegrees=2.5, swing2LimitDegrees=10, twistLimitDegrees=2.5}
 {bone1Name=rope_a_04, bone2Name=rope_a_03, swing1LimitDegrees=2.5, swing2LimitDegrees=10, twistLimitDegrees=2.5}
 {bone1Name=rope_a_05, bone2Name=rope_a_04, swing1LimitDegrees=2.5, swing2LimitDegrees=10, twistLimitDegrees=2.5}
 {
  bone1Name=rope_b_MSTR
  bone2Name=pelvis
  swing1LimitDegrees=45
  swing1Motion=Free
  swing2LimitDegrees=45
  swing2Motion=Free
  twistLimitDegrees=45
  twistMotion=Free
 }
 {bone1Name=rope_b_01, bone2Name=rope_b_MSTR, swing1LimitDegrees=2.5, swing2LimitDegrees=10, twistLimitDegrees=2.5}
 {bone1Name=rope_b_02, bone2Name=rope_b_01, swing1LimitDegrees=2.5, swing2LimitDegrees=10, twistLimitDegrees=2.5}
 {bone1Name=rope_b_03, bone2Name=rope_b_02, swing1LimitDegrees=2.5, swing2LimitDegrees=10, twistLimitDegrees=2.5}
 {bone1Name=rope_b_04, bone2Name=rope_b_03, swing1LimitDegrees=2.5, swing2LimitDegrees=10, twistLimitDegrees=2.5}
 {bone1Name=rope_b_05, bone2Name=rope_b_04, swing1LimitDegrees=2.5, swing2LimitDegrees=10, twistLimitDegrees=2.5}
 {
  bone1Name=rope_c_MSTR
  bone2Name=pelvis
  swing1LimitDegrees=45
  swing1Motion=Free
  swing2LimitDegrees=45
  swing2Motion=Free
  twistLimitDegrees=45
  twistMotion=Free
 }
 {bone1Name=rope_c_01, bone2Name=rope_c_MSTR, swing1LimitDegrees=2.5, swing2LimitDegrees=10, twistLimitDegrees=2.5}
 {bone1Name=rope_c_02, bone2Name=rope_c_01, swing1LimitDegrees=2.5, swing2LimitDegrees=10, twistLimitDegrees=2.5}
 {bone1Name=rope_c_03, bone2Name=rope_c_02, swing1LimitDegrees=2.5, swing2LimitDegrees=10, twistLimitDegrees=2.5}
 {bone1Name=rope_c_04, bone2Name=rope_c_03, swing1LimitDegrees=2.5, swing2LimitDegrees=10, twistLimitDegrees=2.5}
]
