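# Physics asset for the NoPouch variant of the large male CampusSire player body, covering two hanging links and three rope chains around the hips.
# Each link and rope chain starts at a small MSTR sphere body constrained to the pelvis with locked rotation, then connects with narrow single-axis swing limit values.
# Paired capsules outline the links and most rope A and B segments, while pelvis, spine and thigh capsules supply body volumes.
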PhysicsAsset /BRCosmetics/Characters/Player/Male/Large/Bodies/M_LRG_CampusSire/Meshes/M_LRG_CampusSire_Physics_NoPouch
bodies {
 link_a [
  {shapeType=Capsule, length=6, radius=0.8, shapeName=link_a_capsule, center=<3.6024 3.2226 0>, rotation=<rot 90 12.6415 0>}
  {shapeType=Capsule, length=6, radius=0.8, shapeName=link_a_capsule, center=<3.5832 -3.3081 0>, rotation=<rot 90 -12.6415 0>}
  {shapeType=Capsule, length=4.8932, radius=0.8, shapeName=link_a_capsule, center=<8.2619 1.8267 0>, rotation=<rot 90 124.0551 0>}
  {shapeType=Capsule, length=4.753, radius=0.8, shapeName=link_a_capsule, center=<8.1732 -1.958 0>, rotation=<rot 90 -124.0551 0>}
 ]
 link_a_MSTR [{shapeType=Sphere, radius=0.505, shapeName=link_a_MSTR_sphere}]
 link_b [
  {shapeType=Capsule, length=6, radius=0.8, shapeName=link_a_capsule, center=<3.6024 3.2226 0>, rotation=<rot 90 12.6415 0>}
  {shapeType=Capsule, length=6, radius=0.8, shapeName=link_a_capsule, center=<3.5832 -3.3081 0>, rotation=<rot 90 -12.6415 0>}
  {shapeType=Capsule, length=4.8932, radius=0.8, shapeName=link_a_capsule, center=<8.2619 1.8267 0>, rotation=<rot 90 124.0551 0>}
  {shapeType=Capsule, length=4.753, radius=0.8, shapeName=link_a_capsule, center=<8.1732 -1.958 0>, rotation=<rot 90 -124.0551 0>}
 ]
 link_b_MSTR [{shapeType=Sphere, radius=0.505, shapeName=link_b_MSTR_sphere}]
 pelvis [{shapeType=Capsule, length=7.5, radius=18, center=<-1.9952 0 0>}]
 rope_a_01 [
  {shapeType=Capsule, length=5.677, radius=2, shapeName=link_a_capsule, center=<2.8828 2.081 0>, rotation=<rot 90 10.7279 0>}
  {shapeType=Capsule, length=5.677, radius=2, shapeName=rope_a_01_capsule, center=<2.9795 -2.5905 0>, rotation=<rot 90 -11.4057 0>}
 ]
 rope_a_02 [
  {shapeType=Capsule, length=6.559, radius=2.5, shapeName=link_a_capsule, center=<3.2795 3.1263 0>, rotation=<rot 90 8.0888 0>}
  {shapeType=Capsule, length=6.559, radius=2.5, shapeName=rope_a_02_capsule, center=<3.2055 -3.2627 0>, rotation=<rot 90 -6.9063 0>}
 ]
 rope_a_03 [
  {shapeType=Capsule, length=6.585, radius=2.5, shapeName=link_a_capsule, center=<3.2925 3.5527 0>, rotation=<rot 90 -5.5075 0>}
  {shapeType=Capsule, length=6.585, radius=2.5, shapeName=rope_a_03_capsule, center=<3.3209 -3.5117 0>, rotation=<rot 90 4.0789 0>}
 ]
 rope_a_04 [
  {shapeType=Capsule, length=4.093, radius=2.5, shapeName=link_a_capsule, center=<2.3299 2.2969 0>, rotation=<rot 90 -25.9763 0>}
  {shapeType=Capsule, length=4.093, radius=2.5, shapeName=rope_a_04_capsule, center=<2.04 -2.1669 0>, rotation=<rot 90 20.8332 0>}
 ]
 rope_a_05 [{shapeType=Capsule, length=2, radius=2.5, shapeName=link_a_capsule, center=<0.5 0 0>, rotation=<rot 90 0 90>}]
 rope_a_MSTR [{shapeType=Sphere, radius=0.505, shapeName=link_a_MSTR_sphere}]
 rope_b_01 [
  {shapeType=Capsule, length=3.177, radius=2.5, shapeName=link_a_capsule, center=<0.6685 1.897 0>, rotation=<rot 90 12.9551 0>}
  {shapeType=Capsule, length=3.177, radius=2.5, shapeName=rope_b_01_capsule, center=<0.6354 -1.9743 0>, rotation=<rot 90 -13.4458 0>}
 ]
 rope_b_02 [
  {shapeType=Capsule, length=6.997, radius=2.5, shapeName=link_a_capsule, center=<3.4985 3.0366 0>, rotation=<rot 90 2.0971 0>}
  {shapeType=Capsule, length=6.997, radius=2.5, shapeName=rope_b_02_capsule, center=<3.4985 -3.0194 0>, rotation=<rot 90 -4.654 0>}
 ]
 rope_b_03 [
  {shapeType=Capsule, length=4.118, radius=2.5, shapeName=link_a_capsule, center=<2.059 3.011 0>, rotation=<rot 90 -10.0464 0>}
  {shapeType=Capsule, length=4.118, radius=2.5, shapeName=rope_b_03_capsule, center=<2.1098 -3.0132 0>, rotation=<rot 90 11.1543 0>}
 ]
 rope_b_04 [
  {shapeType=Capsule, length=3.359, radius=2.5, shapeName=link_a_capsule, center=<1.6795 2.1101 0>, rotation=<rot 90 -26.1307 0>}
  {shapeType=Capsule, length=3.359, radius=2.5, shapeName=rope_b_04_capsule, center=<2.1037 -1.8329 0>, rotation=<rot 90 19.2613 0>}
 ]
 rope_b_05 [{shapeType=Capsule, length=2, radius=2.5, shapeName=link_a_capsule, center=<0.5 0 0>, rotation=<rot 90 0 90>}]
 rope_b_MSTR [{shapeType=Sphere, radius=0.505, shapeName=link_b_MSTR_sphere}]
 rope_c_01 [{shapeType=Capsule, length=6.718, radius=2.5, shapeName=link_a_capsule, center=<3.359 0 0>, rotation=<rot 90 0 0>}]
 rope_c_02 [{shapeType=Capsule, length=6.594, radius=2.5, shapeName=link_a_capsule, center=<3.297 0 0>, rotation=<rot 90 0 0>}]
 rope_c_03 [{shapeType=Capsule, length=5.961, radius=2.5, shapeName=link_a_capsule, center=<2.9805 0 0>, rotation=<rot 90 0 0>}]
 rope_c_04 [{shapeType=Capsule, length=1, radius=2.5, shapeName=link_a_capsule, center=<0.5 0 0>, rotation=<rot 90 0 0>}]
 rope_c_MSTR [{shapeType=Sphere, radius=1, shapeName=link_b_MSTR_sphere, center=<0.5 0 0>}]
 spine_01 [{shapeType=Capsule, length=8, radius=17, center=<3.9839 3.0237 0>}]
 spine_02 [{shapeType=Capsule, length=0.505, radius=0.505, shapeName=spine_02_capsule, rotation=<rot 0 -90 0>}]
 thigh_l [
  {shapeType=Capsule, length=40, radius=9.5, center=<-21 -1.7 0>, rotation=<rot 90 -5.6 0>}
  {shapeType=Capsule, length=40, radius=9.5, center=<-21 5.5 0>, rotation=<rot 90 -5.6 0>}
 ]
 thigh_r [
  {shapeType=Capsule, length=40, radius=9.75, center=<20.9763 -4.8841 -0.0007>, rotation=<rot 90 -5.6 0>}
  {shapeType=Capsule, length=40, radius=10.3868, center=<20.3465 0 0>, rotation=<rot 90 0 0>}
 ]
}
constraints [
 {
  bone1Name=spine_01
  bone2Name=pelvis
  swing1LimitDegrees=45
  swing1Motion=Free
  swing2LimitDegrees=45
  swing2Motion=Free
  twistLimitDegrees=45
  twistMotion=Free
 }
 {
  bone1Name=thigh_l
  bone2Name=pelvis
  swing1LimitDegrees=45
  swing1Motion=Free
  swing2LimitDegrees=45
  swing2Motion=Free
  twistLimitDegrees=45
  twistMotion=Free
 }
 {
  bone1Name=thigh_r
  bone2Name=pelvis
  swing1LimitDegrees=45
  swing1Motion=Free
  swing2LimitDegrees=45
  swing2Motion=Free
  twistLimitDegrees=45
  twistMotion=Free
 }
 {
  bone1Name=spine_02
  bone2Name=spine_01
  swing1LimitDegrees=45
  swing1Motion=Free
  swing2LimitDegrees=45
  swing2Motion=Free
  twistLimitDegrees=45
  twistMotion=Free
 }
 {bone1Name=link_a, bone2Name=link_a_MSTR, swing1LimitDegrees=0, swing2LimitDegrees=10, twistLimitDegrees=0}
 {bone1Name=link_b, bone2Name=link_b_MSTR, swing1LimitDegrees=0, swing2LimitDegrees=10, twistLimitDegrees=0}
 {
  bone1Name=link_b_MSTR
  bone2Name=pelvis
  swing1LimitDegrees=45
  swing1Motion=Locked
  swing2LimitDegrees=45
  swing2Motion=Locked
  twistLimitDegrees=45
  twistMotion=Locked
 }
 {
  bone1Name=link_a_MSTR
  bone2Name=pelvis
  swing1LimitDegrees=45
  swing1Motion=Locked
  swing2LimitDegrees=45
  swing2Motion=Locked
  twistLimitDegrees=45
  twistMotion=Locked
 }
 {
  bone1Name=rope_a_MSTR
  bone2Name=pelvis
  swing1LimitDegrees=45
  swing1Motion=Locked
  swing2LimitDegrees=45
  swing2Motion=Locked
  twistLimitDegrees=45
  twistMotion=Locked
 }
 {bone1Name=rope_a_01, bone2Name=rope_a_MSTR, swing1LimitDegrees=0, swing2LimitDegrees=10, twistLimitDegrees=0}
 {bone1Name=rope_a_02, bone2Name=rope_a_01, swing1LimitDegrees=0, swing2LimitDegrees=10, twistLimitDegrees=0}
 {bone1Name=rope_a_03, bone2Name=rope_a_02, swing1LimitDegrees=0, swing2LimitDegrees=10, twistLimitDegrees=0}
 {bone1Name=rope_a_04, bone2Name=rope_a_03, swing1LimitDegrees=0, swing2LimitDegrees=10, twistLimitDegrees=0}
 {bone1Name=rope_a_05, bone2Name=rope_a_04, swing1LimitDegrees=0, swing2LimitDegrees=10, twistLimitDegrees=0}
 {
  bone1Name=rope_b_MSTR
  bone2Name=pelvis
  swing1LimitDegrees=45
  swing1Motion=Locked
  swing2LimitDegrees=45
  swing2Motion=Locked
  twistLimitDegrees=45
  twistMotion=Locked
 }
 {bone1Name=rope_b_01, bone2Name=rope_b_MSTR, swing1LimitDegrees=0, swing2LimitDegrees=10, twistLimitDegrees=0}
 {bone1Name=rope_b_02, bone2Name=rope_b_01, swing1LimitDegrees=0, swing2LimitDegrees=10, twistLimitDegrees=0}
 {bone1Name=rope_b_03, bone2Name=rope_b_02, swing1LimitDegrees=0, swing2LimitDegrees=10, twistLimitDegrees=0}
 {bone1Name=rope_b_04, bone2Name=rope_b_03, swing1LimitDegrees=0, swing2LimitDegrees=10, twistLimitDegrees=0}
 {bone1Name=rope_b_05, bone2Name=rope_b_04, swing1LimitDegrees=0, swing2LimitDegrees=10, twistLimitDegrees=0}
 {
  bone1Name=rope_c_MSTR
  bone2Name=pelvis
  swing1LimitDegrees=45
  swing1Motion=Locked
  swing2LimitDegrees=45
  swing2Motion=Locked
  twistLimitDegrees=45
  twistMotion=Locked
 }
 {bone1Name=rope_c_01, bone2Name=rope_c_MSTR, swing1LimitDegrees=0, swing2LimitDegrees=10, twistLimitDegrees=0}
 {bone1Name=rope_c_02, bone2Name=rope_c_01, swing1LimitDegrees=0, swing2LimitDegrees=10, twistLimitDegrees=0}
 {bone1Name=rope_c_03, bone2Name=rope_c_02, swing1LimitDegrees=0, swing2LimitDegrees=10, twistLimitDegrees=0}
 {bone1Name=rope_c_04, bone2Name=rope_c_03, swing1LimitDegrees=0, swing2LimitDegrees=10, twistLimitDegrees=0}
]
