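# Control Rig for the CampusSire face accessory that procedurally poses three beard bone chains (center, left, right) along the body.
# Initial B-splines from beard bones and belt-relative points give each chain bone's length percentage.
# Tick control points follow beard master bones and weighted spine, clavicle and pelvis constraints, with a head-twist offset on the right.
# Aim math orients results by spline tangent and neighboring chains; optional debug controls drive bones directly.
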ControlRigBlueprint /BRCosmetics/Characters/Player/Male/Large/Bodies/M_LRG_CampusSire/Meshes/Parts/M_LRG_CampusSire_FaceAcc_CtrlRig
graphs [
 Add_6_ContainedGraph {nodes [
   n1 = RigVMFunctionEntryNode()
   RigVMFunctionReturnNode(Result=/Script/RigVM.RigVMFunction_MathVectorAdd(A=&RigVMFunction_MathVectorAdd(A=@n1.A, B=@n1.B) {template="Add::Execute(in A,in B,out Result)"}.Result, B=@n1.C) {is=struct RigVMUnitNode, template="Add::Execute(in A,in B,out Result)"}.Result)
  ]}
 New Function_ContainedGraph {nodes [
   n1 = RigVMFunctionEntryNode()
   > RigVMDispatchNode(Array=RigVMVariableNode() {variable=Return_Out Percentages}.Value) {template="DISPATCH_RigVMDispatch_ArrayReset(io Array)"}
   > RigVMDispatchNode(Array=RigVMVariableNode() {variable=Return_Out Transforms}.Value) {template="DISPATCH_RigVMDispatch_ArrayReset(io Array)"}
   > RigVMDispatchNode(Array=RigVMVariableNode() {variable=Return_Out Tangents}.Value) {template="DISPATCH_RigVMDispatch_ArrayReset(io Array)"}
   > RigVMDispatchNode(Array=RigVMVariableNode() {variable=Return_Out Secondaries}.Value) {template="DISPATCH_RigVMDispatch_ArrayReset(io Array)"}
   > n2 = RigVMVariableNode(Value=/Script/ControlRigSpline.RigUnit_ControlRigSplineFromTransforms(Compression=0, SamplesPerSegment=16, SplineMode=BSpline, Stretch=0, bClosed=False, Transforms=@n1.Spline Build Transforms) {is=struct RigVMUnitNode}.Spline) {variable=cons spline}
   RigVMFunction_ControlFlowBranch = &RigVMFunction_ControlFlowBranch(ExecuteContext=@n2.ExecuteContext, Condition=@n1.Debug Display)
   For_Each = RigVMDispatchNode(ExecuteContext=@RigVMFunction_ControlFlowBranch.Completed, Array=@n1.Chain Positions) {template="DISPATCH_RigVMDispatch_ArrayIterator(in Array,out Element,out Index,out Count,out Ratio)"}
   n3 = RigVMVariableNode() {variable=cons spline}
   n4 = /Script/ControlRigSpline.RigUnit_ClosestParameterFromControlRigSpline(Position=@For_Each.Element, Spline=@n3.Value) {is=struct RigVMUnitNode}
   n5 = /Script/ControlRigSpline.RigUnit_GetLengthControlRigSpline(Spline=RigVMVariableNode() {variable=cons spline}.Value) {is=struct RigVMUnitNode}
   Divide = &RigVMFunction_MathDoubleDiv(
    A=/Script/ControlRigSpline.RigUnit_GetLengthAtParamControlRigSpline(U=@n4.U, Spline=@n3.Value) {is=struct RigVMUnitNode}.Length
    B=@n5.Length
   )
   > RigVMDispatchNode(Element=@Divide.Result, Array=RigVMVariableNode() {variable=Return_Out Percentages}.Value) {template="DISPATCH_RigVMDispatch_ArrayAdd(io Array,in Element,out Index)"}
   > Add = RigVMDispatchNode(
    Element=/Script/ControlRigSpline.RigUnit_TransformFromControlRigSpline2(PrimaryAxis=(X=0,Y=0,Z=0), SecondaryAxis=(X=0,Y=0,Z=0), U=@n4.U, Spline=@n3.Value) {is=struct RigVMUnitNode}.Transform
    Array=RigVMVariableNode() {variable=Return_Out Transforms}.Value
   ) {template="DISPATCH_RigVMDispatch_ArrayAdd(io Array,in Element,out Index)"}
   > RigVMDispatchNode(
    Element=/Script/ControlRigSpline.RigUnit_TangentFromControlRigSpline(U=@n4.U, Spline=@n3.Value) {is=struct RigVMUnitNode}.Tangent
    Array=RigVMVariableNode() {variable=Return_Out Tangents}.Value
   ) {template="DISPATCH_RigVMDispatch_ArrayAdd(io Array,in Element,out Index)"}
   > RigVMDispatchNode(Element=(X=0,Y=1,Z=0), Array=RigVMVariableNode() {variable=Return_Out Secondaries}.Value) {template="DISPATCH_RigVMDispatch_ArrayAdd(io Array,in Element,out Index)"}
   RigVMFunctionReturnNode(
    Out Length=@n5.Length
    Out Transforms=RigVMVariableNode() {variable=Return_Out Transforms}.Value
    Out Tangents=RigVMVariableNode() {variable=Return_Out Tangents}.Value
    Out Secondaries=RigVMVariableNode() {variable=Return_Out Secondaries}.Value
    Spline=RigVMVariableNode() {variable=cons spline}.Value
    ExecuteContext=@For_Each.Completed
    Out Percentages=RigVMVariableNode() {variable=Return_Out Percentages}.Value
   )
   /Script/ControlRigSpline.RigUnit_DrawControlRigSpline(
    Color=(R=1,G=0,B=0,A=1)
    Detail=16
    Thickness=1
    ExecutePin=@RigVMFunction_ControlFlowBranch.True
    Spline=RigVMVariableNode() {variable=cons spline}.Value
   ) {is=struct RigVMUnitNode}
  ]}
 New Function_ContainedGraph {nodes [
   n1 = RigVMFunctionEntryNode()
   > RigVMDispatchNode(Array=RigVMVariableNode() {variable=Return_Out Positions}.Value) {template="DISPATCH_RigVMDispatch_ArrayReset(io Array)"}
   > n2 = RigVMVariableNode(Value=&RigUnit_ControlRigSplineFromTransforms(Compression=0, SamplesPerSegment=16, SplineMode=BSpline, Stretch=0, bClosed=False, Transforms=@n1.Spline Build Transforms).Spline) {variable=fwd spline}
   RigVMFunction_ControlFlowBranch = &RigVMFunction_ControlFlowBranch(ExecuteContext=@n2.ExecuteContext, Condition=@n1.Debug Display)
   For_Each_1 = RigVMDispatchNode(ExecuteContext=@RigVMFunction_ControlFlowBranch.Completed, Array=RigVMVariableNode() {variable=Chain Percents}.Value) {template="DISPATCH_RigVMDispatch_ArrayIterator(in Array,out Element,out Index,out Count,out Ratio)"}
   n3 = &RigUnit_GetLengthControlRigSpline(Spline=RigVMVariableNode() {variable=fwd spline}.Value)
   Divide = &RigVMFunction_MathDoubleDiv(A=@n3.Length, B=RigVMVariableNode() {variable=Compare To Length}.Value)
   Multiply = &RigVMFunction_MathDoubleMul(B=@Divide.Result, A=@For_Each_1.Element)
   n4 = RigVMVariableNode() {variable=fwd spline}
   > Add = RigVMDispatchNode(
    Element=/Script/ControlRigSpline.RigUnit_PositionFromControlRigSpline(
     U=/Script/ControlRigSpline.RigUnit_ParameterAtPercentage(Percentage=@Multiply.Result, Spline=@n4.Value) {is=struct RigVMUnitNode}.U
     Spline=@n4.Value
    ) {is=struct RigVMUnitNode}.Position
    Array=RigVMVariableNode() {variable=Return_Out Positions}.Value
   ) {template="DISPATCH_RigVMDispatch_ArrayAdd(io Array,in Element,out Index)"}
   RigVMFunctionReturnNode(
    Out Length=@n3.Length
    Out Percentages=RigVMVariableNode() {variable=Return_Out Positions}.Value
    Spline=RigVMVariableNode() {variable=fwd spline}.Value
    ExecuteContext=@For_Each_1.Completed
   )
   &RigUnit_DrawControlRigSpline(
    Color=(R=1,G=0,B=0,A=1)
    Detail=16
    Thickness=1
    ExecutePin=@RigVMFunction_ControlFlowBranch.True
    Spline=RigVMVariableNode() {variable=fwd spline}.Value
   )
   RerouteNode_3_ArrayMake = RigVMDispatchNode(Values=((X=0,Y=0,Z=0))) {template="DISPATCH_RigVMDispatch_ArrayMake(in Values,out Array)"}
  ]}
 New Function_ContainedGraph {nodes [
   n1 = RigVMFunctionEntryNode()
   > n2 = RigVMDispatchNode(Array=RigVMVariableNode() {variable=Return_Out Controls}.Value) {template="DISPATCH_RigVMDispatch_ArrayReset(io Array)"}
   RigVMFunction_ControlFlowBranch = &RigVMFunction_ControlFlowBranch(Condition=@n1.Debug On, ExecuteContext=@n2.ExecuteContext)
   RigVMFunctionReturnNode(ExecuteContext=@RigVMFunction_ControlFlowBranch.Completed, Out Controls=RigVMVariableNode() {variable=Return_Out Controls}.Value)
   For_Each = RigVMDispatchNode(Array=@n1.Arg Bones, ExecuteContext=@RigVMFunction_ControlFlowBranch.True) {template="DISPATCH_RigVMDispatch_ArrayIterator(in Array,out Element,out Index,out Count,out Ratio)"}
   Concat = /Script/RigVM.RigVMFunction_NameConcat(B=" Ctrl", A=@For_Each.Element.Name) {is=struct RigVMUnitNode, template="Concat::Execute(in A,in B,out Result)"}
   > n3 = /Script/ControlRig.RigUnit_HierarchyAddControlTransform(
    InitialValue=(Rotation=(X=0,Y=0,Z=0,W=1),Translation=(X=0,Y=0,Z=0),Scale3D=(X=1,Y=1,Z=1))
    OffsetSpace=GlobalSpace
    Settings="(InitialSpace=LocalSpace,bUsePreferredRotationOrder=False,PreferredRotationOrder=YZX,Limits=(LimitTranslationX=(bMinimum=False,bMaximum=False),LimitTranslationY=(bMinimum=False,bMaximum=False),LimitTranslationZ=(bMinimum=False,bMaximum=False),LimitPitch=(b ...(truncated)"
    Name=@Concat.Result
    Parent=/Script/ControlRig.RigUnit_HierarchyGetParent(bDefaultParent=True, Child=@For_Each.Element) {is=struct RigVMUnitNode}.Parent
    OffsetTransform=&RigUnit_GetTransform(bInitial=False, Item=@For_Each.Element).Transform
    Settings.Shape=RigVMVariableNode() {variable=Debug Control Shape}.Value
   ) {
    is=struct RigVMUnitNode
    template="SpawnControl::Execute(in InitialValue,in Settings,in OffsetTransform,in OffsetSpace,in Parent,in Name,out Item)"
   }
   > Add = RigVMDispatchNode(Element=@n3.Item, Array=RigVMVariableNode() {variable=Return_Out Controls}.Value) {template="DISPATCH_RigVMDispatch_ArrayAdd(io Array,in Element,out Index)"}
  ]}
 New Function_ContainedGraph {nodes [
   n1 = RigVMFunctionEntryNode()
   RigVMFunction_ControlFlowBranch = &RigVMFunction_ControlFlowBranch(ExecuteContext=@n1.ExecuteContext, Condition=@n1.debug on)
   RigVMFunctionReturnNode(ExecuteContext=@RigVMFunction_ControlFlowBranch.Completed)
   For_Each = RigVMDispatchNode(Array=@n1.Input Debug Controls, ExecuteContext=@RigVMFunction_ControlFlowBranch.True) {template="DISPATCH_RigVMDispatch_ArrayIterator(in Array,out Element,out Index,out Count,out Ratio)"}
   > &RigUnit_SetTransform(
    Weight=1
    bInitial=False
    Item=RigVMDispatchNode(Index=@For_Each.Index, Array=@n1.Input Debug Bones) {template="DISPATCH_RigVMDispatch_ArrayGetAtIndex(in Array,in Index,out Element)"}.Element
    Value=&RigUnit_GetTransform(bInitial=False, Item=@For_Each.Element).Transform
   )
  ]}
 New Function_ContainedGraph {nodes [
   RigVMFunctionEntryNode()
   > RigVMFunction_Sequence = RigVMAggregateNode()
   RigVMDispatchNode(ExecuteContext=@RigVMFunction_Sequence.B, Array=RigVMVariableNode() {variable=Beard Lf Initial Transforms}.Value) {template="DISPATCH_RigVMDispatch_ArrayReset(io Array)"}
   n1 = &RigUnit_GetTransform(Item=(Type=Bone,Name="belt_leg_l"), bInitial=true)
   Make_Array_1 = RigVMDispatchNode(
    Values="((Rotation=(X=0,Y=0,Z=0,W=1),Translation=(X=0,Y=0,Z=0),Scale3D=(X=1,Y=1,Z=1)),((Rotation=(X=0,Y=0,Z=0,W=1),Translation=(X=0,Y=0,Z=0),Scale3D=(X=1,Y=1,Z=1)),Rotation=(X=0,Y=0,Z=0,W=1),Translation=(X=0,Y=0,Z=0),Scale3D=(X=1,Y=1,Z=1)),((Rotation=(X=0,Y=0,Z=0, ...(truncated)"
    Values.3.Translation=&RigVMFunction_MathVectorAdd(
     B=RigVMDispatchNode(Elements=(X=0,Y=5.68,Z=-35)) {template="DISPATCH_RigVMDispatch_MakeStruct(in Elements,out Struct)"}.Struct
     A=@n1.Transform.Translation
    ) {template="Add::Execute(in A,in B,out Result)"}.Result
    Values.2.Translation=&RigVMFunction_MathVectorAdd(
     B=RigVMDispatchNode(Elements=(X=0,Y=5.68,Z=0)) {template="DISPATCH_RigVMDispatch_MakeStruct(in Elements,out Struct)"}.Struct
     A=@n1.Transform.Translation
    ) {template="Add::Execute(in A,in B,out Result)"}.Result
    Values.0=&RigUnit_GetTransform(Item=(Type=Bone,Name="beard_l_01"), bInitial=true).Transform
    Values.1=&RigUnit_GetTransform(Item=(Type=Bone,Name="beard_l_04"), bInitial=true).Transform
   ) {template="DISPATCH_RigVMDispatch_ArrayMake(in Values,out Array)"}
   > RigVMVariableNode(Value=@Make_Array_1.Array) {variable=Beard Lf Initial Transforms}
   RigVMFunctionReturnNode(ExecuteContext=@RigVMFunction_Sequence.E)
   RigVMDispatchNode(Array=RigVMVariableNode() {variable=Beard Ct Initial Transforms}.Value, ExecuteContext=@RigVMFunction_Sequence.A) {template="DISPATCH_RigVMDispatch_ArrayReset(io Array)"}
   n2 = &RigUnit_GetTransform(Item=(Type=Bone,Name="belt_fr"), bInitial=true)
   Make_Array = RigVMDispatchNode(
    Values="((Rotation=(X=0,Y=0,Z=0,W=1),Translation=(X=0,Y=0,Z=0),Scale3D=(X=1,Y=1,Z=1)),((Rotation=(X=0,Y=0,Z=0,W=1),Translation=(X=0,Y=0,Z=0),Scale3D=(X=1,Y=1,Z=1)),Rotation=(X=0,Y=0,Z=0,W=1),Translation=(X=0,Y=0,Z=0),Scale3D=(X=1,Y=1,Z=1)),((Rotation=(X=0,Y=0,Z=0, ...(truncated)"
    Values.3.Translation=&RigVMFunction_MathVectorAdd(
     B=RigVMDispatchNode(Elements=(X=0,Y=0,Z=-35)) {template="DISPATCH_RigVMDispatch_MakeStruct(in Elements,out Struct)"}.Struct
     A=@n2.Transform.Translation
    ) {template="Add::Execute(in A,in B,out Result)"}.Result
    Values.0=&RigUnit_GetTransform(Item=(Type=Bone,Name="beard_ct_01"), bInitial=true).Transform
    Values.2=@n2.Transform
    Values.1=&RigUnit_GetTransform(Item=(Type=Bone,Name="beard_ct_04"), bInitial=true).Transform
   ) {template="DISPATCH_RigVMDispatch_ArrayMake(in Values,out Array)"}
   > RigVMVariableNode(Value=@Make_Array.Array) {variable=Beard Ct Initial Transforms}
   RigVMDispatchNode(Array=RigVMVariableNode() {variable=Beard Rt Initial Transforms}.Value, ExecuteContext=@RigVMFunction_Sequence.C) {template="DISPATCH_RigVMDispatch_ArrayReset(io Array)"}
   n3 = &RigUnit_GetTransform(Item=(Type=Bone,Name="belt_leg_r"), bInitial=true)
   Make_Array_2 = RigVMDispatchNode(
    Values="((Rotation=(X=0,Y=0,Z=0,W=1),Translation=(X=0,Y=0,Z=0),Scale3D=(X=1,Y=1,Z=1)),((Rotation=(X=0,Y=0,Z=0,W=1),Translation=(X=0,Y=0,Z=0),Scale3D=(X=1,Y=1,Z=1)),Rotation=(X=0,Y=0,Z=0,W=1),Translation=(X=0,Y=0,Z=0),Scale3D=(X=1,Y=1,Z=1)),((Rotation=(X=0,Y=0,Z=0, ...(truncated)"
    Values.3.Translation=&RigVMFunction_MathVectorAdd(
     B=RigVMDispatchNode(Elements=(X=0,Y=5.68,Z=-35)) {template="DISPATCH_RigVMDispatch_MakeStruct(in Elements,out Struct)"}.Struct
     A=@n3.Transform.Translation
    ) {template="Add::Execute(in A,in B,out Result)"}.Result
    Values.2.Translation=&RigVMFunction_MathVectorAdd(
     B=RigVMDispatchNode(Elements=(X=0,Y=5.68,Z=0)) {template="DISPATCH_RigVMDispatch_MakeStruct(in Elements,out Struct)"}.Struct
     A=@n3.Transform.Translation
    ) {template="Add::Execute(in A,in B,out Result)"}.Result
    Values.0=&RigUnit_GetTransform(Item=(Type=Bone,Name="beard_r_01"), bInitial=true).Transform
    Values.1=&RigUnit_GetTransform(Item=(Type=Bone,Name="beard_r_04"), bInitial=true).Transform
   ) {template="DISPATCH_RigVMDispatch_ArrayMake(in Values,out Array)"}
   > RigVMVariableNode(Value=@Make_Array_2.Array) {variable=Beard Rt Initial Transforms}
  ]}
 New Function_ContainedGraph {nodes [
   RigVMFunctionEntryNode()
   > RigVMFunction_Sequence = RigVMAggregateNode()
   n1 = &RigUnit_GetTransform(Item=(Type=Bone,Name="belt_leg_l"), bInitial=true)
   n2 = /Script/ControlRig.RigUnit_ParentConstraintMath(
    AdvancedSettings=(InterpolationType=Average)
    Input=(Rotation=(X=0,Y=0,Z=0,W=1),Translation=(X=0,Y=0,Z=0),Scale3D=(X=1,Y=1,Z=1))
    Parents=((Item=(Type=Bone,Name="pelvis"),Weight=1))
    Input.Translation=&RigVMFunction_MathVectorAdd(
     B=RigVMDispatchNode(Elements=(X=0,Y=5.68,Z=0)) {template="DISPATCH_RigVMDispatch_MakeStruct(in Elements,out Struct)"}.Struct
     A=@n1.Transform.Translation
    ) {template="Add::Execute(in A,in B,out Result)"}.Result
   ) {is=struct RigVMUnitNode}
   n3 = &RigUnit_ParentConstraintMath(
    AdvancedSettings=(InterpolationType=Average)
    Input=(Rotation=(X=0,Y=0,Z=0,W=1),Translation=(X=0,Y=0,Z=0),Scale3D=(X=1,Y=1,Z=1))
    Parents=((Item=(Type=Bone,Name="pelvis"),Weight=1))
    Input.Translation=&RigVMFunction_MathVectorAdd(
     B=RigVMDispatchNode(Elements=(X=0,Y=5.68,Z=-35)) {template="DISPATCH_RigVMDispatch_MakeStruct(in Elements,out Struct)"}.Struct
     A=@n1.Transform.Translation
    ) {template="Add::Execute(in A,in B,out Result)"}.Result
   )
   n4 = &RigUnit_ParentConstraintMath(
    AdvancedSettings=(InterpolationType=Average)
    Parents=((Item=(Type=Bone,Name="clavicle_l"),Weight=0.345),(Item=(Type=Bone,Name="spine_05"),Weight=0.56),(Item=(Type=Bone,Name="clavicle_l"),Weight=0.0125),(Item=(Type=Bone,Name="clavicle_r"),Weight=0.0125),(Item=(Type=Bone,Name="pelvis"),Weight=0.05))
    Input=&RigUnit_GetTransform(Item=(Type=Bone,Name="beard_l_04"), bInitial=true).Transform
   )
   Make_Array_1 = RigVMDispatchNode(
    Values="((Rotation=(X=0,Y=0,Z=0,W=1),Translation=(X=0,Y=0,Z=0),Scale3D=(X=1,Y=1,Z=1)),((Rotation=(X=0,Y=0,Z=0,W=1),Translation=(X=0,Y=0,Z=0),Scale3D=(X=1,Y=1,Z=1)),Rotation=(X=0,Y=0,Z=0,W=1),Translation=(X=0,Y=0,Z=0),Scale3D=(X=1,Y=1,Z=1)),((Rotation=(X=0,Y=0,Z=0, ...(truncated)"
    Values.2=@n2.Output
    Values.3=@n3.Output
    Values.1=@n4.Output
    Values.0=&RigUnit_GetTransform(Item=(Type=Bone,Name="beard_l_MSTR"), bInitial=False).Transform
   ) {template="DISPATCH_RigVMDispatch_ArrayMake(in Values,out Array)"}
   RigVMVariableNode(Value=@Make_Array_1.Array, ExecuteContext=@RigVMFunction_Sequence.B) {variable=Beard Lf Tick Transforms}
   RigVMFunctionReturnNode(ExecuteContext=@RigVMFunction_Sequence.D)
   n5 = &RigUnit_GetTransform(Item=(Type=Bone,Name="belt_fr"), bInitial=true)
   Make = RigVMDispatchNode(Elements=(X=0,Y=0,Z=-35)) {template="DISPATCH_RigVMDispatch_MakeStruct(in Elements,out Struct)"}
   Add = &RigVMFunction_MathVectorAdd(B=@Make.Struct, A=@n5.Transform.Translation) {template="Add::Execute(in A,in B,out Result)"}
   n6 = &RigUnit_ParentConstraintMath(
    AdvancedSettings=(InterpolationType=Average)
    Input=(Rotation=(X=0,Y=0,Z=0,W=1),Translation=(X=0,Y=0,Z=0),Scale3D=(X=1,Y=1,Z=1))
    Parents=((Item=(Type=Bone,Name="pelvis"),Weight=1))
    Input.Translation=@Add.Result
   )
   n7 = &RigUnit_ParentConstraintMath(
    AdvancedSettings=(InterpolationType=Average)
    Parents=((Item=(Type=Bone,Name="spine_03"),Weight=0.005),(Item=(Type=Bone,Name="spine_05"),Weight=0.9),(Item=(Type=Bone,Name="clavicle_l"),Weight=0.0125),(Item=(Type=Bone,Name="clavicle_r"),Weight=0.0125),(Item=(Type=Bone,Name="pelvis"),Weight=0.05))
    Input=&RigUnit_GetTransform(Item=(Type=Bone,Name="beard_ct_04"), bInitial=true).Transform
   )
   RigUnit_GetTransform = &RigUnit_GetTransform(Item=(Type=Bone,Name="beard_ct_MSTR"), bInitial=False)
   Make_Array = RigVMDispatchNode(
    Values="((Rotation=(X=0,Y=0,Z=0,W=1),Translation=(X=0,Y=0,Z=0),Scale3D=(X=1,Y=1,Z=1)),((Rotation=(X=0,Y=0,Z=0,W=1),Translation=(X=0,Y=0,Z=0),Scale3D=(X=1,Y=1,Z=1)),Rotation=(X=0,Y=0,Z=0,W=1),Translation=(X=0,Y=0,Z=0),Scale3D=(X=1,Y=1,Z=1)),((Rotation=(X=0,Y=0,Z=0, ...(truncated)"
    Values.2=&RigUnit_ParentConstraintMath(AdvancedSettings=(InterpolationType=Average), Parents=((Item=(Type=Bone,Name="pelvis"),Weight=1)), Input=@n5.Transform).Output
    Values.3=@n6.Output
    Values.1=@n7.Output
    Values.0=@RigUnit_GetTransform.Transform
   ) {template="DISPATCH_RigVMDispatch_ArrayMake(in Values,out Array)"}
   RigVMVariableNode(Value=@Make_Array.Array, ExecuteContext=@RigVMFunction_Sequence.A) {variable=Beard Ct Tick Transforms}
   n8 = RigVMVariableNode() {variable=headTwist_r}
   Multiply = &RigVMFunction_MathDoubleMul(B=15, A=@n8.Value)
   n9 = &RigUnit_GetTransform(Item=(Type=Bone,Name="head"), bInitial=False)
   Scale = /Script/RigVM.RigVMFunction_MathVectorScale(
    Factor=@Multiply.Result
    Value=/Script/RigVM.RigVMFunction_MathTransformRotateVector(Vector=(X=0,Y=0,Z=1), Transform=@n9.Transform) {is=struct RigVMUnitNode, template="Rotate Vector::Execute(in Transform,in Vector,out Result)"}.Result
   ) {is=struct RigVMUnitNode, template="Scale::Execute(in Value,in Factor,out Result)"}
   Add_5 = RigVMAggregateNode(
    C=&RigUnit_GetTransform(Item=(Type=Bone,Name="beard_r_MSTR"), bInitial=False).Transform.Translation
    A=@Scale.Result
    B=&RigVMFunction_MathVectorScale(
     Factor=&RigVMFunction_MathDoubleMul(B=40, A=@n8.Value).Result
     Value=&RigVMFunction_MathTransformRotateVector(Vector=(X=0,Y=1,Z=0), Transform=@n9.Transform) {template="Rotate Vector::Execute(in Transform,in Vector,out Result)"}.Result
    ) {template="Scale::Execute(in Value,in Factor,out Result)"}.Result
   )
   n10 = &RigUnit_GetTransform(Item=(Type=Bone,Name="belt_leg_r"), bInitial=true)
   n11 = &RigUnit_ParentConstraintMath(
    AdvancedSettings=(InterpolationType=Average)
    Input=(Rotation=(X=0,Y=0,Z=0,W=1),Translation=(X=0,Y=0,Z=0),Scale3D=(X=1,Y=1,Z=1))
    Parents=((Item=(Type=Bone,Name="pelvis"),Weight=1))
    Input.Translation=&RigVMFunction_MathVectorAdd(
     B=RigVMDispatchNode(Elements=(X=0,Y=5.68,Z=0)) {template="DISPATCH_RigVMDispatch_MakeStruct(in Elements,out Struct)"}.Struct
     A=@n10.Transform.Translation
    ) {template="Add::Execute(in A,in B,out Result)"}.Result
   )
   n12 = &RigUnit_ParentConstraintMath(
    AdvancedSettings=(InterpolationType=Average)
    Input=(Rotation=(X=0,Y=0,Z=0,W=1),Translation=(X=0,Y=0,Z=0),Scale3D=(X=1,Y=1,Z=1))
    Parents=((Item=(Type=Bone,Name="pelvis"),Weight=1))
    Input.Translation=&RigVMFunction_MathVectorAdd(
     B=RigVMDispatchNode(Elements=(X=0,Y=5.68,Z=-35)) {template="DISPATCH_RigVMDispatch_MakeStruct(in Elements,out Struct)"}.Struct
     A=@n10.Transform.Translation
    ) {template="Add::Execute(in A,in B,out Result)"}.Result
   )
   n13 = &RigUnit_ParentConstraintMath(
    AdvancedSettings=(InterpolationType=Average)
    Parents=((Item=(Type=Bone,Name="clavicle_r"),Weight=0.345),(Item=(Type=Bone,Name="spine_05"),Weight=0.56),(Item=(Type=Bone,Name="clavicle_l"),Weight=0.0125),(Item=(Type=Bone,Name="clavicle_r"),Weight=0.0125),(Item=(Type=Bone,Name="pelvis"),Weight=0.05))
    Input=&RigUnit_GetTransform(Item=(Type=Bone,Name="beard_r_04"), bInitial=true).Transform
   )
   Make_Array_2 = RigVMDispatchNode(
    Values="((Rotation=(X=0,Y=0,Z=0,W=1),Translation=(X=0,Y=0,Z=0),Scale3D=(X=1,Y=1,Z=1)),((Rotation=(X=0,Y=0,Z=0,W=1),Translation=(X=0,Y=0,Z=0),Scale3D=(X=1,Y=1,Z=1)),Rotation=(X=0,Y=0,Z=0,W=1),Translation=(X=0,Y=0,Z=0),Scale3D=(X=1,Y=1,Z=1)),((Rotation=(X=0,Y=0,Z=0, ...(truncated)"
    Values.0.Translation=@Add_5.Result
    Values.2=@n11.Output
    Values.3=@n12.Output
    Values.1=@n13.Output
   ) {template="DISPATCH_RigVMDispatch_ArrayMake(in Values,out Array)"}
   RigVMVariableNode(Value=@Make_Array_2.Array, ExecuteContext=@RigVMFunction_Sequence.C) {variable=Beard Rt Tick Transforms}
  ]}
 RigVMFunctionLibrary {nodes [
   Debug Construct = RigVMCollapseNode(
    Arg Bones=()
    Debug Control Shape=(bVisible=True,Name="Default",Color=(R=1,G=0,B=0,A=1),Transform=(Rotation=(X=0,Y=0,Z=0,W=1),Translation=(X=0,Y=0,Z=0),Scale3D=(X=1,Y=1,Z=1)))
    ExecuteContext=()
   )
   Debug Forward = RigVMCollapseNode(ExecuteContext=(), Input Debug Bones=(), Input Debug Controls=())
   Initial Beard Positions = RigVMCollapseNode(ExecuteContext=())
   Initial Spline = RigVMCollapseNode(Chain Positions=(), ExecuteContext=(), Spline Build Transforms=())
   Tick Beard Positions = RigVMCollapseNode(ExecuteContext=())
   Tick Spline = RigVMCollapseNode(Chain Percents=(), ExecuteContext=(), Spline Build Transforms=())
  ]}
 RigVMFunction_Sequence_1_ContainedGraph {nodes [
   n1 = RigVMFunctionEntryNode()
   RigVMFunction_Sequence = &RigVMFunction_Sequence(ExecuteContext=@n1.ExecuteContext)
   n2 = &RigVMFunction_Sequence(ExecuteContext=@RigVMFunction_Sequence.B)
   n3 = &RigVMFunction_Sequence(ExecuteContext=@n2.B)
   n4 = &RigVMFunction_Sequence(ExecuteContext=@n3.B)
   n5 = &RigVMFunction_Sequence(ExecuteContext=@n4.B)
   n6 = &RigVMFunction_Sequence(ExecuteContext=@n5.B)
   n7 = &RigVMFunction_Sequence(ExecuteContext=@n6.B)
   n8 = &RigVMFunction_Sequence(ExecuteContext=@n7.B)
   n9 = &RigVMFunction_Sequence(ExecuteContext=@n8.B)
   RigVMFunctionReturnNode(A=@RigVMFunction_Sequence.A, B=@n2.A, C=@n3.A, D=@n4.A, E=@n5.A, F=@n6.A, G=@n7.A, H=@n8.A, I=@n9.A, J=@n9.B)
  ]}
 RigVMFunction_Sequence_1_ContainedGraph {nodes [
   n1 = RigVMFunctionEntryNode()
   RigVMFunction_Sequence = &RigVMFunction_Sequence(ExecuteContext=@n1.ExecuteContext)
   n2 = &RigVMFunction_Sequence(ExecuteContext=@RigVMFunction_Sequence.B)
   n3 = &RigVMFunction_Sequence(ExecuteContext=@n2.B)
   n4 = &RigVMFunction_Sequence(ExecuteContext=@n3.B)
   RigVMFunctionReturnNode(A=@RigVMFunction_Sequence.A, B=@n2.A, C=@n3.A, D=@n4.A, E=@n4.B)
  ]}
 RigVMFunction_Sequence_1_ContainedGraph {nodes [
   n1 = RigVMFunctionEntryNode()
   RigVMFunction_Sequence = &RigVMFunction_Sequence(ExecuteContext=@n1.ExecuteContext)
   n2 = &RigVMFunction_Sequence(ExecuteContext=@RigVMFunction_Sequence.B)
   n3 = &RigVMFunction_Sequence(ExecuteContext=@n2.B)
   RigVMFunctionReturnNode(A=@RigVMFunction_Sequence.A, B=@n2.A, C=@n3.A, D=@n3.B)
  ]}
 RigVMModel {nodes [
   &RigUnit_ModifyTransforms(
    ExecutePin=()
    ItemToModify=((Item=(Type=Bone,Name="beard_l_02_scale"),Transform=(Rotation=(X=0,Y=0,Z=0,W=-1),Translation=(X=0,Y=-3.3416,Z=-9.7378),Scale3D=(X=1,Y=1,Z=1))))
    Weight=1
    WeightMinimum=0
   )
   > &RigUnit_ModifyTransforms(
    ItemToModify=((Item=(Type=Bone,Name="beard_l_03_scale"),Transform=(Rotation=(X=-0.2745,Y=0.1423,Z=-0.0411,W=0.9501),Translation=(X=0,Y=-6.5816,Z=-5.8208),Scale3D=(X=1,Y=1,Z=1))))
    Weight=1
    WeightMinimum=0
   )
   > &RigUnit_ModifyTransforms(
    ItemToModify=((Item=(Type=Bone,Name="beard_l_04_scale"),Transform=(Rotation=(X=0,Y=0,Z=0,W=1),Translation=(X=0,Y=-3.461,Z=-9.0003),Scale3D=(X=1,Y=1,Z=1))))
    Weight=1
    WeightMinimum=0
   )
   > &RigUnit_ModifyTransforms(
    ItemToModify=((Item=(Type=Bone,Name="beard_l_01_scale"),Transform=(Rotation=(X=0.0804,Y=0.4121,Z=-0.13,W=0.8982),Translation=(X=0,Y=0,Z=0),Scale3D=(X=1,Y=1,Z=1))))
    Weight=1
    WeightMinimum=0
   )
   &RigUnit_ModifyTransforms(
    ExecutePin=()
    ItemToModify=((Item=(Type=Bone,Name="beard_ct_03_scale"),Transform=(Rotation=(X=0,Y=0,Z=0,W=-1),Translation=(X=0,Y=0,Z=0),Scale3D=(X=1,Y=0.01,Z=1))))
    Weight=1
    WeightMinimum=0
   )
   > &RigUnit_ModifyTransforms(
    ItemToModify=((Item=(Type=Bone,Name="beard_l_03_scale"),Transform=(Rotation=(X=0,Y=0,Z=0,W=-1),Translation=(X=0,Y=0,Z=0),Scale3D=(X=1,Y=0.01,Z=1))))
    Weight=1
    WeightMinimum=0
   )
   > &RigUnit_ModifyTransforms(
    ItemToModify=((Item=(Type=Bone,Name="beard_r_03_scale"),Transform=(Rotation=(X=0,Y=0,Z=0,W=-1),Translation=(X=0,Y=0,Z=0),Scale3D=(X=1,Y=0.01,Z=1))))
    Weight=1
    WeightMinimum=0
   )
   > &RigUnit_ModifyTransforms(
    ItemToModify=((Item=(Type=Bone,Name="beard_ct_04_scale"),Transform=(Rotation=(X=0,Y=0,Z=0,W=1),Translation=(X=0,Y=0,Z=0),Scale3D=(X=1,Y=0.25,Z=1))))
    Weight=1
    WeightMinimum=0
   )
   > &RigUnit_ModifyTransforms(
    ItemToModify=((Item=(Type=Bone,Name="beard_r_04_scale"),Transform=(Rotation=(X=0,Y=0,Z=0,W=-1),Translation=(X=0,Y=0,Z=0),Scale3D=(X=1,Y=0.25,Z=1))))
    Weight=1
    WeightMinimum=0
   )
   > &RigUnit_ModifyTransforms(
    ItemToModify=((Item=(Type=Bone,Name="beard_r_04_scale"),Transform=(Rotation=(X=0,Y=0,Z=0,W=-1),Translation=(X=0,Y=0,Z=0),Scale3D=(X=1,Y=0.25,Z=1))))
    Weight=1
    WeightMinimum=0
   )
   &RigUnit_ModifyTransforms(
    ExecutePin=()
    ItemToModify=((Item=(Type=Bone,Name="beard_r_02_scale"),Transform=(Rotation=(X=0,Y=0,Z=0,W=-1),Translation=(X=0,Y=-3.3416,Z=9.7378),Scale3D=(X=1,Y=1,Z=1))))
    Weight=1
    WeightMinimum=0
   )
   > &RigUnit_ModifyTransforms(
    ItemToModify=((Item=(Type=Bone,Name="beard_r_03_scale"),Transform=(Rotation=(X=0.0467,Y=0,Z=0,W=0.9989),Translation=(X=0,Y=-6.5816,Z=5.8208),Scale3D=(X=1,Y=1,Z=1))))
    Weight=1
    WeightMinimum=0
   )
   > &RigUnit_ModifyTransforms(
    ItemToModify=((Item=(Type=Bone,Name="beard_r_04_scale"),Transform=(Rotation=(X=0,Y=0,Z=0,W=1),Translation=(X=0,Y=-3.461,Z=9.0003),Scale3D=(X=1,Y=1,Z=1))))
    Weight=1
    WeightMinimum=0
   )
   > &RigUnit_ModifyTransforms(
    ItemToModify=((Item=(Type=Bone,Name="beard_r_01_scale"),Transform=(Rotation=(X=0.1706,Y=-0.2708,Z=-0.1995,W=0.9262),Translation=(X=0,Y=0,Z=0),Scale3D=(X=1,Y=1,Z=1))))
    Weight=1
    WeightMinimum=0
   )
   RigUnit_BeginExecution = &RigUnit_BeginExecution()
   RigVMFunction_Sequence = RigVMAggregateNode(ExecuteContext=@RigUnit_BeginExecution.ExecutePin)
   Initial Beard Positions = RigVMFunctionReferenceNode(ExecuteContext=@RigVMFunction_Sequence.B)
   n1 = &RigUnit_ControlRigSplineFromTransforms(
    Compression=0
    SamplesPerSegment=16
    SplineMode=BSpline
    Stretch=0
    bClosed=False
    Transforms=RigVMVariableNode() {variable=Beard Ct Initial Transforms}.Value
   )
   > RigVMVariableNode(Value=@n1.Spline) {variable=Beard Ct Initial Spline}
   > RigVMVariableNode(Value=&RigUnit_GetLengthControlRigSpline(Spline=@n1.Spline).Length) {variable=Beard Ct Initial Length}
   > &RigUnit_DrawControlRigSpline(Color=(R=0,G=0,B=1,A=1), Detail=16, Thickness=0.25, Spline=@n1.Spline)
   n2 = &RigUnit_ControlRigSplineFromTransforms(
    Compression=0
    SamplesPerSegment=16
    SplineMode=BSpline
    Stretch=0
    bClosed=False
    Transforms=RigVMVariableNode() {variable=Beard Lf Initial Transforms}.Value
   )
   > &RigUnit_DrawControlRigSpline(Color=(R=0,G=0,B=1,A=1), Detail=16, Thickness=0.25, Spline=@n2.Spline)
   > RigVMVariableNode(Value=@n2.Spline) {variable=Beard Lf Initial Spline}
   > RigVMVariableNode(Value=&RigUnit_GetLengthControlRigSpline(Spline=@n2.Spline).Length) {variable=Beard Lf Initial Length}
   n3 = &RigUnit_ControlRigSplineFromTransforms(
    Compression=0
    SamplesPerSegment=16
    SplineMode=BSpline
    Stretch=0
    bClosed=False
    Transforms=RigVMVariableNode() {variable=Beard Rt Initial Transforms}.Value
   )
   > &RigUnit_DrawControlRigSpline(Color=(R=0,G=0,B=1,A=1), Detail=16, Thickness=0.25, Spline=@n3.Spline)
   > RigVMVariableNode(Value=@n3.Spline) {variable=Beard Rt Initial Spline}
   > RigVMVariableNode(Value=&RigUnit_GetLengthControlRigSpline(Spline=@n3.Spline).Length) {variable=Beard Rt Initial Length}
   RigVMFunction_ControlFlowBranch = &RigVMFunction_ControlFlowBranch(Condition=RigVMVariableNode() {variable=Debug Controls On}.Value, ExecuteContext=@RigVMFunction_Sequence.A)
   For_Each_2 = RigVMDispatchNode(ExecuteContext=@RigVMFunction_ControlFlowBranch.True, Array=RigVMVariableNode() {variable=Debug Ctrls}.Value) {template="DISPATCH_RigVMDispatch_ArrayIterator(in Array,out Element,out Index,out Count,out Ratio)"}
   > &RigUnit_SetTransform(
    Weight=1
    bInitial=False
    Item=RigVMDispatchNode(Index=@For_Each_2.Index, Array=RigVMVariableNode() {variable=Debug Bones}.Value) {template="DISPATCH_RigVMDispatch_ArrayGetAtIndex(in Array,in Index,out Element)"}.Element
    Value=&RigUnit_GetTransform(bInitial=False, Item=@For_Each_2.Element).Transform
   )
   Make_Array_2 = RigVMDispatchNode(Values=(0,1,2,3)) {template="DISPATCH_RigVMDispatch_ArrayMake(in Values,out Array)"}
   For_Each = RigVMDispatchNode(Array=@Make_Array_2.Array, ExecuteContext=@RigVMFunction_Sequence.G) {template="DISPATCH_RigVMDispatch_ArrayIterator(in Array,out Element,out Index,out Count,out Ratio)"}
   n4 = RigVMVariableNode() {variable=Beard Ct Tick Results}
   n5 = RigVMDispatchNode(Index=@For_Each.Index, Array=@n4.Value) {template="DISPATCH_RigVMDispatch_ArrayGetAtIndex(in Array,in Index,out Element)"}
   n6 = RigVMVariableNode() {variable=Beard Lf Tick Results}
   n7 = RigVMDispatchNode(Index=@For_Each.Index, Array=@n6.Value) {template="DISPATCH_RigVMDispatch_ArrayGetAtIndex(in Array,in Index,out Element)"}
   n8 = RigVMVariableNode() {variable=Beard Rt Tick Results}
   n9 = RigVMDispatchNode(Index=@For_Each.Index, Array=@n8.Value) {template="DISPATCH_RigVMDispatch_ArrayGetAtIndex(in Array,in Index,out Element)"}
   Subtract = /Script/RigVM.RigVMFunction_MathVectorSub(A=@n7.Element.Translation, B=@n9.Element.Translation) {is=struct RigVMUnitNode, template="Subtract::Execute(in A,in B,out Result)"}
   n10 = /Script/ControlRig.RigUnit_AimBoneMath(
    DebugSettings=(bEnabled=False,Scale=10,WorldOffset=(Rotation=(X=0,Y=0,Z=0,W=1),Translation=(X=0,Y=0,Z=0),Scale3D=(X=1,Y=1,Z=1)))
    InputTransform=(Rotation=(X=0,Y=0,Z=0,W=1),Translation=(X=0,Y=0,Z=0),Scale3D=(X=1,Y=1,Z=1))
    Primary=(Weight=1,Axis=(X=1,Y=0,Z=0),Target=(X=1,Y=0,Z=0),Kind=Direction,Space=(Type=Bone,Name="None"))
    Secondary=(Weight=1,Axis=(X=0,Y=0,Z=1),Target=(X=1,Y=0,Z=0),Kind=Direction,Space=(Type=Bone,Name="None"))
    Weight=1
    InputTransform.Translation=@n5.Element.Translation
    Primary.Target=RigVMDispatchNode(Index=@For_Each.Index, Array=RigVMVariableNode() {variable=Beard Ct Tick Tangents}.Value) {template="DISPATCH_RigVMDispatch_ArrayGetAtIndex(in Array,in Index,out Element)"}.Element
    Secondary.Target=@Subtract.Result
   ) {is=struct RigVMUnitNode}
   > Set_At = RigVMDispatchNode(Element=@n10.Result, Index=@For_Each.Index, Array=@n4.Value) {template="DISPATCH_RigVMDispatch_ArraySetAtIndex(io Array,in Index,in Element)"}
   > /Script/RigVM.RigVMFunction_DebugTransformMutableNoSpace(
    Color=(R=1,G=0,B=0,A=1)
    DebugDrawSettings=(DepthPriority=SDPG_Foreground,Lifetime=-1)
    Mode=Axes
    Scale=10
    Thickness=1.1
    WorldOffset=(Rotation=(X=0,Y=0,Z=0,W=1),Translation=(X=0,Y=0,Z=0),Scale3D=(X=1,Y=1,Z=1))
    bEnabled=false
    Transform=@n10.Result
   ) {is=struct RigVMUnitNode}
   n11 = &RigUnit_AimBoneMath(
    DebugSettings=(bEnabled=False,Scale=10,WorldOffset=(Rotation=(X=0,Y=0,Z=0,W=1),Translation=(X=0,Y=0,Z=0),Scale3D=(X=1,Y=1,Z=1)))
    InputTransform=(Rotation=(X=0,Y=0,Z=0,W=1),Translation=(X=0,Y=0,Z=0),Scale3D=(X=1,Y=1,Z=1))
    Primary=(Weight=1,Axis=(X=1,Y=0,Z=0),Target=(X=1,Y=0,Z=0),Kind=Direction,Space=(Type=Bone,Name="None"))
    Secondary=(Weight=1,Axis=(X=0,Y=0,Z=1),Target=(X=1,Y=0,Z=0),Kind=Direction,Space=(Type=Bone,Name="None"))
    Weight=1
    InputTransform.Translation=@n7.Element.Translation
    Primary.Target=RigVMDispatchNode(Index=@For_Each.Index, Array=RigVMVariableNode() {variable=Beard Lf Tick Tangents}.Value) {template="DISPATCH_RigVMDispatch_ArrayGetAtIndex(in Array,in Index,out Element)"}.Element
    Secondary.Target=&RigVMFunction_MathVectorSub(B=@n5.Element.Translation, A=@n7.Element.Translation) {template="Subtract::Execute(in A,in B,out Result)"}.Result
   )
   > Set_At_4 = RigVMDispatchNode(Element=@n11.Result, Index=@For_Each.Index, Array=@n6.Value) {template="DISPATCH_RigVMDispatch_ArraySetAtIndex(io Array,in Index,in Element)"}
   > &RigVMFunction_DebugTransformMutableNoSpace(
    Color=(R=1,G=0,B=0,A=1)
    DebugDrawSettings=(DepthPriority=SDPG_Foreground,Lifetime=-1)
    Mode=Axes
    Scale=10
    Thickness=1.1
    WorldOffset=(Rotation=(X=0,Y=0,Z=0,W=1),Translation=(X=0,Y=0,Z=0),Scale3D=(X=1,Y=1,Z=1))
    bEnabled=false
    Transform=@n11.Result
   )
   n12 = &RigUnit_AimBoneMath(
    DebugSettings=(bEnabled=False,Scale=10,WorldOffset=(Rotation=(X=0,Y=0,Z=0,W=1),Translation=(X=0,Y=0,Z=0),Scale3D=(X=1,Y=1,Z=1)))
    InputTransform=(Rotation=(X=0,Y=0,Z=0,W=1),Translation=(X=0,Y=0,Z=0),Scale3D=(X=1,Y=1,Z=1))
    Primary=(Weight=1,Axis=(X=1,Y=0,Z=0),Target=(X=1,Y=0,Z=0),Kind=Direction,Space=(Type=Bone,Name="None"))
    Secondary=(Weight=1,Axis=(X=0,Y=0,Z=1),Target=(X=1,Y=0,Z=0),Kind=Direction,Space=(Type=Bone,Name="None"))
    Weight=1
    InputTransform.Translation=@n9.Element.Translation
    Primary.Target=RigVMDispatchNode(Index=@For_Each.Index, Array=RigVMVariableNode() {variable=Beard Rt Tick Tangents}.Value) {template="DISPATCH_RigVMDispatch_ArrayGetAtIndex(in Array,in Index,out Element)"}.Element
    Secondary.Target=&RigVMFunction_MathVectorSub(A=@n5.Element.Translation, B=@n9.Element.Translation) {template="Subtract::Execute(in A,in B,out Result)"}.Result
   )
   > Set_At_5 = RigVMDispatchNode(Element=@n12.Result, Index=@For_Each.Index, Array=@n8.Value) {template="DISPATCH_RigVMDispatch_ArraySetAtIndex(io Array,in Index,in Element)"}
   > &RigVMFunction_DebugTransformMutableNoSpace(
    Color=(R=1,G=0,B=0,A=1)
    DebugDrawSettings=(DepthPriority=SDPG_Foreground,Lifetime=-1)
    Mode=Axes
    Scale=10
    Thickness=1.1
    WorldOffset=(Rotation=(X=0,Y=0,Z=0,W=1),Translation=(X=0,Y=0,Z=0),Scale3D=(X=1,Y=1,Z=1))
    bEnabled=false
    Transform=@n12.Result
   )
   Make_Array_1 = RigVMDispatchNode(Values=(0,1,2,3)) {template="DISPATCH_RigVMDispatch_ArrayMake(in Values,out Array)"}
   For_Each_3 = RigVMDispatchNode(Array=@Make_Array_1.Array, ExecuteContext=@RigVMFunction_Sequence.C) {template="DISPATCH_RigVMDispatch_ArrayIterator(in Array,out Element,out Index,out Count,out Ratio)"}
   n13 = RigVMVariableNode() {variable=Beard Ct Initial Spline}
   n14 = &RigUnit_ClosestParameterFromControlRigSpline(
    Position=RigVMDispatchNode(Index=@For_Each_3.Index, Array=RigVMVariableNode() {variable=Beard Ct Chain Positions}.Value) {template="DISPATCH_RigVMDispatch_ArrayGetAtIndex(in Array,in Index,out Element)"}.Element
    Spline=@n13.Value
   )
   > Set_At_6 = RigVMDispatchNode(
    Element=&RigVMFunction_MathDoubleDiv(
     A=&RigUnit_GetLengthAtParamControlRigSpline(U=@n14.U, Spline=@n13.Value).Length
     B=RigVMVariableNode() {variable=Beard Ct Initial Length}.Value
    ).Result
    Index=@For_Each_3.Index
    Array=RigVMVariableNode() {variable=Beard Ct Percentages}.Value
   ) {template="DISPATCH_RigVMDispatch_ArraySetAtIndex(io Array,in Index,in Element)"}
   n15 = RigVMVariableNode() {variable=Beard Lf Initial Spline}
   n16 = &RigUnit_ClosestParameterFromControlRigSpline(
    Position=RigVMDispatchNode(Index=@For_Each_3.Index, Array=RigVMVariableNode() {variable=Beard Lf Chain Positions}.Value) {template="DISPATCH_RigVMDispatch_ArrayGetAtIndex(in Array,in Index,out Element)"}.Element
    Spline=@n15.Value
   )
   > Set_At_7 = RigVMDispatchNode(
    Element=&RigVMFunction_MathDoubleDiv(
     A=&RigUnit_GetLengthAtParamControlRigSpline(U=@n16.U, Spline=@n15.Value).Length
     B=RigVMVariableNode() {variable=Beard Lf Initial Length}.Value
    ).Result
    Index=@For_Each_3.Index
    Array=RigVMVariableNode() {variable=Beard Lf Percentages}.Value
   ) {template="DISPATCH_RigVMDispatch_ArraySetAtIndex(io Array,in Index,in Element)"}
   n17 = RigVMVariableNode() {variable=Beard Rt Initial Spline}
   n18 = &RigUnit_ClosestParameterFromControlRigSpline(
    Position=RigVMDispatchNode(Index=@For_Each_3.Index, Array=RigVMVariableNode() {variable=Beard Rt Chain Positions}.Value) {template="DISPATCH_RigVMDispatch_ArrayGetAtIndex(in Array,in Index,out Element)"}.Element
    Spline=@n17.Value
   )
   > Set_At_8 = RigVMDispatchNode(
    Element=&RigVMFunction_MathDoubleDiv(
     A=&RigUnit_GetLengthAtParamControlRigSpline(U=@n18.U, Spline=@n17.Value).Length
     B=RigVMVariableNode() {variable=Beard Rt Initial Length}.Value
    ).Result
    Index=@For_Each_3.Index
    Array=RigVMVariableNode() {variable=Beard Rt Percentages}.Value
   ) {template="DISPATCH_RigVMDispatch_ArraySetAtIndex(io Array,in Index,in Element)"}
   For_Each_5 = RigVMDispatchNode(Array=RigVMVariableNode() {variable=Beard Ct Percentages}.Value, ExecuteContext=@For_Each_3.Completed) {template="DISPATCH_RigVMDispatch_ArrayIterator(in Array,out Element,out Index,out Count,out Ratio)"}
   n19 = RigVMVariableNode() {variable=Beard Ct Initial Spline}
   n20 = &RigUnit_ParameterAtPercentage(Percentage=@For_Each_5.Element, Spline=@n19.Value)
   n21 = &RigUnit_AimBoneMath(
    DebugSettings=(bEnabled=False,Scale=10,WorldOffset=(Rotation=(X=0,Y=0,Z=0,W=1),Translation=(X=0,Y=0,Z=0),Scale3D=(X=1,Y=1,Z=1)))
    InputTransform=(Rotation=(X=0,Y=0,Z=0,W=1),Translation=(X=0,Y=0,Z=0),Scale3D=(X=1,Y=1,Z=1))
    Primary=(Weight=1,Axis=(X=1,Y=0,Z=0),Target=(X=1,Y=0,Z=0),Kind=Direction,Space=(Type=Bone,Name="None"))
    Secondary=(Weight=1,Axis=(X=0,Y=0,Z=1),Target=(X=1,Y=0,Z=0),Kind=Direction,Space=(Type=Bone,Name="None"))
    Weight=1
    InputTransform.Translation=&RigUnit_PositionFromControlRigSpline(U=@n20.U, Spline=@n19.Value).Position
    Primary.Target=&RigUnit_TangentFromControlRigSpline(U=@n20.U, Spline=@n19.Value).Tangent
   )
   > Set_At_9 = RigVMDispatchNode(Element=@n21.Result, Index=@For_Each_5.Index, Array=RigVMVariableNode() {variable=Beard Ct Initial Results}.Value) {template="DISPATCH_RigVMDispatch_ArraySetAtIndex(io Array,in Index,in Element)"}
   > RigVMFunction_DebugTransformMutableNoSpace = &RigVMFunction_DebugTransformMutableNoSpace(
    Color=(R=1,G=0,B=0,A=1)
    DebugDrawSettings=(DepthPriority=SDPG_Foreground,Lifetime=-1)
    Mode=Axes
    Scale=10
    Thickness=0
    WorldOffset=(Rotation=(X=0,Y=0,Z=0,W=1),Translation=(X=0,Y=0,Z=0),Scale3D=(X=1,Y=1,Z=1))
    bEnabled=false
    Transform=@n21.Result
   )
   For_Each_6 = RigVMDispatchNode(Array=RigVMVariableNode() {variable=Beard Lf Percentages}.Value, ExecuteContext=@For_Each_5.Completed) {template="DISPATCH_RigVMDispatch_ArrayIterator(in Array,out Element,out Index,out Count,out Ratio)"}
   n22 = RigVMVariableNode() {variable=Beard Lf Initial Spline}
   n23 = &RigUnit_ParameterAtPercentage(Percentage=@For_Each_6.Element, Spline=@n22.Value)
   n24 = &RigUnit_AimBoneMath(
    DebugSettings=(bEnabled=False,Scale=10,WorldOffset=(Rotation=(X=0,Y=0,Z=0,W=1),Translation=(X=0,Y=0,Z=0),Scale3D=(X=1,Y=1,Z=1)))
    InputTransform=(Rotation=(X=0,Y=0,Z=0,W=1),Translation=(X=0,Y=0,Z=0),Scale3D=(X=1,Y=1,Z=1))
    Primary=(Weight=1,Axis=(X=1,Y=0,Z=0),Target=(X=1,Y=0,Z=0),Kind=Direction,Space=(Type=Bone,Name="None"))
    Secondary=(Weight=1,Axis=(X=0,Y=0,Z=1),Target=(X=1,Y=0,Z=0),Kind=Direction,Space=(Type=Bone,Name="None"))
    Weight=1
    InputTransform.Translation=&RigUnit_PositionFromControlRigSpline(U=@n23.U, Spline=@n22.Value).Position
    Primary.Target=&RigUnit_TangentFromControlRigSpline(U=@n23.U, Spline=@n22.Value).Tangent
   )
   > Set_At_10 = RigVMDispatchNode(Element=@n24.Result, Index=@For_Each_6.Index, Array=RigVMVariableNode() {variable=Beard Lf Initial Results}.Value) {template="DISPATCH_RigVMDispatch_ArraySetAtIndex(io Array,in Index,in Element)"}
   > &RigVMFunction_DebugTransformMutableNoSpace(
    Color=(R=1,G=0,B=0,A=1)
    DebugDrawSettings=(DepthPriority=SDPG_Foreground,Lifetime=-1)
    Mode=Axes
    Scale=10
    Thickness=0
    WorldOffset=(Rotation=(X=0,Y=0,Z=0,W=1),Translation=(X=0,Y=0,Z=0),Scale3D=(X=1,Y=1,Z=1))
    bEnabled=false
    Transform=@n24.Result
   )
   For_Each_7 = RigVMDispatchNode(Array=RigVMVariableNode() {variable=Beard Rt Percentages}.Value, ExecuteContext=@For_Each_6.Completed) {template="DISPATCH_RigVMDispatch_ArrayIterator(in Array,out Element,out Index,out Count,out Ratio)"}
   n25 = RigVMVariableNode() {variable=Beard Rt Initial Spline}
   n26 = &RigUnit_ParameterAtPercentage(Percentage=@For_Each_7.Element, Spline=@n25.Value)
   n27 = &RigUnit_AimBoneMath(
    DebugSettings=(bEnabled=False,Scale=10,WorldOffset=(Rotation=(X=0,Y=0,Z=0,W=1),Translation=(X=0,Y=0,Z=0),Scale3D=(X=1,Y=1,Z=1)))
    InputTransform=(Rotation=(X=0,Y=0,Z=0,W=1),Translation=(X=0,Y=0,Z=0),Scale3D=(X=1,Y=1,Z=1))
    Primary=(Weight=1,Axis=(X=1,Y=0,Z=0),Target=(X=1,Y=0,Z=0),Kind=Direction,Space=(Type=Bone,Name="None"))
    Secondary=(Weight=1,Axis=(X=0,Y=0,Z=1),Target=(X=1,Y=0,Z=0),Kind=Direction,Space=(Type=Bone,Name="None"))
    Weight=1
    InputTransform.Translation=&RigUnit_PositionFromControlRigSpline(U=@n26.U, Spline=@n25.Value).Position
    Primary.Target=&RigUnit_TangentFromControlRigSpline(U=@n26.U, Spline=@n25.Value).Tangent
   )
   > Set_At_11 = RigVMDispatchNode(Element=@n27.Result, Index=@For_Each_7.Index, Array=RigVMVariableNode() {variable=Beard Rt Initial Results}.Value) {template="DISPATCH_RigVMDispatch_ArraySetAtIndex(io Array,in Index,in Element)"}
   > &RigVMFunction_DebugTransformMutableNoSpace(
    Color=(R=1,G=0,B=0,A=1)
    DebugDrawSettings=(DepthPriority=SDPG_Foreground,Lifetime=-1)
    Mode=Axes
    Scale=10
    Thickness=0
    WorldOffset=(Rotation=(X=0,Y=0,Z=0,W=1),Translation=(X=0,Y=0,Z=0),Scale3D=(X=1,Y=1,Z=1))
    bEnabled=false
    Transform=@n27.Result
   )
   n28 = &RigUnit_SphericalPoseReader(
    ActiveRegionScaleFactors=(PositiveWidth=1,NegativeWidth=1,PositiveHeight=1,NegativeHeight=1)
    ActiveRegionSize=0.001
    Debug=(bDrawDebug=false,bDraw2D=False,bDrawLocalAxes=False,DebugScale=25,DebugSegments=20,DebugThickness=0.25)
    DriverAxis=(X=0,Y=1,Z=0)
    DriverItem=(Type=Bone,Name="head")
    FalloffRegionScaleFactors=(PositiveWidth=1,NegativeWidth=1,PositiveHeight=1,NegativeHeight=1)
    FalloffSize=0.5
    FlipHeightScaling=False
    FlipWidthScaling=False
    OptionalParentItem=(Type=Bone,Name="spine_04")
    RotationOffset=(X=180,Y=0,Z=0)
    ExecutePin=@n63.ExecuteContext
   )
   n29 = &RigUnit_SphericalPoseReader(
    ActiveRegionScaleFactors=(PositiveWidth=1,NegativeWidth=1,PositiveHeight=1,NegativeHeight=1)
    ActiveRegionSize=0.001
    Debug=(bDrawDebug=false,bDraw2D=False,bDrawLocalAxes=False,DebugScale=25,DebugSegments=20,DebugThickness=0.25)
    DriverAxis=(X=0,Y=1,Z=0)
    DriverItem=(Type=Bone,Name="head")
    FalloffRegionScaleFactors=(PositiveWidth=1,NegativeWidth=1,PositiveHeight=1,NegativeHeight=1)
    FalloffSize=0.5
    FlipHeightScaling=False
    FlipWidthScaling=False
    OptionalParentItem=(Type=Bone,Name="spine_04")
    RotationOffset=(X=0,Y=0,Z=0)
    ExecutePin=@RigVMFunction_Sequence.D
   )
   Tick Beard Positions_1 = RigVMFunctionReferenceNode(headTwist_l=@n28.OutputParam, headTwist_r=@n29.OutputParam, ExecuteContext=@RigVMFunction_Sequence.E)
   n30 = &RigUnit_ControlRigSplineFromTransforms(
    Compression=0
    SamplesPerSegment=16
    SplineMode=BSpline
    Stretch=0
    bClosed=False
    Transforms=RigVMVariableNode() {variable=Beard Ct Tick Transforms}.Value
   )
   > &RigUnit_DrawControlRigSpline(Color=(R=1,G=0,B=0.0522,A=1), Detail=16, Thickness=0.375, Spline=@n30.Spline)
   > RigVMVariableNode(Value=@n30.Spline) {variable=Beard Ct Tick Spline}
   > RigVMVariableNode(Value=&RigUnit_GetLengthControlRigSpline(Spline=@n30.Spline).Length) {variable=Beard Ct Tick Length}
   n31 = &RigUnit_ControlRigSplineFromTransforms(
    Compression=0
    SamplesPerSegment=16
    SplineMode=BSpline
    Stretch=0
    bClosed=False
    Transforms=RigVMVariableNode() {variable=Beard Lf Tick Transforms}.Value
   )
   > &RigUnit_DrawControlRigSpline(Color=(R=1,G=0,B=0.0522,A=1), Detail=16, Thickness=0.375, Spline=@n31.Spline)
   > RigVMVariableNode(Value=@n31.Spline) {variable=Beard Lf Tick Spline}
   > RigVMVariableNode(Value=&RigUnit_GetLengthControlRigSpline(Spline=@n31.Spline).Length) {variable=Beard Lf Tick Length}
   n32 = &RigUnit_ControlRigSplineFromTransforms(
    Compression=0
    SamplesPerSegment=16
    SplineMode=BSpline
    Stretch=0
    bClosed=False
    Transforms=RigVMVariableNode() {variable=Beard Rt Tick Transforms}.Value
   )
   > &RigUnit_DrawControlRigSpline(Color=(R=1,G=0,B=0.0522,A=1), Detail=16, Thickness=0.375, Spline=@n32.Spline)
   > RigVMVariableNode(Value=@n32.Spline) {variable=Beard Rt Tick Spline}
   > RigVMVariableNode(Value=&RigUnit_GetLengthControlRigSpline(Spline=@n32.Spline).Length) {variable=Beard Rt Tick Length}
   Make_Array_3 = RigVMDispatchNode(Values=(0,1,2,3)) {template="DISPATCH_RigVMDispatch_ArrayMake(in Values,out Array)"}
   For_Each_1 = RigVMDispatchNode(Array=@Make_Array_3.Array, ExecuteContext=@RigVMFunction_Sequence.H) {template="DISPATCH_RigVMDispatch_ArrayIterator(in Array,out Element,out Index,out Count,out Ratio)"}
   n33 = RigVMDispatchNode(Index=@For_Each_1.Index, Array=RigVMVariableNode() {variable=Beard Ct Chain Bones}.Value) {template="DISPATCH_RigVMDispatch_ArrayGetAtIndex(in Array,in Index,out Element)"}
   n34 = RigVMVariableNode() {variable=Beard Ct Tick Results}
   n35 = RigVMDispatchNode(Index=@For_Each_1.Index, Array=RigVMVariableNode() {variable=Beard Ct Initial Results}.Value) {template="DISPATCH_RigVMDispatch_ArrayGetAtIndex(in Array,in Index,out Element)"}
   n36 = &RigUnit_GetTransform(bInitial=true, Item=@n33.Element)
   > &RigUnit_SetTransform(
    Weight=1
    bInitial=False
    Item=@n33.Element
    Value=/Script/RigVM.RigVMFunction_MathTransformMakeAbsolute(
     Parent=RigVMDispatchNode(Index=@For_Each_1.Index, Array=@n34.Value) {template="DISPATCH_RigVMDispatch_ArrayGetAtIndex(in Array,in Index,out Element)"}.Element
     Local=&RigVMFunction_MathTransformMakeRelative(Parent=@n35.Element, Global=@n36.Transform).Local
    ) {is=struct RigVMUnitNode, template="Make Absolute::Execute(in Local,in Parent,out Global)"}.Global
   )
   n37 = RigVMDispatchNode(Index=@For_Each_1.Index, Array=RigVMVariableNode() {variable=Beard Lf Chain Bones}.Value) {template="DISPATCH_RigVMDispatch_ArrayGetAtIndex(in Array,in Index,out Element)"}
   n38 = RigVMVariableNode() {variable=Beard Lf Tick Results}
   n39 = RigVMDispatchNode(Index=@For_Each_1.Index, Array=RigVMVariableNode() {variable=Beard Lf Initial Results}.Value) {template="DISPATCH_RigVMDispatch_ArrayGetAtIndex(in Array,in Index,out Element)"}
   n40 = &RigUnit_GetTransform(bInitial=true, Item=@n37.Element)
   > &RigUnit_SetTransform(
    Weight=1
    bInitial=False
    Item=@n37.Element
    Value=&RigVMFunction_MathTransformMakeAbsolute(
     Parent=RigVMDispatchNode(Index=@For_Each_1.Index, Array=@n38.Value) {template="DISPATCH_RigVMDispatch_ArrayGetAtIndex(in Array,in Index,out Element)"}.Element
     Local=&RigVMFunction_MathTransformMakeRelative(Parent=@n39.Element, Global=@n40.Transform).Local
    ) {template="Make Absolute::Execute(in Local,in Parent,out Global)"}.Global
   )
   n41 = RigVMDispatchNode(Index=@For_Each_1.Index, Array=RigVMVariableNode() {variable=Beard Rt Chain Bones}.Value) {template="DISPATCH_RigVMDispatch_ArrayGetAtIndex(in Array,in Index,out Element)"}
   n42 = RigVMVariableNode() {variable=Beard Rt Tick Results}
   n43 = RigVMDispatchNode(Index=@For_Each_1.Index, Array=RigVMVariableNode() {variable=Beard Rt Initial Results}.Value) {template="DISPATCH_RigVMDispatch_ArrayGetAtIndex(in Array,in Index,out Element)"}
   n44 = &RigUnit_GetTransform(bInitial=true, Item=@n41.Element)
   > &RigUnit_SetTransform(
    Weight=1
    bInitial=False
    Item=@n41.Element
    Value=&RigVMFunction_MathTransformMakeAbsolute(
     Parent=RigVMDispatchNode(Index=@For_Each_1.Index, Array=@n42.Value) {template="DISPATCH_RigVMDispatch_ArrayGetAtIndex(in Array,in Index,out Element)"}.Element
     Local=&RigVMFunction_MathTransformMakeRelative(Parent=@n43.Element, Global=@n44.Transform).Local
    ) {template="Make Absolute::Execute(in Local,in Parent,out Global)"}.Global
   )
   For_Each_8 = RigVMDispatchNode(Array=RigVMVariableNode() {variable=Beard Ct Percentages}.Value, ExecuteContext=@RigVMFunction_Sequence.F) {template="DISPATCH_RigVMDispatch_ArrayIterator(in Array,out Element,out Index,out Count,out Ratio)"}
   n45 = RigVMVariableNode() {variable=Beard Ct Tick Length}
   n46 = RigVMVariableNode() {variable=Beard Ct Initial Length}
   n47 = RigVMVariableNode() {variable=Beard Ct Tick Spline}
   n48 = &RigUnit_ParameterAtPercentage(
    Percentage=&RigVMFunction_MathDoubleMul(B=&RigVMFunction_MathDoubleDiv(B=@n45.Value, A=@n46.Value).Result, A=@For_Each_8.Element).Result
    Spline=@n47.Value
   )
   n49 = &RigUnit_TangentFromControlRigSpline(U=@n48.U, Spline=@n47.Value)
   n50 = &RigUnit_AimBoneMath(
    DebugSettings=(bEnabled=False,Scale=10,WorldOffset=(Rotation=(X=0,Y=0,Z=0,W=1),Translation=(X=0,Y=0,Z=0),Scale3D=(X=1,Y=1,Z=1)))
    InputTransform=(Rotation=(X=0,Y=0,Z=0,W=1),Translation=(X=0,Y=0,Z=0),Scale3D=(X=1,Y=1,Z=1))
    Primary=(Weight=1,Axis=(X=1,Y=0,Z=0),Target=(X=1,Y=0,Z=0),Kind=Direction,Space=(Type=Bone,Name="None"))
    Secondary=(Weight=1,Axis=(X=0,Y=0,Z=1),Target=(X=1,Y=0,Z=0),Kind=Direction,Space=(Type=Bone,Name="None"))
    Weight=1
    InputTransform.Translation=&RigUnit_PositionFromControlRigSpline(U=@n48.U, Spline=@n47.Value).Position
    Primary.Target=@n49.Tangent
   )
   > Set_At_12 = RigVMDispatchNode(Element=@n50.Result, Index=@For_Each_8.Index, Array=RigVMVariableNode() {variable=Beard Ct Tick Results}.Value) {template="DISPATCH_RigVMDispatch_ArraySetAtIndex(io Array,in Index,in Element)"}
   > &RigVMFunction_DebugTransformMutableNoSpace(
    Color=(R=0.4296,G=1,B=0.9809,A=1)
    DebugDrawSettings=(DepthPriority=SDPG_Foreground,Lifetime=-1)
    Mode=Point
    Scale=3
    Thickness=1
    WorldOffset=(Rotation=(X=0,Y=0,Z=0,W=1),Translation=(X=0,Y=0,Z=0),Scale3D=(X=1,Y=1,Z=1))
    bEnabled=false
    Transform=@n50.Result
   )
   > Set_At_1 = RigVMDispatchNode(Index=@For_Each_8.Index, Element=@n49.Tangent, Array=RigVMVariableNode() {variable=Beard Ct Tick Tangents}.Value) {template="DISPATCH_RigVMDispatch_ArraySetAtIndex(io Array,in Index,in Element)"}
   For_Each_9 = RigVMDispatchNode(Array=RigVMVariableNode() {variable=Beard Lf Percentages}.Value, ExecuteContext=@For_Each_8.Completed) {template="DISPATCH_RigVMDispatch_ArrayIterator(in Array,out Element,out Index,out Count,out Ratio)"}
   n51 = RigVMVariableNode() {variable=Beard Lf Initial Length}
   n52 = RigVMVariableNode() {variable=Beard Lf Tick Length}
   n53 = RigVMVariableNode() {variable=Beard Lf Tick Spline}
   n54 = &RigUnit_ParameterAtPercentage(
    Percentage=&RigVMFunction_MathDoubleMul(B=&RigVMFunction_MathDoubleDiv(A=@n51.Value, B=@n52.Value).Result, A=@For_Each_9.Element).Result
    Spline=@n53.Value
   )
   n55 = &RigUnit_TangentFromControlRigSpline(U=@n54.U, Spline=@n53.Value)
   n56 = &RigUnit_AimBoneMath(
    DebugSettings=(bEnabled=False,Scale=10,WorldOffset=(Rotation=(X=0,Y=0,Z=0,W=1),Translation=(X=0,Y=0,Z=0),Scale3D=(X=1,Y=1,Z=1)))
    InputTransform=(Rotation=(X=0,Y=0,Z=0,W=1),Translation=(X=0,Y=0,Z=0),Scale3D=(X=1,Y=1,Z=1))
    Primary=(Weight=1,Axis=(X=1,Y=0,Z=0),Target=(X=1,Y=0,Z=0),Kind=Direction,Space=(Type=Bone,Name="None"))
    Secondary=(Weight=1,Axis=(X=0,Y=0,Z=1),Target=(X=1,Y=0,Z=0),Kind=Direction,Space=(Type=Bone,Name="None"))
    Weight=1
    InputTransform.Translation=&RigUnit_PositionFromControlRigSpline(U=@n54.U, Spline=@n53.Value).Position
    Primary.Target=@n55.Tangent
   )
   > Set_At_13 = RigVMDispatchNode(Element=@n56.Result, Index=@For_Each_9.Index, Array=RigVMVariableNode() {variable=Beard Lf Tick Results}.Value) {template="DISPATCH_RigVMDispatch_ArraySetAtIndex(io Array,in Index,in Element)"}
   > &RigVMFunction_DebugTransformMutableNoSpace(
    Color=(R=0.4296,G=1,B=0.9809,A=1)
    DebugDrawSettings=(DepthPriority=SDPG_Foreground,Lifetime=-1)
    Mode=Point
    Scale=3
    Thickness=1
    WorldOffset=(Rotation=(X=0,Y=0,Z=0,W=1),Translation=(X=0,Y=0,Z=0),Scale3D=(X=1,Y=1,Z=1))
    bEnabled=false
    Transform=@n56.Result
   )
   > Set_At_2 = RigVMDispatchNode(Index=@For_Each_9.Index, Element=@n55.Tangent, Array=RigVMVariableNode() {variable=Beard Lf Tick Tangents}.Value) {template="DISPATCH_RigVMDispatch_ArraySetAtIndex(io Array,in Index,in Element)"}
   For_Each_10 = RigVMDispatchNode(Array=RigVMVariableNode() {variable=Beard Rt Percentages}.Value, ExecuteContext=@For_Each_9.Completed) {template="DISPATCH_RigVMDispatch_ArrayIterator(in Array,out Element,out Index,out Count,out Ratio)"}
   n57 = RigVMVariableNode() {variable=Beard Rt Initial Length}
   n58 = RigVMVariableNode() {variable=Beard Rt Tick Length}
   n59 = RigVMVariableNode() {variable=Beard Rt Tick Spline}
   n60 = &RigUnit_ParameterAtPercentage(
    Percentage=&RigVMFunction_MathDoubleMul(B=&RigVMFunction_MathDoubleDiv(A=@n57.Value, B=@n58.Value).Result, A=@For_Each_10.Element).Result
    Spline=@n59.Value
   )
   n61 = &RigUnit_TangentFromControlRigSpline(U=@n60.U, Spline=@n59.Value)
   n62 = &RigUnit_AimBoneMath(
    DebugSettings=(bEnabled=False,Scale=10,WorldOffset=(Rotation=(X=0,Y=0,Z=0,W=1),Translation=(X=0,Y=0,Z=0),Scale3D=(X=1,Y=1,Z=1)))
    InputTransform=(Rotation=(X=0,Y=0,Z=0,W=1),Translation=(X=0,Y=0,Z=0),Scale3D=(X=1,Y=1,Z=1))
    Primary=(Weight=1,Axis=(X=1,Y=0,Z=0),Target=(X=1,Y=0,Z=0),Kind=Direction,Space=(Type=Bone,Name="None"))
    Secondary=(Weight=1,Axis=(X=0,Y=0,Z=1),Target=(X=1,Y=0,Z=0),Kind=Direction,Space=(Type=Bone,Name="None"))
    Weight=1
    InputTransform.Translation=&RigUnit_PositionFromControlRigSpline(U=@n60.U, Spline=@n59.Value).Position
    Primary.Target=@n61.Tangent
   )
   > Set_At_14 = RigVMDispatchNode(Element=@n62.Result, Index=@For_Each_10.Index, Array=RigVMVariableNode() {variable=Beard Rt Tick Results}.Value) {template="DISPATCH_RigVMDispatch_ArraySetAtIndex(io Array,in Index,in Element)"}
   > &RigVMFunction_DebugTransformMutableNoSpace(
    Color=(R=0.4296,G=1,B=0.9809,A=1)
    DebugDrawSettings=(DepthPriority=SDPG_Foreground,Lifetime=-1)
    Mode=Point
    Scale=3
    Thickness=1
    WorldOffset=(Rotation=(X=0,Y=0,Z=0,W=1),Translation=(X=0,Y=0,Z=0),Scale3D=(X=1,Y=1,Z=1))
    bEnabled=false
    Transform=@n62.Result
   )
   > Set_At_3 = RigVMDispatchNode(Index=@For_Each_10.Index, Element=@n61.Tangent, Array=RigVMVariableNode() {variable=Beard Rt Tick Tangents}.Value) {template="DISPATCH_RigVMDispatch_ArraySetAtIndex(io Array,in Index,in Element)"}
   RigVMDispatchNode(Enabled=false, Prefix=Turn L, ScreenColor=(R=1,G=1,B=1,A=1), ScreenDuration=0.05, ExecuteContext=@n28.ExecutePin, Value=@n28.OutputParam) {template="DISPATCH_RigVMDispatch_Print(in Prefix,in Value,in Enabled,in ScreenDuration,in ScreenColor)"}
   RigVMRerouteNode(Value=@RigVMFunction_Sequence.I)
   RigVMVariableNode() {variable=NewVar}
   n63 = RigVMDispatchNode(Enabled=false, Prefix=Turn R, ScreenColor=(R=1,G=1,B=1,A=1), ScreenDuration=0.05, ExecuteContext=@n29.ExecutePin, Value=@n29.OutputParam) {template="DISPATCH_RigVMDispatch_Print(in Prefix,in Value,in Enabled,in ScreenDuration,in ScreenColor)"}
  ]}
 RigVMModel Rig {nodes [
   n1 = &RigUnit_PrepareForExecution()
   RigVMFunction_Sequence = &RigVMFunction_Sequence(ExecuteContext=@n1.ExecutePin)
   Initial Beard Positions = RigVMFunctionReferenceNode(ExecuteContext=@RigVMFunction_Sequence.B)
   ItemArray = RigVMDispatchNode(Value=((Type=Bone,Name="beard_ct_MSTR"),(Type=Bone,Name="beard_ct_02"),(Type=Bone,Name="beard_ct_03"),(Type=Bone,Name="beard_ct_04"))) {template="DISPATCH_RigVMDispatch_Constant(io Value)"}
   > RigVMVariableNode(Value=@ItemArray.Value) {variable=Beard Ct Chain Bones}
   ItemArray_1 = RigVMDispatchNode(Value=((Type=Bone,Name="beard_l_MSTR"),(Type=Bone,Name="beard_l_02"),(Type=Bone,Name="beard_l_03"),(Type=Bone,Name="beard_l_04"))) {template="DISPATCH_RigVMDispatch_Constant(io Value)"}
   > RigVMVariableNode(Value=@ItemArray_1.Value) {variable=Beard Lf Chain Bones}
   ItemArray_2 = RigVMDispatchNode(Value=((Type=Bone,Name="beard_r_MSTR"),(Type=Bone,Name="beard_r_02"),(Type=Bone,Name="beard_r_03"),(Type=Bone,Name="beard_r_04"))) {template="DISPATCH_RigVMDispatch_Constant(io Value)"}
   > RigVMVariableNode(Value=@ItemArray_2.Value) {variable=Beard Rt Chain Bones}
   > RigVMDispatchNode(Array=RigVMVariableNode() {variable=Beard Ct Chain Positions}.Value) {template="DISPATCH_RigVMDispatch_ArrayReset(io Array)"}
   > RigVMDispatchNode(Array=RigVMVariableNode() {variable=Beard Lf Chain Positions}.Value) {template="DISPATCH_RigVMDispatch_ArrayReset(io Array)"}
   > RigVMDispatchNode(Array=RigVMVariableNode() {variable=Beard Rt Chain Positions}.Value) {template="DISPATCH_RigVMDispatch_ArrayReset(io Array)"}
   > RigVMDispatchNode(Array=RigVMVariableNode() {variable=Beard Ct Percentages}.Value) {template="DISPATCH_RigVMDispatch_ArrayReset(io Array)"}
   > RigVMDispatchNode(Array=RigVMVariableNode() {variable=Beard Lf Percentages}.Value) {template="DISPATCH_RigVMDispatch_ArrayReset(io Array)"}
   > RigVMDispatchNode(Array=RigVMVariableNode() {variable=Beard Rt Percentages}.Value) {template="DISPATCH_RigVMDispatch_ArrayReset(io Array)"}
   > RigVMDispatchNode(Array=RigVMVariableNode() {variable=Beard Ct Initial Results}.Value) {template="DISPATCH_RigVMDispatch_ArrayReset(io Array)"}
   > RigVMDispatchNode(Array=RigVMVariableNode() {variable=Beard Lf Initial Results}.Value) {template="DISPATCH_RigVMDispatch_ArrayReset(io Array)"}
   > RigVMDispatchNode(Array=RigVMVariableNode() {variable=Beard Rt Initial Results}.Value) {template="DISPATCH_RigVMDispatch_ArrayReset(io Array)"}
   > RigVMDispatchNode(Array=RigVMVariableNode() {variable=Beard Ct Tick Results}.Value) {template="DISPATCH_RigVMDispatch_ArrayReset(io Array)"}
   > RigVMDispatchNode(Array=RigVMVariableNode() {variable=Beard Lf Tick Results}.Value) {template="DISPATCH_RigVMDispatch_ArrayReset(io Array)"}
   > RigVMDispatchNode(Array=RigVMVariableNode() {variable=Beard Rt Tick Results}.Value) {template="DISPATCH_RigVMDispatch_ArrayReset(io Array)"}
   > RigVMDispatchNode(Array=RigVMVariableNode() {variable=Beard Ct Tick Tangents}.Value) {template="DISPATCH_RigVMDispatch_ArrayReset(io Array)"}
   > RigVMDispatchNode(Array=RigVMVariableNode() {variable=Beard Lf Tick Tangents}.Value) {template="DISPATCH_RigVMDispatch_ArrayReset(io Array)"}
   > RigVMDispatchNode(Array=RigVMVariableNode() {variable=Beard Rt Tick Tangents}.Value) {template="DISPATCH_RigVMDispatch_ArrayReset(io Array)"}
   Make_Array_1 = RigVMDispatchNode(Values=(0,1,2,3)) {template="DISPATCH_RigVMDispatch_ArrayMake(in Values,out Array)"}
   > For_Each_4 = RigVMDispatchNode(Array=@Make_Array_1.Array) {template="DISPATCH_RigVMDispatch_ArrayIterator(in Array,out Element,out Index,out Count,out Ratio)"}
   n2 = RigVMVariableNode() {variable=Beard Ct Chain Bones}
   > RigVMDispatchNode(
    Element=&RigUnit_GetTransform(
     bInitial=true
     Item=RigVMDispatchNode(Index=@For_Each_4.Index, Array=@n2.Value) {template="DISPATCH_RigVMDispatch_ArrayGetAtIndex(in Array,in Index,out Element)"}.Element
    ).Transform.Translation
    Array=RigVMVariableNode() {variable=Beard Ct Chain Positions}.Value
   ) {template="DISPATCH_RigVMDispatch_ArrayAdd(io Array,in Element,out Index)"}
   n3 = RigVMVariableNode() {variable=Beard Lf Chain Bones}
   > RigVMDispatchNode(
    Element=&RigUnit_GetTransform(
     bInitial=true
     Item=RigVMDispatchNode(Index=@For_Each_4.Index, Array=@n3.Value) {template="DISPATCH_RigVMDispatch_ArrayGetAtIndex(in Array,in Index,out Element)"}.Element
    ).Transform.Translation
    Array=RigVMVariableNode() {variable=Beard Lf Chain Positions}.Value
   ) {template="DISPATCH_RigVMDispatch_ArrayAdd(io Array,in Element,out Index)"}
   n4 = RigVMVariableNode() {variable=Beard Rt Chain Bones}
   > RigVMDispatchNode(
    Element=&RigUnit_GetTransform(
     bInitial=true
     Item=RigVMDispatchNode(Index=@For_Each_4.Index, Array=@n4.Value) {template="DISPATCH_RigVMDispatch_ArrayGetAtIndex(in Array,in Index,out Element)"}.Element
    ).Transform.Translation
    Array=RigVMVariableNode() {variable=Beard Rt Chain Positions}.Value
   ) {template="DISPATCH_RigVMDispatch_ArrayAdd(io Array,in Element,out Index)"}
   > RigVMDispatchNode(Element=0, Array=RigVMVariableNode() {variable=Beard Ct Percentages}.Value) {template="DISPATCH_RigVMDispatch_ArrayAdd(io Array,in Element,out Index)"}
   > RigVMDispatchNode(Element=-0.25, Array=RigVMVariableNode() {variable=Beard Lf Percentages}.Value) {template="DISPATCH_RigVMDispatch_ArrayAdd(io Array,in Element,out Index)"}
   > RigVMDispatchNode(Element=0.25, Array=RigVMVariableNode() {variable=Beard Rt Percentages}.Value) {template="DISPATCH_RigVMDispatch_ArrayAdd(io Array,in Element,out Index)"}
   > Add = RigVMDispatchNode(
    Element=(Rotation=(X=0,Y=0,Z=0,W=1),Translation=(X=0,Y=0,Z=0),Scale3D=(X=1,Y=1,Z=1))
    Array=RigVMVariableNode() {variable=Beard Ct Initial Results}.Value
   ) {template="DISPATCH_RigVMDispatch_ArrayAdd(io Array,in Element,out Index)"}
   > RigVMDispatchNode(
    Element=(Rotation=(X=0,Y=0,Z=0,W=1),Translation=(X=0,Y=0,Z=0),Scale3D=(X=1,Y=1,Z=1))
    Array=RigVMVariableNode() {variable=Beard Lf Initial Results}.Value
   ) {template="DISPATCH_RigVMDispatch_ArrayAdd(io Array,in Element,out Index)"}
   > RigVMDispatchNode(
    Element=(Rotation=(X=0,Y=0,Z=0,W=1),Translation=(X=0,Y=0,Z=0),Scale3D=(X=1,Y=1,Z=1))
    Array=RigVMVariableNode() {variable=Beard Rt Initial Results}.Value
   ) {template="DISPATCH_RigVMDispatch_ArrayAdd(io Array,in Element,out Index)"}
   > RigVMDispatchNode(
    Element=(Rotation=(X=0,Y=0,Z=0,W=1),Translation=(X=0,Y=0,Z=0),Scale3D=(X=1,Y=1,Z=1))
    Array=RigVMVariableNode() {variable=Beard Ct Tick Results}.Value
   ) {template="DISPATCH_RigVMDispatch_ArrayAdd(io Array,in Element,out Index)"}
   > RigVMDispatchNode(
    Element=(Rotation=(X=0,Y=0,Z=0,W=1),Translation=(X=0,Y=0,Z=0),Scale3D=(X=1,Y=1,Z=1))
    Array=RigVMVariableNode() {variable=Beard Lf Tick Results}.Value
   ) {template="DISPATCH_RigVMDispatch_ArrayAdd(io Array,in Element,out Index)"}
   > RigVMDispatchNode(
    Element=(Rotation=(X=0,Y=0,Z=0,W=1),Translation=(X=0,Y=0,Z=0),Scale3D=(X=1,Y=1,Z=1))
    Array=RigVMVariableNode() {variable=Beard Rt Tick Results}.Value
   ) {template="DISPATCH_RigVMDispatch_ArrayAdd(io Array,in Element,out Index)"}
   > RigVMDispatchNode(Element=(X=0,Y=0,Z=0), Array=RigVMVariableNode() {variable=Beard Ct Tick Tangents}.Value) {template="DISPATCH_RigVMDispatch_ArrayAdd(io Array,in Element,out Index)"}
   > RigVMDispatchNode(Element=(X=0,Y=0,Z=0), Array=RigVMVariableNode() {variable=Beard Lf Tick Tangents}.Value) {template="DISPATCH_RigVMDispatch_ArrayAdd(io Array,in Element,out Index)"}
   > RigVMDispatchNode(Element=(X=0,Y=0,Z=0), Array=RigVMVariableNode() {variable=Beard Rt Tick Tangents}.Value) {template="DISPATCH_RigVMDispatch_ArrayAdd(io Array,in Element,out Index)"}
   RigVMFunction_ControlFlowBranch = &RigVMFunction_ControlFlowBranch(Condition=RigVMVariableNode() {variable=Debug Controls On}.Value, ExecuteContext=@RigVMFunction_Sequence.A)
   RigVMDispatchNode(ExecuteContext=@RigVMFunction_ControlFlowBranch.True, Array=RigVMVariableNode() {variable=Debug Ctrls}.Value) {template="DISPATCH_RigVMDispatch_ArrayReset(io Array)"}
   ItemArray_4 = RigVMDispatchNode(Value=((Type=Bone,Name="spine_03"),(Type=Bone,Name="spine_04"),(Type=Bone,Name="spine_05"),(Type=Bone,Name="neck_01"),(Type=Bone,Name="neck_02"),(Type=Bone,Name="head"))) {template="DISPATCH_RigVMDispatch_Constant(io Value)"}
   > RigVMVariableNode(Value=@ItemArray_4.Value) {variable=Debug Bones}
   > For_Each = RigVMDispatchNode(Array=@ItemArray_4.Value) {template="DISPATCH_RigVMDispatch_ArrayIterator(in Array,out Element,out Index,out Count,out Ratio)"}
   Concat = &RigVMFunction_NameConcat(B=" Ctrl", A=@For_Each.Element.Name) {template="Concat::Execute(in A,in B,out Result)"}
   > n5 = &RigUnit_HierarchyAddControlTransform(
    InitialValue=(Rotation=(X=0,Y=0,Z=0,W=1),Translation=(X=0,Y=0,Z=0),Scale3D=(X=1,Y=1,Z=1))
    OffsetSpace=GlobalSpace
    Settings="(InitialSpace=LocalSpace,bUsePreferredRotationOrder=False,PreferredRotationOrder=YZX,Limits=(LimitTranslationX=(bMinimum=False,bMaximum=False),LimitTranslationY=(bMinimum=False,bMaximum=False),LimitTranslationZ=(bMinimum=False,bMaximum=False),LimitPitch=(b ...(truncated)"
    Name=@Concat.Result
    Parent=&RigUnit_HierarchyGetParent(bDefaultParent=True, Child=@For_Each.Element).Parent
    OffsetTransform=&RigUnit_GetTransform(bInitial=False, Item=@For_Each.Element).Transform
   ) {template="SpawnControl::Execute(in InitialValue,in Settings,in OffsetTransform,in OffsetSpace,in Parent,in Name,out Item)"}
   > RigVMDispatchNode(Element=@n5.Item, Array=RigVMVariableNode() {variable=Debug Ctrls}.Value) {template="DISPATCH_RigVMDispatch_ArrayAdd(io Array,in Element,out Index)"}
  ]}
]
hierarchy [
 root {loc=<0 0 0>, rot=<0 0 0 1>}
 pelvis {parent=root, loc=<0 -0.7321 92.4218>, rot=<0.0127 0.707 -0.0127 -0.707>}
 spine_01 {parent=pelvis, loc=<5.6002 -0.5128 0>, rot=<0 0 0.034 -0.9994>}
 spine_02 {parent=spine_01, loc=<8.0756 0.0139 0>, rot=<0 0 0.0238 -0.9997>}
 spine_03 {parent=spine_02, loc=<7.0395 0 0>, rot=<0 0 0.0149 -0.9999>}
 spine_04 {parent=spine_03, loc=<8.1023 0 0>, rot=<0 0 -0.0132 -0.9999>}
 spine_05 {parent=spine_04, loc=<10.0141 0 0>, rot=<0 0 -0.0015 -1>}
 clavicle_l {parent=spine_05, loc=<2.5856 -0.2866 -4.3777>, rot=<0.0465 -0.7289 -0.009 -0.6829>}
 neck_01 {parent=spine_05, loc=<10.815 1.9196 0>, rot=<0 0 -0.1864 -0.9825>}
 neck_02 {parent=neck_01, loc=<4.7347 0.3045 0>, rot=<0 0 0.0363 -0.9993>}
 head {parent=neck_02, loc=<5.7632 0.5552 0>, rot=<0 0 0.115 -0.9934>}
 stache_ct_MSTR {parent=head, loc=<1.1231 8.645 0>, rot=<0 0 -0.7071 -0.7071>}
 stache_ct {parent=stache_ct_MSTR, loc=<0 0 0>, rot=<0 0 0 1>}
 stache_l_MSTR {parent=head, loc=<1.1231 8.645 -5.1068>, rot=<0.1033 -0.1396 -0.7917 -0.5857>}
 stache_l {parent=stache_l_MSTR, loc=<0 0 0>, rot=<0 0 0 1>}
 stache_l_end {parent=stache_l, loc=<4.5925 0 0>, rot=<0 0 0 1>}
 stache_r_MSTR {parent=head, loc=<1.1231 8.645 5.107>, rot=<-0.1033 0.1396 -0.7917 -0.5857>}
 stache_r {parent=stache_r_MSTR, loc=<0 0 0>, rot=<0 0 0 1>}
 stache_r_end {parent=stache_r, loc=<4.5925 0 0>, rot=<0 0 0 1>}
 handle_l_MSTR {parent=head, loc=<0.8242 4.4795 -9.8827>, rot=<-0.251 0.0423 0.9413 0.2218>}
 handle_l {parent=handle_l_MSTR, loc=<0 0 0>, rot=<0 0 0 1>}
 handle_l_end {parent=handle_l, loc=<6.5429 0 0>, rot=<0 0 0 1>}
 handle_r_MSTR {parent=head, loc=<1.2243 4.3106 10.1457>, rot=<0.2313 -0.0523 0.9507 0.2>}
 handle_r {parent=handle_r_MSTR, loc=<0 0 0>, rot=<0 0 0 1>}
 handle_r_end {parent=handle_r, loc=<6.1656 0 0>, rot=<0 0 0 1>}
 beard_l_MSTR {parent=head, loc=<-2.0854 5.7076 -11.001>, rot=<-0.166 0.0336 0.966 0.1955>}
 beard_l_01 {parent=beard_l_MSTR, loc=<0 0 0>, rot=<0 0 0 1>}
 beard_l_02 {parent=beard_l_01, loc=<9.2705 0 0>, rot=<-0.0024 0.0665 -0.0194 -0.9976>}
 beard_l_02_scale {parent=beard_l_02, loc=<0 0 0>, rot=<0 0 0 1>}
 beard_l_03 {parent=beard_l_02, loc=<10.549 0 0>, rot=<-0.0006 0.0867 -0.0074 -0.9962>}
 beard_l_03_scale {parent=beard_l_03, loc=<0 0 0>, rot=<0 0 0 1>}
 beard_l_04 {parent=beard_l_03, loc=<10.6421 0 0>, rot=<0 0 0 -1>}
 beard_l_04_scale {parent=beard_l_04, loc=<0 0 0>, rot=<0 0 0 1>}
 beard_r_MSTR {parent=head, loc=<-2.0855 5.7075 11.001>, rot=<0.166 -0.0336 0.966 0.1955>}
 beard_r_01 {parent=beard_r_MSTR, loc=<0 0 0>, rot=<0 0 0 1>}
 beard_r_02 {parent=beard_r_01, loc=<9.2708 0 0>, rot=<-0.0118 -0.0668 -0.0184 -0.9975>}
 beard_r_02_scale {parent=beard_r_02, loc=<0 0 0>, rot=<0 0 0 1>}
 beard_r_03 {parent=beard_r_02, loc=<10.5482 0 0>, rot=<-0.0032 -0.0869 -0.0046 -0.9962>}
 beard_r_03_scale {parent=beard_r_03, loc=<0 0 0>, rot=<0 0 0 1>}
 beard_r_04 {parent=beard_r_03, loc=<10.6425 0 0>, rot=<0 -0.033 0.0006 -0.9995>}
 beard_r_04_scale {parent=beard_r_04, loc=<0 0 0>, rot=<0 0 0 1>}
 beard_ct_MSTR {parent=head, loc=<-2.0854 5.7076 0>, rot=<0 0 0.9801 0.1984>}
 beard_ct_01 {parent=beard_ct_MSTR, loc=<0 0 0>, rot=<0 0 0 1>}
 beard_ct_02 {parent=beard_ct_01, loc=<8.739 0 0>, rot=<0 0 -0.0391 -0.9992>}
 beard_ct_02_scale {parent=beard_ct_02, loc=<0 0 0>, rot=<0 0 0 1>}
 beard_ct_03 {parent=beard_ct_02, loc=<10.3236 0 0>, rot=<0 0 -0.006 -1>}
 beard_ct_03_scale {parent=beard_ct_03, loc=<0 0 0>, rot=<0 0 0 1>}
 beard_ct_04 {parent=beard_ct_03, loc=<10.6365 0 0>, rot=<0 0 -0.0011 -1>}
 beard_ct_04_scale {parent=beard_ct_04, loc=<0 0 0>, rot=<0 0 0 1>}
 clavicle_r {parent=spine_05, loc=<2.5855 -0.2866 4.3777>, rot=<0.7289 0.0465 -0.6829 0.009>}
 belt {parent=pelvis, loc=<0 0 0>, rot=<0 0 0 1>}
 belt_fr {parent=belt, loc=<4.5916 23.0737 0>, rot=<0 0 0 1>}
 belt_leg_l {parent=belt, loc=<4.3907 17.3936 -13.6396>, rot=<0 0 0 1>}
 belt_leg_r {parent=belt, loc=<4.3907 17.3936 13.6396>, rot=<0 0 0 1>}
]
variables [
 Beard Ct Chain Bones {type="TArray<FRigElementKey>"}
 Beard Ct Chain Positions {type="TArray<FVector>"}
 Beard Ct Initial Length {type=float, default=5}
 Beard Ct Initial Results {type="TArray<FTransform>"}
 Beard Ct Initial Spline {type=FControlRigSpline, default=()}
 Beard Ct Initial Transforms {type="TArray<FTransform>"}
 Beard Ct Percentages {type="TArray<double>"}
 Beard Ct Tick Length {type=float, default=5}
 Beard Ct Tick Results {type="TArray<FTransform>"}
 Beard Ct Tick Spline {type=FControlRigSpline, default=()}
 Beard Ct Tick Tangents {type="TArray<FVector>"}
 Beard Ct Tick Transforms {type="TArray<FTransform>"}
 Beard Lf Chain Bones {type="TArray<FRigElementKey>"}
 Beard Lf Chain Positions {type="TArray<FVector>"}
 Beard Lf Initial Length {type=float, default=5}
 Beard Lf Initial Results {type="TArray<FTransform>"}
 Beard Lf Initial Spline {type=FControlRigSpline, default=()}
 Beard Lf Initial Transforms {type="TArray<FTransform>"}
 Beard Lf Percentages {type="TArray<double>"}
 Beard Lf Tick Length {type=float, default=5}
 Beard Lf Tick Results {type="TArray<FTransform>"}
 Beard Lf Tick Spline {type=FControlRigSpline, default=()}
 Beard Lf Tick Tangents {type="TArray<FVector>"}
 Beard Lf Tick Transforms {type="TArray<FTransform>"}
 Beard Rt Chain Bones {type="TArray<FRigElementKey>"}
 Beard Rt Chain Positions {type="TArray<FVector>"}
 Beard Rt Initial Length {type=float, default=5}
 Beard Rt Initial Results {type="TArray<FTransform>"}
 Beard Rt Initial Spline {type=FControlRigSpline, default=()}
 Beard Rt Initial Transforms {type="TArray<FTransform>"}
 Beard Rt Percentages {type="TArray<double>"}
 Beard Rt Tick Length {type=float, default=5}
 Beard Rt Tick Results {type="TArray<FTransform>"}
 Beard Rt Tick Spline {type=FControlRigSpline, default=()}
 Beard Rt Tick Tangents {type="TArray<FVector>"}
 Beard Rt Tick Transforms {type="TArray<FTransform>"}
 Debug Bones {type="TArray<FRigElementKey>", default=((Type=Bone,Name="calf_l"),(Type=Bone,Name="calf_r"))}
 Debug Controls On {type=bool, default=False}
 Debug Ctrls {type="TArray<FRigElementKey>", default=(())}
 Debug Shape {
  type=FRigUnit_HierarchyAddControl_ShapeSettings
  default=(bVisible=True,Name="Circle_Thick",Color=(R=1,G=0.9718,B=0,A=1),Transform=(Rotation=(X=0,Y=-0.7071,Z=0,W=0.7071),Translation=(X=0,Y=0,Z=0),Scale3D=(X=2,Y=2,Z=2)))
 }
 NewVar {type="TArray<FString>"}
]
