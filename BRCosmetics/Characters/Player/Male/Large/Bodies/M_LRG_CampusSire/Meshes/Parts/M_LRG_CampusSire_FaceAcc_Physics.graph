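# Physics asset for the CampusSire large male character's face accessory, giving beard, moustache and handle bones physical secondary motion.
# Capsule bodies on pelvis, spine and clavicles form a torso volume, joined by free-rotation constraints.
# Each facial chain hangs from a tiny MSTR anchor body angularly locked to spine_05, with no head body.
# The beard swings widely with crossed capsules on lower segments, while handles and moustache have tighter limits.
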PhysicsAsset /BRCosmetics/Characters/Player/Male/Large/Bodies/M_LRG_CampusSire/Meshes/Parts/M_LRG_CampusSire_FaceAcc_Physics
bodies {
 beard_ct_01 [{shapeType=Capsule, length=8.739, radius=2.875, shapeName=beard_ct_01_capsule, center=<4.1895 0 0>, rotation=<rot 90 0 0>}]
 beard_ct_02 [
  {shapeType=Capsule, length=8.739, radius=2.875, shapeName=beard_ct_02_capsule, center=<4.1895 0 0>, rotation=<rot 90 0 0>}
  {shapeType=Capsule, length=14.8434, radius=2.875, shapeName=beard_ct_02_capsule}
 ]
 beard_ct_03 [
  {shapeType=Capsule, length=10.324, radius=2.875, shapeName=beard_ct_03_capsule, center=<5.162 0 0>, rotation=<rot 90 0 0>}
  {shapeType=Capsule, length=17.8944, radius=2.875, shapeName=beard_ct_03_capsule}
 ]
 beard_ct_04 [
  {shapeType=Capsule, length=6, radius=2.875, shapeName=beard_ct_04_capsule, center=<3 0 0>, rotation=<rot 90 0 0>}
  {shapeType=Capsule, length=16.2, radius=2.875, shapeName=beard_ct_04_capsule}
 ]
 beard_ct_MSTR [{shapeType=Capsule, length=0.505, radius=0.505, shapeName=beard_ct_MSTR_capsule, rotation=<rot 0 -90 0>}]
 clavicle_l [{shapeType=Capsule, length=18, radius=14, center=<18 0 0>, rotation=<rot 90 0 0>}]
 clavicle_r [{shapeType=Capsule, length=18, radius=14, center=<-18 0 0>, rotation=<rot 90 0 0>}]
 handle_l [{shapeType=Capsule, length=11.5, radius=2, shapeName=beard_ct_03_capsule, center=<5.7105 0 0>, rotation=<rot 90 0 0>}]
 handle_l_MSTR [{shapeType=Capsule, length=0.505, radius=0.505, shapeName=beard_ct_MSTR_capsule, rotation=<rot 0 -90 0>}]
 handle_l_end [{shapeType=Capsule, length=2, radius=1.5, shapeName=beard_ct_04_capsule, center=<1 0 0>, rotation=<rot 90 0 0>}]
 handle_r [{shapeType=Capsule, length=10, radius=2, shapeName=beard_ct_03_capsule, center=<4.9232 0 0>, rotation=<rot 90 0 0>}]
 handle_r_MSTR [{shapeType=Capsule, length=0.505, radius=0.505, shapeName=beard_ct_MSTR_capsule, rotation=<rot 0 -90 0>}]
 handle_r_end [{shapeType=Capsule, length=2, radius=1.5, shapeName=beard_ct_04_capsule, center=<1 0 0>, rotation=<rot 90 0 0>}]
 pelvis [{shapeType=Capsule, length=57, radius=14, center=<3.9839 3.0237 0>}]
 spine_01 [{shapeType=Capsule, length=57, radius=14, center=<3.9839 3.0237 0>}]
 spine_02 [{shapeType=Capsule, length=57, radius=14, center=<4.7743 4.3579 0>}]
 spine_03 [{shapeType=Capsule, length=57, radius=14, shapeName=spine_03_capsule, rotation=<rot 0 -90 0>}]
 spine_04 [{shapeType=Capsule, length=57, radius=14, center=<0.7121 2.0074 0>}]
 spine_05 [{shapeType=Capsule, length=57, radius=14, center=<2.6768 -1.5171 0>}]
 stache_ct [{shapeType=Capsule, length=1, radius=2, shapeName=beard_ct_03_capsule, center=<0.5 0 0>, rotation=<rot 90 0 0>}]
 stache_ct_MSTR [{shapeType=Capsule, length=0.505, radius=0.505, shapeName=beard_ct_MSTR_capsule, rotation=<rot 0 -90 0>}]
 stache_l [{shapeType=Capsule, length=1, radius=2, shapeName=beard_ct_03_capsule, center=<0.5 0 0>, rotation=<rot 90 0 0>}]
 stache_l_MSTR [{shapeType=Capsule, length=0.505, radius=0.505, shapeName=beard_ct_MSTR_capsule, rotation=<rot 0 -90 0>}]
 stache_l_end [{shapeType=Capsule, length=1, radius=2, shapeName=beard_ct_04_capsule, center=<0.5 0 0>, rotation=<rot 90 0 0>}]
 stache_r [{shapeType=Capsule, length=1, radius=2, shapeName=beard_ct_03_capsule, center=<0.5 0 0>, rotation=<rot 90 0 0>}]
 stache_r_MSTR [{shapeType=Capsule, length=0.505, radius=0.505, shapeName=beard_ct_MSTR_capsule, rotation=<rot 0 -90 0>}]
 stache_r_end [{shapeType=Capsule, length=1, radius=2, shapeName=beard_ct_04_capsule, center=<0.5 0 0>, rotation=<rot 90 0 0>}]
}
constraints [
 {
  bone1Name=spine_03
  bone2Name=spine_02
  swing1LimitDegrees=45
  swing1Motion=Free
  swing2LimitDegrees=45
  swing2Motion=Free
  twistLimitDegrees=45
  twistMotion=Free
 }
 {
  bone1Name=spine_04
  bone2Name=spine_03
  swing1LimitDegrees=45
  swing1Motion=Free
  swing2LimitDegrees=45
  swing2Motion=Free
  twistLimitDegrees=45
  twistMotion=Free
 }
 {
  bone1Name=spine_05
  bone2Name=spine_04
  swing1LimitDegrees=45
  swing1Motion=Free
  swing2LimitDegrees=45
  swing2Motion=Free
  twistLimitDegrees=45
  twistMotion=Free
 }
 {
  bone1Name=clavicle_l
  bone2Name=spine_05
  swing1LimitDegrees=45
  swing1Motion=Free
  swing2LimitDegrees=45
  swing2Motion=Free
  twistLimitDegrees=45
  twistMotion=Free
 }
 {
  bone1Name=clavicle_r
  bone2Name=spine_05
  swing1LimitDegrees=45
  swing1Motion=Free
  swing2LimitDegrees=45
  swing2Motion=Free
  twistLimitDegrees=45
  twistMotion=Free
 }
 {
  bone1Name=spine_01
  bone2Name=pelvis
  swing1LimitDegrees=45
  swing1Motion=Free
  swing2LimitDegrees=45
  swing2Motion=Free
  twistLimitDegrees=45
  twistMotion=Free
 }
 {
  bone1Name=spine_02
  bone2Name=spine_01
  swing1LimitDegrees=45
  swing1Motion=Free
  swing2LimitDegrees=45
  swing2Motion=Free
  twistLimitDegrees=45
  twistMotion=Free
 }
 {bone1Name=beard_ct_03, bone2Name=beard_ct_02, swing1LimitDegrees=90, swing2LimitDegrees=90, twistLimitDegrees=12}
 {bone1Name=beard_ct_04, bone2Name=beard_ct_03, swing1LimitDegrees=90, swing2LimitDegrees=90, twistLimitDegrees=12}
 {bone1Name=beard_ct_02, bone2Name=beard_ct_01, swing1LimitDegrees=90, swing2LimitDegrees=90, twistLimitDegrees=12}
 {
  bone1Name=beard_ct_MSTR
  bone2Name=spine_05
  swing1LimitDegrees=45
  swing1Motion=Locked
  swing2LimitDegrees=45
  swing2Motion=Locked
  twistLimitDegrees=45
  twistMotion=Locked
 }
 {bone1Name=beard_ct_01, bone2Name=beard_ct_MSTR, swing1LimitDegrees=90, swing2LimitDegrees=90, twistLimitDegrees=12}
 {
  bone1Name=handle_l_MSTR
  bone2Name=spine_05
  swing1LimitDegrees=45
  swing1Motion=Locked
  swing2LimitDegrees=45
  swing2Motion=Locked
  twistLimitDegrees=45
  twistMotion=Locked
 }
 {bone1Name=handle_l, bone2Name=handle_l_MSTR, swing1LimitDegrees=30, swing2LimitDegrees=30, twistLimitDegrees=5}
 {bone1Name=handle_l_end, bone2Name=handle_l, swing1LimitDegrees=45, swing2LimitDegrees=45, twistLimitDegrees=5, twistMotion=Locked}
 {
  bone1Name=handle_r_MSTR
  bone2Name=spine_05
  swing1LimitDegrees=45
  swing1Motion=Locked
  swing2LimitDegrees=45
  swing2Motion=Locked
  twistLimitDegrees=45
  twistMotion=Locked
 }
 {bone1Name=handle_r, bone2Name=handle_r_MSTR, swing1LimitDegrees=30, swing2LimitDegrees=30, twistLimitDegrees=5}
 {bone1Name=handle_r_end, bone2Name=handle_r, swing1LimitDegrees=45, swing2LimitDegrees=45, twistLimitDegrees=5, twistMotion=Locked}
 {
  bone1Name=stache_ct_MSTR
  bone2Name=spine_05
  swing1LimitDegrees=45
  swing1Motion=Locked
  swing2LimitDegrees=45
  swing2Motion=Locked
  twistLimitDegrees=45
  twistMotion=Locked
 }
 {bone1Name=stache_ct, bone2Name=stache_ct_MSTR, swing1LimitDegrees=10, swing2LimitDegrees=10, twistLimitDegrees=5}
 {
  bone1Name=stache_l_MSTR
  bone2Name=spine_05
  swing1LimitDegrees=45
  swing1Motion=Locked
  swing2LimitDegrees=45
  swing2Motion=Locked
  twistLimitDegrees=45
  twistMotion=Locked
 }
 {bone1Name=stache_l, bone2Name=stache_l_MSTR, swing1LimitDegrees=10, swing2LimitDegrees=10, twistLimitDegrees=5}
 {bone1Name=stache_l_end, bone2Name=stache_l, swing1LimitDegrees=30, swing2LimitDegrees=30, twistLimitDegrees=5}
 {
  bone1Name=stache_r_MSTR
  bone2Name=spine_05
  swing1LimitDegrees=45
  swing1Motion=Locked
  swing2LimitDegrees=45
  swing2Motion=Locked
  twistLimitDegrees=45
  twistMotion=Locked
 }
 {bone1Name=stache_r, bone2Name=stache_r_MSTR, swing1LimitDegrees=10, swing2LimitDegrees=10, twistLimitDegrees=5}
 {bone1Name=stache_r_end, bone2Name=stache_r, swing1LimitDegrees=30, swing2LimitDegrees=30, twistLimitDegrees=5}
]
