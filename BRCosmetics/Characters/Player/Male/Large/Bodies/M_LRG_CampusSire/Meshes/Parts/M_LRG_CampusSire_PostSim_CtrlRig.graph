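# Post-simulation Control Rig for the large male CampusSire body that keeps the side beard chains moving with the center chain.
# Forward solve loops over center beard bones 02–04, setting matching left and right bones from each center bone's current transform plus their initial relative offset.
# With Debug Controls On, construction spawns controls for the center chain, which drive those bones before the side bones follow.
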ControlRigBlueprint /BRCosmetics/Characters/Player/Male/Large/Bodies/M_LRG_CampusSire/Meshes/Parts/M_LRG_CampusSire_PostSim_CtrlRig
graphs [
 RigVMFunctionLibrary {links [], nodes []}
 RigVMModel {nodes [
   RigUnit_BeginExecution = &RigUnit_BeginExecution()
   RigVMFunction_ControlFlowBranch = &RigVMFunction_ControlFlowBranch(ExecuteContext=@RigUnit_BeginExecution.ExecutePin, Condition=RigVMVariableNode() {variable=Debug Controls On}.Value)
   ItemArray = RigVMDispatchNode(Value=((Type=Bone,Name="beard_ct_02"),(Type=Bone,Name="beard_ct_03"),(Type=Bone,Name="beard_ct_04"))) {template="DISPATCH_RigVMDispatch_Constant(io Value)"}
   For_Each = RigVMDispatchNode(Array=@ItemArray.Value, ExecuteContext=@RigVMFunction_ControlFlowBranch.Completed) {template="DISPATCH_RigVMDispatch_ArrayIterator(in Array,out Element,out Index,out Count,out Ratio)"}
   ItemArray_2 = RigVMDispatchNode(Value=((Type=Bone,Name="beard_l_02"),(Type=Bone,Name="beard_l_03"),(Type=Bone,Name="beard_l_04"))) {template="DISPATCH_RigVMDispatch_Constant(io Value)"}
   n1 = RigVMDispatchNode(Index=@For_Each.Index, Array=@ItemArray_2.Value) {template="DISPATCH_RigVMDispatch_ArrayGetAtIndex(in Array,in Index,out Element)"}
   n2 = &RigUnit_GetTransform(bInitial=true, Item=@n1.Element)
   n3 = &RigUnit_GetTransform(bInitial=true, Item=@For_Each.Element)
   > &RigUnit_SetTransform(
    Weight=1
    bInitial=False
    Item=@n1.Element
    Value=/Script/RigVM.RigVMFunction_MathTransformMakeAbsolute(
     Parent=&RigUnit_GetTransform(bInitial=False, Item=@For_Each.Element).Transform
     Local=&RigVMFunction_MathTransformMakeRelative(Global=@n2.Transform, Parent=@n3.Transform).Local
    ) {is=struct RigVMUnitNode, template="Make Absolute::Execute(in Local,in Parent,out Global)"}.Global
   )
   ItemArray_1 = RigVMDispatchNode(Value=((Type=Bone,Name="beard_r_02"),(Type=Bone,Name="beard_r_03"),(Type=Bone,Name="beard_r_04"))) {template="DISPATCH_RigVMDispatch_Constant(io Value)"}
   n4 = RigVMDispatchNode(Index=@For_Each.Index, Array=@ItemArray_1.Value) {template="DISPATCH_RigVMDispatch_ArrayGetAtIndex(in Array,in Index,out Element)"}
   n5 = &RigUnit_GetTransform(bInitial=true, Item=@n4.Element)
   n6 = &RigUnit_GetTransform(bInitial=true, Item=@For_Each.Element)
   > &RigUnit_SetTransform(
    Weight=1
    bInitial=False
    Item=@n4.Element
    Value=&RigVMFunction_MathTransformMakeAbsolute(
     Parent=&RigUnit_GetTransform(bInitial=False, Item=@For_Each.Element).Transform
     Local=&RigVMFunction_MathTransformMakeRelative(Global=@n5.Transform, Parent=@n6.Transform).Local
    ) {template="Make Absolute::Execute(in Local,in Parent,out Global)"}.Global
   )
   For_Each_2 = RigVMDispatchNode(ExecuteContext=@RigVMFunction_ControlFlowBranch.True, Array=RigVMVariableNode() {variable=Debug Ctrls}.Value) {template="DISPATCH_RigVMDispatch_ArrayIterator(in Array,out Element,out Index,out Count,out Ratio)"}
   > &RigUnit_SetTransform(
    Weight=1
    bInitial=False
    Item=RigVMDispatchNode(Index=@For_Each_2.Index, Array=RigVMVariableNode() {variable=Debug Bones}.Value) {template="DISPATCH_RigVMDispatch_ArrayGetAtIndex(in Array,in Index,out Element)"}.Element
    Value=&RigUnit_GetTransform(bInitial=False, Item=@For_Each_2.Element).Transform
   )
  ]}
 RigVMModel Rig {nodes [
   n1 = &RigUnit_PrepareForExecution()
   RigVMFunction_ControlFlowBranch = &RigVMFunction_ControlFlowBranch(ExecuteContext=@n1.ExecutePin, Condition=RigVMVariableNode() {variable=Debug Controls On}.Value)
   RigVMDispatchNode(ExecuteContext=@RigVMFunction_ControlFlowBranch.True, Array=RigVMVariableNode() {variable=Debug Ctrls}.Value) {template="DISPATCH_RigVMDispatch_ArrayReset(io Array)"}
   ItemArray = RigVMDispatchNode(Value=((Type=Bone,Name="beard_ct_01"),(Type=Bone,Name="beard_ct_02"),(Type=Bone,Name="beard_ct_03"),(Type=Bone,Name="beard_ct_04"))) {template="DISPATCH_RigVMDispatch_Constant(io Value)"}
   > RigVMVariableNode(Value=@ItemArray.Value) {variable=Debug Bones}
   > For_Each = RigVMDispatchNode(Array=@ItemArray.Value) {template="DISPATCH_RigVMDispatch_ArrayIterator(in Array,out Element,out Index,out Count,out Ratio)"}
   Concat = /Script/RigVM.RigVMFunction_NameConcat(B=" Ctrl", A=@For_Each.Element.Name) {is=struct RigVMUnitNode, template="Concat::Execute(in A,in B,out Result)"}
   > n2 = /Script/ControlRig.RigUnit_HierarchyAddControlTransform(
    InitialValue=(Rotation=(X=0,Y=0,Z=0,W=1),Translation=(X=0,Y=0,Z=0),Scale3D=(X=1,Y=1,Z=1))
    OffsetSpace=GlobalSpace
    Settings="(InitialSpace=LocalSpace,bUsePreferredRotationOrder=False,PreferredRotationOrder=YZX,Limits=(LimitTranslationX=(bMinimum=False,bMaximum=False),LimitTranslationY=(bMinimum=False,bMaximum=False),LimitTranslationZ=(bMinimum=False,bMaximum=False),LimitPitch=(b ...(truncated)"
    Name=@Concat.Result
    Parent=/Script/ControlRig.RigUnit_HierarchyGetParent(bDefaultParent=True, Child=@For_Each.Element) {is=struct RigVMUnitNode}.Parent
    OffsetTransform=&RigUnit_GetTransform(bInitial=False, Item=@For_Each.Element).Transform
   ) {
    is=struct RigVMUnitNode
    template="SpawnControl::Execute(in InitialValue,in Settings,in OffsetTransform,in OffsetSpace,in Parent,in Name,out Item)"
   }
   > RigVMDispatchNode(Element=@n2.Item, Array=RigVMVariableNode() {variable=Debug Ctrls}.Value) {template="DISPATCH_RigVMDispatch_ArrayAdd(io Array,in Element,out Index)"}
  ]}
]
hierarchy [
 root {loc=<0 0 0>, rot=<0 0 0 1>}
 pelvis {parent=root, loc=<0 -0.7321 92.4218>, rot=<0.0127 0.707 -0.0127 -0.707>}
 spine_01 {parent=pelvis, loc=<5.6002 -0.5128 0>, rot=<0 0 0.034 -0.9994>}
 spine_02 {parent=spine_01, loc=<8.0756 0.0139 0>, rot=<0 0 0.0238 -0.9997>}
 spine_03 {parent=spine_02, loc=<7.0395 0 0>, rot=<0 0 0.0149 -0.9999>}
 spine_04 {parent=spine_03, loc=<8.1023 0 0>, rot=<0 0 -0.0132 -0.9999>}
 spine_05 {parent=spine_04, loc=<10.0141 0 0>, rot=<0 0 -0.0015 -1>}
 clavicle_l {parent=spine_05, loc=<2.5856 -0.2866 -4.3777>, rot=<0.0465 -0.7289 -0.009 -0.6829>}
 neck_01 {parent=spine_05, loc=<10.815 1.9196 0>, rot=<0 0 -0.1864 -0.9825>}
 neck_02 {parent=neck_01, loc=<4.7347 0.3045 0>, rot=<0 0 0.0363 -0.9993>}
 head {parent=neck_02, loc=<5.7632 0.5552 0>, rot=<0 0 0.115 -0.9934>}
 stache_ct_MSTR {parent=head, loc=<1.1231 8.645 0>, rot=<0 0 -0.7071 -0.7071>}
 stache_ct {parent=stache_ct_MSTR, loc=<0 0 0>, rot=<0 0 0 1>}
 stache_l_MSTR {parent=head, loc=<1.1231 8.645 -5.1068>, rot=<0.1033 -0.1396 -0.7917 -0.5857>}
 stache_l {parent=stache_l_MSTR, loc=<0 0 0>, rot=<0 0 0 1>}
 stache_l_end {parent=stache_l, loc=<4.5925 0 0>, rot=<0 0 0 1>}
 stache_r_MSTR {parent=head, loc=<1.1231 8.645 5.107>, rot=<-0.1033 0.1396 -0.7917 -0.5857>}
 stache_r {parent=stache_r_MSTR, loc=<0 0 0>, rot=<0 0 0 1>}
 stache_r_end {parent=stache_r, loc=<4.5925 0 0>, rot=<0 0 0 1>}
 handle_l_MSTR {parent=head, loc=<-2.0854 5.7076 -11.001>, rot=<-0.166 0.0336 0.966 0.1955>}
 handle_l {parent=handle_l_MSTR, loc=<0 0 0>, rot=<0 0 0 1>}
 handle_l_end {parent=handle_l, loc=<3.9733 0 0>, rot=<0 0 0 1>}
 handle_r_MSTR {parent=head, loc=<-2.0855 5.7075 11.001>, rot=<0.166 -0.0336 0.966 0.1955>}
 handle_r {parent=handle_r_MSTR, loc=<0 0 0>, rot=<0 0 0 1>}
 handle_r_end {parent=handle_r, loc=<3.9733 0 0>, rot=<0 0 0 1>}
 beard_l_MSTR {parent=head, loc=<-2.0854 5.7076 -11.001>, rot=<-0.166 0.0336 0.966 0.1955>}
 beard_l_01 {parent=beard_l_MSTR, loc=<0 0 0>, rot=<0 0 0 1>}
 beard_l_02 {parent=beard_l_01, loc=<9.2705 0 0>, rot=<-0.0024 0.0665 -0.0194 -0.9976>}
 beard_l_02_scale {parent=beard_l_02, loc=<0 0 0>, rot=<0 0 0 1>}
 beard_l_03 {parent=beard_l_02, loc=<10.549 0 0>, rot=<-0.0006 0.0867 -0.0074 -0.9962>}
 beard_l_03_scale {parent=beard_l_03, loc=<0 0 0>, rot=<0 0 0 1>}
 beard_l_04 {parent=beard_l_03, loc=<10.6421 0 0>, rot=<0 0 0 -1>}
 beard_l_04_scale {parent=beard_l_04, loc=<0 0 0>, rot=<0 0 0 1>}
 beard_r_MSTR {parent=head, loc=<-2.0855 5.7075 11.001>, rot=<0.166 -0.0336 0.966 0.1955>}
 beard_r_01 {parent=beard_r_MSTR, loc=<0 0 0>, rot=<0 0 0 1>}
 beard_r_02 {parent=beard_r_01, loc=<9.2708 0 0>, rot=<-0.0118 -0.0668 -0.0184 -0.9975>}
 beard_r_02_scale {parent=beard_r_02, loc=<0 0 0>, rot=<0 0 0 1>}
 beard_r_03 {parent=beard_r_02, loc=<10.5482 0 0>, rot=<-0.0032 -0.0869 -0.0046 -0.9962>}
 beard_r_03_scale {parent=beard_r_03, loc=<0 0 0>, rot=<0 0 0 1>}
 beard_r_04 {parent=beard_r_03, loc=<10.6425 0 0>, rot=<0 -0.033 0.0006 -0.9995>}
 beard_r_04_scale {parent=beard_r_04, loc=<0 0 0>, rot=<0 0 0 1>}
 beard_ct_MSTR {parent=head, loc=<-2.0854 5.7076 0>, rot=<0 0 0.9801 0.1984>}
 beard_ct_01 {parent=beard_ct_MSTR, loc=<0 0 0>, rot=<0 0 0 1>}
 beard_ct_02 {parent=beard_ct_01, loc=<8.739 0 0>, rot=<0 0 -0.0391 -0.9992>}
 beard_ct_02_scale {parent=beard_ct_02, loc=<0 0 0>, rot=<0 0 0 1>}
 beard_ct_03 {parent=beard_ct_02, loc=<10.3236 0 0>, rot=<0 0 -0.006 -1>}
 beard_ct_03_scale {parent=beard_ct_03, loc=<0 0 0>, rot=<0 0 0 1>}
 beard_ct_04 {parent=beard_ct_03, loc=<10.6365 0 0>, rot=<0 0 -0.0011 -1>}
 beard_ct_04_scale {parent=beard_ct_04, loc=<0 0 0>, rot=<0 0 0 1>}
 clavicle_r {parent=spine_05, loc=<2.5855 -0.2866 4.3777>, rot=<0.7289 0.0465 -0.6829 0.009>}
 belt {parent=pelvis, loc=<0 0 0>, rot=<0 0 0 1>}
 belt_fr {parent=belt, loc=<4.5916 23.0737 0>, rot=<0 0 0 1>}
 belt_leg_l {parent=belt, loc=<4.3907 17.3936 -13.6396>, rot=<0 0 0 1>}
 belt_leg_r {parent=belt, loc=<4.3907 17.3936 13.6396>, rot=<0 0 0 1>}
]
variables [
 Debug Bones {type="TArray<FRigElementKey>", default=((Type=Bone,Name="calf_l"),(Type=Bone,Name="calf_r"))}
 Debug Controls On {type=bool, default=False}
 Debug Ctrls {type="TArray<FRigElementKey>", default=(())}
]
